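# Additive UI material for the quick bar pickup preview, drawing a pulsing arrow and box outline with glow.
# A PulseSpeed-scaled time loop drives ease curves that move and thin the arrow and shrink the box, while an offset sine fades each layer.
# Mirrored, rotated UVs turn one trapezoid into a chevron.
# The box SDF feeds a curve-atlas glow, summed with the arrow and vertically faded outline; adaptive colors supply a bright tint.
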Material /BattleRoyaleGameUI/HUD/QuickBar/Assets/M_UI_QuickBarPickupPreview
graph {
 properties {bAutomaticallySetUsageInEditor=false, blendMode=BLEND_Additive, materialDomain=MD_UI, usedShadingModels=Default Lit}
 nodes [
  n1 = MaterialFunctionCall() {materialFunction=&MF_UI_UVScaling}
  n2 = Subtract(A=ComponentMask(None=@n1.ScaleY) {r=true}) {constB=0.5}
  n3 = Subtract(A=OneMinus(None=Abs(None=@n2))) {constB=0.5}
  n4 = Multiply(A=Time(), B=ScalarParameter() {defaultValue=1.15, parameterName=PulseSpeed})
  n5 = Frac(None=@n4)
  n6 = MaterialFunctionCall("Value (0-1)"=@n5) {materialFunction=&MF_UI_EaseCurves}
  n7 = Subtract(
   A=ComponentMask(None=@n1.ScaleY) {g=true}
   B=LinearInterpolate(
    A=ScalarParameter() {defaultValue=-0.45, parameterName=ArrowOffsetStart}
    B=ScalarParameter() {defaultValue=-0.1, parameterName=ArrowOffsetEnd}
    Alpha=@n6.EaseOutCubic
   ) {constA=-0.45, constB=-0.05}
  )
  n8 = ScalarParameter() {defaultValue=0.25, parameterName=ArrowLength}
  n9 = LinearInterpolate(
   A=ScalarParameter() {defaultValue=0.05, parameterName=ArrowThicknessStart}
   B=ScalarParameter() {defaultValue=0.01, parameterName=ArrowThicknessEnd}
   Alpha=@n6.EaseOutSine
  ) {constA=0.05, constB=0.01}
  n10 = MaterialFunctionCall!(
   UVs=Rotator(Coordinate=AppendVector(A=@n3, B=@n7), Time=Constant() {r=-45}) {speed=0.0175}
   TopWidth=@n8
   BottomWidth=Subtract(A=@n8, B=Multiply(A=@n9) {constB=2})
   Height=@n9
   OutlineThickness
   OutlineAlignment
   GlowMin
   GlowMax
  ) {materialFunction=/FortUILibrary/MaterialFunctions/Shapes/MF_UI_IsoscelesTrapezoid}
  n11 = Add(A=@n4) {constB=0.75}
  n12 = MaterialFunctionCall("Value (0-1)"=ConstantBiasScale(None=Sine(None=@n11))) {materialFunction=&MF_UI_EaseCurves}
  n13 = MaterialFunctionCall(
   UVs=MaterialFunctionCall!(UVs) {materialFunction=/FortUILibrary/MaterialFunctions/Basics/MF_UI_9SliceUVs}.UVs
   Size=LinearInterpolate(
    A=ScalarParameter() {defaultValue=0.75, parameterName=BoxSizeStart}
    B=ScalarParameter() {defaultValue=0.5, parameterName=BoxSizeEnd}
    Alpha=@n6.EaseOutQuint
   ) {constA=0.75, constB=0.5}
   OutlineThickness=LinearInterpolate(
    A=ScalarParameter() {defaultValue=0.03, parameterName=BoxThicknessStart}
    B=ScalarParameter() {defaultValue=0.01, parameterName=BoxThicknessEnd}
    Alpha=@n6.EaseInOutCubic
   ) {constA=0.03, constB=0.01}
   OutlineAlignment=Constant() {r=-1}
  ) {materialFunction=&MF_UI_Box}
  n14 = Multiply(A=@n13.Outline, B=@n12.EaseOutCubic)
  n15 = ComponentMask(None=MaterialFunctionCall() {materialFunction=&GetUserInterfaceUV}.Normalized UV) {g=true}
  n16 = ScalarParameter() {defaultValue=1, parameterName=GlowOpacity}
  n17 = CurveAtlasRowParameter(CurveTime=SmoothStep(Value=@n13.SDFOutline) {constMax=-0.2, constMin=0.2}) {
   atlas=/BattleRoyaleGameUI/Shared/UICurveAtlas_BRGameUI
   curve=/BattleRoyaleGameUI/Shared/UICurve_BRGameUI_StandardFalloffs
   parameterName=GlowCurve
  }
  Add(
   A=Add(A=Multiply(A=@n10.Fill, B=@n12.EaseOutQuint), B=Multiply(A=@n14, B=@n15))
   B=Multiply(A=Multiply(A=@n16, B=@n17.A), B=@n12.EaseInQuint)
  )
  n18 = MaterialFunctionCall!(BufferSource) {materialFunction=/BattleRoyaleGameUI/Shared/MF_UI_AdaptiveColors}
  LinearInterpolate(A=@n18.Bright, B=@n18.Bright, Alpha=NamedRerouteUsage())
  NamedRerouteDeclaration(None=@n5) {name=Time, nodeColor=<rgba 0.5469 1 0 1>}
 ]
}
