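# Translucent UI material for the ammo type icon in a quick bar slot, with selected, clip-empty and out-of-ammo states.
# The icon texture's alpha masks a brightened tint, which shifts to an empty-clip color or the shared adaptive dark color by state.
# Four offset samples of center-scaled UVs merge into a drop shadow around the icon.
# All three states remove the shadow, and selection also darkens the icon and adjusts its opacity.
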Material /BattleRoyaleGameUI/HUD/QuickBar/Assets/M_QuickBarSlotAmmoTypeIcon
graph {
 properties {bAutomaticallySetUsageInEditor=false, blendMode=BLEND_Translucent, materialDomain=MD_UI, usedShadingModels=Default Lit}
 nodes [
  ComponentMask(None=CollectionParameter() {collection=/BattleRoyaleGameUI/Shared/MPC_BRGameUI, parameterName=Yellow}) {b=true, g=true, r=true}
  n1 = VectorParameter() {parameterName=ShadowColor, defaultValue=<rgba 0.013 0.013 0.013 1>}
  n2 = MaterialFunctionCall!(BufferSource) {materialFunction=/BattleRoyaleGameUI/Shared/MF_UI_AdaptiveColors}
  n3 = ScalarParameter() {parameterName=IsOutOfAmmo}
  n4 = MaterialFunctionCall!(
   Blend=ScalarParameter() {defaultValue=0.6667, parameterName=BlendValue}
   Base=VectorParameter() {parameterName=UnselectedColor, defaultValue=<rgba 0.3467 0.9911 0 1>}.RGB
  ) {materialFunction=/Engine/Functions/Engine_MaterialFunctions03/Blends/Blend_LinearDodge}
  n5 = ScalarParameter() {defaultValue=0.25, parameterName=WhiteMix}
  n6 = LinearInterpolate(
   A=Saturate(None=LinearInterpolate(A=@n4.Result, Alpha=@n5))
   B=VectorParameter() {parameterName=ClipEmptyColor, defaultValue=<rgba 0.3467 0.9911 0 1>}.RGB
   Alpha=ScalarParameter() {parameterName=IsClipEmpty}
  )
  n7 = ScalarParameter() {parameterName=IsOutOfAmmo}
  n8 = TextureSampleParameter2D() {parameterName=Icon, texture=/Game/UI/Foundation/Textures/Icons/Ammo/T-Icon-Ammo-Light-32}
  LinearInterpolate(
   A=LinearInterpolate(A=LinearInterpolate(A=@n1.RGB, B=@n2.Dark, Alpha=@n3), B=LinearInterpolate(A=@n6, B=@n2.Dark, Alpha=@n7), Alpha=@n8.A)
   B=@n2.Dark
   Alpha=ScalarParameter() {parameterName=IsSelected}
  ) {constA=1}
  n9 = LinearInterpolate(A=ScalarParameter() {defaultValue=0.25, parameterName=ShadowOpacity}, Alpha=ScalarParameter() {parameterName=IsOutOfAmmo}) {constB=0}
  n10 = ScalarParameter() {parameterName=IsClipEmpty}
  n11 = ScalarParameter() {defaultValue=0.075, parameterName=Shadow_Y_Offset}
  n12 = Add(
   A=NamedRerouteUsage()
   B=AppendVector(A=ScalarParameter() {defaultValue=0.05, parameterName=Shadow_X_Offset}, B=Multiply(A=@n11) {constB=-1})
  )
  n13 = Multiply(A=ScalarParameter() {defaultValue=0.05, parameterName=Shadow_X_Offset}) {constB=-1}
  n14 = ScalarParameter() {defaultValue=0.075, parameterName=Shadow_Y_Offset}
  n15 = TextureSampleParameter2D(UVs=Add(A=NamedRerouteUsage(), B=AppendVector(A=@n13, B=@n14))) {parameterName=Icon, texture=&T-Icon-Ammo-Light-32}
  n16 = NamedRerouteUsage()
  n17 = AppendVector(
   A=ScalarParameter() {defaultValue=0.05, parameterName=Shadow_X_Offset}
   B=ScalarParameter() {defaultValue=0.075, parameterName=Shadow_Y_Offset}
  )
  n18 = Multiply(A=@n17) {constB=-1}
  n19 = NamedRerouteDeclaration(None=MaterialFunctionCall(UVs=TextureCoordinate(), Texture Scale=ScalarParameter() {defaultValue=1.025, parameterName=ShadowScale}) {materialFunction=&ScaleUVsByCenter}.UVs) {name=UVs, nodeColor=<rgba 0 0 0 1>}
  n20 = LinearInterpolate(
   A=TextureSampleParameter2D(UVs=Add(A=@n16, B=@n18)) {parameterName=Icon, texture=&T-Icon-Ammo-Light-32}.A
   Alpha=TextureSampleParameter2D(UVs=Add(A=@n19, B=@n17)) {parameterName=Icon, texture=&T-Icon-Ammo-Light-32}.A
  )
  n21 = Multiply(
   A=LinearInterpolate(A=LinearInterpolate(A=@n9, Alpha=@n10) {constB=0}, Alpha=ScalarParameter() {parameterName=IsSelected}) {constB=0}
   B=LinearInterpolate(
    A=TextureSampleParameter2D(UVs=@n12) {parameterName=Icon, texture=&T-Icon-Ammo-Light-32}.A
    Alpha=LinearInterpolate(A=@n15.A, Alpha=@n20)
   )
  )
  Multiply(
   A=LinearInterpolate(A=LinearInterpolate(A=@n21, Alpha=@n8.A), B=@n8.A, Alpha=ScalarParameter() {parameterName=IsSelected}) {constB=0}
   B=LinearInterpolate(B=ScalarParameter() {defaultValue=0.95, parameterName=SelectedOpacity}, Alpha=ScalarParameter() {parameterName=IsSelected}) {constA=1}
  )
 ]
}
