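# Translucent UI material for quick bar slot text, tinting it by selection, empty clip and out-of-ammo state.
# The unselected color gets a Linear Dodge and white mix, then lerps to the empty-clip color.
# Out-of-ammo and selected states lerp to the Dark color from MF_UI_AdaptiveColors, with selection also setting opacity.
# A separate sine-driven lerp pulses toward the collection's Red as a reload warning.
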Material /BattleRoyaleGameUI/HUD/QuickBar/Assets/M_QuickBarSlotFont
graph {
 properties {bAutomaticallySetUsageInEditor=false, blendMode=BLEND_Translucent, materialDomain=MD_UI, usedShadingModels=Default Lit}
 nodes [
  ComponentMask(None=CollectionParameter() {collection=/BattleRoyaleGameUI/Shared/MPC_BRGameUI, parameterName=Yellow}) {b=true, g=true, r=true}
  n1 = ScalarParameter() {parameterName=IsSelected}
  LinearInterpolate(
   A=ScalarParameter() {defaultValue=1, parameterName=UnselectedOpacity}
   B=ScalarParameter() {defaultValue=0.95, parameterName=SelectedOpacity}
   Alpha=@n1
  ) {constA=1}
  n2 = VectorParameter() {parameterName=UnselectedColor, defaultValue=<rgba 0.3467 0.9911 0 1>}
  n3 = MaterialFunctionCall!(Blend=ScalarParameter() {defaultValue=0.6667, parameterName=BlendValue}, Base=@n2.RGB) {materialFunction=/Engine/Functions/Engine_MaterialFunctions03/Blends/Blend_LinearDodge}
  n4 = ScalarParameter() {defaultValue=0.25, parameterName=WhiteMix}
  n5 = LinearInterpolate(A=@n2.RGB, B=Saturate(None=LinearInterpolate(A=@n3.Result, Alpha=@n4)), Alpha=ScalarParameter() {defaultValue=1, parameterName=UseBlend})
  n6 = VectorParameter() {parameterName=ClipEmptyColor, defaultValue=<rgba 0.3467 0.9911 0 1>}
  n7 = ScalarParameter() {parameterName=IsClipEmpty}
  n8 = MaterialFunctionCall!(BufferSource) {materialFunction=/BattleRoyaleGameUI/Shared/MF_UI_AdaptiveColors}
  LinearInterpolate(
   A=LinearInterpolate(A=LinearInterpolate(A=@n5, B=@n6.RGB, Alpha=@n7), B=@n8.Dark, Alpha=ScalarParameter() {parameterName=IsOutOfAmmo})
   B=@n8.Dark
   Alpha=@n1
  ) {constA=1}
  LinearInterpolate()
  LinearInterpolate(
   A=@n5
   B=ComponentMask(None=CollectionParameter() {collection=&MPC_BRGameUI, parameterName=Red}) {b=true, g=true, r=true}
   Alpha=MaterialFunctionCall(Speed=ScalarParameter() {defaultValue=1, parameterName=ReloadWarningColorSpeed}) {materialFunction=&DebugTimeSine}.Result
  )
 ]
}
