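# UI material for the quick bar's durability bar, drawing a translucent rounded bar that fills bottom to top with Progress.
# MF_UI_Box builds the padded, aspect-corrected shape, while centre-scaled UVs minus Progress form a distance field for an anti-aliased fill mask.
# Named reroutes carry masks into colour and opacity blends between backing and fill.
# Fill colour mixes shared MPC_BRGameUI White with FillColor by FillRarityColorMix.
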Material /BattleRoyaleGameUI/HUD/QuickBar/Assets/M_UI_DurabilityBar
graph {
 properties {blendMode=BLEND_Translucent, materialDomain=MD_UI, pixelDepthOffsetMode=PDOM_Legacy, usedShadingModels=Default Lit}
 nodes [
  n1 = CollectionParameter() {collection=/BattleRoyaleGameUI/Shared/MPC_BRGameUI, parameterName=White}
  LinearInterpolate(
   A=VectorParameter() {parameterName=BackingColor, defaultValue=<rgba 0.0042 0.025 0.0024 1>}.RGB
   B=LinearInterpolate(
    A=ComponentMask(None=@n1) {b=true, g=true, r=true}
    B=VectorParameter() {parameterName=FillColor, defaultValue=<rgba 0.1195 0.521 0 1>}.RGB
    Alpha=ScalarParameter() {defaultValue=0.5, parameterName=FillRarityColorMix}
   )
   Alpha=NamedRerouteUsage()
  )
  Multiply(
   A=NamedRerouteUsage()
   B=LinearInterpolate(
    A=ScalarParameter() {defaultValue=0.75, parameterName=BackgroundOpacity}
    B=ScalarParameter() {defaultValue=1, parameterName=FillOpacity}
    Alpha=NamedRerouteUsage()
   )
  )
  n2 = MaterialFunctionCall() {materialFunction=&GetUserInterfaceUV}
  n3 = ComponentMask(None=@n2.Pixel Size) {g=true}
  NamedRerouteDeclaration(None=@n3) {name=VerticalMask, nodeColor=<rgba 0 0 0 1>}
  n4 = ScalarParameter() {defaultValue=0.2, parameterName=FillPadding}
  n5 = Divide(A=ComponentMask(None=@n2.Pixel Size) {r=true}, B=@n3)
  n6 = OneMinus(None=AppendVector(A=@n4, B=Multiply(A=@n4, B=@n5)))
  n7 = MaterialFunctionCall(Size=@n6, CornerRoundness=Constant() {r=0.05}, KeepAspectRatio=StaticBool() {value=true}) {materialFunction=&MF_UI_Box}
  NamedRerouteDeclaration(None=MaterialFunctionCall(DistanceField=@n7.SDF, Threshold=ScalarParameter() {parameterName=BarPadding}) {materialFunction=&MF_UI_AntiAliasing}.Inner) {name=BarMask, nodeColor=<rgba 0 0 0 1>}
  n8 = MaterialFunctionCall() {materialFunction=&GetUserInterfaceUV}
  n9 = ComponentMask(None=NamedRerouteDeclaration(None=MaterialFunctionCall(UVs=@n8.Normalized UV, Texture Scale=@n6) {materialFunction=&ScaleUVsByCenter}.UVs) {name=BarUVs, nodeColor=<rgba 0 0 0 1>}) {g=true}
  n10 = NamedRerouteDeclaration(None=Subtract(A=OneMinus(None=@n9), B=ScalarParameter() {defaultValue=0.5, group=Fill, parameterName=Progress, sortPriority=0}) {constB=0.75}) {name=ProgressFillSDF, nodeColor=<rgba 0 0 0 1>}
  NamedRerouteDeclaration(None=Multiply(
    A=MaterialFunctionCall(DistanceField=@n10) {materialFunction=&MF_UI_AntiAliasing}.Inner
    B=@n7.Fill
   )) {name=ProgressFillMask, nodeColor=<rgba 0 0 0 1>}
 ]
}
