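# UI material for the Battle Royale HUD quick bar's item card, in a variant-supporting version.
# Chained lerps driven by many scalar and vector parameters blend card layers and states, drawing shared white and red from the MPC_BRGameUI parameter collection.
# Shapes are built procedurally from SDF masks, skewed and rotated UVs and pixel-size aspect correction.
# Easing functions shape several blends, alongside a panner and a desaturated texture sample.
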Material /BattleRoyaleGameUI/HUD/QuickBar/Assets/M_UI_QuickBarItemCard_WithVariant
graph {
 expressions {
  MaterialExpressionAdd_0 {
   class=MaterialExpressionAdd
   inputs [
    A {from=MaterialExpressionNamedRerouteUsage_4}
    B {from=MaterialExpressionConstant_45}
   ]
  }
  MaterialExpressionAdd_1 {
   class=MaterialExpressionAdd
   inputs [
    A {from=MaterialExpressionComponentMask_20}
    B {from=MaterialExpressionScalarParameter_1}
   ]
  }
  MaterialExpressionAdd_12 {
   class=MaterialExpressionAdd
   inputs [
    A {from=MaterialExpressionLength_0}
    B {from=MaterialExpressionMin_4}
   ]
  }
  MaterialExpressionAdd_14 {
   class=MaterialExpressionAdd
   inputs [
    A {from=MaterialExpressionMaterialFunctionCall_37, output=Result}
    B {from=MaterialExpressionMaterialFunctionCall_38, output=Result}
   ]
   properties {constB=0.25}
  }
  MaterialExpressionAdd_18 {
   class=MaterialExpressionAdd
   inputs [
    A {from=MaterialExpressionMaterialFunctionCall_22, output=Main Color}
    B {from=""}
   ]
   properties {constB=0.4}
  }
  MaterialExpressionAdd_2 {
   class=MaterialExpressionAdd
   inputs [
    A {from=MaterialExpressionNamedRerouteUsage_76}
    B {from=MaterialExpressionNamedRerouteUsage_62}
   ]
  }
  MaterialExpressionAdd_3 {
   class=MaterialExpressionAdd
   inputs [
    A {from=MaterialExpressionNamedRerouteUsage_38}
    B {from=MaterialExpressionMultiply_17}
   ]
  }
  MaterialExpressionAdd_4 {
   class=MaterialExpressionAdd
   inputs [
    A {from=MaterialExpressionMultiply_11}
    B {from=MaterialExpressionScalarParameter_1}
   ]
  }
  MaterialExpressionAdd_5 {
   class=MaterialExpressionAdd
   inputs [
    A {from=MaterialExpressionLinearInterpolate_9}
    B {from=MaterialExpressionNamedRerouteDeclaration_34}
   ]
  }
  MaterialExpressionAdd_6 {
   class=MaterialExpressionAdd
   inputs [
    A {from=MaterialExpressionLinearInterpolate_52}
    B {from=""}
   ]
   properties {constA=1}
  }
  MaterialExpressionAdd_7 {
   class=MaterialExpressionAdd
   inputs [
    A {from=MaterialExpressionMaterialFunctionCall_34, output=Result}
    B {from=MaterialExpressionMaterialFunctionCall_35, output=Result}
   ]
   properties {constB=0.25}
  }
  MaterialExpressionAdd_8 {
   class=MaterialExpressionAdd
   inputs [
    A {from=MaterialExpressionLinearInterpolate_35}
    B {from=MaterialExpressionLinearInterpolate_44}
   ]
  }
  MaterialExpressionAdd_9 {
   class=MaterialExpressionAdd
   inputs [
    A {from=MaterialExpressionNamedRerouteUsage_22}
    B {from=MaterialExpressionNamedRerouteUsage_65}
   ]
  }
  MaterialExpressionAppendVector_1 {
   class=MaterialExpressionAppendVector
   inputs [
    A {from=MaterialExpressionConstant_1, output=""}
    B {from=MaterialExpressionNamedRerouteUsage_6, output=""}
   ]
  }
  MaterialExpressionAppendVector_2 {
   class=MaterialExpressionAppendVector
   inputs [
    A {from=MaterialExpressionMultiply_24, output=""}
    B {from=MaterialExpressionMultiply_25, output=""}
   ]
  }
  MaterialExpressionAppendVector_3 {
   class=MaterialExpressionAppendVector
   inputs [
    A {from=MaterialExpressionReroute_15, output=""}
    B {from=MaterialExpressionLinearInterpolate_29, output=""}
   ]
  }
  MaterialExpressionAppendVector_4 {
   class=MaterialExpressionAppendVector
   inputs [
    A {from=MaterialExpressionConstant_0, output=""}
    B {from=MaterialExpressionConstant_33, output=""}
   ]
  }
  MaterialExpressionAppendVector_5 {
   class=MaterialExpressionAppendVector
   inputs [
    A {from=MaterialExpressionComponentMask_8, output=""}
    B {from=MaterialExpressionSubtract_2, output=""}
   ]
  }
  MaterialExpressionAppendVector_6 {
   class=MaterialExpressionAppendVector
   inputs [
    A {from=MaterialExpressionConstant_5, output=""}
    B {from=MaterialExpressionConstant_47, output=""}
   ]
  }
  MaterialExpressionCollectionParameter_0 {class=MaterialExpressionCollectionParameter, inputs [], properties {collection=/BattleRoyaleGameUI/Shared/MPC_BRGameUI, parameterName=White}}
  MaterialExpressionCollectionParameter_5 {class=MaterialExpressionCollectionParameter, inputs [], properties {collection=&MPC_BRGameUI, parameterName=Red}}
  MaterialExpressionComponentMask_0 {class=MaterialExpressionComponentMask, inputs [None {from=MaterialExpressionPanner_1}], properties {g=true}}
  MaterialExpressionComponentMask_1 {
   class=MaterialExpressionComponentMask
   inputs [None {from=MaterialExpressionMaterialFunctionCall_26, output=Normalized UV}]
   properties {g=true}
  }
  MaterialExpressionComponentMask_10 {class=MaterialExpressionComponentMask, inputs [None {from=MaterialExpressionAdd_4}], properties {g=true}}
  MaterialExpressionComponentMask_11 {class=MaterialExpressionComponentMask, inputs [None {from=MaterialExpressionNamedRerouteUsage_88}], properties {g=true}}
  MaterialExpressionComponentMask_13 {class=MaterialExpressionComponentMask, inputs [None {from=MaterialExpressionReroute_6}], properties {g=true}}
  MaterialExpressionComponentMask_14 {
   class=MaterialExpressionComponentMask
   inputs [None {from=MaterialExpressionMaterialFunctionCall_15, output=Pixel Size}]
   properties {r=true}
  }
  MaterialExpressionComponentMask_15 {class=MaterialExpressionComponentMask, inputs [None {from=MaterialExpressionReroute_11}], properties {r=true}}
  MaterialExpressionComponentMask_16 {class=MaterialExpressionComponentMask, inputs [None {from=MaterialExpressionMaterialFunctionCall_7, output=Skewed UVs}], properties {g=true}}
  MaterialExpressionComponentMask_17 {
   class=MaterialExpressionComponentMask
   inputs [None {from=MaterialExpressionMaterialFunctionCall_15, output=Pixel Size}]
   properties {g=true}
  }
  MaterialExpressionComponentMask_18 {class=MaterialExpressionComponentMask, inputs [None {from=MaterialExpressionNamedRerouteUsage_40}], properties {r=true}}
  MaterialExpressionComponentMask_2 {class=MaterialExpressionComponentMask, inputs [None {from=MaterialExpressionMaterialFunctionCall_6, output=Skewed UVs}], properties {g=true}}
  MaterialExpressionComponentMask_20 {class=MaterialExpressionComponentMask, inputs [None {from=MaterialExpressionConvert_1}], properties {g=true}}
  MaterialExpressionComponentMask_22 {class=MaterialExpressionComponentMask, inputs [None {from=MaterialExpressionCollectionParameter_5}], properties {b=true, g=true, r=true}}
  MaterialExpressionComponentMask_23 {class=MaterialExpressionComponentMask, inputs [None {from=MaterialExpressionNamedRerouteUsage_33}], properties {g=true}}
  MaterialExpressionComponentMask_3 {class=MaterialExpressionComponentMask, inputs [None {from=MaterialExpressionCollectionParameter_0}], properties {b=true, g=true, r=true}}
  MaterialExpressionComponentMask_4 {class=MaterialExpressionComponentMask, inputs [None {from=MaterialExpressionOneMinus_10}], properties {g=true}}
  MaterialExpressionComponentMask_6 {class=MaterialExpressionComponentMask, inputs [None {from=MaterialExpressionMultiply_18}], properties {b=true, g=true, r=true}}
  MaterialExpressionComponentMask_7 {class=MaterialExpressionComponentMask, inputs [None {from=MaterialExpressionMultiply_18}], properties {a=true}}
  MaterialExpressionComponentMask_8 {class=MaterialExpressionComponentMask, inputs [None {from=MaterialExpressionMaterialFunctionCall_2, output=ScaleY}], properties {r=true}}
  MaterialExpressionComponentMask_9 {class=MaterialExpressionComponentMask, inputs [None {from=MaterialExpressionMaterialFunctionCall_2, output=ScaleY}], properties {g=true}}
  MaterialExpressionConstant2Vector_0 {class=MaterialExpressionConstant2Vector, properties {g=0.71, r=0.88}}
  MaterialExpressionConstant2Vector_1 {class=MaterialExpressionConstant2Vector, properties {g=0.76, r=0.88}}
  MaterialExpressionConstant_0 {class=MaterialExpressionConstant}
  MaterialExpressionConstant_1 {class=MaterialExpressionConstant}
  MaterialExpressionConstant_10 {class=MaterialExpressionConstant, properties {r=0.25}}
  MaterialExpressionConstant_12 {class=MaterialExpressionConstant, properties {r=0.2}}
  MaterialExpressionConstant_13 {class=MaterialExpressionConstant}
  MaterialExpressionConstant_14 {class=MaterialExpressionConstant, properties {r=0.1}}
  MaterialExpressionConstant_15 {class=MaterialExpressionConstant, properties {r=0.5}}
  MaterialExpressionConstant_16 {class=MaterialExpressionConstant, properties {r=0.125}}
  MaterialExpressionConstant_17 {class=MaterialExpressionConstant, properties {r=0.025}}
  MaterialExpressionConstant_18 {class=MaterialExpressionConstant, properties {r=0.0625}}
  MaterialExpressionConstant_19 {class=MaterialExpressionConstant, properties {r=0.125}}
  MaterialExpressionConstant_2 {class=MaterialExpressionConstant, properties {r=1}}
  MaterialExpressionConstant_20 {class=MaterialExpressionConstant, properties {r=0.5}}
  MaterialExpressionConstant_21 {class=MaterialExpressionConstant, properties {r=0.5}}
  MaterialExpressionConstant_22 {class=MaterialExpressionConstant, properties {r=4}}
  MaterialExpressionConstant_23 {class=MaterialExpressionConstant, properties {r=1}}
  MaterialExpressionConstant_24 {class=MaterialExpressionConstant, properties {r=0.9}}
  MaterialExpressionConstant_3 {class=MaterialExpressionConstant, properties {r=1}}
  MaterialExpressionConstant_30 {class=MaterialExpressionConstant, properties {r=0.1725}}
  MaterialExpressionConstant_31 {class=MaterialExpressionConstant, properties {r=0.285}}
  MaterialExpressionConstant_32 {class=MaterialExpressionConstant, properties {r=0.1875}}
  MaterialExpressionConstant_33 {class=MaterialExpressionConstant, properties {r=2.25}}
  MaterialExpressionConstant_34 {class=MaterialExpressionConstant, properties {r=0.055}}
  MaterialExpressionConstant_35 {class=MaterialExpressionConstant, properties {r=0.275}}
  MaterialExpressionConstant_36 {class=MaterialExpressionConstant, properties {r=0.0075}}
  MaterialExpressionConstant_37 {class=MaterialExpressionConstant, properties {r=0.3333}}
  MaterialExpressionConstant_38 {class=MaterialExpressionConstant, properties {r=0.1667}}
  MaterialExpressionConstant_39 {class=MaterialExpressionConstant, properties {r=0.3333}}
  MaterialExpressionConstant_4 {class=MaterialExpressionConstant, properties {r=1}}
  MaterialExpressionConstant_40 {class=MaterialExpressionConstant, properties {r=0.268}}
  MaterialExpressionConstant_41 {class=MaterialExpressionConstant, properties {r=0.121}}
  MaterialExpressionConstant_43 {class=MaterialExpressionConstant, properties {r=0.1875}}
  MaterialExpressionConstant_44 {class=MaterialExpressionConstant, properties {r=0.078}}
  MaterialExpressionConstant_45 {class=MaterialExpressionConstant, properties {r=0.5}}
  MaterialExpressionConstant_46 {class=MaterialExpressionConstant, properties {r=0.85}}
  MaterialExpressionConstant_47 {class=MaterialExpressionConstant, properties {r=-0.25}}
  MaterialExpressionConstant_48 {class=MaterialExpressionConstant, properties {r=0.4}}
  MaterialExpressionConstant_49 {class=MaterialExpressionConstant, properties {r=0.1668}}
  MaterialExpressionConstant_5 {class=MaterialExpressionConstant}
  MaterialExpressionConstant_50 {class=MaterialExpressionConstant, properties {r=0.268}}
  MaterialExpressionConstant_51 {class=MaterialExpressionConstant, properties {r=0.268}}
  MaterialExpressionConstant_54 {class=MaterialExpressionConstant, properties {r=0.025}}
  MaterialExpressionConstant_55 {class=MaterialExpressionConstant, properties {r=0.05}}
  MaterialExpressionConstant_6 {class=MaterialExpressionConstant, properties {r=1}}
  MaterialExpressionConstant_63 {class=MaterialExpressionConstant, properties {r=0.8}}
  MaterialExpressionConstant_64 {class=MaterialExpressionConstant, properties {r=0.4}}
  MaterialExpressionConstant_65 {class=MaterialExpressionConstant, properties {r=0.7}}
  MaterialExpressionConstant_7 {class=MaterialExpressionConstant, properties {r=0.25}}
  MaterialExpressionConstant_8 {class=MaterialExpressionConstant, properties {r=1}}
  MaterialExpressionConstant_9 {class=MaterialExpressionConstant, properties {r=0.121}}
  MaterialExpressionConvert_0 {
   class=MaterialExpressionConvert
   inputs [None {from=MaterialExpressionMaterialFunctionCall_25, output=Rotated Values}]
   properties {
    nodeName=Break Float2
    convertInputs [{
      type=Vector2
      expressionInput {
       expression=&self:MaterialExpressionMaterialFunctionCall_25
       inputName=None
       mask=0
       maskA=0
       maskB=0
       maskG=0
       maskR=0
       outputIndex=0
      }
     }]
    convertOutputs [{type=Scalar}, {type=Scalar}]
   }
  }
  MaterialExpressionConvert_1 {
   class=MaterialExpressionConvert
   inputs [
    None {from=MaterialExpressionConvert_0}
    None {from=MaterialExpressionLinearInterpolate_51}
   ]
   properties {
    nodeName=Make Vector2
    convertInputs [
     {
      type=Scalar
      expressionInput {
       expression=&self:MaterialExpressionConvert_0
       inputName=None
       mask=0
       maskA=0
       maskB=0
       maskG=0
       maskR=0
       outputIndex=0
      }
     }
     {
      type=Scalar
      expressionInput {
       expression=&self:MaterialExpressionLinearInterpolate_51
       inputName=None
       mask=0
       maskA=0
       maskB=0
       maskG=0
       maskR=0
       outputIndex=0
      }
     }
    ]
    convertOutputs [{type=Vector2}]
   }
  }
  MaterialExpressionConvert_2 {
   class=MaterialExpressionConvert
   inputs [None {from=MaterialExpressionNamedRerouteUsage_58}]
   properties {
    nodeName=Break Float2
    convertInputs [{
      type=Vector2
      expressionInput {
       expression=&self:MaterialExpressionNamedRerouteUsage_58
       inputName=None
       mask=0
       maskA=0
       maskB=0
       maskG=0
       maskR=0
       outputIndex=0
      }
     }]
    convertOutputs [{type=Scalar}, {type=Scalar}]
   }
  }
  MaterialExpressionDesaturation_0 {
   class=MaterialExpressionDesaturation
   inputs [
    None {from=MaterialExpressionTextureSample_1, output=RGB}
    Fraction {from="", output=""}
   ]
   properties {}
  }
  MaterialExpressionDivide_1 {
   class=MaterialExpressionDivide
   inputs [
    A {from=MaterialExpressionSubtract_10}
    B {from=MaterialExpressionReroute_0}
   ]
  }
  MaterialExpressionDivide_2 {
   class=MaterialExpressionDivide
   inputs [
    A {from=MaterialExpressionComponentMask_17}
    B {from=MaterialExpressionComponentMask_14}
   ]
  }
  MaterialExpressionDivide_3 {
   class=MaterialExpressionDivide
   inputs [
    A {from=MaterialExpressionSubtract_13}
    B {from=MaterialExpressionReroute_5}
   ]
  }
  MaterialExpressionDivide_4 {
   class=MaterialExpressionDivide
   inputs [
    A {from=MaterialExpressionSubtract_14}
    B {from=MaterialExpressionReroute_12}
   ]
  }
  MaterialExpressionLength_0 {class=MaterialExpressionLength, inputs [None {from=MaterialExpressionMax_1, output=""}], properties {}}
  MaterialExpressionLinearInterpolate_0 {
   class=MaterialExpressionLinearInterpolate
   inputs [
    A {from=MaterialExpressionLinearInterpolate_7}
    B {from=MaterialExpressionNamedRerouteUsage_25}
    Alpha {from=MaterialExpressionMaterialFunctionCall_11, output=SDF}
   ]
   properties {constA=0.8, constB=0.4}
  }
  MaterialExpressionLinearInterpolate_1 {
   class=MaterialExpressionLinearInterpolate
   inputs [
    A {from=""}
    B {from=""}
    Alpha {from=MaterialExpressionNamedRerouteUsage_16}
   ]
   properties {constA=0.9, constB=0.85}
  }
  MaterialExpressionLinearInterpolate_10 {
   class=MaterialExpressionLinearInterpolate
   inputs [
    A {from=MaterialExpressionLinearInterpolate_77}
    B {from=MaterialExpressionComponentMask_3}
    Alpha {from=MaterialExpressionScalarParameter_2}
   ]
   properties {constA=1}
  }
  MaterialExpressionLinearInterpolate_100 {
   class=MaterialExpressionLinearInterpolate
   inputs [
    A {from=MaterialExpressionScalarParameter_105}
    B {from=""}
    Alpha {from=MaterialExpressionScalarParameter_106}
   ]
   properties {constA=1, constB=0}
  }
  MaterialExpressionLinearInterpolate_101 {
   class=MaterialExpressionLinearInterpolate
   inputs [
    A {from=MaterialExpressionNamedRerouteDeclaration_51}
    B {from=MaterialExpressionNamedRerouteUsage_94}
    Alpha {from=MaterialExpressionSaturate_9}
   ]
  }
  MaterialExpressionLinearInterpolate_103 {
   class=MaterialExpressionLinearInterpolate
   inputs [
    A {from=MaterialExpressionLinearInterpolate_101}
    B {from=MaterialExpressionNamedRerouteUsage_70}
    Alpha {from=MaterialExpressionSaturate_11}
   ]
  }
  MaterialExpressionLinearInterpolate_104 {
   class=MaterialExpressionLinearInterpolate
   inputs [
    A {from=MaterialExpressionLinearInterpolate_4}
    B {from=MaterialExpressionLinearInterpolate_56}
    Alpha {from=MaterialExpressionScalarParameter_67}
   ]
   properties {constB=0}
  }
  MaterialExpressionLinearInterpolate_11 {
   class=MaterialExpressionLinearInterpolate
   inputs [
    A {from=MaterialExpressionLinearInterpolate_83}
    B {from=MaterialExpressionLinearInterpolate_110}
    Alpha {from=MaterialExpressionScalarParameter_15}
   ]
   properties {constA=1}
  }
  MaterialExpressionLinearInterpolate_110 {
   class=MaterialExpressionLinearInterpolate
   inputs [
    A {from=MaterialExpressionLinearInterpolate_20}
    B {from=MaterialExpressionNamedRerouteUsage_71}
    Alpha {from=MaterialExpressionScalarParameter_108}
   ]
  }
  MaterialExpressionLinearInterpolate_114 {
   class=MaterialExpressionLinearInterpolate
   inputs [
    A {from=MaterialExpressionConvert_2}
    B {from=MaterialExpressionMaterialFunctionCall_40, output=EaseOutCubic}
    Alpha {from=MaterialExpressionScalarParameter_113}
   ]
  }
  MaterialExpressionLinearInterpolate_117 {
   class=MaterialExpressionLinearInterpolate
   inputs [
    A {from=MaterialExpressionLinearInterpolate_85}
    B {from=MaterialExpressionSaturate_8}
    Alpha {from=MaterialExpressionScalarParameter_115}
   ]
   properties {constA=1}
  }
  MaterialExpressionLinearInterpolate_12 {
   class=MaterialExpressionLinearInterpolate
   inputs [
    A {from=MaterialExpressionNamedRerouteUsage_29}
    B {from=MaterialExpressionNamedRerouteUsage_23}
    Alpha {from=""}
   ]
  }
  MaterialExpressionLinearInterpolate_13 {
   class=MaterialExpressionLinearInterpolate
   inputs [
    A {from=MaterialExpressionLinearInterpolate_11}
    B {from=""}
    Alpha {from=MaterialExpressionNamedRerouteUsage_8}
   ]
   properties {constA=1, constB=0.95}
  }
  MaterialExpressionLinearInterpolate_14 {
   class=MaterialExpressionLinearInterpolate
   inputs [
    A {from=MaterialExpressionSaturate_6}
    B {from=MaterialExpressionReroute_18}
    Alpha {from=MaterialExpressionNamedRerouteDeclaration_47}
   ]
  }
  MaterialExpressionLinearInterpolate_15 {
   class=MaterialExpressionLinearInterpolate
   inputs [
    A {from=MaterialExpressionLinearInterpolate_35}
    B {from=MaterialExpressionConstant_51}
    Alpha {from=MaterialExpressionNamedRerouteUsage_28}
   ]
  }
  MaterialExpressionLinearInterpolate_16 {
   class=MaterialExpressionLinearInterpolate
   inputs [
    A {from=MaterialExpressionLinearInterpolate_19}
    B {from=MaterialExpressionMultiply_10}
    Alpha {from=MaterialExpressionScalarParameter_33}
   ]
  }
  MaterialExpressionLinearInterpolate_17 {
   class=MaterialExpressionLinearInterpolate
   inputs [
    A {from=""}
    B {from=MaterialExpressionNamedRerouteUsage_26}
    Alpha {from=MaterialExpressionScalarParameter_12}
   ]
   properties {constA=-0.05, constB=-0.2}
  }
  MaterialExpressionLinearInterpolate_18 {
   class=MaterialExpressionLinearInterpolate
   inputs [
    A {from=MaterialExpressionLinearInterpolate_79}
    B {from=""}
    Alpha {from=MaterialExpressionMaterialFunctionCall_47, output=EaseInQuint}
   ]
  }
  MaterialExpressionLinearInterpolate_19 {
   class=MaterialExpressionLinearInterpolate
   inputs [
    A {from=MaterialExpressionLinearInterpolate_87}
    B {from=MaterialExpressionLinearInterpolate_88}
    Alpha {from=MaterialExpressionNamedRerouteUsage_89}
   ]
  }
  MaterialExpressionLinearInterpolate_2 {
   class=MaterialExpressionLinearInterpolate
   inputs [
    A {from=MaterialExpressionLinearInterpolate_55}
    B {from=MaterialExpressionLinearInterpolate_64}
    Alpha {from=MaterialExpressionNamedRerouteUsage_2}
   ]
  }
  MaterialExpressionLinearInterpolate_20 {
   class=MaterialExpressionLinearInterpolate
   inputs [
    A {from=""}
    B {from=""}
    Alpha {from=MaterialExpressionNamedRerouteUsage_41}
   ]
   properties {constA=1, constB=0.85}
  }
  MaterialExpressionLinearInterpolate_21 {
   class=MaterialExpressionLinearInterpolate
   inputs [
    A {from=MaterialExpressionReroute_16}
    B {from=MaterialExpressionReroute_17}
    Alpha {from=MaterialExpressionNamedRerouteUsage_0}
   ]
   properties {constB=0.81}
  }
  MaterialExpressionLinearInterpolate_22 {
   class=MaterialExpressionLinearInterpolate
   inputs [
    A {from=MaterialExpressionLinearInterpolate_35}
    B {from=MaterialExpressionAdd_8}
    Alpha {from=MaterialExpressionNamedRerouteUsage_4}
   ]
  }
  MaterialExpressionLinearInterpolate_23 {
   class=MaterialExpressionLinearInterpolate
   inputs [
    A {from=MaterialExpressionNamedRerouteUsage_30}
    B {from=MaterialExpressionNamedRerouteUsage_52}
    Alpha {from=MaterialExpressionNamedRerouteUsage_35}
   ]
  }
  MaterialExpressionLinearInterpolate_24 {
   class=MaterialExpressionLinearInterpolate
   inputs [
    A {from=""}
    B {from=MaterialExpressionConstant_46}
    Alpha {from=MaterialExpressionScalarParameter_22}
   ]
   properties {constA=1, constB=0.5}
  }
  MaterialExpressionLinearInterpolate_25 {
   class=MaterialExpressionLinearInterpolate
   inputs [
    A {from=MaterialExpressionLinearInterpolate_67}
    B {from=MaterialExpressionNamedRerouteUsage_77}
    Alpha {from=MaterialExpressionScalarParameter_6}
   ]
  }
  MaterialExpressionLinearInterpolate_26 {
   class=MaterialExpressionLinearInterpolate
   inputs [
    A {from=""}
    B {from=MaterialExpressionConstant_12}
    Alpha {from=MaterialExpressionNamedRerouteUsage_13}
   ]
   properties {constA=0.3}
  }
  MaterialExpressionLinearInterpolate_27 {
   class=MaterialExpressionLinearInterpolate
   inputs [
    A {from=MaterialExpressionNamedRerouteUsage_45}
    B {from=MaterialExpressionNamedRerouteUsage_44}
    Alpha {from=""}
   ]
  }
  MaterialExpressionLinearInterpolate_28 {
   class=MaterialExpressionLinearInterpolate
   inputs [
    A {from=MaterialExpressionMaterialFunctionCall_0, output=EaseInQuint}
    B {from=""}
    Alpha {from=MaterialExpressionMaterialFunctionCall_17, output=IsGreaterOrEqual}
   ]
   properties {constB=0}
  }
  MaterialExpressionLinearInterpolate_29 {
   class=MaterialExpressionLinearInterpolate
   inputs [
    A {from=MaterialExpressionDivide_1}
    B {from=MaterialExpressionDivide_3}
    Alpha {from=MaterialExpressionDivide_4}
   ]
  }
  MaterialExpressionLinearInterpolate_3 {
   class=MaterialExpressionLinearInterpolate
   inputs [
    A {from=MaterialExpressionVectorParameter_5, output=RGBA}
    B {from=MaterialExpressionVectorParameter_4, output=RGBA}
    Alpha {from=MaterialExpressionNamedRerouteDeclaration_0}
   ]
  }
  MaterialExpressionLinearInterpolate_30 {
   class=MaterialExpressionLinearInterpolate
   inputs [
    A {from=MaterialExpressionNamedRerouteDeclaration_41}
    B {from=MaterialExpressionLinearInterpolate_32}
    Alpha {from=MaterialExpressionNamedRerouteUsage_18}
   ]
  }
  MaterialExpressionLinearInterpolate_31 {
   class=MaterialExpressionLinearInterpolate
   inputs [
    A {from=MaterialExpressionLinearInterpolate_47}
    B {from=""}
    Alpha {from=MaterialExpressionNamedRerouteUsage_37}
   ]
   properties {constA=0.4}
  }
  MaterialExpressionLinearInterpolate_32 {
   class=MaterialExpressionLinearInterpolate
   inputs [
    A {from=MaterialExpressionLinearInterpolate_49}
    B {from=MaterialExpressionNamedRerouteDeclaration_39}
    Alpha {from=MaterialExpressionNamedRerouteUsage_50}
   ]
  }
  MaterialExpressionLinearInterpolate_33 {
   class=MaterialExpressionLinearInterpolate
   inputs [
    A {from=MaterialExpressionLinearInterpolate_58}
    B {from=MaterialExpressionNamedRerouteDeclaration_42}
    Alpha {from=MaterialExpressionScalarParameter_40}
   ]
  }
  MaterialExpressionLinearInterpolate_34 {
   class=MaterialExpressionLinearInterpolate
   inputs [
    A {from=MaterialExpressionLinearInterpolate_22}
    B {from=MaterialExpressionAdd_0}
    Alpha {from=MaterialExpressionNamedRerouteUsage_21}
   ]
  }
  MaterialExpressionLinearInterpolate_35 {
   class=MaterialExpressionLinearInterpolate
   inputs [
    A {from=MaterialExpressionMaterialFunctionCall_39, output=Result}
    B {from=MaterialExpressionMaterialFunctionCall_42, output=Result}
    Alpha {from=MaterialExpressionNamedRerouteDeclaration_24}
   ]
  }
  MaterialExpressionLinearInterpolate_36 {
   class=MaterialExpressionLinearInterpolate
   inputs [
    A {from=MaterialExpressionMaterialFunctionCall_27, output=Inner}
    B {from=MaterialExpressionReroute_10}
    Alpha {from=MaterialExpressionNamedRerouteUsage_48}
   ]
  }
  MaterialExpressionLinearInterpolate_37 {
   class=MaterialExpressionLinearInterpolate
   inputs [
    A {from=MaterialExpressionMultiply_26}
    B {from=MaterialExpressionNamedRerouteUsage_43}
    Alpha {from=MaterialExpressionScalarParameter_45}
   ]
  }
  MaterialExpressionLinearInterpolate_38 {
   class=MaterialExpressionLinearInterpolate
   inputs [
    A {from=MaterialExpressionMaterialFunctionCall_4, output=Inner}
    B {from=MaterialExpressionReroute_3}
    Alpha {from=MaterialExpressionNamedRerouteUsage_61}
   ]
  }
  MaterialExpressionLinearInterpolate_39 {
   class=MaterialExpressionLinearInterpolate
   inputs [
    A {from=MaterialExpressionNamedRerouteUsage_72}
    B {from=MaterialExpressionNamedRerouteDeclaration_44}
    Alpha {from=MaterialExpressionReroute_22}
   ]
  }
  MaterialExpressionLinearInterpolate_4 {
   class=MaterialExpressionLinearInterpolate
   inputs [
    A {from=MaterialExpressionLinearInterpolate_56}
    B {from=MaterialExpressionLinearInterpolate_0}
    Alpha {from=MaterialExpressionScalarParameter_4}
   ]
   properties {constA=1}
  }
  MaterialExpressionLinearInterpolate_40 {
   class=MaterialExpressionLinearInterpolate
   inputs [
    A {from=MaterialExpressionLinearInterpolate_70}
    B {from=MaterialExpressionLinearInterpolate_71}
    Alpha {from=MaterialExpressionNamedRerouteUsage_17}
   ]
  }
  MaterialExpressionLinearInterpolate_41 {
   class=MaterialExpressionLinearInterpolate
   inputs [
    A {from=MaterialExpressionNamedRerouteUsage_67}
    B {from=MaterialExpressionNamedRerouteUsage_84}
    Alpha {from=MaterialExpressionNamedRerouteUsage_14}
   ]
  }
  MaterialExpressionLinearInterpolate_42 {
   class=MaterialExpressionLinearInterpolate
   inputs [
    A {from=MaterialExpressionLinearInterpolate_75}
    B {from=MaterialExpressionNamedRerouteUsage_93}
    Alpha {from=MaterialExpressionScalarParameter_0}
   ]
  }
  MaterialExpressionLinearInterpolate_44 {
   class=MaterialExpressionLinearInterpolate
   inputs [
    A {from=""}
    B {from=""}
    Alpha {from=MaterialExpressionNamedRerouteUsage_1}
   ]
   properties {constA=0.25}
  }
  MaterialExpressionLinearInterpolate_45 {
   class=MaterialExpressionLinearInterpolate
   inputs [
    A {from=MaterialExpressionSaturate_1}
    B {from=MaterialExpressionNamedRerouteUsage_66}
    Alpha {from=MaterialExpressionNamedRerouteUsage_69}
   ]
  }
  MaterialExpressionLinearInterpolate_46 {
   class=MaterialExpressionLinearInterpolate
   inputs [
    A {from=MaterialExpressionLinearInterpolate_41}
    B {from=MaterialExpressionNamedRerouteUsage_67}
    Alpha {from=MaterialExpressionNamedRerouteUsage_49}
   ]
  }
  MaterialExpressionLinearInterpolate_47 {
   class=MaterialExpressionLinearInterpolate
   inputs [
    A {from=MaterialExpressionReroute_26}
    B {from=MaterialExpressionLinearInterpolate_74}
    Alpha {from=MaterialExpressionNamedRerouteUsage_51}
   ]
  }
  MaterialExpressionLinearInterpolate_48 {
   class=MaterialExpressionLinearInterpolate
   inputs [
    A {from=MaterialExpressionMaterialFunctionCall_16, output=EaseInQuint}
    B {from=""}
    Alpha {from=MaterialExpressionMaterialFunctionCall_17, output=IsGreaterOrEqual}
   ]
   properties {constB=0}
  }
  MaterialExpressionLinearInterpolate_49 {
   class=MaterialExpressionLinearInterpolate
   inputs [
    A {from=MaterialExpressionLinearInterpolate_9}
    B {from=MaterialExpressionSaturate_3}
    Alpha {from=MaterialExpressionNamedRerouteUsage_53}
   ]
  }
  MaterialExpressionLinearInterpolate_5 {
   class=MaterialExpressionLinearInterpolate
   inputs [
    A {from=MaterialExpressionConstant_32}
    B {from=MaterialExpressionLinearInterpolate_94}
    Alpha {from=MaterialExpressionScalarParameter_11}
   ]
   properties {constA=0.175, constB=0.285}
  }
  MaterialExpressionLinearInterpolate_50 {
   class=MaterialExpressionLinearInterpolate
   inputs [
    A {from=MaterialExpressionSaturate_2}
    B {from=""}
    Alpha {from=MaterialExpressionScalarParameter_51}
   ]
   properties {constB=0}
  }
  MaterialExpressionLinearInterpolate_51 {
   class=MaterialExpressionLinearInterpolate
   inputs [
    A {from=""}
    B {from=MaterialExpressionAdd_6}
    Alpha {from=MaterialExpressionConvert_0}
   ]
  }
  MaterialExpressionLinearInterpolate_52 {
   class=MaterialExpressionLinearInterpolate
   inputs [
    A {from=MaterialExpressionConstant_34}
    B {from=MaterialExpressionConstant_35}
    Alpha {from=MaterialExpressionScalarParameter_56}
   ]
  }
  MaterialExpressionLinearInterpolate_54 {
   class=MaterialExpressionLinearInterpolate
   inputs [
    A {from=MaterialExpressionMultiply_36}
    B {from=""}
    Alpha {from=""}
   ]
   properties {constAlpha=0.25}
  }
  MaterialExpressionLinearInterpolate_55 {
   class=MaterialExpressionLinearInterpolate
   inputs [
    A {from=MaterialExpressionNamedRerouteUsage_7}
    B {from=MaterialExpressionLinearInterpolate_62}
    Alpha {from=MaterialExpressionScalarParameter_70}
   ]
  }
  MaterialExpressionLinearInterpolate_56 {
   class=MaterialExpressionLinearInterpolate
   inputs [
    A {from=MaterialExpressionLinearInterpolate_23}
    B {from=MaterialExpressionNamedRerouteDeclaration_50}
    Alpha {from=MaterialExpressionScalarParameter_68}
   ]
  }
  MaterialExpressionLinearInterpolate_57 {
   class=MaterialExpressionLinearInterpolate
   inputs [
    A {from=MaterialExpressionNamedRerouteUsage_9}
    B {from=MaterialExpressionNamedRerouteUsage_46}
    Alpha {from=MaterialExpressionScalarParameter_39}
   ]
  }
  MaterialExpressionLinearInterpolate_58 {
   class=MaterialExpressionLinearInterpolate
   inputs [
    A {from=MaterialExpressionNamedRerouteUsage_32}
    B {from=MaterialExpressionNamedRerouteDeclaration_43}
    Alpha {from=MaterialExpressionLinearInterpolate_45}
   ]
  }
  MaterialExpressionLinearInterpolate_59 {
   class=MaterialExpressionLinearInterpolate
   inputs [
    A {from=MaterialExpressionNamedRerouteUsage_11}
    B {from=MaterialExpressionNamedRerouteUsage_20}
    Alpha {from=MaterialExpressionScalarParameter_60}
   ]
  }
  MaterialExpressionLinearInterpolate_6 {
   class=MaterialExpressionLinearInterpolate
   inputs [
    A {from=""}
    B {from=MaterialExpressionNamedRerouteUsage_31}
    Alpha {from=MaterialExpressionScalarParameter_17}
   ]
   properties {constA=0.55, constB=0.75}
  }
  MaterialExpressionLinearInterpolate_60 {
   class=MaterialExpressionLinearInterpolate
   inputs [
    A {from=MaterialExpressionLinearInterpolate_6}
    B {from=MaterialExpressionMultiply_4}
    Alpha {from=MaterialExpressionScalarParameter_23}
   ]
  }
  MaterialExpressionLinearInterpolate_61 {
   class=MaterialExpressionLinearInterpolate
   inputs [
    A {from=MaterialExpressionNamedRerouteUsage_55}
    B {from=MaterialExpressionNamedRerouteUsage_56}
    Alpha {from=MaterialExpressionNamedRerouteUsage_53}
   ]
  }
  MaterialExpressionLinearInterpolate_62 {
   class=MaterialExpressionLinearInterpolate
   inputs [
    A {from=MaterialExpressionMaterialFunctionCall_18, output=Result}
    B {from=MaterialExpressionMaterialFunctionCall_28, output=Result}
    Alpha {from=MaterialExpressionNamedRerouteUsage_73}
   ]
  }
  MaterialExpressionLinearInterpolate_63 {
   class=MaterialExpressionLinearInterpolate
   inputs [
    A {from=MaterialExpressionConstant_3}
    B {from=MaterialExpressionMaterialFunctionCall_9, output=Outer}
    Alpha {from=MaterialExpressionMaterialFunctionCall_3, output=IsLess}
   ]
  }
  MaterialExpressionLinearInterpolate_64 {
   class=MaterialExpressionLinearInterpolate
   inputs [
    A {from=MaterialExpressionLinearInterpolate_57}
    B {from=MaterialExpressionReroute_23}
    Alpha {from=MaterialExpressionScalarParameter_75}
   ]
   properties {constA=1}
  }
  MaterialExpressionLinearInterpolate_65 {
   class=MaterialExpressionLinearInterpolate
   inputs [
    A {from=MaterialExpressionReroute_9}
    B {from=MaterialExpressionLinearInterpolate_66}
    Alpha {from=MaterialExpressionMaterialFunctionCall_21, output=IsLess}
   ]
  }
  MaterialExpressionLinearInterpolate_66 {
   class=MaterialExpressionLinearInterpolate
   inputs [
    A {from=MaterialExpressionNamedRerouteUsage_78}
    B {from=MaterialExpressionLinearInterpolate_72}
    Alpha {from=MaterialExpressionNamedRerouteUsage_75}
   ]
  }
  MaterialExpressionLinearInterpolate_67 {
   class=MaterialExpressionLinearInterpolate
   inputs [
    A {from=MaterialExpressionNamedRerouteUsage_97}
    B {from=MaterialExpressionNamedRerouteUsage_77}
    Alpha {from=MaterialExpressionNamedRerouteUsage_74}
   ]
  }
  MaterialExpressionLinearInterpolate_69 {
   class=MaterialExpressionLinearInterpolate
   inputs [
    A {from=MaterialExpressionMultiply_39}
    B {from=MaterialExpressionNamedRerouteUsage_12}
    Alpha {from=MaterialExpressionScalarParameter_79}
   ]
   properties {constA=1}
  }
  MaterialExpressionLinearInterpolate_7 {
   class=MaterialExpressionLinearInterpolate
   inputs [
    A {from=MaterialExpressionNamedRerouteUsage_10}
    B {from=MaterialExpressionNamedRerouteUsage_24}
    Alpha {from=MaterialExpressionLinearInterpolate_24}
   ]
  }
  MaterialExpressionLinearInterpolate_70 {
   class=MaterialExpressionLinearInterpolate
   inputs [
    A {from=MaterialExpressionMaterialFunctionCall_34, output=Result}
    B {from=MaterialExpressionAdd_7}
    Alpha {from=MaterialExpressionScalarParameter_81}
   ]
   properties {constA=1}
  }
  MaterialExpressionLinearInterpolate_71 {
   class=MaterialExpressionLinearInterpolate
   inputs [
    A {from=""}
    B {from=""}
    Alpha {from=MaterialExpressionScalarParameter_82}
   ]
   properties {constA=0.5, constB=0.75}
  }
  MaterialExpressionLinearInterpolate_72 {
   class=MaterialExpressionLinearInterpolate
   inputs [
    A {from=MaterialExpressionLinearInterpolate_42}
    B {from=MaterialExpressionLinearInterpolate_18}
    Alpha {from=MaterialExpressionScalarParameter_9}
   ]
  }
  MaterialExpressionLinearInterpolate_73 {
   class=MaterialExpressionLinearInterpolate
   inputs [
    A {from=""}
    B {from=MaterialExpressionConstant_36}
    Alpha {from=MaterialExpressionComponentMask_1}
   ]
  }
  MaterialExpressionLinearInterpolate_74 {
   class=MaterialExpressionLinearInterpolate
   inputs [
    A {from=MaterialExpressionMaterialFunctionCall_37, output=Result}
    B {from=MaterialExpressionAdd_14}
    Alpha {from=MaterialExpressionScalarParameter_85}
   ]
   properties {constA=1}
  }
  MaterialExpressionLinearInterpolate_75 {
   class=MaterialExpressionLinearInterpolate
   inputs [
    A {from=MaterialExpressionNamedRerouteUsage_91}
    B {from=MaterialExpressionNamedRerouteUsage_68}
    Alpha {from=""}
   ]
  }
  MaterialExpressionLinearInterpolate_77 {
   class=MaterialExpressionLinearInterpolate
   inputs [
    A {from=MaterialExpressionLinearInterpolate_8}
    B {from=MaterialExpressionNamedRerouteUsage_79}
    Alpha {from=MaterialExpressionScalarParameter_74}
   ]
  }
  MaterialExpressionLinearInterpolate_79 {
   class=MaterialExpressionLinearInterpolate
   inputs [
    A {from=MaterialExpressionLinearInterpolate_27}
    B {from=MaterialExpressionNamedRerouteUsage_93}
    Alpha {from=MaterialExpressionScalarParameter_86}
   ]
  }
  MaterialExpressionLinearInterpolate_8 {
   class=MaterialExpressionLinearInterpolate
   inputs [
    A {from=MaterialExpressionReroute_19}
    B {from=MaterialExpressionLinearInterpolate_12}
    Alpha {from=MaterialExpressionMaterialFunctionCall_19, output=EaseInCubic}
   ]
  }
  MaterialExpressionLinearInterpolate_81 {
   class=MaterialExpressionLinearInterpolate
   inputs [
    A {from=MaterialExpressionComponentMask_6}
    B {from=MaterialExpressionNamedRerouteUsage_27}
    Alpha {from=MaterialExpressionScalarParameter_88}
   ]
  }
  MaterialExpressionLinearInterpolate_83 {
   class=MaterialExpressionLinearInterpolate
   inputs [
    A {from=MaterialExpressionLinearInterpolate_1}
    B {from=MaterialExpressionNamedRerouteUsage_71}
    Alpha {from=MaterialExpressionScalarParameter_90}
   ]
   properties {constB=0.5}
  }
  MaterialExpressionLinearInterpolate_84 {
   class=MaterialExpressionLinearInterpolate
   inputs [
    A {from=MaterialExpressionVectorParameter_0, output=RGB}
    B {from=MaterialExpressionNamedRerouteUsage_47}
    Alpha {from=MaterialExpressionScalarParameter_16}
   ]
  }
  MaterialExpressionLinearInterpolate_85 {
   class=MaterialExpressionLinearInterpolate
   inputs [
    A {from=""}
    B {from=MaterialExpressionLinearInterpolate_86}
    Alpha {from=MaterialExpressionLinearInterpolate_114}
   ]
  }
  MaterialExpressionLinearInterpolate_86 {
   class=MaterialExpressionLinearInterpolate
   inputs [
    A {from=""}
    B {from=""}
    Alpha {from=MaterialExpressionNamedRerouteDeclaration_49}
   ]
   properties {constA=0.5}
  }
  MaterialExpressionLinearInterpolate_87 {
   class=MaterialExpressionLinearInterpolate
   inputs [
    A {from=MaterialExpressionConstant_18}
    B {from=""}
    Alpha {from=MaterialExpressionScalarParameter_92}
   ]
   properties {constB=0}
  }
  MaterialExpressionLinearInterpolate_88 {
   class=MaterialExpressionLinearInterpolate
   inputs [
    A {from=MaterialExpressionConstant_19}
    B {from=MaterialExpressionMultiply_43}
    Alpha {from=MaterialExpressionLinearInterpolate_92}
   ]
   properties {constB=0}
  }
  MaterialExpressionLinearInterpolate_89 {
   class=MaterialExpressionLinearInterpolate
   inputs [
    A {from=MaterialExpressionVectorParameter_1, output=RGB}
    B {from=MaterialExpressionNamedRerouteUsage_64}
    Alpha {from=MaterialExpressionScalarParameter_18}
   ]
  }
  MaterialExpressionLinearInterpolate_9 {
   class=MaterialExpressionLinearInterpolate
   inputs [
    A {from=MaterialExpressionLinearInterpolate_2}
    B {from=MaterialExpressionLinearInterpolate_59}
    Alpha {from=MaterialExpressionLinearInterpolate_69}
   ]
  }
  MaterialExpressionLinearInterpolate_90 {
   class=MaterialExpressionLinearInterpolate
   inputs [
    A {from=MaterialExpressionVectorParameter_2, output=RGB}
    B {from=MaterialExpressionNamedRerouteUsage_80}
    Alpha {from=MaterialExpressionScalarParameter_19}
   ]
  }
  MaterialExpressionLinearInterpolate_91 {
   class=MaterialExpressionLinearInterpolate
   inputs [
    A {from=""}
    B {from=""}
    Alpha {from=MaterialExpressionNamedRerouteUsage_86}
   ]
   properties {constA=1, constB=3}
  }
  MaterialExpressionLinearInterpolate_92 {
   class=MaterialExpressionLinearInterpolate
   inputs [
    A {from=MaterialExpressionScalarParameter_93}
    B {from=""}
    Alpha {from=MaterialExpressionScalarParameter_96}
   ]
   properties {constB=0}
  }
  MaterialExpressionLinearInterpolate_93 {
   class=MaterialExpressionLinearInterpolate
   inputs [
    A {from=MaterialExpressionVectorParameter_3, output=RGB}
    B {from=MaterialExpressionNamedRerouteUsage_87}
    Alpha {from=MaterialExpressionScalarParameter_20}
   ]
  }
  MaterialExpressionLinearInterpolate_94 {
   class=MaterialExpressionLinearInterpolate
   inputs [
    A {from=MaterialExpressionConstant_31}
    B {from=MaterialExpressionConstant_30}
    Alpha {from=MaterialExpressionScalarParameter_99}
   ]
  }
  MaterialExpressionLinearInterpolate_95 {
   class=MaterialExpressionLinearInterpolate
   inputs [
    A {from=MaterialExpressionMultiply_45}
    B {from=""}
    Alpha {from=MaterialExpressionLinearInterpolate_117}
   ]
  }
  MaterialExpressionLinearInterpolate_96 {
   class=MaterialExpressionLinearInterpolate
   inputs [
    A {from=""}
    B {from=""}
    Alpha {from=MaterialExpressionScalarParameter_100}
   ]
   properties {constA=-0.095, constB=-0.035}
  }
  MaterialExpressionLinearInterpolate_97 {
   class=MaterialExpressionLinearInterpolate
   inputs [
    A {from=""}
    B {from=""}
    Alpha {from=MaterialExpressionScalarParameter_102}
   ]
   properties {constA=0.625, constB=0.65}
  }
  MaterialExpressionLinearInterpolate_99 {
   class=MaterialExpressionLinearInterpolate
   inputs [
    A {from=MaterialExpressionNamedRerouteUsage_12}
    B {from=MaterialExpressionMultiply_46}
    Alpha {from=MaterialExpressionLinearInterpolate_100}
   ]
   properties {constB=0.5}
  }
  MaterialExpressionMaterialFunctionCall_0 {
   class=MaterialExpressionMaterialFunctionCall
   inputs ["Value (0-1)" {from=MaterialExpressionSmoothStep_1}]
   properties {materialFunction=&MF_UI_EaseCurves}
  }
  MaterialExpressionMaterialFunctionCall_1 {
   class=MaterialExpressionMaterialFunctionCall
   inputs [
    UVs {from=MaterialExpressionAppendVector_5}
    Texture Scale {from=MaterialExpressionLinearInterpolate_60}
   ]
   properties {materialFunction=&ScaleUVsByCenter}
  }
  MaterialExpressionMaterialFunctionCall_10 {
   class=MaterialExpressionMaterialFunctionCall
   inputs [Float4 {from=MaterialExpressionLinearInterpolate_3}]
   properties {materialFunction=&BreakOutFloat4Components}
  }
  MaterialExpressionMaterialFunctionCall_11 {
   class=MaterialExpressionMaterialFunctionCall
   inputs [
    UVs {from=MaterialExpressionSubtract_3}
    Radius {from=MaterialExpressionConstant_48}
   ]
   properties {materialFunction=/FortUILibrary/MaterialFunctions/Shapes/SDFs/MF_SDF_Circle}
  }
  MaterialExpressionMaterialFunctionCall_12 {
   class=MaterialExpressionMaterialFunctionCall
   inputs [
    DistanceField {from=MaterialExpressionMaterialFunctionCall_23, output=SDF}
    Threshold {from=MaterialExpressionConstant_17}
    Smoothing {from=MaterialExpressionConstant_14}
   ]
   properties {materialFunction=&MF_UI_AntiAliasing}
  }
  MaterialExpressionMaterialFunctionCall_13 {
   class=MaterialExpressionMaterialFunctionCall
   inputs []
   properties {materialFunction=/BattleRoyaleGameUI/Shared/FX/MF_UI_BR_StormTechPattern}
  }
  MaterialExpressionMaterialFunctionCall_14 {
   class=MaterialExpressionMaterialFunctionCall
   inputs [BufferSource {from=""}]
   properties {materialFunction=/BattleRoyaleGameUI/Shared/MF_UI_AdaptiveColors}
  }
  MaterialExpressionMaterialFunctionCall_15 {class=MaterialExpressionMaterialFunctionCall, inputs [], properties {materialFunction=&GetUserInterfaceUV}}
  MaterialExpressionMaterialFunctionCall_16 {
   class=MaterialExpressionMaterialFunctionCall
   inputs ["Value (0-1)" {from=MaterialExpressionSmoothStep_7}]
   properties {materialFunction=&MF_UI_EaseCurves}
  }
  MaterialExpressionMaterialFunctionCall_17 {
   class=MaterialExpressionMaterialFunctionCall
   inputs [
    A {from=MaterialExpressionScalarParameter_55}
    B {from=MaterialExpressionConstant_4}
   ]
   properties {materialFunction=/FortUILibrary/MaterialFunctions/Basics/MF_UI_BooleanOperations}
  }
  MaterialExpressionMaterialFunctionCall_18 {
   class=MaterialExpressionMaterialFunctionCall
   inputs [
    A {from=MaterialExpressionConstant_8}
    B {from=MaterialExpressionNamedRerouteUsage_46}
    C {from=MaterialExpressionNamedRerouteUsage_60}
    Alpha {from=MaterialExpressionNamedRerouteUsage_34}
   ]
   properties {materialFunction=&3ColorBlend}
  }
  MaterialExpressionMaterialFunctionCall_19 {
   class=MaterialExpressionMaterialFunctionCall
   inputs ["Value (0-1)" {from=MaterialExpressionComponentMask_18}]
   properties {materialFunction=&MF_UI_EaseCurves}
  }
  MaterialExpressionMaterialFunctionCall_2 {
   class=MaterialExpressionMaterialFunctionCall
   inputs [
    UV {from=""}
    TextureSize {from=""}
    ScalePivotPoint {from=""}
    ContainerSize {from=""}
   ]
   properties {materialFunction=&MF_UI_UVScaling}
  }
  MaterialExpressionMaterialFunctionCall_20 {
   class=MaterialExpressionMaterialFunctionCall
   inputs ["Value (0-1)" {from=MaterialExpressionMaterialFunctionCall_30, output=EaseInQuint}]
   properties {materialFunction=&MF_UI_EaseCurves}
  }
  MaterialExpressionMaterialFunctionCall_21 {
   class=MaterialExpressionMaterialFunctionCall
   inputs [
    A {from=MaterialExpressionScalarParameter_76}
    B {from=MaterialExpressionConstant_6}
   ]
   properties {materialFunction=&MF_UI_BooleanOperations}
  }
  MaterialExpressionMaterialFunctionCall_22 {
   class=MaterialExpressionMaterialFunctionCall
   inputs [
    Gradient Rotation {from=""}
    Gradient Tiling {from=MaterialExpressionConstant_63}
    Noise Size {from=MaterialExpressionConstant_64}
    Noise Opacity {from=""}
    Hue Color {from=""}
    Accent Intensity {from=MaterialExpressionConstant_65}
   ]
   properties {materialFunction=/BattleRoyaleGameUI/Shared/MF_UI_BRGameUI_RainbowStyling}
  }
  MaterialExpressionMaterialFunctionCall_23 {
   class=MaterialExpressionMaterialFunctionCall
   inputs [
    UVs {from=MaterialExpressionMaterialFunctionCall_43, output=UVs}
    Size {from=MaterialExpressionLinearInterpolate_21}
    CornerRoundness {from=MaterialExpressionLinearInterpolate_26}
    KeepAspectRatio {from=""}
    PixelSize {from=""}
   ]
   properties {materialFunction=/FortUILibrary/MaterialFunctions/Shapes/SDFs/MF_SDF_Box}
  }
  MaterialExpressionMaterialFunctionCall_24 {
   class=MaterialExpressionMaterialFunctionCall
   inputs [
    Default {from=MaterialExpressionMultiply_33}
    LowEnd {from=MaterialExpressionReroute_8}
   ]
   properties {materialFunction=/BattleRoyaleGameUI/Shared/MF_UI_LowEnd_ValueSwitcher}
  }
  MaterialExpressionMaterialFunctionCall_25 {
   class=MaterialExpressionMaterialFunctionCall
   inputs [
    UVs {from=MaterialExpressionNamedRerouteUsage_42}
    Rotation Center {from=""}
    "Rotation Angle (0-1)" {from=MaterialExpressionLinearInterpolate_73}
   ]
   properties {materialFunction=&CustomRotator}
  }
  MaterialExpressionMaterialFunctionCall_26 {class=MaterialExpressionMaterialFunctionCall, inputs [], properties {materialFunction=&GetUserInterfaceUV}}
  MaterialExpressionMaterialFunctionCall_27 {
   class=MaterialExpressionMaterialFunctionCall
   inputs [
    DistanceField {from=MaterialExpressionNamedRerouteDeclaration_18}
    Threshold {from=MaterialExpressionConstant_54}
    Smoothing {from=MaterialExpressionConstant_55}
   ]
   properties {materialFunction=&MF_UI_AntiAliasing}
  }
  MaterialExpressionMaterialFunctionCall_28 {
   class=MaterialExpressionMaterialFunctionCall
   inputs [
    A {from=MaterialExpressionNamedRerouteUsage_46}
    B {from=MaterialExpressionNamedRerouteUsage_60}
    C {from=MaterialExpressionNamedRerouteUsage_36}
    Alpha {from=MaterialExpressionNamedRerouteUsage_34}
   ]
   properties {materialFunction=&3ColorBlend}
  }
  MaterialExpressionMaterialFunctionCall_29 {
   class=MaterialExpressionMaterialFunctionCall
   inputs [Speed {from=MaterialExpressionConstant_20}]
   properties {materialFunction=&DebugTimeSine}
  }
  MaterialExpressionMaterialFunctionCall_3 {
   class=MaterialExpressionMaterialFunctionCall
   inputs [
    A {from=MaterialExpressionScalarParameter_71}
    B {from=MaterialExpressionConstant_2}
   ]
   properties {materialFunction=&MF_UI_BooleanOperations}
  }
  MaterialExpressionMaterialFunctionCall_30 {
   class=MaterialExpressionMaterialFunctionCall
   inputs ["Value (0-1)" {from=MaterialExpressionMaterialFunctionCall_31, output=EaseInQuint}]
   properties {materialFunction=&MF_UI_EaseCurves}
  }
  MaterialExpressionMaterialFunctionCall_31 {
   class=MaterialExpressionMaterialFunctionCall
   inputs ["Value (0-1)" {from=MaterialExpressionSine_0}]
   properties {materialFunction=&MF_UI_EaseCurves}
  }
  MaterialExpressionMaterialFunctionCall_32 {
   class=MaterialExpressionMaterialFunctionCall
   inputs [
    DistanceField {from=MaterialExpressionMaterialFunctionCall_23, output=SDF}
    Threshold {from=""}
    Smoothing {from=""}
   ]
   properties {materialFunction=&MF_UI_AntiAliasing}
  }
  MaterialExpressionMaterialFunctionCall_33 {
   class=MaterialExpressionMaterialFunctionCall
   inputs [
    Default {from=MaterialExpressionConstant_23}
    LowEnd {from=MaterialExpressionConstant_24}
   ]
   properties {materialFunction=&MF_UI_LowEnd_ValueSwitcher}
  }
  MaterialExpressionMaterialFunctionCall_34 {
   class=MaterialExpressionMaterialFunctionCall
   inputs [
    Default {from=MaterialExpressionConstant_15}
    LowEnd {from=MaterialExpressionConstant_40}
   ]
   properties {materialFunction=&MF_UI_LowEnd_ValueSwitcher}
  }
  MaterialExpressionMaterialFunctionCall_35 {
   class=MaterialExpressionMaterialFunctionCall
   inputs [
    Default {from=MaterialExpressionConstant_7}
    LowEnd {from=MaterialExpressionConstant_9}
   ]
   properties {materialFunction=&MF_UI_LowEnd_ValueSwitcher}
  }
  MaterialExpressionMaterialFunctionCall_36 {
   class=MaterialExpressionMaterialFunctionCall
   inputs [
    Default {from=MaterialExpressionMaterialFunctionCall_13, output=Result}
    LowEnd {from=MaterialExpressionConstant_13}
   ]
   properties {materialFunction=&MF_UI_LowEnd_ValueSwitcher}
  }
  MaterialExpressionMaterialFunctionCall_37 {
   class=MaterialExpressionMaterialFunctionCall
   inputs [
    Default {from=MaterialExpressionConstant_21}
    LowEnd {from=MaterialExpressionConstant_50}
   ]
   properties {materialFunction=&MF_UI_LowEnd_ValueSwitcher}
  }
  MaterialExpressionMaterialFunctionCall_38 {
   class=MaterialExpressionMaterialFunctionCall
   inputs [
    Default {from=MaterialExpressionConstant_10}
    LowEnd {from=MaterialExpressionConstant_41}
   ]
   properties {materialFunction=&MF_UI_LowEnd_ValueSwitcher}
  }
  MaterialExpressionMaterialFunctionCall_39 {
   class=MaterialExpressionMaterialFunctionCall
   inputs [
    Default {from=MaterialExpressionConstant_37}
    LowEnd {from=MaterialExpressionConstant_49}
   ]
   properties {materialFunction=&MF_UI_LowEnd_ValueSwitcher}
  }
  MaterialExpressionMaterialFunctionCall_4 {
   class=MaterialExpressionMaterialFunctionCall
   inputs [
    DistanceField {from=MaterialExpressionNamedRerouteDeclaration_18}
    Threshold {from=""}
    Smoothing {from=""}
   ]
   properties {materialFunction=&MF_UI_AntiAliasing}
  }
  MaterialExpressionMaterialFunctionCall_40 {
   class=MaterialExpressionMaterialFunctionCall
   inputs ["Value (0-1)" {from=MaterialExpressionConvert_2}]
   properties {materialFunction=&MF_UI_EaseCurves}
  }
  MaterialExpressionMaterialFunctionCall_41 {
   class=MaterialExpressionMaterialFunctionCall
   inputs ["Value (0-1)" {from=MaterialExpressionDesaturation_0}]
   properties {materialFunction=&MF_UI_EaseCurves}
  }
  MaterialExpressionMaterialFunctionCall_42 {
   class=MaterialExpressionMaterialFunctionCall
   inputs [
    Default {from=MaterialExpressionConstant_38}
    LowEnd {from=MaterialExpressionConstant_44}
   ]
   properties {materialFunction=&MF_UI_LowEnd_ValueSwitcher}
  }
  MaterialExpressionMaterialFunctionCall_43 {
   class=MaterialExpressionMaterialFunctionCall
   inputs [UVs {from=""}]
   properties {materialFunction=/FortUILibrary/MaterialFunctions/Basics/MF_UI_9SliceUVs}
  }
  MaterialExpressionMaterialFunctionCall_44 {
   class=MaterialExpressionMaterialFunctionCall
   inputs [
    ScreenUV {from=""}
    TextureSize {from=""}
    ScalePivotPoint {from=""}
   ]
   properties {materialFunction=/FortUILibrary/MaterialFunctions/Basics/MF_UI_ScreenSpaceUVScaling}
  }
  MaterialExpressionMaterialFunctionCall_45 {
   class=MaterialExpressionMaterialFunctionCall
   inputs ["Value (0-1)" {from=MaterialExpressionNamedRerouteUsage_96}]
   properties {materialFunction=&MF_UI_EaseCurves}
  }
  MaterialExpressionMaterialFunctionCall_46 {
   class=MaterialExpressionMaterialFunctionCall
   inputs [
    Default {from=MaterialExpressionLinearInterpolate_34}
    LowEnd {from=MaterialExpressionLinearInterpolate_15}
   ]
   properties {materialFunction=&MF_UI_LowEnd_ValueSwitcher}
  }
  MaterialExpressionMaterialFunctionCall_47 {
   class=MaterialExpressionMaterialFunctionCall
   inputs ["Value (0-1)" {from=MaterialExpressionNamedRerouteUsage_92}]
   properties {materialFunction=&MF_UI_EaseCurves}
  }
  MaterialExpressionMaterialFunctionCall_49 {
   class=MaterialExpressionMaterialFunctionCall
   inputs [
    Default {from=MaterialExpressionMaterialFunctionCall_14, output=Dim}
    LowEnd {from=MaterialExpressionMaterialFunctionCall_14, output=Dim}
   ]
   properties {materialFunction=&MF_UI_LowEnd_ValueSwitcher}
  }
  MaterialExpressionMaterialFunctionCall_5 {
   class=MaterialExpressionMaterialFunctionCall
   inputs [
    DistanceField {from=MaterialExpressionSubtract_1}
    Threshold {from=""}
    Smoothing {from=""}
   ]
   properties {materialFunction=&MF_UI_AntiAliasing}
  }
  MaterialExpressionMaterialFunctionCall_6 {
   class=MaterialExpressionMaterialFunctionCall
   inputs [
    UV {from=""}
    Skew Angle {from=MaterialExpressionNamedRerouteDeclaration_11}
   ]
   properties {materialFunction=/FortUILibrary/MaterialFunctions/Basics/MF_UI_Skew}
  }
  MaterialExpressionMaterialFunctionCall_7 {
   class=MaterialExpressionMaterialFunctionCall
   inputs [
    UV {from=""}
    Skew Angle {from=MaterialExpressionAppendVector_1}
   ]
   properties {materialFunction=&MF_UI_Skew}
  }
  MaterialExpressionMaterialFunctionCall_8 {
   class=MaterialExpressionMaterialFunctionCall
   inputs ["Value (0-1)" {from=MaterialExpressionOneMinus_12}]
   properties {materialFunction=&MF_UI_EaseCurves}
  }
  MaterialExpressionMaterialFunctionCall_9 {
   class=MaterialExpressionMaterialFunctionCall
   inputs [
    DistanceField {from=MaterialExpressionComponentMask_20}
    Threshold {from=MaterialExpressionOneMinus_21}
    Smoothing {from=""}
   ]
   properties {materialFunction=&MF_UI_AntiAliasing}
  }
  MaterialExpressionMax_0 {
   class=MaterialExpressionMax
   inputs [
    A {from=MaterialExpressionLinearInterpolate_99}
    B {from=MaterialExpressionLinearInterpolate_13}
   ]
  }
  MaterialExpressionMax_1 {
   class=MaterialExpressionMax
   inputs [
    A {from=MaterialExpressionAppendVector_2}
    B {from=""}
   ]
   properties {constB=0}
  }
  MaterialExpressionMax_5 {
   class=MaterialExpressionMax
   inputs [
    A {from=MaterialExpressionMultiply_24}
    B {from=MaterialExpressionMultiply_25}
   ]
  }
  MaterialExpressionMin_1 {
   class=MaterialExpressionMin
   inputs [
    A {from=MaterialExpressionReroute_14, output=""}
    B {from=MaterialExpressionOneMinus_18, output=""}
   ]
   properties {}
  }
  MaterialExpressionMin_2 {
   class=MaterialExpressionMin
   inputs [
    A {from=MaterialExpressionLinearInterpolate_37, output=""}
    B {from=MaterialExpressionLinearInterpolate_14, output=""}
   ]
   properties {}
  }
  MaterialExpressionMin_3 {
   class=MaterialExpressionMin
   inputs [
    A {from=MaterialExpressionReroute_13, output=""}
    B {from=MaterialExpressionOneMinus_20, output=""}
   ]
   properties {}
  }
  MaterialExpressionMin_4 {
   class=MaterialExpressionMin
   inputs [
    A {from=MaterialExpressionMax_5, output=""}
    B {from="", output=""}
   ]
   properties {constB=0}
  }
  MaterialExpressionMultiply_0 {
   class=MaterialExpressionMultiply
   inputs [
    A {from=MaterialExpressionLinearInterpolate_39}
    B {from=""}
   ]
  }
  MaterialExpressionMultiply_10 {
   class=MaterialExpressionMultiply
   inputs [
    A {from=MaterialExpressionLinearInterpolate_19}
    B {from=MaterialExpressionConstant_22}
   ]
   properties {constB=2}
  }
  MaterialExpressionMultiply_11 {
   class=MaterialExpressionMultiply
   inputs [
    A {from=MaterialExpressionOneMinus_22}
    B {from=""}
   ]
   properties {constB=-1}
  }
  MaterialExpressionMultiply_16 {
   class=MaterialExpressionMultiply
   inputs [
    A {from=MaterialExpressionMaterialFunctionCall_22, output=Accent Color}
    B {from=""}
   ]
   properties {constB=0.3}
  }
  MaterialExpressionMultiply_17 {
   class=MaterialExpressionMultiply
   inputs [
    A {from=MaterialExpressionOneMinus_5}
    B {from=MaterialExpressionOneMinus_6}
   ]
  }
  MaterialExpressionMultiply_18 {
   class=MaterialExpressionMultiply
   inputs [
    A {from=MaterialExpressionMultiply_5}
    B {from=MaterialExpressionReroute_2}
   ]
  }
  MaterialExpressionMultiply_22 {
   class=MaterialExpressionMultiply
   inputs [
    A {from=MaterialExpressionOneMinus_15}
    B {from=""}
   ]
   properties {constB=0.5}
  }
  MaterialExpressionMultiply_23 {
   class=MaterialExpressionMultiply
   inputs [
    A {from=MaterialExpressionOneMinus_19}
    B {from=""}
   ]
   properties {constB=0.5}
  }
  MaterialExpressionMultiply_24 {
   class=MaterialExpressionMultiply
   inputs [
    A {from=MaterialExpressionMin_1}
    B {from=""}
   ]
   properties {constB=-1}
  }
  MaterialExpressionMultiply_25 {
   class=MaterialExpressionMultiply
   inputs [
    A {from=MaterialExpressionMin_3}
    B {from=""}
   ]
   properties {constB=-1}
  }
  MaterialExpressionMultiply_26 {
   class=MaterialExpressionMultiply
   inputs [
    A {from=MaterialExpressionMax_0}
    B {from=MaterialExpressionLinearInterpolate_31}
   ]
  }
  MaterialExpressionMultiply_31 {
   class=MaterialExpressionMultiply
   inputs [
    A {from=MaterialExpressionNamedRerouteUsage_15}
    B {from=MaterialExpressionLinearInterpolate_16}
   ]
  }
  MaterialExpressionMultiply_33 {
   class=MaterialExpressionMultiply
   inputs [
    A {from=MaterialExpressionNamedRerouteDeclaration_15}
    B {from=MaterialExpressionNamedRerouteDeclaration_13}
   ]
  }
  MaterialExpressionMultiply_36 {
   class=MaterialExpressionMultiply
   inputs [
    A {from=MaterialExpressionNamedRerouteUsage_54}
    B {from=""}
   ]
   properties {constB=2}
  }
  MaterialExpressionMultiply_37 {
   class=MaterialExpressionMultiply
   inputs [
    A {from=MaterialExpressionLinearInterpolate_50}
    B {from=MaterialExpressionNamedRerouteUsage_63}
   ]
  }
  MaterialExpressionMultiply_39 {
   class=MaterialExpressionMultiply
   inputs [
    A {from=""}
    B {from=MaterialExpressionNamedRerouteUsage_12}
   ]
   properties {constA=0.95}
  }
  MaterialExpressionMultiply_4 {
   class=MaterialExpressionMultiply
   inputs [
    A {from=MaterialExpressionLinearInterpolate_6}
    B {from=""}
   ]
   properties {constB=1.125}
  }
  MaterialExpressionMultiply_43 {
   class=MaterialExpressionMultiply
   inputs [
    A {from=MaterialExpressionConstant_19}
    B {from=MaterialExpressionLinearInterpolate_91}
   ]
   properties {constB=2}
  }
  MaterialExpressionMultiply_45 {
   class=MaterialExpressionMultiply
   inputs [
    A {from=MaterialExpressionLinearInterpolate_117}
    B {from=MaterialExpressionNamedRerouteUsage_90}
   ]
  }
  MaterialExpressionMultiply_46 {
   class=MaterialExpressionMultiply
   inputs [
    A {from=MaterialExpressionNamedRerouteUsage_12}
    B {from=MaterialExpressionMaterialFunctionCall_33, output=Result}
   ]
  }
  MaterialExpressionMultiply_5 {
   class=MaterialExpressionMultiply
   inputs [
    A {from=MaterialExpressionTextureSample_1, output=RGBA}
    B {from=MaterialExpressionScalarParameter_14}
   ]
  }
  MaterialExpressionMultiply_9 {
   class=MaterialExpressionMultiply
   inputs [
    A {from=MaterialExpressionOneMinus_7}
    B {from=""}
   ]
   properties {constB=0.5}
  }
  MaterialExpressionNamedRerouteDeclaration_0 {
   class=MaterialExpressionNamedRerouteDeclaration
   inputs [None {from=MaterialExpressionScalarParameter_21}]
   properties {name=IsSelected, nodeColor=<rgba 0.7813 0 1 1>}
  }
  MaterialExpressionNamedRerouteDeclaration_1 {
   class=MaterialExpressionNamedRerouteDeclaration
   inputs [None {from=MaterialExpressionLinearInterpolate_38}]
   properties {name=TileMask, nodeColor=<rgba 0 0 0 1>}
  }
  MaterialExpressionNamedRerouteDeclaration_10 {
   class=MaterialExpressionNamedRerouteDeclaration
   inputs [None {from=MaterialExpressionLinearInterpolate_81}]
   properties {name=IconColor, nodeColor=<rgba 0 0 0 1>}
  }
  MaterialExpressionNamedRerouteDeclaration_11 {
   class=MaterialExpressionNamedRerouteDeclaration
   inputs [None {from=MaterialExpressionAppendVector_4}]
   properties {name=StripAngle, nodeColor=<rgba 0 0 0 1>}
  }
  MaterialExpressionNamedRerouteDeclaration_12 {
   class=MaterialExpressionNamedRerouteDeclaration
   inputs [None {from=MaterialExpressionLinearInterpolate_63}]
   properties {name=CoolDownMask, nodeColor=<rgba 0 0 0 1>}
  }
  MaterialExpressionNamedRerouteDeclaration_13 {
   class=MaterialExpressionNamedRerouteDeclaration
   inputs [None {from=MaterialExpressionMaterialFunctionCall_36, output=Result}]
   properties {name=StormTechPatternMask, nodeColor=<rgba 0 0 0 1>}
  }
  MaterialExpressionNamedRerouteDeclaration_14 {
   class=MaterialExpressionNamedRerouteDeclaration
   inputs [None {from=MaterialExpressionLinearInterpolate_65}]
   properties {name=GlowColor, nodeColor=<rgba 0 0 0 1>}
  }
  MaterialExpressionNamedRerouteDeclaration_15 {
   class=MaterialExpressionNamedRerouteDeclaration
   inputs [None {from=MaterialExpressionSaturate_4}]
   properties {name=SwipeMask, nodeColor=<rgba 0 0 0 1>}
  }
  MaterialExpressionNamedRerouteDeclaration_16 {
   class=MaterialExpressionNamedRerouteDeclaration
   inputs [None {from=MaterialExpressionMaterialFunctionCall_8, output=EaseInQuint}]
   properties {name=CoolDownShadowMask, nodeColor=<rgba 0 0 0 1>}
  }
  MaterialExpressionNamedRerouteDeclaration_17 {
   class=MaterialExpressionNamedRerouteDeclaration
   inputs [None {from=MaterialExpressionLinearInterpolate_28}]
   properties {name=CoolDownHighlight, nodeColor=<rgba 0 0 0 1>}
  }
  MaterialExpressionNamedRerouteDeclaration_18 {
   class=MaterialExpressionNamedRerouteDeclaration
   inputs [None {from=MaterialExpressionAdd_12}]
   properties {name=TileSDF, nodeColor=<rgba 0 0 0 1>}
  }
  MaterialExpressionNamedRerouteDeclaration_19 {
   class=MaterialExpressionNamedRerouteDeclaration
   inputs [None {from=MaterialExpressionLinearInterpolate_89}]
   properties {name=Rarity2, nodeColor=<rgba 0 0 0 1>}
  }
  MaterialExpressionNamedRerouteDeclaration_2 {
   class=MaterialExpressionNamedRerouteDeclaration
   inputs [None {from=MaterialExpressionMaterialFunctionCall_5, output=Outer}]
   properties {name=BottomStripMask, nodeColor=<rgba 0 0 0 1>}
  }
  MaterialExpressionNamedRerouteDeclaration_20 {
   class=MaterialExpressionNamedRerouteDeclaration
   inputs [None {from=MaterialExpressionLinearInterpolate_90}]
   properties {name=Rarity3, nodeColor=<rgba 0 0 0 1>}
  }
  MaterialExpressionNamedRerouteDeclaration_21 {
   class=MaterialExpressionNamedRerouteDeclaration
   inputs [None {from=MaterialExpressionLinearInterpolate_48}]
   properties {name=CoolDownGradientMask, nodeColor=<rgba 0 0 0 1>}
  }
  MaterialExpressionNamedRerouteDeclaration_22 {
   class=MaterialExpressionNamedRerouteDeclaration
   inputs [None {from=MaterialExpressionLinearInterpolate_36}]
   properties {name=GlowMask, nodeColor=<rgba 0 0 0 1>}
  }
  MaterialExpressionNamedRerouteDeclaration_23 {
   class=MaterialExpressionNamedRerouteDeclaration
   inputs [None {from=MaterialExpressionLinearInterpolate_84}]
   properties {name=Rarity1, nodeColor=<rgba 0 0 0 1>}
  }
  MaterialExpressionNamedRerouteDeclaration_24 {
   class=MaterialExpressionNamedRerouteDeclaration
   inputs [None {from=MaterialExpressionSmoothStep_2}]
   properties {name=EmptyTileVerticalMask, nodeColor=<rgba 0 0 0 1>}
  }
  MaterialExpressionNamedRerouteDeclaration_25 {
   class=MaterialExpressionNamedRerouteDeclaration
   inputs [None {from=MaterialExpressionLinearInterpolate_96}]
   properties {name=IconYPositionSelected, nodeColor=<rgba 0 0 0 1>}
  }
  MaterialExpressionNamedRerouteDeclaration_26 {
   class=MaterialExpressionNamedRerouteDeclaration
   inputs [None {from=MaterialExpressionLinearInterpolate_97}]
   properties {name=IconScale_Selected, nodeColor=<rgba 0 0 0 1>}
  }
  MaterialExpressionNamedRerouteDeclaration_27 {
   class=MaterialExpressionNamedRerouteDeclaration
   inputs [None {from=MaterialExpressionLinearInterpolate_93}]
   properties {name=Rarity4, nodeColor=<rgba 0 0 0 1>}
  }
  MaterialExpressionNamedRerouteDeclaration_28 {
   class=MaterialExpressionNamedRerouteDeclaration
   inputs [None {from=MaterialExpressionMaterialFunctionCall_14, output=Dark}]
   properties {name=PostProcessDark, nodeColor=<rgba 0 0 0 1>}
  }
  MaterialExpressionNamedRerouteDeclaration_29 {
   class=MaterialExpressionNamedRerouteDeclaration
   inputs [None {from=MaterialExpressionMaterialFunctionCall_32, output=Inner}]
   properties {name=RoundedMask, nodeColor=<rgba 0.0078 0 1 1>}
  }
  MaterialExpressionNamedRerouteDeclaration_3 {
   class=MaterialExpressionNamedRerouteDeclaration
   inputs [None {from=MaterialExpressionMaterialFunctionCall_10, output=R}]
   properties {name=TileWidth, nodeColor=<rgba 0 0 0 1>}
  }
  MaterialExpressionNamedRerouteDeclaration_30 {
   class=MaterialExpressionNamedRerouteDeclaration
   inputs [None {from=MaterialExpressionAppendVector_3}]
   properties {name=TileUVs, nodeColor=<rgba 0 0 0 1>}
  }
  MaterialExpressionNamedRerouteDeclaration_31 {
   class=MaterialExpressionNamedRerouteDeclaration
   inputs [None {from=MaterialExpressionMaterialFunctionCall_49, output=Result}]
   properties {name=PostProcessDim, nodeColor=<rgba 0 0 0 1>}
  }
  MaterialExpressionNamedRerouteDeclaration_32 {
   class=MaterialExpressionNamedRerouteDeclaration
   inputs [None {from=MaterialExpressionMaterialFunctionCall_46, output=Result}]
   properties {name=EmptyMask, nodeColor=<rgba 0 0 0 1>}
  }
  MaterialExpressionNamedRerouteDeclaration_33 {
   class=MaterialExpressionNamedRerouteDeclaration
   inputs [None {from=MaterialExpressionMaterialFunctionCall_24, output=Result}]
   properties {name=StormTechSwipeMask, nodeColor=<rgba 0 0 0 1>}
  }
  MaterialExpressionNamedRerouteDeclaration_34 {
   class=MaterialExpressionNamedRerouteDeclaration
   inputs [None {from=MaterialExpressionLinearInterpolate_61}]
   properties {name=CooldowHighlightColors, nodeColor=<rgba 0 0 0 1>}
  }
  MaterialExpressionNamedRerouteDeclaration_35 {
   class=MaterialExpressionNamedRerouteDeclaration
   inputs [None {from=MaterialExpressionMaterialFunctionCall_44, output=Scaled to Fit}]
   properties {name=ScreenUVs, nodeColor=<rgba 0 0 0 1>}
  }
  MaterialExpressionNamedRerouteDeclaration_36 {
   class=MaterialExpressionNamedRerouteDeclaration
   inputs [None {from=MaterialExpressionMaterialFunctionCall_12, output=Inner}]
   properties {name=RoundedMaskGlow, nodeColor=<rgba 0.3125 1 0 1>}
  }
  MaterialExpressionNamedRerouteDeclaration_37 {
   class=MaterialExpressionNamedRerouteDeclaration
   inputs [None {from=MaterialExpressionLinearInterpolate_103}]
   properties {name=EmptyAmmoIconColor, nodeColor=<rgba 0 0 0 1>}
  }
  MaterialExpressionNamedRerouteDeclaration_38 {
   class=MaterialExpressionNamedRerouteDeclaration
   inputs [None {from=MaterialExpressionScalarParameter_3}]
   properties {name=IsRoundedMask, nodeColor=<rgba 1 0 0.6328 1>}
  }
  MaterialExpressionNamedRerouteDeclaration_39 {
   class=MaterialExpressionNamedRerouteDeclaration
   inputs [None {from=MaterialExpressionSaturate_5}]
   properties {name=BrightRarityColor, nodeColor=<rgba 0 0 0 1>}
  }
  MaterialExpressionNamedRerouteDeclaration_4 {
   class=MaterialExpressionNamedRerouteDeclaration
   inputs [None {from=MaterialExpressionMaterialFunctionCall_10, output=G}]
   properties {name=TileLeftEdgeHeight, nodeColor=<rgba 0 0 0 1>}
  }
  MaterialExpressionNamedRerouteDeclaration_40 {
   class=MaterialExpressionNamedRerouteDeclaration
   inputs [None {from=MaterialExpressionMultiply_37}]
   properties {name=GlowOpacityMask, nodeColor=<rgba 0 0 0 1>}
  }
  MaterialExpressionNamedRerouteDeclaration_41 {
   class=MaterialExpressionNamedRerouteDeclaration
   inputs [None {from=MaterialExpressionLinearInterpolate_46}]
   properties {name=CooldownEmptyProgressColor, nodeColor=<rgba 0 0 0 1>}
  }
  MaterialExpressionNamedRerouteDeclaration_42 {
   class=MaterialExpressionNamedRerouteDeclaration
   inputs [None {from=MaterialExpressionNamedRerouteUsage_19}]
   properties {name=EmptyTileColors, nodeColor=<rgba 0 0 0 1>}
  }
  MaterialExpressionNamedRerouteDeclaration_43 {
   class=MaterialExpressionNamedRerouteDeclaration
   inputs [None {from=MaterialExpressionLinearInterpolate_30}]
   properties {name=FilledTileColors, nodeColor=<rgba 0 0 0 1>}
  }
  MaterialExpressionNamedRerouteDeclaration_44 {
   class=MaterialExpressionNamedRerouteDeclaration
   inputs [None {from=MaterialExpressionMin_2}]
   properties {name=TileOpacity, nodeColor=<rgba 0 0 0 1>}
  }
  MaterialExpressionNamedRerouteDeclaration_45 {
   class=MaterialExpressionNamedRerouteDeclaration
   inputs [None {from=MaterialExpressionComponentMask_11}]
   properties {name=TileVerticalGradientMask, nodeColor=<rgba 0 0 0 1>}
  }
  MaterialExpressionNamedRerouteDeclaration_46 {
   class=MaterialExpressionNamedRerouteDeclaration
   inputs [None {from=MaterialExpressionLinearInterpolate_95}]
   properties {name=EmptyAmmoGradientMask, nodeColor=<rgba 0 0 0 1>}
  }
  MaterialExpressionNamedRerouteDeclaration_47 {
   class=MaterialExpressionNamedRerouteDeclaration
   inputs [None {from=MaterialExpressionScalarParameter_7}]
   properties {name=IsIconMasked, nodeColor=<rgba 0 0 0 1>}
  }
  MaterialExpressionNamedRerouteDeclaration_48 {
   class=MaterialExpressionNamedRerouteDeclaration
   inputs [None {from=MaterialExpressionComponentMask_22}]
   properties {name=RedColor, nodeColor=<rgba 0 0 0 1>}
  }
  MaterialExpressionNamedRerouteDeclaration_49 {
   class=MaterialExpressionNamedRerouteDeclaration
   inputs [None {from=MaterialExpressionMaterialFunctionCall_29, output=Result}]
   properties {name=SineOnOff, nodeColor=<rgba 0 0 0 1>}
  }
  MaterialExpressionNamedRerouteDeclaration_5 {
   class=MaterialExpressionNamedRerouteDeclaration
   inputs [None {from=MaterialExpressionMaterialFunctionCall_10, output=B}]
   properties {name=TileRightEdgeHeight, nodeColor=<rgba 0 0 0 1>}
  }
  MaterialExpressionNamedRerouteDeclaration_50 {
   class=MaterialExpressionNamedRerouteDeclaration
   inputs [None {from=MaterialExpressionLinearInterpolate_25}]
   properties {name=EmptyAmmoBGColor, nodeColor=<rgba 0 0 0 1>}
  }
  MaterialExpressionNamedRerouteDeclaration_51 {
   class=MaterialExpressionNamedRerouteDeclaration
   inputs [None {from=MaterialExpressionNamedRerouteUsage_100}]
   properties {name=EmptyAmmoDarkColor, nodeColor=<rgba 0 0 0 1>}
  }
  MaterialExpressionNamedRerouteDeclaration_52 {
   class=MaterialExpressionNamedRerouteDeclaration
   inputs [None {from=MaterialExpressionMaterialFunctionCall_22, output=Main Color}]
   properties {name=Variant2, nodeColor=<rgba 0 0 0 1>}
  }
  MaterialExpressionNamedRerouteDeclaration_53 {
   class=MaterialExpressionNamedRerouteDeclaration
   inputs [None {from=MaterialExpressionMaterialFunctionCall_22, output=Accent Color}]
   properties {name=Variant3, nodeColor=<rgba 0 0 0 1>}
  }
  MaterialExpressionNamedRerouteDeclaration_54 {
   class=MaterialExpressionNamedRerouteDeclaration
   inputs [None {from=MaterialExpressionSaturate_15}]
   properties {name=Variant1, nodeColor=<rgba 0 0 0 1>}
  }
  MaterialExpressionNamedRerouteDeclaration_55 {
   class=MaterialExpressionNamedRerouteDeclaration
   inputs [None {from=MaterialExpressionMultiply_16}]
   properties {name=Variant4, nodeColor=<rgba 0 0 0 1>}
  }
  MaterialExpressionNamedRerouteDeclaration_6 {
   class=MaterialExpressionNamedRerouteDeclaration
   inputs [None {from=MaterialExpressionMaterialFunctionCall_10, output=A}]
   properties {name=TileSkewVertical, nodeColor=<rgba 0 0 0 1>}
  }
  MaterialExpressionNamedRerouteDeclaration_7 {
   class=MaterialExpressionNamedRerouteDeclaration
   inputs [None {from=MaterialExpressionSaturate_0}]
   properties {name=BackgroundColor, nodeColor=<rgba 0 0 0 1>}
  }
  MaterialExpressionNamedRerouteDeclaration_8 {
   class=MaterialExpressionNamedRerouteDeclaration
   inputs [None {from=MaterialExpressionLinearInterpolate_10}]
   properties {name=BottomStripColor, nodeColor=<rgba 0 0 0 1>}
  }
  MaterialExpressionNamedRerouteDeclaration_9 {
   class=MaterialExpressionNamedRerouteDeclaration
   inputs [None {from=MaterialExpressionComponentMask_7}]
   properties {name=IconMask, nodeColor=<rgba 0 0 0 1>}
  }
  MaterialExpressionNamedRerouteUsage_0 {class=MaterialExpressionNamedRerouteUsage}
  MaterialExpressionNamedRerouteUsage_1 {class=MaterialExpressionNamedRerouteUsage}
  MaterialExpressionNamedRerouteUsage_10 {class=MaterialExpressionNamedRerouteUsage}
  MaterialExpressionNamedRerouteUsage_100 {class=MaterialExpressionNamedRerouteUsage}
  MaterialExpressionNamedRerouteUsage_11 {class=MaterialExpressionNamedRerouteUsage}
  MaterialExpressionNamedRerouteUsage_12 {class=MaterialExpressionNamedRerouteUsage}
  MaterialExpressionNamedRerouteUsage_13 {class=MaterialExpressionNamedRerouteUsage}
  MaterialExpressionNamedRerouteUsage_14 {class=MaterialExpressionNamedRerouteUsage}
  MaterialExpressionNamedRerouteUsage_15 {class=MaterialExpressionNamedRerouteUsage}
  MaterialExpressionNamedRerouteUsage_16 {class=MaterialExpressionNamedRerouteUsage}
  MaterialExpressionNamedRerouteUsage_17 {class=MaterialExpressionNamedRerouteUsage}
  MaterialExpressionNamedRerouteUsage_18 {class=MaterialExpressionNamedRerouteUsage}
  MaterialExpressionNamedRerouteUsage_19 {class=MaterialExpressionNamedRerouteUsage}
  MaterialExpressionNamedRerouteUsage_2 {class=MaterialExpressionNamedRerouteUsage}
  MaterialExpressionNamedRerouteUsage_20 {class=MaterialExpressionNamedRerouteUsage}
  MaterialExpressionNamedRerouteUsage_21 {class=MaterialExpressionNamedRerouteUsage}
  MaterialExpressionNamedRerouteUsage_22 {class=MaterialExpressionNamedRerouteUsage}
  MaterialExpressionNamedRerouteUsage_23 {class=MaterialExpressionNamedRerouteUsage}
  MaterialExpressionNamedRerouteUsage_24 {class=MaterialExpressionNamedRerouteUsage}
  MaterialExpressionNamedRerouteUsage_25 {class=MaterialExpressionNamedRerouteUsage}
  MaterialExpressionNamedRerouteUsage_26 {class=MaterialExpressionNamedRerouteUsage}
  MaterialExpressionNamedRerouteUsage_27 {class=MaterialExpressionNamedRerouteUsage}
  MaterialExpressionNamedRerouteUsage_28 {class=MaterialExpressionNamedRerouteUsage}
  MaterialExpressionNamedRerouteUsage_29 {class=MaterialExpressionNamedRerouteUsage}
  MaterialExpressionNamedRerouteUsage_3 {class=MaterialExpressionNamedRerouteUsage}
  MaterialExpressionNamedRerouteUsage_30 {class=MaterialExpressionNamedRerouteUsage}
  MaterialExpressionNamedRerouteUsage_31 {class=MaterialExpressionNamedRerouteUsage}
  MaterialExpressionNamedRerouteUsage_32 {class=MaterialExpressionNamedRerouteUsage}
  MaterialExpressionNamedRerouteUsage_33 {class=MaterialExpressionNamedRerouteUsage}
  MaterialExpressionNamedRerouteUsage_34 {class=MaterialExpressionNamedRerouteUsage}
  MaterialExpressionNamedRerouteUsage_35 {class=MaterialExpressionNamedRerouteUsage}
  MaterialExpressionNamedRerouteUsage_36 {class=MaterialExpressionNamedRerouteUsage}
  MaterialExpressionNamedRerouteUsage_37 {class=MaterialExpressionNamedRerouteUsage}
  MaterialExpressionNamedRerouteUsage_38 {class=MaterialExpressionNamedRerouteUsage}
  MaterialExpressionNamedRerouteUsage_39 {class=MaterialExpressionNamedRerouteUsage}
  MaterialExpressionNamedRerouteUsage_4 {class=MaterialExpressionNamedRerouteUsage}
  MaterialExpressionNamedRerouteUsage_40 {class=MaterialExpressionNamedRerouteUsage}
  MaterialExpressionNamedRerouteUsage_41 {class=MaterialExpressionNamedRerouteUsage}
  MaterialExpressionNamedRerouteUsage_42 {class=MaterialExpressionNamedRerouteUsage}
  MaterialExpressionNamedRerouteUsage_43 {class=MaterialExpressionNamedRerouteUsage}
  MaterialExpressionNamedRerouteUsage_44 {class=MaterialExpressionNamedRerouteUsage}
  MaterialExpressionNamedRerouteUsage_45 {class=MaterialExpressionNamedRerouteUsage}
  MaterialExpressionNamedRerouteUsage_46 {class=MaterialExpressionNamedRerouteUsage}
  MaterialExpressionNamedRerouteUsage_47 {class=MaterialExpressionNamedRerouteUsage}
  MaterialExpressionNamedRerouteUsage_48 {class=MaterialExpressionNamedRerouteUsage}
  MaterialExpressionNamedRerouteUsage_49 {class=MaterialExpressionNamedRerouteUsage}
  MaterialExpressionNamedRerouteUsage_5 {class=MaterialExpressionNamedRerouteUsage}
  MaterialExpressionNamedRerouteUsage_50 {class=MaterialExpressionNamedRerouteUsage}
  MaterialExpressionNamedRerouteUsage_51 {class=MaterialExpressionNamedRerouteUsage}
  MaterialExpressionNamedRerouteUsage_52 {class=MaterialExpressionNamedRerouteUsage}
  MaterialExpressionNamedRerouteUsage_53 {class=MaterialExpressionNamedRerouteUsage}
  MaterialExpressionNamedRerouteUsage_54 {class=MaterialExpressionNamedRerouteUsage}
  MaterialExpressionNamedRerouteUsage_55 {class=MaterialExpressionNamedRerouteUsage}
  MaterialExpressionNamedRerouteUsage_56 {class=MaterialExpressionNamedRerouteUsage}
  MaterialExpressionNamedRerouteUsage_57 {class=MaterialExpressionNamedRerouteUsage}
  MaterialExpressionNamedRerouteUsage_58 {class=MaterialExpressionNamedRerouteUsage}
  MaterialExpressionNamedRerouteUsage_59 {class=MaterialExpressionNamedRerouteUsage}
  MaterialExpressionNamedRerouteUsage_6 {class=MaterialExpressionNamedRerouteUsage}
  MaterialExpressionNamedRerouteUsage_60 {class=MaterialExpressionNamedRerouteUsage}
  MaterialExpressionNamedRerouteUsage_61 {class=MaterialExpressionNamedRerouteUsage}
  MaterialExpressionNamedRerouteUsage_62 {class=MaterialExpressionNamedRerouteUsage}
  MaterialExpressionNamedRerouteUsage_63 {class=MaterialExpressionNamedRerouteUsage}
  MaterialExpressionNamedRerouteUsage_64 {class=MaterialExpressionNamedRerouteUsage}
  MaterialExpressionNamedRerouteUsage_65 {class=MaterialExpressionNamedRerouteUsage}
  MaterialExpressionNamedRerouteUsage_66 {class=MaterialExpressionNamedRerouteUsage}
  MaterialExpressionNamedRerouteUsage_67 {class=MaterialExpressionNamedRerouteUsage}
  MaterialExpressionNamedRerouteUsage_68 {class=MaterialExpressionNamedRerouteUsage}
  MaterialExpressionNamedRerouteUsage_69 {class=MaterialExpressionNamedRerouteUsage}
  MaterialExpressionNamedRerouteUsage_7 {class=MaterialExpressionNamedRerouteUsage}
  MaterialExpressionNamedRerouteUsage_70 {class=MaterialExpressionNamedRerouteUsage}
  MaterialExpressionNamedRerouteUsage_71 {class=MaterialExpressionNamedRerouteUsage}
  MaterialExpressionNamedRerouteUsage_72 {class=MaterialExpressionNamedRerouteUsage}
  MaterialExpressionNamedRerouteUsage_73 {class=MaterialExpressionNamedRerouteUsage}
  MaterialExpressionNamedRerouteUsage_74 {class=MaterialExpressionNamedRerouteUsage}
  MaterialExpressionNamedRerouteUsage_75 {class=MaterialExpressionNamedRerouteUsage}
  MaterialExpressionNamedRerouteUsage_76 {class=MaterialExpressionNamedRerouteUsage}
  MaterialExpressionNamedRerouteUsage_77 {class=MaterialExpressionNamedRerouteUsage}
  MaterialExpressionNamedRerouteUsage_78 {class=MaterialExpressionNamedRerouteUsage}
  MaterialExpressionNamedRerouteUsage_79 {class=MaterialExpressionNamedRerouteUsage}
  MaterialExpressionNamedRerouteUsage_8 {class=MaterialExpressionNamedRerouteUsage}
  MaterialExpressionNamedRerouteUsage_80 {class=MaterialExpressionNamedRerouteUsage}
  MaterialExpressionNamedRerouteUsage_84 {class=MaterialExpressionNamedRerouteUsage}
  MaterialExpressionNamedRerouteUsage_86 {class=MaterialExpressionNamedRerouteUsage}
  MaterialExpressionNamedRerouteUsage_87 {class=MaterialExpressionNamedRerouteUsage}
  MaterialExpressionNamedRerouteUsage_88 {class=MaterialExpressionNamedRerouteUsage}
  MaterialExpressionNamedRerouteUsage_89 {class=MaterialExpressionNamedRerouteUsage}
  MaterialExpressionNamedRerouteUsage_9 {class=MaterialExpressionNamedRerouteUsage}
  MaterialExpressionNamedRerouteUsage_90 {class=MaterialExpressionNamedRerouteUsage}
  MaterialExpressionNamedRerouteUsage_91 {class=MaterialExpressionNamedRerouteUsage}
  MaterialExpressionNamedRerouteUsage_92 {class=MaterialExpressionNamedRerouteUsage}
  MaterialExpressionNamedRerouteUsage_93 {class=MaterialExpressionNamedRerouteUsage}
  MaterialExpressionNamedRerouteUsage_94 {class=MaterialExpressionNamedRerouteUsage}
  MaterialExpressionNamedRerouteUsage_96 {class=MaterialExpressionNamedRerouteUsage}
  MaterialExpressionNamedRerouteUsage_97 {class=MaterialExpressionNamedRerouteUsage}
  MaterialExpressionOneMinus_1 {class=MaterialExpressionOneMinus, inputs [None {from=MaterialExpressionLinearInterpolate_5}]}
  MaterialExpressionOneMinus_10 {class=MaterialExpressionOneMinus, inputs [None {from=MaterialExpressionAdd_1}]}
  MaterialExpressionOneMinus_12 {class=MaterialExpressionOneMinus, inputs [None {from=MaterialExpressionSmoothStep_5}]}
  MaterialExpressionOneMinus_15 {class=MaterialExpressionOneMinus, inputs [None {from=MaterialExpressionNamedRerouteUsage_5}]}
  MaterialExpressionOneMinus_18 {class=MaterialExpressionOneMinus, inputs [None {from=MaterialExpressionReroute_14}]}
  MaterialExpressionOneMinus_19 {class=MaterialExpressionOneMinus, inputs [None {from=MaterialExpressionAdd_3}]}
  MaterialExpressionOneMinus_20 {class=MaterialExpressionOneMinus, inputs [None {from=MaterialExpressionReroute_13}]}
  MaterialExpressionOneMinus_21 {class=MaterialExpressionOneMinus, inputs [None {from=MaterialExpressionScalarParameter_1}]}
  MaterialExpressionOneMinus_22 {class=MaterialExpressionOneMinus, inputs [None {from=MaterialExpressionConvert_1}]}
  MaterialExpressionOneMinus_23 {class=MaterialExpressionOneMinus, inputs [None {from=MaterialExpressionComponentMask_10}]}
  MaterialExpressionOneMinus_5 {class=MaterialExpressionOneMinus, inputs [None {from=MaterialExpressionNamedRerouteUsage_38}]}
  MaterialExpressionOneMinus_6 {class=MaterialExpressionOneMinus, inputs [None {from=MaterialExpressionDivide_2}]}
  MaterialExpressionOneMinus_7 {class=MaterialExpressionOneMinus, inputs [None {from=MaterialExpressionNamedRerouteUsage_3}]}
  MaterialExpressionOneMinus_8 {class=MaterialExpressionOneMinus, inputs [None {from=MaterialExpressionConstant_43}]}
  MaterialExpressionPanner_1 {
   class=MaterialExpressionPanner
   inputs [
    Coordinate {from=MaterialExpressionNamedRerouteDeclaration_35}
    Time {from=MaterialExpressionTime_0}
    Speed {from=MaterialExpressionConstant_16}
   ]
   properties {speedX=1}
  }
  MaterialExpressionReroute_0 {class=MaterialExpressionReroute, inputs [None {from=MaterialExpressionNamedRerouteUsage_3}]}
  MaterialExpressionReroute_1 {class=MaterialExpressionReroute, inputs [None {from=MaterialExpressionMaterialFunctionCall_1, output=0-1 mask}]}
  MaterialExpressionReroute_10 {class=MaterialExpressionReroute, inputs [None {from=MaterialExpressionNamedRerouteUsage_57}]}
  MaterialExpressionReroute_11 {class=MaterialExpressionReroute, inputs [None {from=MaterialExpressionMaterialFunctionCall_7, output=Skewed UVs}]}
  MaterialExpressionReroute_12 {class=MaterialExpressionReroute, inputs [None {from=MaterialExpressionAdd_3}]}
  MaterialExpressionReroute_13 {class=MaterialExpressionReroute, inputs [None {from=MaterialExpressionReroute_15}]}
  MaterialExpressionReroute_14 {class=MaterialExpressionReroute, inputs [None {from=MaterialExpressionLinearInterpolate_29}]}
  MaterialExpressionReroute_15 {class=MaterialExpressionReroute, inputs [None {from=MaterialExpressionDivide_4}]}
  MaterialExpressionReroute_16 {class=MaterialExpressionReroute, inputs [None {from=MaterialExpressionConstant2Vector_0}]}
  MaterialExpressionReroute_17 {class=MaterialExpressionReroute, inputs [None {from=MaterialExpressionConstant2Vector_1}]}
  MaterialExpressionReroute_18 {class=MaterialExpressionReroute, inputs [None {from=MaterialExpressionNamedRerouteUsage_22}]}
  MaterialExpressionReroute_19 {class=MaterialExpressionReroute, inputs [None {from=MaterialExpressionNamedRerouteUsage_23}]}
  MaterialExpressionReroute_2 {class=MaterialExpressionReroute, inputs [None {from=MaterialExpressionReroute_1}]}
  MaterialExpressionReroute_22 {class=MaterialExpressionReroute, inputs [None {from=MaterialExpressionReroute_24}]}
  MaterialExpressionReroute_23 {class=MaterialExpressionReroute, inputs [None {from=MaterialExpressionNamedRerouteUsage_9}]}
  MaterialExpressionReroute_24 {class=MaterialExpressionReroute, inputs [None {from=MaterialExpressionLinearInterpolate_14}]}
  MaterialExpressionReroute_26 {class=MaterialExpressionReroute, inputs [None {from=MaterialExpressionLinearInterpolate_40}]}
  MaterialExpressionReroute_3 {class=MaterialExpressionReroute, inputs [None {from=MaterialExpressionNamedRerouteUsage_59}]}
  MaterialExpressionReroute_4 {class=MaterialExpressionReroute, inputs [None {from=MaterialExpressionMaterialFunctionCall_1, output=UVs}]}
  MaterialExpressionReroute_5 {class=MaterialExpressionReroute, inputs [None {from=MaterialExpressionNamedRerouteUsage_5}]}
  MaterialExpressionReroute_6 {class=MaterialExpressionReroute, inputs [None {from=MaterialExpressionMaterialFunctionCall_7, output=Skewed UVs}]}
  MaterialExpressionReroute_7 {class=MaterialExpressionReroute, inputs [None {from=MaterialExpressionLinearInterpolate_39}]}
  MaterialExpressionReroute_8 {class=MaterialExpressionReroute, inputs [None {from=MaterialExpressionConstant_13}]}
  MaterialExpressionReroute_9 {class=MaterialExpressionReroute, inputs [None {from=MaterialExpressionLinearInterpolate_72}]}
  MaterialExpressionSaturate_0 {class=MaterialExpressionSaturate, inputs [None {from=MaterialExpressionLinearInterpolate_104}]}
  MaterialExpressionSaturate_1 {class=MaterialExpressionSaturate, inputs [None {from=MaterialExpressionAdd_2}]}
  MaterialExpressionSaturate_11 {class=MaterialExpressionSaturate, inputs [None {from=MaterialExpressionMaterialFunctionCall_45, output=EaseOutQuint}]}
  MaterialExpressionSaturate_15 {class=MaterialExpressionSaturate, inputs [None {from=MaterialExpressionAdd_18}]}
  MaterialExpressionSaturate_2 {class=MaterialExpressionSaturate, inputs [None {from=MaterialExpressionMultiply_31}]}
  MaterialExpressionSaturate_3 {class=MaterialExpressionSaturate, inputs [None {from=MaterialExpressionAdd_5}]}
  MaterialExpressionSaturate_4 {class=MaterialExpressionSaturate, inputs [None {from=MaterialExpressionMaterialFunctionCall_20, output=EaseInQuint}]}
  MaterialExpressionSaturate_5 {class=MaterialExpressionSaturate, inputs [None {from=MaterialExpressionLinearInterpolate_54}]}
  MaterialExpressionSaturate_6 {class=MaterialExpressionSaturate, inputs [None {from=MaterialExpressionAdd_9}]}
  MaterialExpressionSaturate_8 {class=MaterialExpressionSaturate, inputs [None {from=MaterialExpressionMaterialFunctionCall_40, output=EaseInOutSine}]}
  MaterialExpressionSaturate_9 {class=MaterialExpressionSaturate, inputs [None {from=MaterialExpressionMaterialFunctionCall_41, output=EaseInOutSine}]}
  MaterialExpressionScalarParameter_0 {class=MaterialExpressionScalarParameter, properties {parameterName=IsAmmoEmpty}}
  MaterialExpressionScalarParameter_1 {class=MaterialExpressionScalarParameter, properties {defaultValue=1, parameterName=CooldownProgress}}
  MaterialExpressionScalarParameter_100 {class=MaterialExpressionScalarParameter, properties {parameterName=IsHarvestingTool}}
  MaterialExpressionScalarParameter_102 {class=MaterialExpressionScalarParameter, properties {parameterName=IsHarvestingTool}}
  MaterialExpressionScalarParameter_105 {class=MaterialExpressionScalarParameter, properties {parameterName=IsAmmoEmpty}}
  MaterialExpressionScalarParameter_106 {class=MaterialExpressionScalarParameter, properties {parameterName=IsSelected}}
  MaterialExpressionScalarParameter_108 {class=MaterialExpressionScalarParameter, properties {parameterName=IsAmmoEmpty}}
  MaterialExpressionScalarParameter_11 {class=MaterialExpressionScalarParameter, properties {parameterName=IsSelected}}
  MaterialExpressionScalarParameter_113 {class=MaterialExpressionScalarParameter, properties {parameterName=IsSelected}}
  MaterialExpressionScalarParameter_115 {class=MaterialExpressionScalarParameter, properties {parameterName=IsSelected}}
  MaterialExpressionScalarParameter_12 {class=MaterialExpressionScalarParameter, properties {parameterName=IsSelected}}
  MaterialExpressionScalarParameter_14 {class=MaterialExpressionScalarParameter, properties {defaultValue=1, parameterName=IsIconLoaded}}
  MaterialExpressionScalarParameter_15 {class=MaterialExpressionScalarParameter, properties {parameterName=IsSelected}}
  MaterialExpressionScalarParameter_16 {class=MaterialExpressionScalarParameter, properties {parameterName=IsSpecialStyling}}
  MaterialExpressionScalarParameter_17 {class=MaterialExpressionScalarParameter, properties {parameterName=IsSelected}}
  MaterialExpressionScalarParameter_18 {class=MaterialExpressionScalarParameter, properties {parameterName=IsSpecialStyling}}
  MaterialExpressionScalarParameter_19 {class=MaterialExpressionScalarParameter, properties {parameterName=IsSpecialStyling}}
  MaterialExpressionScalarParameter_2 {class=MaterialExpressionScalarParameter, properties {parameterName=IsSelected}}
  MaterialExpressionScalarParameter_20 {class=MaterialExpressionScalarParameter, properties {parameterName=IsSpecialStyling}}
  MaterialExpressionScalarParameter_21 {class=MaterialExpressionScalarParameter, properties {parameterName=IsSelected}}
  MaterialExpressionScalarParameter_22 {class=MaterialExpressionScalarParameter, properties {parameterName=IsSelected}}
  MaterialExpressionScalarParameter_23 {class=MaterialExpressionScalarParameter, properties {parameterName=AnimCoolDownComplete}}
  MaterialExpressionScalarParameter_3 {class=MaterialExpressionScalarParameter, properties {parameterName=isRoundedMask}}
  MaterialExpressionScalarParameter_33 {class=MaterialExpressionScalarParameter, properties {parameterName=AnimCoolDownComplete}}
  MaterialExpressionScalarParameter_39 {class=MaterialExpressionScalarParameter, properties {parameterName=AnimCoolDownComplete}}
  MaterialExpressionScalarParameter_4 {class=MaterialExpressionScalarParameter, properties {parameterName=IsSelected}}
  MaterialExpressionScalarParameter_40 {class=MaterialExpressionScalarParameter, properties {parameterName=IsEmpty}}
  MaterialExpressionScalarParameter_45 {class=MaterialExpressionScalarParameter, properties {parameterName=IsEmpty}}
  MaterialExpressionScalarParameter_5 {class=MaterialExpressionScalarParameter, properties {parameterName=IconYOffset}}
  MaterialExpressionScalarParameter_51 {class=MaterialExpressionScalarParameter, properties {parameterName=IsEmpty}}
  MaterialExpressionScalarParameter_55 {class=MaterialExpressionScalarParameter, properties {defaultValue=1, parameterName=CooldownProgress}}
  MaterialExpressionScalarParameter_56 {class=MaterialExpressionScalarParameter, properties {parameterName=IsSelected}}
  MaterialExpressionScalarParameter_6 {class=MaterialExpressionScalarParameter, properties {parameterName=IsSelected}}
  MaterialExpressionScalarParameter_60 {class=MaterialExpressionScalarParameter, properties {parameterName=AnimCoolDownComplete}}
  MaterialExpressionScalarParameter_67 {class=MaterialExpressionScalarParameter, properties {parameterName=IsAmmoEmpty}}
  MaterialExpressionScalarParameter_68 {class=MaterialExpressionScalarParameter, properties {parameterName=IsAmmoEmpty}}
  MaterialExpressionScalarParameter_7 {class=MaterialExpressionScalarParameter, properties {parameterName=IsIconMasked}}
  MaterialExpressionScalarParameter_70 {class=MaterialExpressionScalarParameter, properties {parameterName=AnimCoolDownComplete}}
  MaterialExpressionScalarParameter_71 {class=MaterialExpressionScalarParameter, properties {defaultValue=1, parameterName=CooldownProgress}}
  MaterialExpressionScalarParameter_74 {class=MaterialExpressionScalarParameter, properties {parameterName=IsAmmoEmpty}}
  MaterialExpressionScalarParameter_75 {class=MaterialExpressionScalarParameter, properties {parameterName=IsSelected}}
  MaterialExpressionScalarParameter_76 {class=MaterialExpressionScalarParameter, properties {defaultValue=1, parameterName=CooldownProgress}}
  MaterialExpressionScalarParameter_79 {class=MaterialExpressionScalarParameter, properties {parameterName=IsSelected}}
  MaterialExpressionScalarParameter_81 {class=MaterialExpressionScalarParameter, properties {parameterName=IsSelected}}
  MaterialExpressionScalarParameter_82 {class=MaterialExpressionScalarParameter, properties {parameterName=IsSelected}}
  MaterialExpressionScalarParameter_85 {class=MaterialExpressionScalarParameter, properties {parameterName=IsSelected}}
  MaterialExpressionScalarParameter_86 {class=MaterialExpressionScalarParameter, properties {parameterName=IsAmmoEmpty}}
  MaterialExpressionScalarParameter_88 {class=MaterialExpressionScalarParameter, properties {parameterName=IsAmmoEmpty}}
  MaterialExpressionScalarParameter_9 {class=MaterialExpressionScalarParameter, properties {parameterName=IsSelected}}
  MaterialExpressionScalarParameter_90 {class=MaterialExpressionScalarParameter, properties {parameterName=IsAmmoEmpty}}
  MaterialExpressionScalarParameter_92 {class=MaterialExpressionScalarParameter, properties {parameterName=IsAmmoEmpty}}
  MaterialExpressionScalarParameter_93 {class=MaterialExpressionScalarParameter, properties {parameterName=IsAmmoEmpty}}
  MaterialExpressionScalarParameter_96 {class=MaterialExpressionScalarParameter, properties {parameterName=IsSelected}}
  MaterialExpressionScalarParameter_99 {class=MaterialExpressionScalarParameter, properties {parameterName=IsHarvestingTool}}
  MaterialExpressionSine_0 {class=MaterialExpressionSine, inputs [None {from=MaterialExpressionComponentMask_0}], properties {period=2}}
  MaterialExpressionSmoothStep_1 {
   class=MaterialExpressionSmoothStep
   inputs [
    Min {from="", output=""}
    Max {from="", output=""}
    Value {from=MaterialExpressionOneMinus_23, output=""}
   ]
   properties {constMin=0.75}
  }
  MaterialExpressionSmoothStep_2 {
   class=MaterialExpressionSmoothStep
   inputs [
    Min {from=MaterialExpressionConstant_43, output=""}
    Max {from=MaterialExpressionOneMinus_8, output=""}
    Value {from=MaterialExpressionComponentMask_23, output=""}
   ]
   properties {}
  }
  MaterialExpressionSmoothStep_5 {
   class=MaterialExpressionSmoothStep
   inputs [
    Min {from="", output=""}
    Max {from=MaterialExpressionConstant_39, output=""}
    Value {from=MaterialExpressionComponentMask_4, output=""}
   ]
   properties {constMax=0.25}
  }
  MaterialExpressionSmoothStep_7 {
   class=MaterialExpressionSmoothStep
   inputs [
    Min {from="", output=""}
    Max {from="", output=""}
    Value {from=MaterialExpressionOneMinus_23, output=""}
   ]
   properties {}
  }
  MaterialExpressionSubtract_0 {
   class=MaterialExpressionSubtract
   inputs [
    A {from=MaterialExpressionLinearInterpolate_17}
    B {from=MaterialExpressionScalarParameter_5}
   ]
  }
  MaterialExpressionSubtract_1 {
   class=MaterialExpressionSubtract
   inputs [
    A {from=MaterialExpressionComponentMask_2}
    B {from=MaterialExpressionOneMinus_1}
   ]
  }
  MaterialExpressionSubtract_10 {
   class=MaterialExpressionSubtract
   inputs [
    A {from=MaterialExpressionComponentMask_13}
    B {from=MaterialExpressionMultiply_9}
   ]
  }
  MaterialExpressionSubtract_13 {
   class=MaterialExpressionSubtract
   inputs [
    A {from=MaterialExpressionComponentMask_16}
    B {from=MaterialExpressionMultiply_22}
   ]
  }
  MaterialExpressionSubtract_14 {
   class=MaterialExpressionSubtract
   inputs [
    A {from=MaterialExpressionComponentMask_15}
    B {from=MaterialExpressionMultiply_23}
   ]
  }
  MaterialExpressionSubtract_2 {
   class=MaterialExpressionSubtract
   inputs [
    A {from=MaterialExpressionComponentMask_9}
    B {from=MaterialExpressionSubtract_0}
   ]
  }
  MaterialExpressionSubtract_3 {
   class=MaterialExpressionSubtract
   inputs [
    A {from=MaterialExpressionNamedRerouteUsage_39}
    B {from=MaterialExpressionAppendVector_6}
   ]
  }
  MaterialExpressionTextureObjectParameter_0 {
   class=MaterialExpressionTextureObjectParameter
   inputs [
    UVs {from="", output=""}
    Apply View MipBias {from="", output=""}
   ]
   properties {parameterName=Icon, texture=/Game/UI/Foundation/Textures/Icons/Weapons/Items/T-Icon-Pickaxes-DefaultMarkIIIPickaxe-L}
  }
  MaterialExpressionTextureSample_1 {
   class=MaterialExpressionTextureSample
   inputs [
    UVs {from=MaterialExpressionReroute_4}
    Tex {from=MaterialExpressionTextureObjectParameter_0}
    Apply View MipBias {from=""}
   ]
   properties {samplerSource=SSM_Clamp_WorldGroupSettings}
  }
  MaterialExpressionTime_0 {class=MaterialExpressionTime}
  MaterialExpressionVectorParameter_0 {
   class=MaterialExpressionVectorParameter
   properties {group=Rarity Colors, parameterName=Rarity 1, sortPriority=0, defaultValue=<rgba 0.3564 0.9911 0 1>}
  }
  MaterialExpressionVectorParameter_1 {
   class=MaterialExpressionVectorParameter
   properties {group=Rarity Colors, parameterName=Rarity 2, sortPriority=1, defaultValue=<rgba 0.1195 0.521 0 1>}
  }
  MaterialExpressionVectorParameter_2 {
   class=MaterialExpressionVectorParameter
   properties {group=Rarity Colors, parameterName=Rarity 3, sortPriority=2, defaultValue=<rgba 0 0.2159 0.0015 1>}
  }
  MaterialExpressionVectorParameter_3 {
   class=MaterialExpressionVectorParameter
   properties {group=Rarity Colors, parameterName=Rarity 4, sortPriority=3, defaultValue=<rgba 0.0006 0.0782 0.0009 1>}
  }
  MaterialExpressionVectorParameter_4 {class=MaterialExpressionVectorParameter, properties {parameterName=SelectedTileConfig, defaultValue=<rgba 0.9 0.8 0.825 1>}}
  MaterialExpressionVectorParameter_5 {class=MaterialExpressionVectorParameter, properties {parameterName=DefaultTileConfig, defaultValue=<rgba 0.9 0.75 0.75 0>}}
 }
 properties {bAutomaticallySetUsageInEditor=false, blendMode=BLEND_Translucent, materialDomain=MD_UI, usedShadingModels=Default Lit}
}
