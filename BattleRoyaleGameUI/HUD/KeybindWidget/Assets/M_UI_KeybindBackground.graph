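# Translucent UI material for the background of a HUD keybind widget, drawing a rounded box with a vertical opacity gradient.
# 9-slice UVs feed a box SDF sized by KeybindSize with fixed corner roundness, which is anti-aliased into an inner mask.
# That mask multiplies a blend between FillOpacityTop and FillOpacityBottom along normalized vertical UV.
# A shared adaptive-colors function is also present, and a sine-of-UV branch feeds nothing.
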Material /BattleRoyaleGameUI/HUD/KeybindWidget/Assets/M_UI_KeybindBackground
graph {
 properties {bAutomaticallySetUsageInEditor=false, blendMode=BLEND_Translucent, materialDomain=MD_UI, usedShadingModels=Default Lit}
 nodes [
  n1 = MaterialFunctionCall() {materialFunction=&GetUserInterfaceUV}
  ComponentMask(None=Sine(None=@n1.Normalized UV) {period=2}) {g=true}
  MaterialFunctionCall!(BufferSource) {materialFunction=/BattleRoyaleGameUI/Shared/MF_UI_AdaptiveColors}
  n2 = MaterialFunctionCall!(UVs) {materialFunction=/FortUILibrary/MaterialFunctions/Basics/MF_UI_9SliceUVs}
  n3 = ScalarParameter() {defaultValue=0.75, parameterName=KeybindSize}
  n4 = Constant() {r=0.15}
  Multiply(
   A=MaterialFunctionCall(DistanceField=MaterialFunctionCall(UVs=@n2.UVs, Size=@n3, CornerRoundness=@n4) {materialFunction=&MF_UI_Box}.SDF) {materialFunction=&MF_UI_AntiAliasing}.Inner
   B=LinearInterpolate(
    A=ScalarParameter() {defaultValue=0.5, parameterName=FillOpacityTop}
    B=ScalarParameter() {defaultValue=0.25, parameterName=FillOpacityBottom}
    Alpha=NamedRerouteUsage()
   )
  )
  NamedRerouteDeclaration(None=ComponentMask(None=@n1.Normalized UV) {g=true}) {name=VerticalFade, nodeColor=<rgba 0 0 0 1>}
 ]
}
