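# Battle Royale HUD widget that lays out the keybind hint for the bottom of the quickbar, built on FortQuickbarKeybindWidget.
# A hit-test-invisible SizeBox holds the inherited MainSwitcher with three layouts: single key, key pair, and key pair with arrows.
# The pairs default to GamepadAbility1 and GamepadAbility2; arrows and the single layout's Hold label start collapsed.
# The event graph is empty, so layout switching is left to the parent class.
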WidgetBlueprint /BattleRoyaleGameUI/HUD/QuickBar/WBP_BRGameUI_QuickbarKeybindBottom
graphs [EventGraph {nodes [
   Event PreConstruct()
   Event Construct()
   Event Tick()
  ]}]
variables []
widgets {
 info {inheritedWidgetCount=11, parentClass=/Script/FortniteUI.FortQuickbarKeybindWidget, rootWidgetClass=&SizeBox, widgetCount=22}
 widgets [
  SizeBox_BottomKeybind {
   widgetClassPath=&SizeBox
   widget=SizeBox_BottomKeybind
   properties {bOverride_MinDesiredHeight=true, minDesiredHeight=45, visibility=HitTestInvisible}
  }
  MainSwitcher {
   widgetClassPath=&CommonActivatableWidgetSwitcher
   bInherited=true
   parent=SizeBox_BottomKeybind
   widget=MainSwitcher
   properties {bOverride_Cursor=true}
   slot {class=SizeBoxSlot}
  }
  HorizontalBox {widgetClassPath=&HorizontalBox, parent=MainSwitcher, widget=HorizontalBox, slot {class=WidgetSwitcherSlot}}
  LeftArrowImage {
   widgetClassPath=&Image
   bInherited=true
   parent=HorizontalBox
   widget=LeftArrowImage
   properties {
    visibility=Collapsed
    brush {
     bIsDynamicallyLoaded=false
     drawAs=Image
     imageType=NoImage
     mirroring=NoMirror
     resourceName=None
     resourceObject=/Game/UI/Foundation/Textures/Buttons/T-HUD-Slider-Arrow
     tiling=NoTile
     imageSize=<12 12>
     margin=<ltrb 0 0 0 0>
     outlineSettings {
      bUseBrushTransparency=false
      roundingType=HalfHeightRadius
      width=0
      color {colorUseRule=UseColor_Specified, specifiedColor=<rgba 0 0 0 0>}
      cornerRadii=<0 0 0 0>
     }
     tintColor {colorUseRule=UseColor_Specified, specifiedColor=<rgba 1 1 1 1>}
     uVRegion {bIsValid=false, max=<0 0>, min=<0 0>}
    }
   }
   slot {class=HorizontalBoxSlot, properties {horizontalAlignment=HAlign_Right, verticalAlignment=VAlign_Center}}
  }
  SizeBoxBottomKeybind {
   widgetClassPath=&SizeBox
   parent=HorizontalBox
   widget=SizeBoxBottomKeybind
   properties {bOverride_HeightOverride=true, bOverride_WidthOverride=true, heightOverride=40, minDesiredWidth=45, widthOverride=40}
   slot {class=HorizontalBoxSlot, properties {horizontalAlignment=HAlign_Center, verticalAlignment=VAlign_Center}}
  }
  KeybindWidget {
   widgetClassPath=/BattleRoyaleGameUI/HUD/KeybindWidget/WBP_BRGameUI_KeybindWidget
   bInherited=true
   parent=SizeBoxBottomKeybind
   widget=KeybindWidget
   slot {class=SizeBoxSlot}
  }
  HoldText {
   widgetClassPath=&CommonTextBlock
   bInherited=true
   parent=HorizontalBox
   widget=HoldText
   properties {
    style=/Game/UI/Foundation/Items/Parts/TextStyles/ItemCardPowerRatingTextStyle_XS.ItemCardPowerRatingTextStyle_XS_C
    text=Hold
    visibility=Collapsed
    font {
     bForceMonospaced=false
     bMaterialIsStencil=false
     fontMaterial=null
     fontObject=&NotoSans
     letterSpacing=0
     monospacedWidth=1
     size=15
     skewAmount=0
     typefaceFontName=NotoSans-Bold
     outlineSettings {
      bApplyOutlineToDropShadows=false
      bMiteredCorners=false
      bSeparateFillAlpha=false
      outlineMaterial=null
      outlineSize=0
      outlineColor=<rgba 0 0 0 1>
     }
    }
    shadowOffset=<0 0>
   }
   slot {class=HorizontalBoxSlot, properties {verticalAlignment=VAlign_Center, padding=<ltrb 2 0 0 0>}}
  }
  RightArrowImage {
   widgetClassPath=&Image
   bInherited=true
   parent=HorizontalBox
   widget=RightArrowImage
   properties {
    visibility=Collapsed
    brush {
     bIsDynamicallyLoaded=false
     drawAs=Image
     imageType=NoImage
     mirroring=NoMirror
     resourceName=None
     resourceObject=&T-HUD-Slider-Arrow
     tiling=NoTile
     imageSize=<12 12>
     margin=<ltrb 0 0 0 0>
     outlineSettings {
      bUseBrushTransparency=false
      roundingType=HalfHeightRadius
      width=0
      color {colorUseRule=UseColor_Specified, specifiedColor=<rgba 0 0 0 0>}
      cornerRadii=<0 0 0 0>
     }
     tintColor {colorUseRule=UseColor_Specified, specifiedColor=<rgba 1 1 1 1>}
     uVRegion {bIsValid=false, max=<0 0>, min=<0 0>}
    }
    renderTransform {angle=180, scale=<1 1>, shear=<0 0>, translation=<0 0>}
   }
   slot {class=HorizontalBoxSlot, properties {horizontalAlignment=HAlign_Left, verticalAlignment=VAlign_Center}}
  }
  HBoxDoubleKeybind {
   widgetClassPath=&HorizontalBox
   parent=MainSwitcher
   widget=HBoxDoubleKeybind
   properties {renderTransform {angle=0, scale=<1 1>, shear=<0 0>, translation=<0 5>}}
   slot {class=WidgetSwitcherSlot}
  }
  SizeBoxBottomCombo1 {
   widgetClassPath=&SizeBox
   parent=HBoxDoubleKeybind
   widget=SizeBoxBottomCombo1
   properties {bOverride_HeightOverride=true, bOverride_WidthOverride=true, heightOverride=32, widthOverride=32}
   slot {class=HorizontalBoxSlot, properties {verticalAlignment=VAlign_Center}}
  }
  KeybindWidgetCombo1 {
   widgetClassPath=&WBP_BRGameUI_KeybindWidget
   bInherited=true
   parent=SizeBoxBottomCombo1
   widget=KeybindWidgetCombo1
   properties {
    boundAction=GamepadAbility1
    keyBindTextBorder {
     bIsDynamicallyLoaded=false
     drawAs=Box
     imageType=FullColor
     mirroring=NoMirror
     resourceName=None
     resourceObject=/BattleRoyaleGameUI/HUD/KeybindWidget/Assets/M_UI_KeybindBackground
     tiling=NoTile
     imageSize=<24 24>
     margin=<ltrb 0.5 0.5 0.5 0.5>
     outlineSettings {
      bUseBrushTransparency=false
      roundingType=HalfHeightRadius
      width=0
      color {colorUseRule=UseColor_Specified, specifiedColor=<rgba 0 0 0 0>}
      cornerRadii=<0 0 0 0>
     }
     tintColor {colorUseRule=UseColor_Specified, specifiedColor=<rgba 1 1 1 1>}
     uVRegion {bIsValid=false, max=<0 0>, min=<0 0>}
    }
    keyBindTextFont {
     bForceMonospaced=false
     bMaterialIsStencil=false
     fontMaterial=null
     fontObject=&BurbankSmall
     letterSpacing=-10
     monospacedWidth=1
     size=13.5
     skewAmount=0
     typefaceFontName=Black
     outlineSettings {
      bApplyOutlineToDropShadows=false
      bMiteredCorners=false
      bSeparateFillAlpha=true
      outlineMaterial=null
      outlineSize=1
      outlineColor=<rgba 0 0 0 0>
     }
    }
   }
   slot {class=SizeBoxSlot}
  }
  Spacer_0 {
   widgetClassPath=&Spacer
   parent=HBoxDoubleKeybind
   widget=Spacer_0
   slot {class=HorizontalBoxSlot, properties {verticalAlignment=VAlign_Center}}
  }
  SizeBoxBottomCombo2 {
   widgetClassPath=&SizeBox
   parent=HBoxDoubleKeybind
   widget=SizeBoxBottomCombo2
   properties {bOverride_HeightOverride=true, bOverride_WidthOverride=true, heightOverride=32, widthOverride=32}
   slot {class=HorizontalBoxSlot, properties {verticalAlignment=VAlign_Center}}
  }
  KeybindWidgetCombo2 {
   widgetClassPath=&WBP_BRGameUI_KeybindWidget
   bInherited=true
   parent=SizeBoxBottomCombo2
   widget=KeybindWidgetCombo2
   properties {
    boundAction=GamepadAbility2
    keyBindTextBorder {
     bIsDynamicallyLoaded=false
     drawAs=Box
     imageType=FullColor
     mirroring=NoMirror
     resourceName=None
     resourceObject=&M_UI_KeybindBackground
     tiling=NoTile
     imageSize=<24 24>
     margin=<ltrb 0.5 0.5 0.5 0.5>
     outlineSettings {
      bUseBrushTransparency=false
      roundingType=HalfHeightRadius
      width=0
      color {colorUseRule=UseColor_Specified, specifiedColor=<rgba 0 0 0 0>}
      cornerRadii=<0 0 0 0>
     }
     tintColor {colorUseRule=UseColor_Specified, specifiedColor=<rgba 1 1 1 1>}
     uVRegion {bIsValid=false, max=<0 0>, min=<0 0>}
    }
    keyBindTextFont {
     bForceMonospaced=false
     bMaterialIsStencil=false
     fontMaterial=null
     fontObject=&BurbankSmall
     letterSpacing=-10
     monospacedWidth=1
     size=13.5
     skewAmount=0
     typefaceFontName=Black
     outlineSettings {
      bApplyOutlineToDropShadows=false
      bMiteredCorners=false
      bSeparateFillAlpha=true
      outlineMaterial=null
      outlineSize=1
      outlineColor=<rgba 0 0 0 0>
     }
    }
   }
   slot {class=SizeBoxSlot}
  }
  HorizontalBoxBottomDoubleBindArrows {
   widgetClassPath=&HorizontalBox
   parent=MainSwitcher
   widget=HorizontalBoxBottomDoubleBindArrows
   properties {renderTransform {angle=0, scale=<1 1>, shear=<0 0>, translation=<0 5>}}
   slot {class=WidgetSwitcherSlot}
  }
  LeftArrowImage2 {
   widgetClassPath=&Image
   bInherited=true
   parent=HorizontalBoxBottomDoubleBindArrows
   widget=LeftArrowImage2
   properties {
    visibility=Collapsed
    brush {
     bIsDynamicallyLoaded=false
     drawAs=Image
     imageType=NoImage
     mirroring=NoMirror
     resourceName=None
     resourceObject=&T-HUD-Slider-Arrow
     tiling=NoTile
     imageSize=<16 16>
     margin=<ltrb 0 0 0 0>
     outlineSettings {
      bUseBrushTransparency=false
      roundingType=HalfHeightRadius
      width=0
      color {colorUseRule=UseColor_Specified, specifiedColor=<rgba 0 0 0 0>}
      cornerRadii=<0 0 0 0>
     }
     tintColor {colorUseRule=UseColor_Specified, specifiedColor=<rgba 1 1 1 1>}
     uVRegion {bIsValid=false, max=<0 0>, min=<0 0>}
    }
   }
   slot {class=HorizontalBoxSlot, properties {horizontalAlignment=HAlign_Right, verticalAlignment=VAlign_Center}}
  }
  SizeBox_2 {
   widgetClassPath=&SizeBox
   parent=HorizontalBoxBottomDoubleBindArrows
   widget=SizeBox_2
   properties {bOverride_HeightOverride=true, bOverride_WidthOverride=true, heightOverride=32, widthOverride=32}
   slot {class=HorizontalBoxSlot, properties {verticalAlignment=VAlign_Center}}
  }
  KeybindWidgetComboArrow1 {
   widgetClassPath=&WBP_BRGameUI_KeybindWidget
   bInherited=true
   bIsVariable=true
   parent=SizeBox_2
   widget=KeybindWidgetComboArrow1
   properties {
    boundAction=GamepadAbility1
    keyBindTextBorder {
     bIsDynamicallyLoaded=false
     drawAs=Box
     imageType=FullColor
     mirroring=NoMirror
     resourceName=None
     resourceObject=&M_UI_KeybindBackground
     tiling=NoTile
     imageSize=<24 24>
     margin=<ltrb 0.5 0.5 0.5 0.5>
     outlineSettings {
      bUseBrushTransparency=false
      roundingType=HalfHeightRadius
      width=0
      color {colorUseRule=UseColor_Specified, specifiedColor=<rgba 0 0 0 0>}
      cornerRadii=<0 0 0 0>
     }
     tintColor {colorUseRule=UseColor_Specified, specifiedColor=<rgba 1 1 1 1>}
     uVRegion {bIsValid=false, max=<0 0>, min=<0 0>}
    }
    keyBindTextFont {
     bForceMonospaced=false
     bMaterialIsStencil=false
     fontMaterial=null
     fontObject=&BurbankSmall
     letterSpacing=-10
     monospacedWidth=1
     size=13.5
     skewAmount=0
     typefaceFontName=Black
     outlineSettings {
      bApplyOutlineToDropShadows=false
      bMiteredCorners=false
      bSeparateFillAlpha=true
      outlineMaterial=null
      outlineSize=1
      outlineColor=<rgba 0 0 0 0>
     }
    }
   }
   slot {class=SizeBoxSlot}
  }
  Spacer_2 {
   widgetClassPath=&Spacer
   parent=HorizontalBoxBottomDoubleBindArrows
   widget=Spacer_2
   slot {class=HorizontalBoxSlot, properties {verticalAlignment=VAlign_Center}}
  }
  SizeBox_3 {
   widgetClassPath=&SizeBox
   parent=HorizontalBoxBottomDoubleBindArrows
   widget=SizeBox_3
   properties {bOverride_HeightOverride=true, bOverride_WidthOverride=true, heightOverride=32, widthOverride=32}
   slot {class=HorizontalBoxSlot, properties {verticalAlignment=VAlign_Center}}
  }
  KeybindWidgetComboArrow2 {
   widgetClassPath=&WBP_BRGameUI_KeybindWidget
   bInherited=true
   bIsVariable=true
   parent=SizeBox_3
   widget=KeybindWidgetComboArrow2
   properties {
    boundAction=GamepadAbility2
    keyBindTextBorder {
     bIsDynamicallyLoaded=false
     drawAs=Box
     imageType=FullColor
     mirroring=NoMirror
     resourceName=None
     resourceObject=&M_UI_KeybindBackground
     tiling=NoTile
     imageSize=<24 24>
     margin=<ltrb 0.5 0.5 0.5 0.5>
     outlineSettings {
      bUseBrushTransparency=false
      roundingType=HalfHeightRadius
      width=0
      color {colorUseRule=UseColor_Specified, specifiedColor=<rgba 0 0 0 0>}
      cornerRadii=<0 0 0 0>
     }
     tintColor {colorUseRule=UseColor_Specified, specifiedColor=<rgba 1 1 1 1>}
     uVRegion {bIsValid=false, max=<0 0>, min=<0 0>}
    }
    keyBindTextFont {
     bForceMonospaced=false
     bMaterialIsStencil=false
     fontMaterial=null
     fontObject=&BurbankSmall
     letterSpacing=-10
     monospacedWidth=1
     size=13.5
     skewAmount=0
     typefaceFontName=Black
     outlineSettings {
      bApplyOutlineToDropShadows=false
      bMiteredCorners=false
      bSeparateFillAlpha=true
      outlineMaterial=null
      outlineSize=1
      outlineColor=<rgba 0 0 0 0>
     }
    }
   }
   slot {class=SizeBoxSlot}
  }
  RightArrowImage2 {
   widgetClassPath=&Image
   bInherited=true
   parent=HorizontalBoxBottomDoubleBindArrows
   widget=RightArrowImage2
   properties {
    visibility=Collapsed
    brush {
     bIsDynamicallyLoaded=false
     drawAs=Image
     imageType=NoImage
     mirroring=NoMirror
     resourceName=None
     resourceObject=&T-HUD-Slider-Arrow
     tiling=NoTile
     imageSize=<16 16>
     margin=<ltrb 0 0 0 0>
     outlineSettings {
      bUseBrushTransparency=false
      roundingType=HalfHeightRadius
      width=0
      color {colorUseRule=UseColor_Specified, specifiedColor=<rgba 0 0 0 0>}
      cornerRadii=<0 0 0 0>
     }
     tintColor {colorUseRule=UseColor_Specified, specifiedColor=<rgba 1 1 1 1>}
     uVRegion {bIsValid=false, max=<0 0>, min=<0 0>}
    }
    renderTransform {angle=180, scale=<1 1>, shear=<0 0>, translation=<0 0>}
   }
   slot {class=HorizontalBoxSlot, properties {horizontalAlignment=HAlign_Left, verticalAlignment=VAlign_Center}}
  }
 ]
}
