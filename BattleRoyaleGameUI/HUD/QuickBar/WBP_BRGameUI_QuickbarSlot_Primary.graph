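# Primary quickbar slot widget for the Battle Royale HUD, showing an item's icon, rarity, ammo or count, durability or fuel, and cooldown.
# A slot viewmodel and inventory events feed handlers that update text, visibility and dynamic material parameters, with icons loaded asynchronously.
# Selection plays material and transform animations, guarded by a cached state.
# It also handles drag-and-drop, touch selection and special styling via a stateful event.
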WidgetBlueprint /BattleRoyaleGameUI/HUD/QuickBar/WBP_BRGameUI_QuickbarSlot_Primary
animations [
 Selected {
  binding_count=8
  display_label=""
  length_seconds=0.4
  object=&self:Selected
  sequence {
   bindings [
    ReloadIndicator_Attachment {
     id=#1
     possessed_class=FortVisualAttachment
     tracks [{
       type=MovieSceneFloatTrack
       display_name=Render Opacity
       sections [{type=MovieSceneFloatSection, channels [None {type=MovieSceneScriptingFloatChannel, keys [[0,1],[3,0]]}]}]
      }]
    }
    Image_Durability {
     id=#2
     possessed_class=Image
     tracks [{
       type=MovieScene2DTransformTrack
       sections [{
         type=MovieScene2DTransformSection
         end_frame=24
         start_frame=0
         channels [
          Translation.X {type=MovieSceneScriptingFloatChannel, keys [[0,0],[6,4]]}
          Translation.Y {type=MovieSceneScriptingFloatChannel, keys [[0,0],[6,-13],[24,-10]]}
          Angle {type=MovieSceneScriptingFloatChannel, default=0}
          Scale.X {type=MovieSceneScriptingFloatChannel, default=1}
          Scale.Y {type=MovieSceneScriptingFloatChannel, default=1}
          Shear.X {type=MovieSceneScriptingFloatChannel, default=0}
          Shear.Y {type=MovieSceneScriptingFloatChannel, default=-3}
         ]
        }]
      }]
    }
    Overlay_TileContents {
     id=#3
     display_name=Overlay_0
     possessed_class=Overlay
     tracks [{
       type=MovieScene2DTransformTrack
       sections [{
         type=MovieScene2DTransformSection
         channels [
          Translation.X {type=MovieSceneScriptingFloatChannel, default=0}
          Translation.Y {type=MovieSceneScriptingFloatChannel, default=0, keys [[0,0],[12,-4]]}
          Angle {type=MovieSceneScriptingFloatChannel}
          Scale.X {type=MovieSceneScriptingFloatChannel, const [1,2]}
          Scale.Y {type=MovieSceneScriptingFloatChannel, const [1,2]}
          Shear.X {type=MovieSceneScriptingFloatChannel}
          Shear.Y {type=MovieSceneScriptingFloatChannel}
         ]
        }]
      }]
    }
    ItemDisabledIconOverlay {
     id=#4
     possessed_class=Overlay
     tracks [{
       type=MovieSceneFloatTrack
       display_name=Render Opacity
       sections [{type=MovieSceneFloatSection, channels [None {type=MovieSceneScriptingFloatChannel, keys [[0,1],[3,0]]}]}]
      }]
    }
    Image_AmmoTypeIcon {
     id=#5
     possessed_class=Image
     tracks [{
       type=MovieSceneWidgetMaterialTrack
       sections [{type=MovieSceneParameterSection, channels [IsSelected {type=MovieSceneScriptingFloatChannel, keys [[0,0],[3,1]]}]}]
      }]
    }
    Image_TileImage {
     id=#6
     possessed_class=Image
     tracks [{
       type=MovieSceneWidgetMaterialTrack
       sections [{
         type=MovieSceneParameterSection
         channels [
          IsSelected {type=MovieSceneScriptingFloatChannel, keys [[0,0],[6,1.5],[24,1]]}
          SelectedTileConfigR {type=MovieSceneScriptingFloatChannel, keys [[0,0.9],[11,1]]}
          SelectedTileConfigG {type=MovieSceneScriptingFloatChannel, keys [[0,0.8],[11,0.9]]}
          SelectedTileConfigB {type=MovieSceneScriptingFloatChannel, keys [[0,0.825],[11,0.925]]}
          SelectedTileConfigA {type=MovieSceneScriptingFloatChannel, const [1,1]}
         ]
        }]
      }]
    }
    HBox_ItemQuantityAndAmmoType {
     id=#7
     possessed_class=VerticalBox
     tracks [{
       type=MovieScene2DTransformTrack
       sections [{
         type=MovieScene2DTransformSection
         channels [
          Translation.X {type=MovieSceneScriptingFloatChannel, default=0, keys [[0,0],[9,2]]}
          Translation.Y {type=MovieSceneScriptingFloatChannel, keys [[0,0],[9,14]]}
          Angle {type=MovieSceneScriptingFloatChannel, default=0}
          Scale.X {type=MovieSceneScriptingFloatChannel, default=1}
          Scale.Y {type=MovieSceneScriptingFloatChannel, default=1}
          Shear.X {type=MovieSceneScriptingFloatChannel, default=0}
          Shear.Y {type=MovieSceneScriptingFloatChannel, default=0}
         ]
        }]
      }]
    }
    Text_ItemQuantity {
     id=#8
     possessed_class=CommonTextBlock
     tracks [
      {
       type=MovieSceneWidgetMaterialTrack
       display_name=Font.FontMaterial
       sections [{type=MovieSceneParameterSection, channels [IsSelected {type=MovieSceneScriptingFloatChannel, keys [[0,0],[3,1]]}]}]
      }
      {
       type=MovieSceneWidgetMaterialTrack
       display_name=Font.Outline Settings.OutlineMaterial
       sections [{type=MovieSceneParameterSection, channels [AnimOpacity {type=MovieSceneScriptingFloatChannel, keys [[0,1],[3,0]]}]}]
      }
     ]
    }
   ]
   display_rate {numerator=60}
  }
 }
 Unselected {
  binding_count=8
  length_seconds=0.1167
  object=&self:Unselected
  sequence {
   bindings [
    ItemDisabledIconOverlay {
     id=#9
     possessed_class=Overlay
     tracks [{
       type=MovieSceneFloatTrack
       display_name=Render Opacity
       sections [{type=MovieSceneFloatSection, channels [None {type=MovieSceneScriptingFloatChannel, keys [[0,0],[5,1]]}]}]
      }]
    }
    Image_TileImage {
     id=#10
     possessed_class=Image
     tracks [{
       type=MovieSceneWidgetMaterialTrack
       sections [{
         type=MovieSceneParameterSection
         channels [
          IsSelected {type=MovieSceneScriptingFloatChannel, keys [[0,1],[7,0]]}
          SelectedTileConfigR {type=MovieSceneScriptingFloatChannel, keys [[0,1],[7,0.9]]}
          SelectedTileConfigG {type=MovieSceneScriptingFloatChannel, keys [[0,0.9],[7,0.8]]}
          SelectedTileConfigB {type=MovieSceneScriptingFloatChannel, keys [[0,0.925],[7,0.825]]}
          SelectedTileConfigA {type=MovieSceneScriptingFloatChannel, const [1,1]}
         ]
        }]
      }]
    }
    Image_Durability {
     id=#2
     possessed_class=Image
     tracks [{
       type=MovieScene2DTransformTrack
       sections [{
         type=MovieScene2DTransformSection
         end_frame=7
         start_frame=0
         channels [
          Translation.X {type=MovieSceneScriptingFloatChannel, keys [[0,4],[7,0]]}
          Translation.Y {type=MovieSceneScriptingFloatChannel, keys [[0,-10],[7,0]]}
          Angle {type=MovieSceneScriptingFloatChannel, default=0}
          Scale.X {type=MovieSceneScriptingFloatChannel, default=1}
          Scale.Y {type=MovieSceneScriptingFloatChannel, default=1}
          Shear.X {type=MovieSceneScriptingFloatChannel, default=0}
          Shear.Y {type=MovieSceneScriptingFloatChannel, default=-3}
         ]
        }]
      }]
    }
    ReloadIndicator_Attachment {
     id=#11
     possessed_class=FortVisualAttachment
     tracks [{
       type=MovieSceneFloatTrack
       display_name=Render Opacity
       sections [{type=MovieSceneFloatSection, channels [None {type=MovieSceneScriptingFloatChannel, keys [[0,0],[5,1]]}]}]
      }]
    }
    Overlay_TileContents {
     id=#3
     display_name=Overlay_0
     possessed_class=Overlay
     tracks [{
       type=MovieScene2DTransformTrack
       sections [{
         type=MovieScene2DTransformSection
         channels [
          Translation.X {type=MovieSceneScriptingFloatChannel, default=0}
          Translation.Y {type=MovieSceneScriptingFloatChannel, default=0, keys [[0,-4],[7,0]]}
          Angle {type=MovieSceneScriptingFloatChannel}
          Scale.X {type=MovieSceneScriptingFloatChannel, const [1,2]}
          Scale.Y {type=MovieSceneScriptingFloatChannel, const [1,2]}
          Shear.X {type=MovieSceneScriptingFloatChannel}
          Shear.Y {type=MovieSceneScriptingFloatChannel}
         ]
        }]
      }]
    }
    Image_AmmoTypeIcon {
     id=#5
     possessed_class=Image
     tracks [{
       type=MovieSceneWidgetMaterialTrack
       sections [{type=MovieSceneParameterSection, channels [IsSelected {type=MovieSceneScriptingFloatChannel, keys [[0,1],[7,0]]}]}]
      }]
    }
    HBox_ItemQuantityAndAmmoType {
     id=#7
     possessed_class=VerticalBox
     tracks [{
       type=MovieScene2DTransformTrack
       sections [{
         type=MovieScene2DTransformSection
         channels [
          Translation.X {type=MovieSceneScriptingFloatChannel, default=0, keys [[0,2],[7,0]]}
          Translation.Y {type=MovieSceneScriptingFloatChannel, keys [[0,14],[7,0]]}
          Angle {type=MovieSceneScriptingFloatChannel, default=0}
          Scale.X {type=MovieSceneScriptingFloatChannel, default=1}
          Scale.Y {type=MovieSceneScriptingFloatChannel, default=1}
          Shear.X {type=MovieSceneScriptingFloatChannel, default=0}
          Shear.Y {type=MovieSceneScriptingFloatChannel, default=0}
         ]
        }]
      }]
    }
    Text_ItemQuantity {
     id=#8
     possessed_class=CommonTextBlock
     tracks [
      {
       type=MovieSceneWidgetMaterialTrack
       display_name=Font.FontMaterial
       sections [{type=MovieSceneParameterSection, channels [IsSelected {type=MovieSceneScriptingFloatChannel, keys [[0,1],[7,0]]}]}]
      }
      {
       type=MovieSceneWidgetMaterialTrack
       display_name=Font.Outline Settings.OutlineMaterial
       sections [{type=MovieSceneParameterSection, channels [AnimOpacity {type=MovieSceneScriptingFloatChannel, keys [[0,0],[7,1]]}]}]
      }
     ]
    }
   ]
   display_rate {numerator=60}
  }
 }
 CooldownComplete {
  binding_count=1
  length_seconds=1
  object=&self:CooldownComplete
  sequence {bindings [Image_TileImage {
     id=#12
     possessed_class=Image
     tracks [{
       type=MovieSceneWidgetMaterialTrack
       sections [{type=MovieSceneParameterSection, channels [AnimCoolDownComplete {type=MovieSceneScriptingFloatChannel, keys [[0,0],[1,1],[20,0]]}]}]
      }]
    }]}
 }
 HideFontOutline {
  binding_count=1
  length_seconds=1
  object=&self:HideFontOutline
  sequence {bindings [Text_ItemQuantity {
     id=#13
     possessed_class=CommonTextBlock
     tracks [{
       type=MovieSceneWidgetMaterialTrack
       display_name=Font.Outline Settings.OutlineMaterial
       sections [{type=MovieSceneParameterSection, channels [Hide {type=MovieSceneScriptingFloatChannel, keys [[0,0],[20,1]]}]}]
      }]
    }]}
 }
]
graphs [
 EventGraph {nodes [
   n1 = Handle OnQuickbarContentsChanged!() -> (OutputDelegate, then, QuickbarIndex:EFortQuickBars Enum=Primary, ChangedSlots:Array of Integers) {is=title K2Node_CustomEvent}
   > RefreshItemIfSlotChanged!(
    execute
    self:Fort Quick Bar Slot Base Object Reference
    InQuickBar:EFortQuickBars Enum=@n1.QuickbarIndex
    InSlot:Array of Integers=@n1.ChangedSlots
   ) -> (then)
   > SetHarvestingToolVisualParameters()
   > SetItemType()
   n2 = HandleItemIconChanged!() -> (OutputDelegate, then, Texture:Texture 2D Soft Object Reference) {is=title K2Node_CustomEvent}
   > n3 = Async Load Asset(Asset=@n2.Texture)
   n4 = Get Image_TileImage() -> (Image_TileImage:Image Object Reference)
   > n5 = GetDynamicMaterial(self=@n4)
   > SetScalarParameterValue(self=@n5, ParameterName=IsIconLoaded, Value=0)
   n6 = GetDynamicMaterial(execute=@n3.Completed, self=@n4)
   > SetScalarParameterValue(self=@n6, ParameterName=IsIconLoaded, Value=1)
   > n7 = Cast To Texture2D(Object=@n3)
   > SetTextureParameterValue(self=@n6, ParameterName=Icon, Value=@n7)
   Event MaximizeSlot!() -> (OutputDelegate, then, bShouldSkipAnimation:bool) {is=title K2Node_Event}
   > Resize!(execute, self:Fort Quick Bar Slot Base Object Reference, CardSize:EFortItemCardSize Enum=S) -> (then)
   Event MinimizeSlot!() -> (OutputDelegate, then, bShouldSkipAnimation:bool) {is=title K2Node_Event}
   > Resize(CardSize=XS)
   Event PreConstruct()
   > UpdateItemBoxSize()
   > UpdateKeyBindingText(self:Fort Quick Bar Slot Base Object Reference)
   Event Construct()
   n8 = Get FortInventoryContext!() -> (ReturnValue:Fort Inventory Context Object Reference) {is=title K2Node_GetBlueprintContextEx}
   > Bind Event to On Quickbar Contents Changed(self:Fort Inventory Context Object Reference=@n8, Delegate=@n1) {is=title K2Node_AddDelegate}
   > InitializeInvalidation()
   > RefreshItem(self:Fort Quick Bar Slot Base Object Reference)
   > SetHarvestingToolVisualParameters()
   > SetItemType()
   > RefreshCooldownTimeVisibility()
   > n9 = FindParentWidgetOfType!(execute, StartingWidget:Widget Object Reference=Self-Reference(), Type:Widget Class Reference=&HorizontalBox) -> (then, ReturnValue:Horizontal Box Object Reference)
   > Set ParentHB(ParentHB:Horizontal Box Object Reference=@n9) -> (Output_Get:Horizontal Box Object Reference)
   > n10 = StartListeningToStatefulEvent!(
    execute:Exec
    Target:Gameplay Event Router Component Object Reference=GetGlobalEventRouter()
    EventType:Script Struct Object Reference=/BattleRoyaleGameUI/Data/Struct_BRGameUI_QuickbarMaterialOverride
    Context:Object Reference=GetOwningPlayer()
    EventBubblingRule:EEventBubblingRule Enum
   ) -> (
    then:Exec
    AsyncTaskProxy:Async Action Start Listening To Stateful Event Object Reference
    OnEventReceived:Exec
    OnSavedEventStateExists:Exec
    OnEventStateCleared:Exec
    Payload:Struct BRGame UI Quickbar Material Override Structure
   ) {is=title K2Node_StartListeningToStatefulEventAsyncAction}
   n11 = Break Struct BRGame UI Quickbar Material Override!(Struct_BRGameUI_QuickbarMaterialOverride:"Struct BRGame UI Quickbar Material Override Structure (by ref)"=@n10.Payload) -> (
    UseSpecialStyling_2_#14:bool=~
    QuickbarType_8_#15:EFortQuickBars Enum
    ItemSlot_5_#16:int
   ) {is=title K2Node_BreakStruct}
   Branch(
    execute=[@n10.OnEventReceived, @n10.OnSavedEventStateExists]
    Condition="Equal (Integer)"(A=Get Slot Num() -> (Slot Num:int=0), B=@n11.ItemSlot_5_#16)
   )
   > Set Use Special Styling(Use Special Styling:bool=@n11) -> (Output_Get:bool)
   n12 = Event SetSlotSelected!() -> (OutputDelegate, then, bSelected:bool) {is=title K2Node_Event}
   > n13 = Set bIsSelected(bIsSelected:bool=@n12.bSelected) -> (Output_Get:bool)
   > SetIsSelected!(
    execute
    self:Athena Quick Bar Native Cell Object Reference=Get NativeCell() -> (NativeCell:Athena Quick Bar Native Cell Object Reference)
    bInSelected:bool=@n13
   ) -> (then)
   > SetRenderTranslation(
    self=Get OverlayExtensionWidgets() -> (OverlayExtensionWidgets:Overlay Object Reference)
    Translation=Select(Option 0:Vector 2D Structure, Option 1:Vector 2D Structure=(X=0,Y=-13.5), Index=@n13) -> (ReturnValue:Vector 2D Structure)
   )
   > UpdateKeyBindingText()
   > UpdateKeyBindingVisibility(self:Fort Quick Bar Slot Base Object Reference)
   > UpdateGameplayTags()
   > n14 = Branch(Condition=@n13)
   > n15 = Branch(Condition=Get DoNotPlaySelectionAnimation() -> (DoNotPlaySelectionAnimation:bool))
   n16 = Get bCacheIsSelected() -> (bCacheIsSelected:bool)
   n17 = Branch(execute=@n15.else, Condition=@n16)
   Set bCacheIsSelected(execute=@n17.else, bCacheIsSelected:bool=true) -> (Output_Get:bool)
   n18 = Get Unselected() -> (Unselected:Widget Animation Object Reference)
   > n19 = Branch(Condition=IsAnimationPlaying(InAnimation=@n18))
   > n20 = StopAnimation(InAnimation=@n18)
   n21 = Get Selected() -> (Selected:Widget Animation Object Reference)
   PlayAnimation(execute=[@n20.then, @n19.else], InAnimation=@n21)
   n22 = Branch(execute=@n14.else, Condition=Get DoNotPlaySelectionAnimation())
   Branch(execute=@n22.else, Condition=@n16)
   > Set bCacheIsSelected(bCacheIsSelected=false)
   > n23 = Branch(Condition=IsAnimationPlaying(InAnimation=@n21))
   > n24 = StopAnimation(InAnimation=@n21)
   PlayAnimation(execute=[@n24.then, @n23.else], InAnimation=@n18)
   Event OnRefreshItem() {is=title K2Node_Event}
   > UpdateItemBoxSize()
   > UpdateGameplayTags()
   > SetItemType()
   n25 = Event OnExtensionWidgetAdded!() -> (OutputDelegate, then, ExtensionSlot:EUIExtensionSlot Enum=Invalid, ExtensionWidget:User Widget Object Reference) {is=title K2Node_Event}
   > n26 = Cast To FortWeaponFuelGaugeIndicator!(execute, Object=@n25.ExtensionWidget) -> (then, CastFailed, AsFort Weapon Fuel Gauge Indicator:Fort Weapon Fuel Gauge Indicator Object Reference) {is=title K2Node_DynamicCast}
   n27 = AddChildToOverlay!(execute=@n26.CastFailed, self:Overlay Object Reference=Get OverlayExtensionWidgets(), Content:Widget Object Reference=@n25.ExtensionWidget) -> (then, ReturnValue:Overlay Slot Object Reference)
   > SetHorizontalAlignment(self=@n27, InHorizontalAlignment=HAlign_Fill)
   > SetVerticalAlignment(self=@n27, InVerticalAlignment=VAlign_Fill)
   Event Destruct()
   > Unbind Event from On Quickbar Contents Changed(self:Fort Inventory Context Object Reference=@n8, Delegate=@n1) {is=title K2Node_RemoveDelegate}
   > ClearChildren(self=Get OverlayExtensionWidgets())
   Event OnInitialized()
   > n28 = Construct Object from Class(Class=/Script/BattleRoyaleGameUI.InstituteChirpQuickbarSlotViewModel) -> (ReturnValue:Institute Chirp - HUD Controller Quickbar Slot Viewmodel Object Reference)
   > n29 = Set InstituteChirpQuickbarSlotViewModel(InstituteChirpQuickbarSlotViewModel:Institute Chirp - HUD Controller Quickbar Slot Viewmodel Object Reference=@n28) -> (Output_Get:Institute Chirp - HUD Controller Quickbar Slot Viewmodel Object Reference)
   > Set CooldownTypesSupported!(
    execute
    CooldownTypesSupported:Array of EFortItemCooldownType Enums=Get CooldownTypesSupported() -> (CooldownTypesSupported:Array of EFortItemCooldownType Enums=None)
    self:Item Cooldown Viewmodel Object Reference=Get ItemCooldownViewmodel!(self:Device - HUD Controller Quickbar Slot Viewmodel Object Reference=@n29) -> (ItemCooldownViewmodel:Item Cooldown Viewmodel Object Reference)
   ) -> (then, Output_Get:Array of EFortItemCooldownType Enums=None)
   > InitializeDataModel!(execute, self, SlotNum:int=Get Slot Num() -> (Slot Num=0), Quick Bar Type:EFortQuickBars Enum=Primary) -> (then)
  ]}
 HandleAmmoTypeIconChanged {nodes [
   n1 = HandleAmmoTypeIconChanged!() -> (then, AmmoTypeIcon:Slate Brush Structure) {is=title K2Node_FunctionEntry}
   n2 = GetBrushResourceAsTexture2D!(Brush:"Slate Brush Structure (by ref)"=@n1) -> (ReturnValue:Texture 2D Object Reference)
   > n3 = Is Valid(InputObject=@n2)
   n4 = Get Image_AmmoTypeIcon() -> (Image_AmmoTypeIcon:Image Object Reference)
   SetVisibility(execute=@n3.Is Not Valid, self=@n4, InVisibility=Hidden)
   SetVisibility(execute=@n3.Is Valid, self=@n4, InVisibility=HitTestInvisible)
   > n5 = GetDynamicMaterial(self=Get Image_AmmoTypeIcon())
   > SetTextureParameterValue(self=@n5, ParameterName=Icon, Value=@n2)
  ]}
 HandleCooldownRemainingChanged {nodes [
   HandleCooldownRemainingChanged!() -> (then, CooldownRemaining:double=~) {is=title K2Node_FunctionEntry}
   n1 = GetCooldownPercentage!(self:Item Cooldown Viewmodel Object Reference=Get ItemCooldownViewmodel(self=Get InstituteChirpQuickbarSlotViewModel() -> (InstituteChirpQuickbarSlotViewModel:Institute Chirp - HUD Controller Quickbar Slot Viewmodel Object Reference))) -> (ReturnValue:float)
   > Set CooldownPercent(CooldownPercent:double=SelectFloat(A=@n1, B=1, bPickA="float > float"(A=@n1))) -> (Output_Get:double)
   > n2 = GetDynamicMaterial(self=Get Image_TileImage())
   > SetScalarParameterValue(self=@n2, ParameterName=CooldownProgress, Value=Get CooldownPercent() -> (CooldownPercent:double))
   n3 = Get CooldownRemaining() -> (CooldownRemaining:double)
   > "SetText (Text)"(
    self=Get Text_CooldownTime() -> (Text_CooldownTime:Common Text Object Reference)
    InText="To Text (Integer)"(Value=Ceil(A=@n3) {is=title K2Node_CallFunction})
   )
   > n4 = Delta Gate Wildcard!(
    execute:Exec
    Variable:"Boolean (by ref)"=Get bIsCooldownActive() -> (bIsCooldownActive:bool)
    New Value:bool="float < float"(A=Get CooldownPercent(), B=1)
   ) -> (Success:Exec, Updated Variable:"Boolean (by ref)", Previous Variable:bool=~) {is=title K2Node_MacroInstance}
   > RefreshCooldownTimeVisibility()
   > n5 = Branch(Condition=@n4)
   n6 = Get CurrentCooldownType!(self:Item Cooldown Viewmodel Object Reference=Get ItemCooldownViewmodel(self=Get InstituteChirpQuickbarSlotViewModel())) -> (CurrentCooldownType:EFortItemCooldownType Enum=None)
   Branch(
    execute=@n5.else
    Condition=OR Boolean(
     A="Not Equal (Enum)"!(A:EFortItemCooldownType Enum=@n6, B:EFortItemCooldownType Enum=WeaponReloading) -> (ReturnValue:bool)
     B=Get bCacheIsSelected()
    )
   )
   n7 = Get CooldownComplete() -> (CooldownComplete:Widget Animation Object Reference)
   > PlayAnimation(
    InAnimation=@n7
    StartAtTime=Select(Option 0:float=GetEndTime(self=@n7), Option 1:float=0, Index=Get bCacheHasItem() -> (bCacheHasItem:bool)) -> (ReturnValue:float)
   )
  ]}
 HandleCurrentCooldownTypeChanged {nodes [
   HandleCurrentCooldownTypeChanged!() -> (then, CooldownType:EFortItemCooldownType Enum=None) {is=title K2Node_FunctionEntry}
   > RefreshCooldownTimeVisibility()
  ]}
 HandleFuelChargeCountChanged {nodes [
   n1 = HandleFuelChargeCountChanged!() -> (then, FuelChargeCount:double=~) {is=title K2Node_FunctionEntry}
   > n2 = Set CacheFuelChargeCount(CacheFuelChargeCount:int=Truncate(A=@n1)) -> (Output_Get:int=0)
   > HandleItemCountChanged!(execute, self, ItemCount:int=@n2) -> (then)
  ]}
 HandleFuelPercentageChanged {nodes [
   n1 = HandleFuelPercentageChanged!() -> (then, FuelPercentage:double=~) {is=title K2Node_FunctionEntry}
   > Set CacheFuelPercentage(CacheFuelPercentage:double=@n1) -> (Output_Get:double)
   > Branch(Condition="Equal (Enum)"(
     A:EFortFuelGadgetVisualType Enum=Get CacheFuelVisualType() -> (CacheFuelVisualType:EFortFuelGadgetVisualType Enum=FuelMeter)
     B:EFortFuelGadgetVisualType Enum=FuelMeter
    ))
   > RefreshDurabilityOrFuelMeter()
  ]}
 HandleFuelVisualTypeChanged {nodes [
   n1 = HandleFuelVisualTypeChanged!() -> (then, FuelVisualType:EFortFuelGadgetVisualType Enum=FuelMeter) {is=title K2Node_FunctionEntry}
   > Set CacheFuelVisualType(CacheFuelVisualType:EFortFuelGadgetVisualType Enum=@n1) -> (Output_Get:EFortFuelGadgetVisualType Enum=FuelMeter)
   > RefreshDurabilityOrFuelVisibility()
  ]}
 HandleHasItemChanged {nodes [
   n1 = HandleHasItemChanged!() -> (then, HasItem:bool=~) {is=title K2Node_FunctionEntry}
   > n2 = Set bCacheHasItem(bCacheHasItem:bool=@n1) -> (Output_Get:bool)
   > n3 = GetDynamicMaterial(self=Get Image_TileImage())
   > SetScalarParameterValue(self=@n3, ParameterName=IsEmpty, Value="To Float (Boolean)"(InBool=NOT Boolean(A=@n2)))
   > HandleItemQuantityAndAmmoTypeIconVisibility!(execute, self, IsVisible:bool=false) -> (then)
   > RefreshCooldownTimeVisibility()
   > RefreshDurabilityOrFuelVisibility()
   > SetOutOfAmmo!(execute, self, OutOfAmmo:bool=false) -> (then)
  ]}
 HandleItemCountChanged {nodes [
   HandleItemCountChanged!() -> (then, ItemCount:int) {is=title K2Node_FunctionEntry}
   > Branch(Condition=IsVisible(self=Get HBox_ItemQuantityAndAmmoType() -> (HBox_ItemQuantityAndAmmoType:Horizontal Box Object Reference)))
   n1 = Get InstituteChirpQuickbarSlotViewModel()
   > n2 = Set L_TotalAmmo(L_TotalAmmo:int=Total Ammo!(self:Device - HUD Controller Quickbar Slot Viewmodel Object Reference=@n1) -> (ReturnValue:int=0)) -> (Output_Get:int=0)
   n3 = "To Text (Integer)"(Value=@n2)
   n4 = SelectInt(
    A=Stack Count!(self:Device - HUD Controller Quickbar Slot Viewmodel Object Reference=@n1) -> (ReturnValue:int=0)
    B=Get CacheFuelChargeCount() -> (CacheFuelChargeCount:int=0)
    bPickA="Not Equal (Enum)"(A:EFortFuelGadgetVisualType Enum=Get CacheFuelVisualType() -> (CacheFuelVisualType=FuelMeter), B:EFortFuelGadgetVisualType Enum=ChargeText)
   )
   n5 = "To Text (Integer)"(Value=@n4)
   n6 = "integer >= integer"(A=@n2)
   n7 = "integer < integer"(A=Get ItemCount() -> (ItemCount:int=0), B=999999)
   n8 = "integer < integer"(A=@n4, B=999999)
   > "SetText (Text)"(
    self=Get Text_ItemQuantity() -> (Text_ItemQuantity:Common Text Object Reference)
    InText=SelectText(A=SelectText(A=@n3, B=@n5, bPickA=@n6), B=GetInfinitySymbolText!() -> (ReturnValue:"Text (by ref)"), bPickA=AND Boolean(A=@n8, B=@n7))
   )
   n9 = "Equal (Enum)"(A:EFortItemType Enum=Get ItemType() -> (ItemType:EFortItemType Enum=WorldItem), B:EFortItemType Enum=WeaponRanged)
   n10 = "integer <= integer"(A=Get MaxStackSize() -> (MaxStackSize:int=0), B=1)
   n11 = Get InstituteChirpQuickbarSlotViewModel()
   n12 = Loaded Ammo!(self:Device - HUD Controller Quickbar Slot Viewmodel Object Reference=@n11) -> (ReturnValue:int=0)
   n13 = Max Clip Size!(self:Device - HUD Controller Quickbar Slot Viewmodel Object Reference=@n11) -> (ReturnValue:int=0)
   > n14 = Branch(Condition=AND Boolean!(A=AND Boolean!(A=@n9, B=@n10) -> (ReturnValue), B="integer <= integer"(A=@n12), C:bool="integer > integer"(A=@n13)) -> (ReturnValue))
   > SetOutOfAmmo(OutOfAmmo="integer <= integer"(A=Get L_TotalAmmo() -> (L_TotalAmmo:int=0)))
   > SetIsClipEmpty!(execute, self, ClipIsEmpty:bool="integer > integer"(A=Get L_TotalAmmo() -> (L_TotalAmmo=0))) -> (then)
   SetOutOfAmmo(execute=@n14.else, OutOfAmmo=false)
   > SetIsClipEmpty(ClipIsEmpty=false)
  ]}
 HandleItemQuantityAndAmmoTypeIconVisibility {nodes [
   HandleItemQuantityAndAmmoTypeIconVisibility!() -> (then, IsVisible:bool=~) {is=title K2Node_FunctionEntry}
   n1 = NOT Boolean(A=Get bShouldShowAmmoAsPercent() -> (bShouldShowAmmoAsPercent:bool))
   n2 = Get InstituteChirpQuickbarSlotViewModel()
   n3 = NOT Boolean(A=ItemHasDurability!(self:Item Durability Viewmodel Object Reference=Get ItemDurabilityViewModel!(self:Device - HUD Controller Quickbar Slot Viewmodel Object Reference=@n2) -> (ItemDurabilityViewModel:Item Durability Viewmodel Object Reference)) -> (ReturnValue:bool))
   n4 = NOT Boolean(A=Is Harvest Tool!(self:Device - HUD Controller Quickbar Slot Viewmodel Object Reference=@n2) -> (ReturnValue:bool))
   n5 = "Not Equal (Enum)"(A=Get CacheFuelVisualType() -> (CacheFuelVisualType=FuelMeter), B=FuelMeter)
   n6 = Get bWeaponUsesAmmo() -> (bWeaponUsesAmmo:bool)
   n7 = Get bCacheHasItem()
   > SetVisibility(
    self=Get HBox_ItemQuantityAndAmmoType()
    InVisibility=Select(
     Option 0:ESlateVisibility Enum=Hidden
     Option 1:ESlateVisibility Enum=HitTestInvisible
     Index=AND Boolean!(A=@n7, B=@n1, C:bool=@n3, D:bool=@n4, E:bool=@n5, F:bool=@n6) -> (ReturnValue)
    ) -> (ReturnValue:ESlateVisibility Enum)
   )
  ]}
 HandleItemRarityChanged {nodes [
   n1 = HandleItemRarityChanged!() -> (then, ItemRarity:EFortRarity Enum=Common) {is=title K2Node_FunctionEntry}
   > n2 = GetRarityData!(execute, Rarity:EFortRarity Enum=@n1) -> (
    then
    ReturnValue_Name:Text
    ReturnValue_Color1:Linear Color Structure
    ReturnValue_Color2:Linear Color Structure
    ReturnValue_Color3:Linear Color Structure
    ReturnValue_Color4:Linear Color Structure
    ReturnValue_Color5:Linear Color Structure
    ReturnValue_Radius:float=~
    ReturnValue_Falloff:float=~
    ReturnValue_Brightness:float=~
    ReturnValue_Roughness:float=~
    ReturnValue_Glow:float=~
   )
   > n3 = Sequence!(execute) -> (then_0, then_1, then_2:Exec, then_3:Exec, then_4:Exec)
   n4 = GetDynamicMaterial(execute=@n3.then_4, self=Get Image_Durability() -> (Image_Durability:Image Object Reference))
   > SetVectorParameterValue(self=@n4, ParameterName=BackingColor, Value=@n2.ReturnValue_Color4)
   > SetVectorParameterValue(self=@n4, ParameterName=FillColor, Value=@n2.ReturnValue_Color1)
   n5 = Branch(execute=@n3.then_0, Condition=Get Use Special Styling() -> (Use Special Styling:bool))
   > n6 = GetDynamicMaterial(self=Get Image_TileImage())
   > SetScalarParameterValue(self=@n6, ParameterName=IsSpecialStyling, Value=1)
   n7 = GetDynamicMaterial(execute=@n5.else, self=Get Image_TileImage())
   > SetVectorParameterValue(self=@n7, ParameterName=Rarity 1, Value=@n2.ReturnValue_Color1)
   > SetVectorParameterValue(self=@n7, ParameterName=Rarity 2, Value=@n2.ReturnValue_Color2)
   > SetVectorParameterValue(self=@n7, ParameterName=Rarity 3, Value=@n2.ReturnValue_Color3)
   > SetVectorParameterValue(self=@n7, ParameterName=Rarity 4, Value=@n2.ReturnValue_Color4)
   > SetScalarParameterValue(self=@n7, ParameterName=IsSpecialStyling, Value=0)
   n8 = GetDynamicMaterial(execute=@n3.then_1, self=Get Image_AmmoTypeIcon())
   > SetVectorParameterValue(self=@n8, ParameterName=UnselectedColor, Value=@n2.ReturnValue_Color1)
   > SetVectorParameterValue(self=@n8, ParameterName=ShadowColor, Value=@n2.ReturnValue_Color4)
   > SetVectorParameterValue(self=@n8, ParameterName=ClipEmptyColor, Value=@n2.ReturnValue_Color2)
   > n9 = GetDynamicMaterial(self=Get Image_ReloadIndicator() -> (Image_ReloadIndicator:Image Object Reference))
   > SetVectorParameterValue(self=@n9, ParameterName=ClipEmptyColor, Value=@n2.ReturnValue_Color2)
   n10 = GetDynamicFontMaterial!(execute=@n3.then_2, self:Text Object Reference=Get Text_ItemQuantity()) -> (then, ReturnValue:Material Instance Dynamic Object Reference)
   > SetVectorParameterValue(self=@n10, ParameterName=UnselectedColor, Value=@n2.ReturnValue_Color1)
   > SetVectorParameterValue(self=@n10, ParameterName=ClipEmptyColor, Value=@n2.ReturnValue_Color2)
   > n11 = GetDynamicOutlineMaterial!(execute, self:Text Object Reference=Get Text_ItemQuantity()) -> (then, ReturnValue:Material Instance Dynamic Object Reference)
   > SetVectorParameterValue(self=@n11, ParameterName=Color, Value=@n2.ReturnValue_Color4)
   n12 = GetDynamicOutlineMaterial(execute=@n3.then_3, self=Get Text_CooldownTime())
   > SetVectorParameterValue(self=@n12, ParameterName=Color, Value=@n2.ReturnValue_Color4)
   > n13 = GetDynamicMaterial(self=Get Image_ReloadIndicator())
   > SetVectorParameterValue(self=@n13, ParameterName=CustomOutlineColor, Value=@n2.ReturnValue_Color4)
  ]}
 HandleLoadedAmmoChanged {nodes [
   n1 = HandleLoadedAmmoChanged!() -> (then, LoadedAmmo:int) {is=title K2Node_FunctionEntry}
   > Set CacheLoadedAmmo(CacheLoadedAmmo:int=@n1) -> (Output_Get:int=0)
   > Branch(Condition=Get bShouldShowAmmoAsPercent())
   > RefreshDurabilityOrFuelMeter()
  ]}
 HandleNormalizedDurabilityChanged {nodes [
   n1 = HandleNormalizedDurabilityChanged!() -> (then, NormalizedDurability:double=~) {is=title K2Node_FunctionEntry}
   > Set CacheNormalizedDurability(CacheNormalizedDurability:double=@n1) -> (Output_Get:double)
   > Branch(Condition=Get bCacheShowDurability() -> (bCacheShowDurability:bool))
   > RefreshDurabilityOrFuelMeter()
  ]}
 HandleNumSlotsTakenChanged {nodes [
   n1 = HandleNumSlotsTakenChanged!() -> (then, ItemCount:int) {is=title K2Node_FunctionEntry}
   > SetWidthOverride(
    self=Get SizeBox_TileContents() -> (SizeBox_TileContents:Size Box Object Reference)
    InWidthOverride=float * float(A:int=@n1, B=Get SlotWidth() -> (SlotWidth:double)) {is=title K2Node_PromotableOperator}
   )
  ]}
 HandleShouldShowCooldownChanged {nodes [
   HandleShouldShowCooldownChanged!() -> (then, ShowCooldown:bool=~) {is=title K2Node_FunctionEntry}
   > RefreshCooldownTimeVisibility()
  ]}
 HandleShouldShowDurabilityChanged {nodes [
   n1 = HandleShouldShowDurabilityChanged!() -> (then, ShowDurability:bool=~) {is=title K2Node_FunctionEntry}
   > Set bCacheShowDurability(bCacheShowDurability:bool=@n1) -> (Output_Get:bool)
   > RefreshDurabilityOrFuelVisibility()
  ]}
 InitializeDataModel {nodes [
   n1 = InitializeDataModel!() -> (then, SlotNum:int, Quick Bar Type:EFortQuickBars Enum=Primary) {is=title K2Node_FunctionEntry}
   > Branch(Condition="integer >= integer"(A=@n1))
   > n2 = Set Slot Num(Slot Num:int=@n1) -> (Output_Get:int=0)
   n3 = Get InstituteChirpQuickbarSlotViewModel()
   > Is Valid(InputObject=@n3)
   > InitializeWithType!(
    execute
    self:Institute Chirp - HUD Controller Quickbar Slot Viewmodel Object Reference=@n3
    InPlayerController:Player Controller Object Reference=GetOwningPlayer()
    InQuickBarType:EFortQuickBars Enum=@n1.Quick Bar Type
    InSlot:int=@n2
   ) -> (then)
  ]}
 InitializeInvalidation {nodes [InitializeInvalidation() {is=title K2Node_FunctionEntry}]}
 OnCooldownStarted {nodes [
   n1 = OnCooldownStarted!() -> (then, CooldownType:EFortItemCooldownType Enum=None, CooldownDuration:float=~) {is=title K2Node_FunctionEntry}
   > n2 = Switch on EFortItemCooldownType!(execute, Selection:EFortItemCooldownType Enum=@n1) -> (
    None:Exec
    AmmoRegeneration:Exec
    ItemActivation:Exec
    WeaponReloading:Exec
    Death:Exec
    AthenaWeaponFireCooldown:Exec
    AbilitySetActivateByInputAbility:Exec
    DirectProgressUpdate:Exec
   ) {is=title K2Node_SwitchEnum}
   Set bKeybindsHidden(execute=[@n2.Death, @n2.ItemActivation], bKeybindsHidden:bool=true) -> (Output_Get:bool)
   > UpdateKeyBindingVisibility()
  ]}
 OnCooldownStopped {nodes [
   n1 = OnCooldownStopped!() -> (then, CooldownType:EFortItemCooldownType Enum=None) {is=title K2Node_FunctionEntry}
   > n2 = Switch on EFortItemCooldownType(Selection=@n1)
   Set bKeybindsHidden(execute=[@n2.Death, @n2.ItemActivation], bKeybindsHidden=false)
   > UpdateKeyBindingVisibility()
  ]}
 OnDragDetected {nodes [
   OnDragDetected!() -> (then, MyGeometry:Geometry Structure, PointerEvent:"Pointer Event Structure (by ref)") {is=title K2Node_FunctionEntry}
   > Set bDragDetected(bDragDetected:bool=true) -> (Output_Get:bool)
   > n1 = Branch(Condition=Get bDragAndDropEnabled() -> (bDragAndDropEnabled:bool))
   n2 = Get Item() -> (Item:Fort Item Object Reference)
   > n3 = Branch(Condition=IsValid(Object=@n2))
   > n4 = Create Widget!(
    execute:Exec
    Class:User Widget Class Reference=/Game/Athena/HUD/Inventory/AthenaInventoryDragVisual.AthenaInventoryDragVisual_C
    OwningPlayer:Player Controller Object Reference=GetOwningPlayer()
    Item:Fort Item Object Reference=@n2
   ) -> (then:Exec, ReturnValue:Athena Inventory Drag Visual Object Reference)
   > n5 = "Create Drag && Drop Operation"!(
    execute:Exec
    Class:Drag Drop Operation Class Reference=/Game/Athena/HUD/Inventory/AthenaInventoryDragDrop.AthenaInventoryDragDrop_C
    Tag:String=FromQuickbar
    Payload:Object Reference=Get Item()
    DefaultDragVisual:Widget Object Reference=@n4
    Pivot:EDragPivot Enum=CenterCenter
    Offset:Vector 2D Structure=(X=0,Y=0)
    SplitTheStack:bool=False
    bIsFromCreativeInventory:bool=False
    InstigatorSlotIndex:int=Get SlotIndex() -> (SlotIndex:int=0)
   ) -> (then:Exec, ReturnValue:Athena Inventory Drag Drop Object Reference) {is=title K2Node_CreateDragDropOperation}
   > Set bDragOpStarted(bDragOpStarted:bool=true) -> (Output_Get:bool)
   > Return Node!(execute, Operation:Drag Drop Operation Object Reference=@n5)
   Return Node(execute=[@n3.else, @n1.else])
  ]}
 OnDragOver {nodes [
   n1 = OnDragOver!() -> (then, MyGeometry:Geometry Structure, PointerEvent:Pointer Event Structure, Operation:Drag Drop Operation Object Reference) {is=title K2Node_FunctionEntry}
   > Branch(Condition=NOT Boolean(A=Get bIsLocked() -> (bIsLocked:bool)))
   > n2 = Cast To AthenaInventoryDragDrop!(execute, Object=@n1.Operation) -> (then, CastFailed, AsAthena Inventory Drag Drop:Athena Inventory Drag Drop Object Reference) {is=title K2Node_DynamicCast}
   > n3 = Cast To FortItem!(execute, Object=Get Payload!(self:Drag Drop Operation Object Reference=@n1.Operation) -> (Payload:Object Reference)) -> (then, CastFailed, AsFort Item:Fort Item Object Reference) {is=title K2Node_DynamicCast}
   > n4 = CanSwapItem!(
    execute
    self:Fort Inventory Context Object Reference=Get FortInventoryContext()
    Item:Fort Item Object Reference=@n3
    TargetQuickbar:EFortQuickBars Enum=Get QuickBarType() -> (QuickBarType:EFortQuickBars Enum=Primary)
    TargetSlot:int=Get SlotIndex() -> (SlotIndex=0)
   ) -> (then, ReturnValue:bool)
   > n5 = Branch(Condition=@n4)
   > ShowDropIcon!(execute, self:Athena Inventory Drag Drop Object Reference=@n2, Drop Action:AthenaDragDropAction Enum=NewEnumerator2) -> (then)
   > n6 = SetHovered!(execute, self, NewHovered:bool=true) -> (then)
   Return Node!(execute=[@n6.then, @n7.then], ReturnValue:bool=true)
   n7 = ShowDropIcon(execute=@n5.else, self=@n2, Drop Action=NewEnumerator3)
  ]}
 OnDrop {nodes [
   n1 = OnDrop!() -> (then, MyGeometry:Geometry Structure, PointerEvent:Pointer Event Structure, Operation:Drag Drop Operation Object Reference) {is=title K2Node_FunctionEntry}
   > n2 = Branch(Condition=NOT Boolean(A=Get bIsLocked()))
   > SetHovered(NewHovered=false)
   > n3 = Cast To FortItem(Object=Get Payload(self=@n1.Operation))
   n4 = Get SlotIndex() -> (SlotIndex=0)
   > DropItemOnQuickBar!(
    execute
    self:Fort Inventory Context Object Reference=Get FortInventoryContext()
    Item:Fort Item Object Reference=@n3
    TargetQuickbar:EFortQuickBars Enum=Get QuickBarType() -> (QuickBarType=Primary)
    TargetSlot:int=@n4
   ) -> (then, ReturnValue:bool)
   > n5 = Cast To AthenaInventoryDragDrop(Object=Get Operation() -> (Operation:Drag Drop Operation Object Reference))
   > n6 = Branch(Condition="Equal (Integer)"(A=Get InstigatorSlotIndex!(self:Athena Inventory Drag Drop Object Reference=@n5) -> (InstigatorSlotIndex:int=0), B=@n4) {is=title K2Node_CallFunction})
   > n7 = GetConsoleVariableBoolValue!(execute, VariableName:String=Fort.Rollback.IndependentTapAnywherePrimaryAction) -> (then, ReturnValue:bool)
   > n8 = Branch(Condition=@n7)
   > StartSelectSlot!(execute, self:Athena Quick Bar Slot Object Reference, bReloadOrUseIfAlreadySelected:bool=true, bStopUseImmediately:bool=true) -> (then)
   > Return Node(ReturnValue=true)
   SlotSelected!(execute=@n8.else, self:Bacchus Quickbar Slot Base Object Reference, bStopUseImmediately:bool=true) -> (then)
   > Return Node(ReturnValue=true)
   Return Node(execute=[@n5.CastFailed, @n6.else, @n3.CastFailed], ReturnValue=true)
   Return Node(execute=@n2.else, ReturnValue=true)
  ]}
 OnTouchEnded {nodes [
   OnTouchEnded!() -> (then, MyGeometry:Geometry Structure, InTouchEvent:"Pointer Event Structure (by ref)") {is=title K2Node_FunctionEntry}
   > n1 = Branch(Condition=AND Boolean!(A=IsTouchInputPlatform!() -> (ReturnValue:bool), B=Get bShouldTakeItemSelectionInputFromHUD() -> (bShouldTakeItemSelectionInputFromHUD:bool)) -> (ReturnValue))
   > n2 = GetConsoleVariableBoolValue(VariableName=Fort.Rollback.IndependentTapAnywherePrimaryAction)
   > n3 = Branch(Condition=@n2)
   > StartSelectSlot(bReloadOrUseIfAlreadySelected=true, bStopUseImmediately=true)
   > Return Node(ReturnValue:Event Reply Structure=Handled())
   TakeItemSelectionInputFromHUD!(
    execute=@n3.else
    self:Athena HUDContext Object Reference=Get AthenaHUDContext!() -> (ReturnValue:Athena HUDContext Object Reference) {is=title K2Node_GetBlueprintContextEx}
    TouchIndex:int=Get Pointer Index!(Input:"Pointer Event Structure (by ref)"=Get InTouchEvent() -> (InTouchEvent:"Pointer Event Structure (by ref)")) -> (ReturnValue:int=0) {is=title K2Node_CallFunction}
    SlotIdx:int=Get SlotIndex() -> (SlotIndex=0)
    InQuickBar:EFortQuickBars Enum=Get QuickBarType() -> (QuickBarType=Primary)
    bReloadOrUseIfAlreadySelected:bool=true
    bStopUseImmediately:bool=true
   ) -> (then, bIsUsing:bool)
   > Return Node(ReturnValue=Handled())
   Return Node(execute=@n1.else, ReturnValue=Unhandled())
  ]}
 RefreshCooldownTimeVisibility {nodes [
   RefreshCooldownTimeVisibility() {is=title K2Node_FunctionEntry}
   n1 = Get CurrentCooldownType(self=Get ItemCooldownViewmodel(self=Get InstituteChirpQuickbarSlotViewModel())) -> (CurrentCooldownType=None)
   n2 = "Not Equal (Enum)"(A:EFortItemCooldownType Enum=@n1, B:EFortItemCooldownType Enum=DirectProgressUpdate)
   n3 = "Not Equal (Enum)"(A=@n1, B=WeaponReloading)
   n4 = Get bShouldShowCooldown!(self:Item Cooldown Viewmodel Object Reference=Get ItemCooldownViewmodel(self=Get InstituteChirpQuickbarSlotViewModel())) -> (bShouldShowCooldown:bool)
   n5 = Get bIsCooldownActive()
   n6 = Get bCacheHasItem()
   > SetVisibility(
    self=Get Text_CooldownTime()
    InVisibility=Select(Option 0=Hidden, Option 1=SelfHitTestInvisible, Index=AND Boolean!(A=@n5, B=@n6, C:bool=@n3, D:bool=@n2, E:bool=@n4) -> (ReturnValue))
   )
  ]}
 RefreshDurabilityOrFuelMeter {nodes [
   RefreshDurabilityOrFuelMeter() {is=title K2Node_FunctionEntry}
   > Branch(Condition=IsVisible(self=Get Image_Durability()))
   > HandleItemQuantityAndAmmoTypeIconVisibility(IsVisible=false) {is=title K2Node_CallFunction}
   > n1 = GetDynamicMaterial(self=Get Image_Durability())
   n2 = SelectFloat(
    A=Get CacheFuelPercentage() -> (CacheFuelPercentage:double)
    B=1
    bPickA="Equal (Enum)"(A:EFortFuelGadgetVisualType Enum=Get CacheFuelVisualType() -> (CacheFuelVisualType=FuelMeter), B:EFortFuelGadgetVisualType Enum=FuelMeter)
   )
   n3 = Get CacheNormalizedDurability() -> (CacheNormalizedDurability:double)
   n4 = Get bCacheShowDurability()
   n5 = Get InstituteChirpQuickbarSlotViewModel()
   n6 = SafeDivide(
    A="To Float (Integer)"(InInt=Get CacheLoadedAmmo() -> (CacheLoadedAmmo:int=0))
    B="To Float (Integer)"(InInt=Max Clip Size(self=@n5) -> (ReturnValue=0))
   )
   > SetScalarParameterValue(
    self=@n1
    ParameterName=Progress
    Value=SelectFloat(A="Max (Float)"(A=@n6, B=0), B=SelectFloat(A=@n3, B=@n2, bPickA=@n4), bPickA=Get bShouldShowAmmoAsPercent())
   )
  ]}
 RefreshDurabilityOrFuelVisibility {nodes [
   RefreshDurabilityOrFuelVisibility() {is=title K2Node_FunctionEntry}
   n1 = "Equal (Enum)"(A=Get CacheFuelVisualType() -> (CacheFuelVisualType=FuelMeter), B=FuelMeter)
   n2 = Get bShouldShowAmmoAsPercent()
   n3 = Get bCacheShowDurability()
   > n4 = Branch(Condition=AND Boolean!(A=Get bCacheHasItem(), B=OR Boolean!(A=@n1, B=@n3, C:bool=@n2) -> (ReturnValue)) -> (ReturnValue))
   n5 = Get Image_Durability()
   > SetVisibility(self=@n5, InVisibility=HitTestInvisible)
   > RefreshDurabilityOrFuelMeter() {is=title K2Node_CallFunction}
   SetVisibility(execute=@n4.else, self=@n5, InVisibility=Hidden)
  ]}
 SetAnimatedBoxScaleFactor {nodes [
   n1 = SetAnimatedBoxScaleFactor!() -> (then, Value:double=~) {is=title K2Node_FunctionEntry}
   > Set AnimatedBoxScaleFactor(AnimatedBoxScaleFactor:double=@n1) -> (Output_Get:double)
   > UpdateItemBoxSize()
  ]}
 SetHarvestingToolVisualParameters {nodes [
   SetHarvestingToolVisualParameters() {is=title K2Node_FunctionEntry}
   > n1 = GetDynamicMaterial(self=Get Image_TileImage())
   n2 = Get InstituteChirpQuickbarSlotViewModel()
   > SetScalarParameterValue(
    self=@n1
    ParameterName=IsHarvestingTool
    Value=Select(Option 0:float=0, Option 1:float=1, Index=Is Harvest Tool(self=@n2)) -> (ReturnValue:float)
   )
  ]}
 SetHovered {nodes [
   n1 = SetHovered!() -> (then, NewHovered:bool=~) {is=title K2Node_FunctionEntry}
   > Branch(Condition="Not Equal (Boolean)"!(A:bool=@n1, B:bool=Get IsDragHovered() -> (IsDragHovered:bool)) -> (ReturnValue:bool) {is=title K2Node_CallFunction})
   > n2 = Set IsDragHovered(IsDragHovered:bool=@n1) -> (Output_Get:bool)
   > SetVisibility(
    self=Get DropTarget() -> (DropTarget:Image Object Reference)
    InVisibility=Select(Option 0:ESlateVisibility Enum=Collapsed, Option 1:ESlateVisibility Enum=HitTestInvisible, Index=@n2) -> (ReturnValue:ESlateVisibility Enum)
   )
   > InvalidateLayoutAndVolatility(self:Widget Object Reference=Get SizeBox_TileContents())
  ]}
 SetIsClipEmpty {nodes [
   n1 = SetIsClipEmpty!() -> (then, ClipIsEmpty:bool=~) {is=title K2Node_FunctionEntry}
   > Branch(Condition="Not Equal (Boolean)"(A=@n1, B=Get bCacheIsClipEmpty() -> (bCacheIsClipEmpty:bool)) -> (ReturnValue=~) {is=title K2Node_PromotableOperator})
   > n2 = Set bCacheIsClipEmpty(bCacheIsClipEmpty:bool=Get ClipIsEmpty() -> (ClipIsEmpty:bool)) -> (Output_Get:bool)
   > Set ClipIsEmptyAsFloat(ClipIsEmptyAsFloat:double="To Float (Boolean)"(InBool=@n2)) -> (Output_Get:double)
   > SetVisibility(self=Get Image_ReloadIndicator(), InVisibility=Select(Option 0=Hidden, Option 1=SelfHitTestInvisible, Index=Get ClipIsEmpty()))
   > n3 = GetDynamicMaterial(self=Get Image_TileImage())
   n4 = Get ClipIsEmptyAsFloat() -> (ClipIsEmptyAsFloat:double)
   > SetScalarParameterValue(self=@n3, ParameterName=IsClipEmpty, Value=@n4)
   > n5 = GetDynamicMaterial(self=Get Image_AmmoTypeIcon())
   > SetScalarParameterValue(self=@n5, ParameterName=IsClipEmpty, Value=@n4)
   > n6 = GetDynamicFontMaterial(self=Get Text_ItemQuantity())
   > SetScalarParameterValue(self=@n6, ParameterName=IsClipEmpty, Value=@n4)
  ]}
 SetItemType {nodes [
   SetItemType() {is=title K2Node_FunctionEntry}
   n1 = Get Item()
   > Is Valid(InputObject=@n1)
   n2 = GetItemDefinition!(self:Fort Item Object Reference=@n1) -> (ReturnValue:Fort Item Definition Object Reference)
   > Set ItemType(ItemType:EFortItemType Enum=GetItemType!(ItemDefinition:Item Definition Base Object Reference=@n2) -> (ReturnValue:EFortItemType Enum=WorldItem)) -> (Output_Get:EFortItemType Enum=WorldItem)
   > Set MaxStackSize(MaxStackSize:int=GetMaxStackSize!(ItemDefinition:Item Definition Base Object Reference=@n2, AbilityComp:Ability System Component Object Reference) -> (ReturnValue:int=0)) -> (Output_Get:int=0)
   > n3 = Cast To FortWeaponItemDefinition!(execute, Object=@n2) -> (then, CastFailed, AsFort Weapon Item Definition:Fort Weapon Item Definition Object Reference) {is=title K2Node_DynamicCast}
   > Set bShouldShowAmmoAsPercent(bShouldShowAmmoAsPercent:bool=ShouldShowAmmoAsPercent!(self:Fort Weapon Item Definition Object Reference=@n3) -> (ReturnValue:bool)) -> (Output_Get:bool)
   > Set bWeaponUsesAmmo(bWeaponUsesAmmo:bool=RequiresAmmoToFire!(self:Fort Weapon Item Definition Object Reference=@n3, bPrimaryFire:bool=true) -> (ReturnValue:bool)) -> (Output_Get:bool)
   Set bShouldShowAmmoAsPercent(execute=@n3.CastFailed, bShouldShowAmmoAsPercent=false)
   > Set bWeaponUsesAmmo(bWeaponUsesAmmo=true)
  ]}
 SetOutOfAmmo {nodes [
   n1 = SetOutOfAmmo!() -> (then, OutOfAmmo:bool=~) {is=title K2Node_FunctionEntry}
   > Branch(Condition="Not Equal (Boolean)"(A=@n1, B=Get bCacheNoAmmo() -> (bCacheNoAmmo:bool)) -> (ReturnValue=~))
   > n2 = Set bCacheNoAmmo(bCacheNoAmmo:bool=Get OutOfAmmo() -> (OutOfAmmo:bool)) -> (Output_Get:bool)
   > Set NoAmmoAsFloat(NoAmmoAsFloat:double="To Float (Boolean)"(InBool=@n2)) -> (Output_Get:double)
   > n3 = GetDynamicMaterial(self=Get Image_TileImage())
   > SetScalarParameterValue(self=@n3, ParameterName=IsAmmoEmpty, Value=Get NoAmmoAsFloat() -> (NoAmmoAsFloat:double))
   > PlayAnimation(
    InAnimation=Get HideFontOutline() -> (HideFontOutline:Widget Animation Object Reference)
    StartAtTime=Get NoAmmoAsFloat()
    PlayMode=Select(Option 0:EUMGSequencePlayMode Enum=Reverse, Option 1:EUMGSequencePlayMode Enum, Index=Get OutOfAmmo()) -> (ReturnValue:EUMGSequencePlayMode Enum)
   )
   > n4 = GetDynamicMaterial(self=Get Image_AmmoTypeIcon())
   n5 = Get NoAmmoAsFloat()
   > SetScalarParameterValue(self=@n4, ParameterName=IsOutOfAmmo, Value=@n5)
   > n6 = GetDynamicFontMaterial(self=Get Text_ItemQuantity())
   > SetScalarParameterValue(self=@n6, ParameterName=IsOutOfAmmo, Value=@n5)
   n7 = Get bCacheHasItem()
   n8 = Get OutOfAmmo()
   > SetVisibility(
    self=Get ItemDisabledIconOverlay() -> (ItemDisabledIconOverlay:Overlay Object Reference)
    InVisibility=Select(Option 0:ESlateVisibility Enum=Hidden, Option 1:ESlateVisibility Enum=SelfHitTestInvisible, Index=AND Boolean(A=@n8, B=@n7)) -> (ReturnValue:ESlateVisibility Enum)
   )
  ]}
 SetRandomFXValues {nodes [
   SetRandomFXValues() {is=title K2Node_FunctionEntry}
   > n1 = For Each Loop(Array:Array of Image Object References=Make Array!("[0]":Image Object Reference, "[1]":Image Object Reference) -> (Array:Array of Image Object References)) -> (Array Element:Image Object Reference)
   Return Node!(execute=@n1.Completed)
   > n2 = GetDynamicMaterial(self=@n1)
   > SetScalarParameterValue(
    self=@n2
    ParameterName=Layer1_TileCount
    Value="To Float (Integer)"(InInt=RandomIntegerInRange!(Min:int=6, Max:int=7) -> (ReturnValue:int=0))
   )
   > SetScalarParameterValue(self=@n2, ParameterName=Layer2_TileCount, Value="To Float (Integer)"(InInt=RandomIntegerInRange(Min=3, Max=4) -> (ReturnValue=0)))
   > SetScalarParameterValue(self=@n2, ParameterName=Layer1_TilePanSpeed, Value=RandomFloatInRange(Min=-0.1, Max=-0.15))
   > SetScalarParameterValue(self=@n2, ParameterName=Layer2_TilePanSpeed, Value=RandomFloatInRange(Min=-0.75, Max=-1))
   Reroute Node(InputPin:Wildcard) -> (OutputPin:Wildcard)
  ]}
 UpdateGameplayTags {nodes [
   UpdateGameplayTags() {is=title K2Node_FunctionEntry}
   > n1 = HasMatchingGameplayTag!(execute:Exec, self:Object Reference=Get Item(), TagToCheck:Gameplay Tag Structure=(TagName="Athena.HUD.EquippedItem.HideQuickbarItemCount")) -> (then:Exec, ReturnValue:bool) {is=title K2Node_Message}
   n2 = Get bIsSelected() -> (bIsSelected:bool)
   > SetVisibility(
    self=Get Text_ItemQuantity()
    InVisibility=Select(Option 0=SelfHitTestInvisible, Option 1=Select(Option 0=Collapsed, Option 1=Hidden, Index=@n2), Index=@n1)
   )
  ]}
 UpdateItemBoxSize {nodes [
   UpdateItemBoxSize() {is=title K2Node_FunctionEntry}
   > n1 = Branch(Condition=Get bUseSlotSizeBySlotUsage() -> (bUseSlotSizeBySlotUsage:bool))
   n2 = GetSlotSizeForItem!(self:Fort Quick Bar Slot Base Object Reference, InItem:Fort Item Object Reference=Get Item()) -> (ReturnValue:Vector 2D Structure)
   n3 = BreakVector2D(InVec=@n2)
   n4 = BreakVector2D(InVec=vector2d * float!(A:Vector 2D Structure=@n2, B:double=0.7) -> (ReturnValue:Vector 2D Structure))
   n5 = Get AnimatedBoxScaleFactor() -> (AnimatedBoxScaleFactor:double)
   n6 = Lerp(A=@n4, B=@n3, Alpha=@n5)
   n7 = Lerp(A=@n4.Y, B=@n3.Y, Alpha=@n5)
   > DesiredSizeOverride!(execute, self:Athena Quick Bar Native Cell Object Reference=Get NativeCell(), WidthOverride:float=@n6, HeightOverride:float=@n7) -> (then)
   n8 = Get SizeBox_ExtensionOverlay() -> (SizeBox_ExtensionOverlay:Size Box Object Reference)
   > SetHeightOverride(self=@n8, InHeightOverride=@n7)
   > SetWidthOverride(self=@n8, InWidthOverride=@n6)
   n9 = Lerp(A=64, B=92, Alpha=@n5)
   DesiredSizeOverride(execute=@n1.else, self=Get NativeCell(), WidthOverride=@n9, HeightOverride=@n9)
   n10 = Get SizeBox_ExtensionOverlay()
   > SetHeightOverride(self=@n10, InHeightOverride=@n9)
   > SetWidthOverride(self=@n10, InWidthOverride=@n9)
  ]}
]
variables [
 AnimatedBoxScaleFactor {type=number}
 bCacheHasItem {type=boolean}
 bCacheIsClipEmpty {type=boolean}
 bCacheIsSelected {type=boolean}
 bCacheNoAmmo {type=boolean}
 bCacheShowDurability {type=boolean}
 bHasCooldownIntroAnimPlayed {type=boolean}
 bIsCooldownActive {type=boolean}
 bShouldShowAmmoAsPercent {type=boolean}
 bShouldTakeItemSelectionInputFromHUD {type=boolean, default=true}
 bWeaponUsesAmmo {type=boolean, default=true}
 CacheFuelChargeCount {type=integer, default=-1}
 CacheFuelPercentage {type=number, default=-1}
 CacheFuelVisualType {type=EFortFuelGadgetVisualType, default=Invalid}
 CacheLoadedAmmo {type=integer, default=-1}
 CacheNormalizedDurability {type=number, default=-1}
 CooldownIntroHasPlayed {type=boolean}
 CooldownTypesSupported {type="array<EFortItemCooldownType>", default []}
 DropTarget {type=Image, default=null}
 HaveOffset {type=boolean}
 IsDragHovered {type=boolean}
 ItemDisplayStyle {type=ItemDisplayStyle, default=World}
 ItemType {type=EFortItemType, default=WorldItem}
 LastPlayedOffset {type=WidgetAnimation, default=null}
 MaxStackSize {type=integer}
 ParentHB {type=HorizontalBox, default=null}
 Slot Num {type=integer, default=-1}
 SlotWidth {type=number, default=95}
 Use Special Styling {type=boolean}
]
widgets {
 info {inheritedWidgetCount=15, parentClass=/Script/FortniteUI.BacchusQuickbarSlotBase, rootWidgetClass=&VerticalBox, widgetCount=37}
 widgets [
  VerticalBox_0 {widgetClassPath=&VerticalBox, widget=VerticalBox_0}
  SizeBox_TileContents {
   widgetClassPath=&SizeBox
   bIsVariable=true
   parent=VerticalBox_0
   widget=SizeBox_TileContents
   properties {bOverride_HeightOverride=true, bOverride_WidthOverride=true, heightOverride=110, widthOverride=95}
   slot {class=VerticalBoxSlot, properties {horizontalAlignment=HAlign_Center, verticalAlignment=VAlign_Center}}
  }
  Overlay_TileContents {
   widgetClassPath=&Overlay
   bIsVariable=true
   parent=SizeBox_TileContents
   widget=Overlay_TileContents
   properties {pixelSnapping=Disabled}
   slot {class=SizeBoxSlot}
  }
  Slot_BehindImage {
   widgetClassPath=&NamedSlot
   bIsVariable=true
   parent=Overlay_TileContents
   widget=Slot_BehindImage
   properties {bExposeOnInstanceOnly=true, visibility=Collapsed}
   slot {class=OverlaySlot, properties {horizontalAlignment=HAlign_Fill, verticalAlignment=VAlign_Fill}}
  }
  Image_TileImage {
   widgetClassPath=&Image
   bIsVariable=true
   parent=Overlay_TileContents
   widget=Image_TileImage
   properties {brush {
     bIsDynamicallyLoaded=false
     drawAs=Image
     imageType=FullColor
     mirroring=NoMirror
     resourceName=None
     resourceObject=/BattleRoyaleGameUI/HUD/QuickBar/Assets/M_UI_QuickBarItemCard_WithVariant
     tiling=NoTile
     imageSize=<32 32>
     margin=<ltrb 0 0 0 0>
     outlineSettings {
      bUseBrushTransparency=false
      roundingType=HalfHeightRadius
      width=0
      color {colorUseRule=UseColor_Specified, specifiedColor=<rgba 0 0 0 0>}
      cornerRadii=<0 0 0 0>
     }
     tintColor {colorUseRule=UseColor_Specified, specifiedColor=<rgba 1 1 1 1>}
     uVRegion {bIsValid=false, max=<0 0>, min=<0 0>}
    }}
   slot {class=OverlaySlot, properties {horizontalAlignment=HAlign_Fill, verticalAlignment=VAlign_Fill}}
  }
  Image_Durability {
   widgetClassPath=&Image
   bIsVariable=true
   parent=Overlay_TileContents
   widget=Image_Durability
   properties {
    visibility=Hidden
    brush {
     bIsDynamicallyLoaded=false
     drawAs=Image
     imageType=FullColor
     mirroring=NoMirror
     resourceName=None
     resourceObject=/BattleRoyaleGameUI/HUD/QuickBar/Assets/M_UI_DurabilityBar
     tiling=NoTile
     imageSize=<8 10>
     margin=<ltrb 0 0 0 0>
     outlineSettings {
      bUseBrushTransparency=false
      roundingType=HalfHeightRadius
      width=0
      color {colorUseRule=UseColor_Specified, specifiedColor=<rgba 0 0 0 0>}
      cornerRadii=<0 0 0 0>
     }
     tintColor {colorUseRule=UseColor_Specified, specifiedColor=<rgba 1 1 1 1>}
     uVRegion {bIsValid=false, max=<0 0>, min=<0 0>}
    }
    renderTransform {angle=0, scale=<1 1>, shear=<0 -3>, translation=<0 0>}
   }
   slot {class=OverlaySlot, properties {horizontalAlignment=HAlign_Right, verticalAlignment=VAlign_Fill, padding=<ltrb 10 20 10 28>}}
  }
  Text_CooldownTime {
   widgetClassPath=&CommonTextBlock
   bIsVariable=true
   parent=Overlay_TileContents
   widget=Text_CooldownTime
   properties {
    justification=Center
    minDesiredWidth=75
    mobileFontSizeMultiplier=0.7
    text=66
    visibility=Hidden
    font {
     bForceMonospaced=false
     bMaterialIsStencil=false
     fontMaterial=/BattleRoyaleGameUI/Shared/M_UI_Font_White
     fontObject=&Burbank
     letterSpacing=0
     monospacedWidth=1
     size=36
     skewAmount=0
     typefaceFontName=BurbankRegular-Black
     outlineSettings {
      bApplyOutlineToDropShadows=false
      bMiteredCorners=false
      bSeparateFillAlpha=true
      outlineMaterial=/BattleRoyaleGameUI/HUD/QuickBar/Assets/MI_UI_Font_SoftShadow_ItemCoolDown
      outlineSize=16
      outlineColor=<rgba 1 1 1 1>
     }
    }
    margin=<ltrb 0 -4 0 -6>
    renderTransform {angle=0, scale=<1 1>, shear=<-16 0>, translation=<0 0>}
    shadowOffset=<0 0>
   }
   slot {class=OverlaySlot, properties {horizontalAlignment=HAlign_Center, verticalAlignment=VAlign_Center, padding=<ltrb 0 0 0 8>}}
  }
  HBox_ItemQuantityAndAmmoType {
   widgetClassPath=&HorizontalBox
   bIsVariable=true
   parent=Overlay_TileContents
   widget=HBox_ItemQuantityAndAmmoType
   properties {bOverride_Cursor=true, clipping=OnDemand}
   slot {class=OverlaySlot, properties {horizontalAlignment=HAlign_Right, verticalAlignment=VAlign_Bottom, padding=<ltrb 0 0 9 24>}}
  }
  Image_AmmoTypeIcon {
   widgetClassPath=&Image
   bIsVariable=true
   parent=HBox_ItemQuantityAndAmmoType
   widget=Image_AmmoTypeIcon
   properties {
    brush {
     bIsDynamicallyLoaded=false
     drawAs=Image
     imageType=FullColor
     mirroring=NoMirror
     resourceName=None
     resourceObject=/BattleRoyaleGameUI/HUD/QuickBar/Assets/M_QuickBarSlotAmmoTypeIcon
     tiling=NoTile
     imageSize=<17 17>
     margin=<ltrb 0 0 0 0>
     outlineSettings {
      bUseBrushTransparency=false
      roundingType=HalfHeightRadius
      width=0
      color {colorUseRule=UseColor_Specified, specifiedColor=<rgba 0 0 0 0>}
      cornerRadii=<0 0 0 0>
     }
     tintColor {colorUseRule=UseColor_Specified, specifiedColor=<rgba 1 1 1 1>}
     uVRegion {bIsValid=false, max=<0 0>, min=<0 0>}
    }
    renderTransformPivot=<0.35 1.65>
   }
   slot {class=HorizontalBoxSlot, properties {horizontalAlignment=HAlign_Center, verticalAlignment=VAlign_Center, padding=<ltrb 0 0 3 0>}}
  }
  SizeBox_1 {
   widgetClassPath=&SizeBox
   parent=HBox_ItemQuantityAndAmmoType
   widget=SizeBox_1
   properties {bOverride_MaxDesiredWidth=true, maxDesiredWidth=60}
   slot {class=HorizontalBoxSlot}
  }
  Text_ItemQuantity {
   widgetClassPath=&CommonTextBlock
   bIsVariable=true
   parent=SizeBox_1
   widget=Text_ItemQuantity
   properties {
    justification=Center
    text=99
    font {
     bForceMonospaced=false
     bMaterialIsStencil=false
     fontMaterial=/BattleRoyaleGameUI/HUD/QuickBar/Assets/M_QuickBarSlotFont
     fontObject=&BurbankSmall
     letterSpacing=0
     monospacedWidth=1
     size=13.5
     skewAmount=0
     typefaceFontName=Black
     outlineSettings {
      bApplyOutlineToDropShadows=false
      bMiteredCorners=false
      bSeparateFillAlpha=true
      outlineMaterial=/BattleRoyaleGameUI/Shared/MI_UI_Font_SoftShadow_SmallText_CustomColor
      outlineSize=8
      outlineColor=<rgba 1 1 1 1>
     }
    }
    margin=<ltrb -12 -8 0 -8>
    shadowOffset=<0 0>
   }
   slot {class=SizeBoxSlot, properties {verticalAlignment=VAlign_Center}}
  }
  ReloadIndicator_Attachment {
   widgetClassPath=&FortVisualAttachment
   parent=HBox_ItemQuantityAndAmmoType
   widget=ReloadIndicator_Attachment
   properties {contentAnchor=<1 1>}
   slot {class=HorizontalBoxSlot, properties {horizontalAlignment=HAlign_Right}}
  }
  Image_ReloadIndicator {
   widgetClassPath=&Image
   bIsVariable=true
   parent=ReloadIndicator_Attachment
   widget=Image_ReloadIndicator
   properties {
    visibility=Collapsed
    brush {
     bIsDynamicallyLoaded=false
     drawAs=Image
     imageType=FullColor
     mirroring=NoMirror
     resourceName=None
     resourceObject=/BattleRoyaleGameUI/HUD/QuickBar/Assets/MI_UI_Quickbar_AmmoEmptyIndicator
     tiling=NoTile
     imageSize=<24 24>
     margin=<ltrb 0 0 0 0>
     outlineSettings {
      bUseBrushTransparency=false
      roundingType=HalfHeightRadius
      width=0
      color {colorUseRule=UseColor_Specified, specifiedColor=<rgba 0 0 0 0>}
      cornerRadii=<0 0 0 0>
     }
     tintColor {colorUseRule=UseColor_Specified, specifiedColor=<rgba 1 1 1 1>}
     uVRegion {bIsValid=false, max=<0 0>, min=<0 0>}
    }
    renderTransform {angle=0, scale=<1 1>, shear=<0 0>, translation=<0 2>}
   }
   slot {class=SizeBoxSlot, properties {horizontalAlignment=HAlign_Right, verticalAlignment=VAlign_Bottom}}
  }
  SizeBox_ExtensionOverlay {
   widgetClassPath=&SizeBox
   bIsVariable=true
   parent=Overlay_TileContents
   widget=SizeBox_ExtensionOverlay
   properties {bOverride_HeightOverride=true, heightOverride=132}
   slot {class=OverlaySlot, properties {horizontalAlignment=HAlign_Fill, verticalAlignment=VAlign_Fill}}
  }
  OverlayExtensionWidgets {
   widgetClassPath=&Overlay
   bIsVariable=true
   parent=SizeBox_ExtensionOverlay
   widget=OverlayExtensionWidgets
   slot {class=SizeBoxSlot}
  }
  TutorialHighlight {
   widgetClassPath=&TutorialHighlight
   parent=Overlay_TileContents
   widget=TutorialHighlight
   properties {customHighlightWidget=/Game/Athena/HUD/Tutorial/Tutorial_Outline_Square.Tutorial_Outline_Square_C}
   slot {class=OverlaySlot, properties {horizontalAlignment=HAlign_Fill, verticalAlignment=VAlign_Fill}}
  }
  ItemDisabledIconOverlay {
   widgetClassPath=&Overlay
   bIsVariable=true
   parent=Overlay_TileContents
   widget=ItemDisabledIconOverlay
   properties {visibility=Collapsed}
   slot {class=OverlaySlot, properties {horizontalAlignment=HAlign_Center, verticalAlignment=VAlign_Bottom, padding=<ltrb 0 0 0 4>}}
  }
  Image_ItemDisabled_Glow {
   widgetClassPath=&Image
   parent=ItemDisabledIconOverlay
   widget=Image_ItemDisabled_Glow
   properties {
    visibility=SelfHitTestInvisible
    brush {
     bIsDynamicallyLoaded=false
     drawAs=Image
     imageType=FullColor
     mirroring=NoMirror
     resourceName=None
     resourceObject=/BattleRoyaleGameUI/Shared/FX/M_UI_ElimGlow
     tiling=NoTile
     imageSize=<22 22>
     margin=<ltrb 0 0 0 0>
     outlineSettings {
      bUseBrushTransparency=false
      roundingType=HalfHeightRadius
      width=0
      color {colorUseRule=UseColor_Specified, specifiedColor=<rgba 0 0 0 0>}
      cornerRadii=<0 0 0 0>
     }
     tintColor {colorUseRule=UseColor_Specified, specifiedColor=<rgba 1 1 1 1>}
     uVRegion {bIsValid=false, max=<0 0>, min=<0 0>}
    }
    renderTransform {angle=0, scale=<4 3>, shear=<0 0>, translation=<0 0>}
   }
   slot {class=OverlaySlot, properties {horizontalAlignment=HAlign_Fill, verticalAlignment=VAlign_Fill}}
  }
  Image_ItemDisabled {
   widgetClassPath=&Image
   parent=ItemDisabledIconOverlay
   widget=Image_ItemDisabled
   properties {
    visibility=SelfHitTestInvisible
    brush {
     bIsDynamicallyLoaded=false
     drawAs=Image
     imageType=FullColor
     mirroring=NoMirror
     resourceName=None
     resourceObject=/BattleRoyaleGameUI/HUD/QuickBar/Assets/MI_UI_DisabledItemIndicator
     tiling=NoTile
     imageSize=<22 22>
     margin=<ltrb 0 0 0 0>
     outlineSettings {
      bUseBrushTransparency=false
      roundingType=HalfHeightRadius
      width=0
      color {colorUseRule=UseColor_Specified, specifiedColor=<rgba 0 0 0 0>}
      cornerRadii=<0 0 0 0>
     }
     tintColor {colorUseRule=UseColor_Specified, specifiedColor=<rgba 1 1 1 1>}
     uVRegion {bIsValid=false, max=<0 0>, min=<0 0>}
    }
   }
   slot {class=OverlaySlot, properties {horizontalAlignment=HAlign_Fill, verticalAlignment=VAlign_Fill, padding=<ltrb 2 0 0 2>}}
  }
  LockIcon {
   widgetClassPath=&Image
   bInherited=true
   bIsVariable=true
   parent=Overlay_TileContents
   widget=LockIcon
   properties {
    visibility=Collapsed
    brush {
     bIsDynamicallyLoaded=false
     drawAs=Image
     imageType=NoImage
     mirroring=NoMirror
     resourceName=None
     resourceObject=/Game/UI/Foundation/Textures/Icons/Locks/T-Icon-Lock-64
     tiling=NoTile
     imageSize=<32 32>
     margin=<ltrb 0 0 0 0>
     outlineSettings {
      bUseBrushTransparency=false
      roundingType=HalfHeightRadius
      width=0
      color {colorUseRule=UseColor_Specified, specifiedColor=<rgba 0 0 0 0>}
      cornerRadii=<0 0 0 0>
     }
     tintColor {colorUseRule=UseColor_Specified, specifiedColor=<rgba 1 1 1 1>}
     uVRegion {bIsValid=false, max=<0 0>, min=<0 0>}
    }
   }
   slot {class=OverlaySlot}
  }
  HorizontalBox_0 {
   widgetClassPath=&HorizontalBox
   parent=VerticalBox_0
   widget=HorizontalBox_0
   slot {class=VerticalBoxSlot, properties {horizontalAlignment=HAlign_Center, verticalAlignment=VAlign_Top}}
  }
  Overlay_0 {
   widgetClassPath=&Overlay
   parent=HorizontalBox_0
   widget=Overlay_0
   slot {class=HorizontalBoxSlot, properties {horizontalAlignment=HAlign_Center, verticalAlignment=VAlign_Center}}
  }
  Slot_BehindKeybind {
   widgetClassPath=&NamedSlot
   bIsVariable=true
   parent=Overlay_0
   widget=Slot_BehindKeybind
   slot {class=OverlaySlot, properties {horizontalAlignment=HAlign_Center, verticalAlignment=VAlign_Center}}
  }
  QuickbarKeybindWidgetBottom {
   widgetClassPath=&/WBP_BRGameUI_QuickbarKeybindBottom
   bInherited=true
   bIsVariable=true
   parent=Overlay_0
   widget=QuickbarKeybindWidgetBottom
   slot {class=OverlaySlot, properties {horizontalAlignment=HAlign_Center, verticalAlignment=VAlign_Center}}
  }
  Slot_NextToKeybind {
   widgetClassPath=&NamedSlot
   bIsVariable=true
   parent=HorizontalBox_0
   widget=Slot_NextToKeybind
   slot {class=HorizontalBoxSlot, properties {horizontalAlignment=HAlign_Center, verticalAlignment=VAlign_Center}}
  }
  SelectionHighlight {widgetClassPath=&Image, bInherited=true, widget=null}
  Image_Disabled {widgetClassPath=&Image, bInherited=true, widget=null}
  SizeBox_BottomKeybind {widgetClassPath=&SizeBox, bInherited=true, widget=null}
  SizeBox_TopKeybind {widgetClassPath=&SizeBox, bInherited=true, widget=null}
  QuickbarKeybindWidgetTop {widgetClassPath=/Script/FortniteUI.FortQuickbarKeybindWidget, bInherited=true, widget=null}
  EmptyImage {widgetClassPath=&Image, bInherited=true, widget=null}
  ItemCardMaximized {widgetClassPath=/Script/FortniteUI.FortMultiSizeItemCard, bInherited=true, widget=null}
  ItemCardMinimized {widgetClassPath=&FortMultiSizeItemCard, bInherited=true, widget=null}
  QuickbarSlotCooldown {widgetClassPath=/Script/FortniteUI.FortItemCooldownWidget, bInherited=true, widget=null}
  QuickbarSlotDurability {widgetClassPath=/Script/FortniteUI.AthenaItemElementWidgetBase, bInherited=true, widget=null}
  SlotInteraction {widgetClassPath=&Image, bInherited=true, widget=null}
  NativeCell {widgetClassPath=/Script/FortniteUI.AthenaQuickBarNativeCell, bInherited=true, widget=null}
 ]
}
