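# Quickbar slot widget for the battle royale HUD, a slottable variant that derives from WBP_BRGameUI_QuickbarSlot_Primary and adds no widgets, variables or logic of its own.
# Anim_Show fades in the tile contents while sliding the slot vertically into place.
# Anim_Selected raises the tile and animates the image material's IconYOffset parameter alongside it.
# The event graph only forwards PreConstruct and Construct to the parent, and nothing here plays the animations.
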WidgetBlueprint /BattleRoyaleGameUI/HUD/QuickBar/WBP_BRGameUI_QuickbarSlot_Slottable
animations [
 Anim_Show {
  binding_count=2
  length_seconds=0.2
  object=&self:Anim_Show
  sequence {bindings [
    SizeBox_TileContents {
     id=#1
     possessed_class=SizeBox
     tracks [{
       type=MovieSceneFloatTrack
       display_name=Render Opacity
       sections [{type=MovieSceneFloatSection, end_frame=4, start_frame=0, channels [None {type=MovieSceneScriptingFloatChannel, keys [[0,0],[4,1]]}]}]
      }]
    }
    WBP_BRGameUI_QuickbarSlot_Slottable_C_4 {
     id=#2
     possessed_class=WBP_BRGameUI_QuickbarSlot_Slottable_C
     tracks [{
       type=MovieScene2DTransformTrack
       sections [{
         type=MovieScene2DTransformSection
         end_frame=4
         start_frame=0
         channels [
          Translation.X {type=MovieSceneScriptingFloatChannel, const [0,1]}
          Translation.Y {type=MovieSceneScriptingFloatChannel, keys [[0,-45],[4,0]]}
          Angle {type=MovieSceneScriptingFloatChannel, const [0,1]}
          Scale.X {type=MovieSceneScriptingFloatChannel, const [1,1]}
          Scale.Y {type=MovieSceneScriptingFloatChannel, const [1,1]}
          Shear.X {type=MovieSceneScriptingFloatChannel, const [0,1]}
          Shear.Y {type=MovieSceneScriptingFloatChannel, const [0,1]}
         ]
        }]
      }]
    }
   ]}
 }
 Anim_Selected {
  binding_count=2
  length_seconds=0.15
  object=&self:Anim_Selected
  sequence {bindings [
    SizeBox_TileContents {
     id=#3
     possessed_class=SizeBox
     tracks [{
       type=MovieScene2DTransformTrack
       sections [{
         type=MovieScene2DTransformSection
         end_frame=3
         start_frame=0
         channels [
          Translation.X {type=MovieSceneScriptingFloatChannel, const [0,1]}
          Translation.Y {type=MovieSceneScriptingFloatChannel, keys [[0,-20],[3,-35]]}
          Angle {type=MovieSceneScriptingFloatChannel, const [0,1]}
          Scale.X {type=MovieSceneScriptingFloatChannel, const [1,1]}
          Scale.Y {type=MovieSceneScriptingFloatChannel, const [1,1]}
          Shear.X {type=MovieSceneScriptingFloatChannel, const [0,1]}
          Shear.Y {type=MovieSceneScriptingFloatChannel, const [0,1]}
         ]
        }]
      }]
    }
    Image_TileImage {
     id=#4
     possessed_class=Image
     tracks [{
       type=MovieSceneWidgetMaterialTrack
       sections [{
         type=MovieSceneParameterSection
         end_frame=3
         start_frame=0
         channels [IconYOffset {type=MovieSceneScriptingFloatChannel, keys [[0,0.125],[3,0.175]]}]
        }]
      }]
    }
   ]}
 }
]
graphs [EventGraph {nodes [
   n1 = Event PreConstruct()
   > Parent: PreConstruct(IsDesignTime=@n1.IsDesignTime)
   Event Construct()
   > Parent: Construct()
   Event Tick()
  ]}]
variables []
widgets {
 info {
  inheritedWidgetCount=0
  parentClass=&/WBP_BRGameUI_QuickbarSlot_Primary.WBP_BRGameUI_QuickbarSlot_Primary_C
  rootWidgetClass=null
  widgetCount=0
 }
 widgets []
}
