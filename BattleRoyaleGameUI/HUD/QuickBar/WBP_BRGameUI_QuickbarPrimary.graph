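# Battle Royale HUD widget for the primary quickbar: six fixed slots, pickup preview and a scale animation for maximize/minimize.
# Inventory content changes create or remove extra "gizmo" slot widgets in a map keyed by slot index.
# Focus changes swap slot 1 with the active gizmo between overlays, updating swap icon and pickaxe-toggle tag.
# Interaction context drives the preview's position and width, while global event router messages inject or remove per-slot overlay widgets.
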WidgetBlueprint /BattleRoyaleGameUI/HUD/QuickBar/WBP_BRGameUI_QuickbarPrimary
animations [AnimMaximize {
  binding_count=1
  length_seconds=0.0583
  object=&self:AnimMaximize
  sequence {
   bindings [HorizontalBox_1 {
     id=#1
     possessed_class=HorizontalBox
     tracks [{
       type=MovieScene2DTransformTrack
       sections [{
         type=MovieScene2DTransformSection
         end_frame=7
         start_frame=0
         channels [
          Translation.X {type=MovieSceneScriptingFloatChannel, default=0}
          Translation.Y {type=MovieSceneScriptingFloatChannel, default=0}
          Angle {type=MovieSceneScriptingFloatChannel, default=0}
          Scale.X {type=MovieSceneScriptingFloatChannel, keys [[0,0.85],[7,1]]}
          Scale.Y {type=MovieSceneScriptingFloatChannel, keys [[0,0.85],[7,1]]}
          Shear.X {type=MovieSceneScriptingFloatChannel, default=0}
          Shear.Y {type=MovieSceneScriptingFloatChannel, default=0}
         ]
        }]
      }]
    }]
   display_rate {numerator=120}
  }
 }]
graphs [
 EventGraph {nodes [
   HandleInputMethodChanged_Bind!() -> (OutputDelegate, then, InputType:ECommonInputType Enum) {is=title K2Node_CustomEvent}
   > HandleShowHideWeaponRail()
   OnCreativeSwapPhoneHoldChanged!() -> (OutputDelegate, then, bIsSwapped:bool=~) {is=title K2Node_CustomEvent}
   > UpdateKeyBindingText(self:Fort Quick Bar Slot Base Object Reference=Get QuickbarSlot1() -> (QuickbarSlot1:WBP BRGame UI Quickbar Slot Primary Object Reference))
   n1 = OnQuickBarSlotFocusChanged_Bind!() -> (OutputDelegate, then, QuickbarIndex:EFortQuickBars Enum=Primary, Slot:int) {is=title K2Node_CustomEvent}
   > RefreshPickupPreview()
   > Switch on EFortQuickBars!(execute, Selection:EFortQuickBars Enum=@n1.QuickbarIndex) -> (Primary:Exec, Secondary:Exec, Creative:Exec, Max_None:Exec) {is=title K2Node_SwitchEnum}
   > n2 = Set CurrentWeaponSlot(CurrentWeaponSlot:int=@n1.Slot) -> (Output_Get:int=0)
   > n3 = Branch(Condition="integer >= integer"(A=@n1.Slot, B=6))
   n4 = GetSelectedGizmoSlot!(CurrentSlot:int=@n2) -> (GizmoSlot:WBP BRGame UI Quickbar Slot Primary Object Reference)
   > Set GizmoSlot(GizmoSlot:WBP BRGame UI Quickbar Slot Primary Object Reference=@n4) -> (Output_Get:WBP BRGame UI Quickbar Slot Primary Object Reference)
   > Set HasAdditionalSlot(HasAdditionalSlot:bool=IsValid(Object=@n4)) -> (Output_Get:bool)
   > n5 = Swap Slots 6 to 1()
   n6 = Get CurrentWeaponSlot() -> (CurrentWeaponSlot:int=0)
   n7 = OR Boolean(A="Equal (Integer)"(A=@n6), B="integer >= integer"(A=@n6, B=6))
   HandleGizmoSlotPreview!(execute=[@n5.then, @n9.then, @n8.else], self, ShowGizmoPreview:bool=@n7) -> (then)
   > HandleUnboundAndSwapIconVisibility_Swap!(execute, self, Is6thSlotActive:bool=@n7) -> (then)
   > HandleAdditionalSlotTag!(execute, self, IsAddTag:bool=AND Boolean(A=HasAnyValidAdditionalSlot!() -> (HasAnyAdditionalSlot:bool), B=@n7)) -> (then)
   > Set PreviousWeaponSlot(PreviousWeaponSlot:int=Get CurrentWeaponSlot() -> (CurrentWeaponSlot=0)) -> (Output_Get:int=0)
   Branch(execute=@n3.else, Condition=Get HasAdditionalSlot() -> (HasAdditionalSlot:bool))
   > n8 = Branch(Condition="integer >= integer"(A=Get PreviousWeaponSlot() -> (PreviousWeaponSlot:int=0), B=6))
   > n9 = Swap Slots 1 to 6()
   QuickbarSwapSlottable() {is=title K2Node_CustomEvent}
   > Branch(Condition=Get HasAdditionalSlot())
   > n10 = Branch(Condition=Get IsGizmoSlotFocused() -> (IsGizmoSlotFocused:bool))
   > Swap Slots 1 to 6()
   Swap Slots 6 to 1(execute=@n10.else)
   Swap Slots 1 to 6() {is=title K2Node_CustomEvent}
   > HideGizmoSlots()
   > SetSwapableVisuals!(execute, self, QuickbarSlot:WBP BRGame UI Quickbar Slot Primary Object Reference=Get QuickbarSlot1(), IsFocused:bool=true) -> (then)
   > AddChild(self=Get QuickbarSlot1_Overlay() -> (QuickbarSlot1_Overlay:Overlay Object Reference), Content=Get QuickbarSlot1())
   > Set IsGizmoSlotFocused(IsGizmoSlotFocused:bool=false) -> (Output_Get:bool)
   > Is Valid(InputObject=Get GizmoSlot() -> (GizmoSlot:WBP BRGame UI Quickbar Slot Primary Object Reference))
   n11 = Get GizmoSlot()
   > AddChild(self=Get 6thSlot_Overlay() -> (6thSlot_Overlay:Overlay Object Reference), Content=@n11)
   > SetSlotSelected!(execute, self:Fort Quick Bar Slot Base Object Reference=Get GizmoSlot(), bSelected:bool=false) -> (then)
   > SetSwapableVisuals(QuickbarSlot=Get GizmoSlot(), IsFocused=false)
   n12 = Get CurrentWeaponSlot() -> (CurrentWeaponSlot=0)
   n13 = "integer >= integer"(A=@n12, B=6)
   n14 = "Equal (Integer)"(A=@n12)
   > SetRenderOpacity(
    self=Get SizeBox_TileContents!(self:WBP BRGame UI Quickbar Slot Primary Object Reference=@n11) -> (SizeBox_TileContents:Size Box Object Reference)
    InOpacity="To Float (Boolean)"(InBool=OR Boolean(A=@n14, B=@n13))
   )
   Swap Slots 6 to 1() {is=title K2Node_CustomEvent}
   > HideGizmoSlots()
   > AddChild(self=Get 6thSlot_Overlay(), Content=Get QuickbarSlot1())
   > Set IsGizmoSlotFocused(IsGizmoSlotFocused=true)
   > SetSwapableVisuals(QuickbarSlot=Get QuickbarSlot1(), IsFocused=false)
   > Is Valid(InputObject=Get GizmoSlot())
   n15 = Get GizmoSlot()
   > AddChild(self=Get QuickbarSlot1_Overlay(), Content=@n15)
   > SetSlotSelected(self=Get GizmoSlot(), bSelected=false)
   > SetSlotSelected(self=Get GizmoSlot(), bSelected=true)
   > SetSwapableVisuals(QuickbarSlot=Get GizmoSlot(), IsFocused=true)
   n16 = Get CurrentWeaponSlot() -> (CurrentWeaponSlot=0)
   n17 = "integer >= integer"(A=@n16, B=6)
   n18 = "Equal (Integer)"(B=@n16)
   > SetRenderOpacity(self=Get SizeBox_TileContents(self=@n15), InOpacity="To Float (Boolean)"(InBool=OR Boolean(A=@n18, B=@n17)))
   Event HandleKeybindsChanged() {is=title K2Node_Event}
   > Parent: HandleKeybindsChanged() {is=title K2Node_CallParentFunction}
   Event Destruct()
   > n19 = Sequence()
   n20 = Get CommonInputSubsystem()
   > Is Valid(InputObject=@n20)
   > Unbind Event from On Input Method Changed(self=@n20, Delegate=Create Event())
   > Clear and Invalidate Timer by Handle(Handle=Get FillRadialInputTimerHandle() -> (FillRadialInputTimerHandle:Timer Handle Structure))
   > Parent: Destruct()
   StopListeningToEvent(
    execute=@n19.then_1
    self:Async Action Start Listening To Event Object Reference=Get Async Task Ref() -> (Async Task Ref:Async Action Start Listening To Event Object Reference)
   )
   Event Maximize() {is=title K2Node_Event}
   > Parent: Maximize() {is=title K2Node_CallParentFunction}
   > HandleShowHideWeaponRail()
   > n21 = Branch(Condition=Get MaximizeAnimHasPlayed() -> (MaximizeAnimHasPlayed:bool))
   Set MinimizeAnimHasPlayed(execute=[@n21.then, @n22.then], MinimizeAnimHasPlayed:bool=false) -> (Output_Get:bool)
   > SetVisibility(self=Get SizeBox_PickupPreview() -> (SizeBox_PickupPreview:Size Box Object Reference), InVisibility=Hidden)
   > RefreshPickupPreview()
   PlayAnimationForward(execute=@n21.else, InAnimation=Get AnimMaximize() -> (AnimMaximize:Widget Animation Object Reference))
   > n22 = Set MaximizeAnimHasPlayed(MaximizeAnimHasPlayed:bool=true) -> (Output_Get:bool)
   Event Minimize() {is=title K2Node_Event}
   > Parent: Minimize() {is=title K2Node_CallParentFunction}
   > Show Primary Quickbar Rail!(execute, self, Show:bool=false) -> (then)
   > n23 = Branch(Condition=Get MinimizeAnimHasPlayed() -> (MinimizeAnimHasPlayed:bool))
   Set MaximizeAnimHasPlayed(execute=[@n23.then, @n24.then], MaximizeAnimHasPlayed=false)
   > SetVisibility(self=Get SizeBox_PickupPreview(), InVisibility=Hidden)
   > RefreshPickupPreview()
   PlayAnimationReverse(execute=@n23.else, InAnimation=Get AnimMaximize())
   > n24 = Set MinimizeAnimHasPlayed(MinimizeAnimHasPlayed=true)
   n25 = Event OnQuickbarContentsChanged!() -> (OutputDelegate, then, QuickbarIndex:EFortQuickBars Enum=Primary, ChangedSlots:Array of Integers) {is=title K2Node_Event}
   > Parent: OnQuickbarContentsChanged!(execute, QuickbarIndex:EFortQuickBars Enum=@n25.QuickbarIndex, ChangedSlots:Array of Integers=@n25.ChangedSlots) -> (then) {is=title K2Node_CallParentFunction}
   > HandleShowHideWeaponRail()
   > RefreshPickupPreview()
   Event Construct()
   > n26 = Sequence!(execute) -> (then_0, then_1, then_2:Exec, then_3:Exec, then_4:Exec, then_5:Exec, then_6:Exec, then_7:Exec, then_8:Exec, then_9:Exec, then_10:Exec)
   Parent: Construct(execute=@n26.then_1)
   n27 = Create Event()
   n28 = Get FortHUDContext!() -> (ReturnValue:Fort HUDContext Object Reference) {is=title K2Node_GetBlueprintContextEx}
   Bind Event to On Player Can Interact(execute=@n26.then_6, self:Fort HUDContext Object Reference=@n28, Delegate=@n27) {is=title K2Node_AddDelegate}
   n29 = PlaylistDataReadyAsync!(execute:Exec=@n26.then_8) -> (
    then:Exec
    OnReady:Exec
    GameState:Fort Game State Athena Object Reference
    Playlist:Fort Playlist Object Reference
    PlaylistContextTags:"Gameplay Tag Container Structure (by ref)"
   ) {is=title K2Node_AsyncAction}
   Set bPlaylistHidesQuickbar(
    execute=@n29.OnReady
    bPlaylistHidesQuickbar:bool=Compare Tag Container to Other Tags!(
     TagCase_0:"Gameplay Tag Structure (by ref)"=Get QuickbarTag() -> (QuickbarTag:Gameplay Tag Structure)
     Gameplay Tag Container:"Gameplay Tag Container Structure (by ref)"=Get HUDElementsToHide!(self:Fort Playlist Object Reference=@n29.Playlist) -> (HUDElementsToHide:Gameplay Tag Container Structure)
    ) -> (Case_0 True:bool=~) {is=title GameplayTagsK2Node_MultiCompareGameplayTagContainerSingleTags}
   ) -> (Output_Get:bool)
   > SetVisibility(self=Get Overlay_AdditionalSlot() -> (Overlay_AdditionalSlot:Overlay Object Reference), InVisibility=Hidden)
   n30 = GetDynamicMaterial(
    execute=@n26.then_9
    self=Get Image_TileImage!(self:WBP BRGame UI Quickbar Slot Primary Object Reference=Get QuickbarSlot1()) -> (Image_TileImage:Image Object Reference)
   )
   > Set QuickbarSlot1TileImageMID(QuickbarSlot1TileImageMID:Material Instance Dynamic Object Reference=@n30) -> (Output_Get:Material Instance Dynamic Object Reference)
   > SetVisibility(self=Get Swap() -> (Swap:Image Object Reference), InVisibility=Collapsed)
   > Clear!(
    execute
    TargetMap:Map of Integers to WBP BRGame UI Quickbar Slot Primary Object References=Get GizmoSlotMap() -> (GizmoSlotMap:Map of Integers to WBP BRGame UI Quickbar Slot Primary Object References)
   ) -> (then) {is=title K2Node_CallFunction}
   > n31 = Set PlayerController(PlayerController:Player Controller Object Reference=GetOwningPlayer()) -> (Output_Get:Player Controller Object Reference)
   > n32 = Cast To FortPlayerController(Object=@n31)
   > Set FortPlayerController(FortPlayerController:Fort Player Controller Object Reference=@n32) -> (Output_Get:Fort Player Controller Object Reference)
   n33 = GetGlobalEventRouter()
   n34 = StartListeningToEvent!(
    execute:Exec=@n26.then_10
    Target:Gameplay Event Router Component Object Reference=@n33
    EventType:Script Struct Object Reference=/BattleRoyaleGameUI/Data/ST_AddOverlayToQuickBarData
    Context:Object Reference
    EventBubblingRule:EEventBubblingRule Enum
    None:%=@n33
   ) -> (
    then:Exec
    AsyncTaskProxy:Async Action Start Listening To Event Object Reference
    OnEventReceived:Exec
    Payload:ST Add Overlay To Quick Bar Data Structure
   )
   > n35 = Get!(execute:Exec) -> (then:Exec, else:Exec, Async Task Ref:Async Action Start Listening To Event Object Reference)
   > StopListeningToEvent(self=@n35)
   Set Async Task Ref(execute=@n35.else, Async Task Ref:Async Action Start Listening To Event Object Reference=@n34) -> (Output_Get:Async Action Start Listening To Event Object Reference)
   n36 = Break ST Add Overlay to Quick Bar Data!(ST_AddOverlayToQuickBarData:"ST Add Overlay To Quick Bar Data Structure (by ref)"=@n34.Payload) -> (Widget_16_#2:User Widget Class Reference, QuickBarType_5_#3:EFortQuickBars Enum) {is=title K2Node_BreakStruct}
   Branch(
    execute=@n34.OnEventReceived
    Condition=AND Boolean(
     A="Equal (Enum)"(A:EFortQuickBars Enum=@n36.QuickBarType_5_#3, B:EFortQuickBars Enum=Primary)
     B=IsValidClass!(Class:Object Class Reference=@n36) -> (ReturnValue:bool)
    )
   )
   > HandleQuickBarOverlay!(execute, self, Data:"ST Add Overlay To Quick Bar Data Structure (by ref)"=@n34.Payload) -> (then)
   Bind Event to On Interact Updated(execute=@n26.then_7, self:Fort HUDContext Object Reference=@n28, Delegate=@n27) {is=title K2Node_AddDelegate}
   n37 = Set QuickbarSlots(
    execute=@n26.then_0
    QuickbarSlots:Array of Fort Quick Bar Slot Base Object References=Make Array!(
     "[0]":Fort Quick Bar Slot Base Object Reference=Get QuickbarSlot1()
     "[1]":Fort Quick Bar Slot Base Object Reference=Get QuickbarSlot2() -> (QuickbarSlot2:WBP BRGame UI Quickbar Slot Primary Object Reference)
     "[2]":Fort Quick Bar Slot Base Object Reference=Get QuickbarSlot3() -> (QuickbarSlot3:WBP BRGame UI Quickbar Slot Primary Object Reference)
     "[3]":Fort Quick Bar Slot Base Object Reference=Get QuickbarSlot4() -> (QuickbarSlot4:WBP BRGame UI Quickbar Slot Primary Object Reference)
     "[4]":Fort Quick Bar Slot Base Object Reference=Get QuickbarSlot5() -> (QuickbarSlot5:WBP BRGame UI Quickbar Slot Primary Object Reference)
     "[5]":Fort Quick Bar Slot Base Object Reference=Get QuickbarSlot6() -> (QuickbarSlot6:WBP BRGame UI Quickbar Slot Primary Object Reference)
    ) -> (Array:Array of Fort Quick Bar Slot Base Object References)
   ) -> (Output_Get:Array of Fort Quick Bar Slot Base Object References)
   > Set NonPickaxeSlotsNum(NonPickaxeSlotsNum:int=int - int(A=Length(TargetArray:Array of Fort Quick Bar Slot Base Object References=@n37))) -> (Output_Get:int=0)
   > n38 = For Each Loop(Array:Array of Fort Quick Bar Slot Base Object References=@n37) -> (Array Element:Fort Quick Bar Slot Base Object Reference)
   > n39 = Cast To WBP_BRGameUI_QuickbarSlot_Primary!(execute, Object=@n38) -> (then, CastFailed, AsWBP BRGame UI Quickbar Slot Primary:WBP BRGame UI Quickbar Slot Primary Object Reference) {is=title K2Node_DynamicCast}
   > InitializeDataModel!(
    execute
    self:WBP BRGame UI Quickbar Slot Primary Object Reference=@n39
    SlotNum:int=@n38.Array Index
    Quick Bar Type:EFortQuickBars Enum=Primary
   ) -> (then)
   n40 = Cast To FortPlayerControllerGameplay!(execute=@n26.then_4, Object=GetOwningPlayer()) -> (then, CastFailed, AsFort Player Controller Gameplay:Fort Player Controller Gameplay Object Reference) {is=title K2Node_DynamicCast}
   > Bind Event to On Creative Swap Phone and Harvest Hold Changed Delegate(self:Fort Player Controller Gameplay Object Reference=@n40, Delegate=Create Event()) {is=title K2Node_AddDelegate}
   Bind Event to On Input Method Changed(execute=@n26.then_3, self=Get CommonInputSubsystem(), Delegate=Create Event())
   n41 = GetQuickbarFocus!(self:Fort Inventory Context Object Reference=Get FortInventoryContext!() -> (ReturnValue:Fort Inventory Context Object Reference) {is=title K2Node_GetBlueprintContextEx}) -> (OutQuickBar:EFortQuickBars Enum=Primary, OutSlot:int=0, OutSecondarySlot:int=0, OutPreviousFocusedSlot:int=0)
   OnQuickbarSlotFocusChanged!(execute=@n26.then_2, self:Fort Quickbar Base Object Reference, QuickbarIndex:EFortQuickBars Enum=@n41, SlotIndex:int=@n41.OutSlot) -> (then)
   n42 = Get FortInventoryContext()
   Assign On Quickbar Slot Focus Changed!(execute:Exec=@n26.then_5, self:Fort Inventory Context Object Reference=@n42, Delegate:"Delegate (by ref)"=Create Event()) -> (then:Exec) {is=title K2Node_AssignDelegate}
   > Bind Event to On Quickbar Contents Changed(self:Fort Inventory Context Object Reference=@n42, Delegate=Create Event()) {is=title K2Node_AddDelegate}
   "integer >= integer"(A=Get CurrentWeaponSlot() -> (CurrentWeaponSlot=0), B=6)
  ]}
 GetFirstGizmoWithQuantity {nodes [
   GetFirstGizmoWithQuantity() {is=title K2Node_FunctionEntry}
   > n1 = For Loop with Break!(Execute:Exec, FirstIndex:int=6, LastIndex:int=int - int(A=Get TotalSlots() -> (TotalSlots:int=0)), Break:Exec=@n5.then) -> (LoopBody:Exec, Index:int, Completed:Exec) {is=title K2Node_MacroInstance}
   n2 = Get GizmoSlotMap()
   > Branch(Condition="integer > integer"(A=Length!(TargetMap:Map of Integers to WBP BRGame UI Quickbar Slot Primary Object References=@n2) -> (ReturnValue:int=0) {is=title K2Node_CallFunction}))
   n3 = Find(TargetMap:Map of Integers to WBP BRGame UI Quickbar Slot Primary Object References=@n2, Key:"Integer (by ref)"=@n1) -> (Value:WBP BRGame UI Quickbar Slot Primary Object Reference)
   > Is Valid(InputObject=@n3)
   n4 = Get InstituteChirpQuickbarSlotViewModel!(self:WBP BRGame UI Quickbar Slot Primary Object Reference=@n3) -> (InstituteChirpQuickbarSlotViewModel:Institute Chirp - HUD Controller Quickbar Slot Viewmodel Object Reference)
   > Is Valid(InputObject=@n4)
   > Branch(Condition="integer > integer"(A=Stack Count!(self:Device - HUD Controller Quickbar Slot Viewmodel Object Reference=@n4) -> (ReturnValue:int=0)))
   > Set GizmoSlot(GizmoSlot=@n3)
   > n5 = Set CurrentPreviewGizmoSlot(CurrentPreviewGizmoSlot:WBP BRGame UI Quickbar Slot Primary Object Reference=@n3) -> (Output_Get:WBP BRGame UI Quickbar Slot Primary Object Reference)
   Return Node!(execute=@n1.Completed, GizmoSlot:WBP BRGame UI Quickbar Slot Primary Object Reference=Get GizmoSlot())
  ]}
 GetFirstWidgetOfType {nodes [
   n1 = GetFirstWidgetOfType!() -> (then, Type:User Widget Class Reference, Panel:Panel Widget Object Reference) {is=title K2Node_FunctionEntry}
   > n2 = For Each Loop(Array:Array of Widget Object References=GetAllChildren(self=@n1.Panel)) -> (Array Element:Widget Object Reference)
   Return Node!(execute=@n2.Completed, Content:Widget Object Reference)
   > Branch(Condition=IsA!(Object:Object Reference=@n2, ObjectClass:Object Class Reference=@n1) -> (ReturnValue:bool))
   > Return Node(Content=@n2)
  ]}
 GetSelectedGizmoSlot {nodes [
   n1 = GetSelectedGizmoSlot!() -> (then, CurrentSlot:int) {is=title K2Node_FunctionEntry}
   n2 = Get GizmoSlotMap()
   > n3 = Branch(Condition="integer > integer"(A=Length(TargetMap=@n2) -> (ReturnValue=0)))
   > Return Node!(execute, GizmoSlot:WBP BRGame UI Quickbar Slot Primary Object Reference=Find(TargetMap=@n2, Key=@n1))
   Return Node(execute=@n3.else)
  ]}
 GetSpecificSlot {nodes [
   n1 = GetSpecificSlot!() -> (then, SlotNumber:int=0) {is=title K2Node_FunctionEntry}
   n2 = "Get (a copy)"(
    Array:Array of Fort Quick Bar Slot Base Object References=Get QuickbarSlots() -> (QuickbarSlots:Array of Fort Quick Bar Slot Base Object References)
    Dimension 1=@n1
   ) -> (Output:Fort Quick Bar Slot Base Object Reference)
   > Branch(Condition=IsValid(Object=@n2))
   > Return Node!(execute, ReturnValue:Fort Quick Bar Slot Base Object Reference=@n2)
  ]}
 HandleAdditionalSlotTag {nodes [
   n1 = HandleAdditionalSlotTag!() -> (then, IsAddTag:bool=~) {is=title K2Node_FunctionEntry}
   > n2 = Delta Gate Wildcard!(execute:Exec, Variable:"Boolean (by ref)"=Get IsAdditionalSlotTagAdded() -> (IsAdditionalSlotTagAdded:bool), New Value:bool=@n1) -> (Success:Exec, Updated Variable:"Boolean (by ref)", Previous Variable:bool=~) {is=title K2Node_MacroInstance}
   > n3 = Branch(Condition=@n2)
   > n4 = Branch(Condition=HasAdditionalSlotTag!() -> (HasTag:bool))
   AddLooseGameplayTags!(
    execute=@n4.else
    Actor:Actor Object Reference=Get PlayerController() -> (PlayerController:Player Controller Object Reference)
    GameplayTags:"Gameplay Tag Container Structure (by ref)"=Get AllowTogglePickAxeTag() -> (AllowTogglePickAxeTag:Gameplay Tag Container Structure)
    bShouldReplicate:bool=true
   ) -> (then, ReturnValue:bool)
   RemoveLooseGameplayTags!(
    execute=@n3.else
    Actor:Actor Object Reference=Get PlayerController()
    GameplayTags:"Gameplay Tag Container Structure (by ref)"=Get AllowTogglePickAxeTag()
    bShouldReplicate:bool=true
   ) -> (then, ReturnValue:bool)
  ]}
 HandleGizmoSlotPreview {nodes [
   HandleGizmoSlotPreview!() -> (then, ShowGizmoPreview:bool=~) {is=title K2Node_FunctionEntry}
   > n1 = Sequence!(execute) -> (then_0, then_1)
   > Is Valid(InputObject=Get GizmoSlot())
   > n2 = Cast To WBP_BRGameUI_QuickbarSlot_Slottable!(execute, Object=Get GizmoSlot()) -> (then, CastFailed, AsWBP BRGame UI Quickbar Slot Slottable:WBP BRGame UI Quickbar Slot Slottable Object Reference) {is=title K2Node_DynamicCast}
   > Set SlottableWidget(SlottableWidget:WBP BRGame UI Quickbar Slot Slottable Object Reference=@n2) -> (Output_Get:WBP BRGame UI Quickbar Slot Slottable Object Reference)
   > n3 = Sequence()
   > n4 = Branch(Condition="Not Equal (Boolean)"(A=Get ShowGizmoPreview() -> (ShowGizmoPreview:bool), B=Get IsGizmoSlotPreviewed() -> (IsGizmoSlotPreviewed:bool)) -> (ReturnValue=~))
   > n5 = Branch(Condition=Get ShowGizmoPreview())
   n6 = Get SlottableWidget() -> (SlottableWidget:WBP BRGame UI Quickbar Slot Slottable Object Reference)
   > StopAnimation(
    self=@n6
    InAnimation=Get Anim_Show!(self:WBP BRGame UI Quickbar Slot Slottable Object Reference=@n6) -> (Anim_Show:Widget Animation Object Reference)
   )
   n7 = Get SizeBox_TileContents(self=Get GizmoSlot())
   > SetRenderOpacity(self=@n7, InOpacity=1)
   > SetSwapableVisuals(QuickbarSlot=Get SlottableWidget(), IsFocused="integer >= integer"(A=Get CurrentWeaponSlot() -> (CurrentWeaponSlot=0), B=6))
   n8 = Get SlottableWidget()
   > PlayAnimationForward(self=@n8, InAnimation=Get Anim_Show(self=@n8))
   n9 = Get SlottableWidget()
   StopAnimation(execute=@n5.else, self=@n9, InAnimation=Get Anim_Show(self=@n9))
   > SetRenderOpacity(self=@n7, InOpacity=0)
   n10 = Get SlottableWidget()
   > PlayAnimationReverse(self=@n10, InAnimation=Get Anim_Show(self=@n10))
   n11 = Get SlottableWidget()
   SetAnimationCurrentTime!(
    execute=@n4.else
    self:User Widget Object Reference=@n11
    InAnimation:Widget Animation Object Reference=Get Anim_Show(self=@n11)
    InTime:float=Select(Option 0:float=0, Option 1:float=1, Index=Get ShowGizmoPreview()) -> (ReturnValue:float)
   ) -> (then)
   n12 = Get InstituteChirpQuickbarSlotViewModel(self=Get SlottableWidget())
   Is Valid(exec=@n3.then_1, InputObject=@n12)
   > SetHidden!(execute, self:Device - HUD Controller Quickbar Slot Viewmodel Object Reference=@n12, bInHidden:bool=NOT Boolean(A=Get ShowGizmoPreview())) -> (then)
   Set IsGizmoSlotPreviewed(execute=@n1.then_1, IsGizmoSlotPreviewed:bool=Get ShowGizmoPreview()) -> (Output_Get:bool)
  ]}
 HandleInteractionChanged {nodes [
   n1 = HandleInteractionChanged!() -> (then, Context Info:Fort Interact Context Info Object Reference) {is=title K2Node_FunctionEntry}
   > n2 = Is Valid(InputObject=@n1)
   > n3 = Cast To FortPickup!(execute, Object=Get ReceivingActor!(self:Fort Interact Context Info Object Reference=@n1) -> (ReceivingActor:Actor Weak object)) -> (then, CastFailed, AsFort Pickup:Fort Pickup Object Reference) {is=title K2Node_DynamicCast}
   > n4 = Set CurrentPickupPreviewItem(CurrentPickupPreviewItem:Fort Item Object Reference=Get PrimaryPickupDummyItem!(self:Fort Pickup Object Reference=@n3) -> (PrimaryPickupDummyItem:Fort Item Object Reference)) -> (Output_Get:Fort Item Object Reference)
   RefreshPickupPreview(execute=[@n5.then, @n4.then])
   n5 = Set CurrentPickupPreviewItem(execute=[@n3.CastFailed, @n2.Is Not Valid])
  ]}
 HandleQuickBarOverlay {nodes [
   n1 = HandleQuickBarOverlay!() -> (then, Data:"ST Add Overlay To Quick Bar Data Structure (by ref)") {is=title K2Node_FunctionEntry}
   n2 = Break ST Add Overlay to Quick Bar Data!(ST_AddOverlayToQuickBarData=@n1) -> (
    AddOrRemove_12_#4:bool=~
    Widget_16_#2
    QuickBarType_5_#3
    Slot_8_#5:int
   )
   n3 = Get Slot3Overlay() -> (Slot3Overlay:Overlay Object Reference)
   n4 = Get Slot4Overlay() -> (Slot4Overlay:Overlay Object Reference)
   n5 = Get Slot5Overlay() -> (Slot5Overlay:Overlay Object Reference)
   n6 = Get Slot6Overlay() -> (Slot6Overlay:Overlay Object Reference)
   n7 = Get Slot1Overlay() -> (Slot1Overlay:Overlay Object Reference)
   n8 = Get Slot2Overlay() -> (Slot2Overlay:Overlay Object Reference)
   > Set CurrentSlot(CurrentSlot:Overlay Object Reference="Get (a copy)"(
     Array:Array of Overlay Object References=Make Array(
      "[0]":Overlay Object Reference=@n7
      "[1]":Overlay Object Reference=@n8
      "[2]":Overlay Object Reference=@n3
      "[3]":Overlay Object Reference=@n4
      "[4]":Overlay Object Reference=@n5
      "[5]":Overlay Object Reference=@n6
     ) -> (Array:Array of Overlay Object References)
     Dimension 1=int - int(A=@n2.Slot_8_#5)
    ) -> (Output:Overlay Object Reference)) -> (Output_Get:Overlay Object Reference)
   > n9 = Branch(Condition=@n2)
   n10 = Get CurrentSlot() -> (CurrentSlot:Overlay Object Reference)
   > n11 = Is Valid(InputObject=GetFirstWidgetOfType!(self, Type:User Widget Class Reference=@n2.Widget_16_#2, Panel:Panel Widget Object Reference=@n10) -> (Content:Widget Object Reference) {is=title K2Node_CallFunction})
   n12 = Create Widget!(
    execute:Exec=@n11.Is Not Valid
    Class:User Widget Class Reference=@n2.Widget_16_#2
    OwningPlayer:Player Controller Object Reference=GetOwningPlayer()
   ) -> (then:Exec, ReturnValue:User Widget Object Reference)
   > n13 = AddChildToOverlay!(execute, self:Overlay Object Reference=@n10, Content:Widget Object Reference=@n12) -> (then, ReturnValue:Overlay Slot Object Reference)
   > SetVerticalAlignment(self=@n13, InVerticalAlignment=VAlign_Fill)
   > SetHorizontalAlignment(self=@n13, InHorizontalAlignment=HAlign_Fill)
   > InvalidateLayoutAndVolatility(self:Widget Object Reference=@n10)
   n14 = GetFirstWidgetOfType(Type=@n2.Widget_16_#2, Panel=Get CurrentSlot())
   Is Valid(exec=@n9.else, InputObject=@n14)
   > RemoveChild(self=@n10, Content=@n14)
  ]}
 HandleShowHideWeaponRail {nodes [
   HandleShowHideWeaponRail() {is=title K2Node_FunctionEntry}
   n1 = Get CommonInputSubsystem!() -> (ReturnValue:Common Input Object Reference) {is=title K2Node_GetBlueprintContextEx}
   > n2 = Branch(Condition="Equal (Enum)"(A:ECommonInputType Enum=GetCurrentInputType(self=@n1), B:ECommonInputType Enum=Gamepad))
   > n3 = Branch(Condition=HasOneNonHarvestWeapon!() -> (HasNonHarvestWeapon:bool))
   > Show Primary Quickbar Rail(Show=false)
   Show Primary Quickbar Rail(execute=[@n3.else, @n2.else], Show=false)
  ]}
 HandleUnboundAndSwapIconVisibility_Swap {nodes [
   n1 = HandleUnboundAndSwapIconVisibility_Swap!() -> (then, Is6thSlotActive:bool=~) {is=title K2Node_FunctionEntry}
   n2 = AND Boolean(A=@n1, B=HasAnyValidAdditionalSlot())
   n3 = Get QuickbarSlot1()
   > Set bShowUnboundStatus!(
    execute
    bShowUnboundStatus:bool=@n2
    self:Fort Keybind Widget Object Reference=Get KeybindWidget!(self:Fort Quickbar Keybind Widget Object Reference=Get QuickbarKeybindWidgetBottom!(self:Fort Quick Bar Slot Base Object Reference=@n3) -> (QuickbarKeybindWidgetBottom:Fort Quickbar Keybind Widget Object Reference)) -> (KeybindWidget:Fort Keybind Widget Object Reference)
   ) -> (then, Output_Get:bool)
   > SetVisibility(
    self=Get Swap()
    InVisibility=Select(Option 0:ESlateVisibility Enum=Collapsed, Option 1:ESlateVisibility Enum=HitTestInvisible, Index=@n2) -> (ReturnValue:ESlateVisibility Enum)
   )
  ]}
 HasAdditionalSlotTag {nodes [
   HasAdditionalSlotTag() {is=title K2Node_FunctionEntry}
   n1 = Get PlayerController()
   > Return Node!(
    execute
    HasTag:bool=HasMatchingGameplayTag!(
     self:Gameplay Tag Asset Interface Interface=To GameplayTagAssetInterface!(InObject:Object Reference=@n1) -> (ReturnValue:Gameplay Tag Asset Interface Interface)
     TagToCheck:Gameplay Tag Structure=(TagName="Gameplay.Status.TogglePickAxeEnabled")
    ) -> (ReturnValue:bool)
   )
  ]}
 HasAnyValidAdditionalSlot {nodes [
   HasAnyValidAdditionalSlot() {is=title K2Node_FunctionEntry}
   n1 = Get GizmoSlotMap()
   > n2 = Branch(Condition="integer > integer"(A=Length(TargetMap=@n1) -> (ReturnValue=0)))
   > n3 = "For Each Loop (Map)"!(execute:Exec, MapPin:Map of Integers to WBP BRGame UI Quickbar Slot Primary Object References=Get GizmoSlotMap(), BreakPin:Exec=@n5.then) -> (then:Exec, KeyPin:int, ValuePin:WBP BRGame UI Quickbar Slot Primary Object Reference, CompletedPin:Exec)
   n4 = IsValid(Object=@n3.ValuePin)
   > n5 = Branch(Condition=@n4)
   Return Node!(execute=@n3.CompletedPin, HasAnyAdditionalSlot:bool=@n4)
   Return Node(execute=@n2.else, HasAnyAdditionalSlot=false)
  ]}
 HasOneNonHarvestWeapon {nodes [
   HasOneNonHarvestWeapon() {is=title K2Node_FunctionEntry}
   > n1 = For Loop(FirstIndex=1, LastIndex=5)
   Return Node!(execute=@n1.Completed, HasNonHarvestWeapon:bool=false)
   n2 = Get QuickbarSlots()
   > Is Valid(InputObject=Get Item!(self:Fort Quick Bar Slot Base Object Reference="Get (a copy)"(Array:Array of Fort Quick Bar Slot Base Object References=@n2, Dimension 1=@n1) -> (Output:Fort Quick Bar Slot Base Object Reference)) -> (Item:Fort Item Object Reference))
   > Return Node(HasNonHarvestWeapon=true)
  ]}
 HideGizmoSlots {nodes [
   HideGizmoSlots() {is=title K2Node_FunctionEntry}
   > n1 = "For Each Loop (Map)"(MapPin=Get GizmoSlotMap())
   n2 = GetSelectedGizmoSlot(CurrentSlot=GetFocusedSlot!(
     self:Fort Player Controller Object Reference=Get FortPlayerController() -> (FortPlayerController:Fort Player Controller Object Reference)
     InQuickbar:EFortQuickBars Enum=Primary
    ) -> (ReturnValue:int=0)) {is=title K2Node_CallFunction}
   n3 = Get GizmoSlot()
   > n4 = Branch(Condition=OR Boolean(A="Equal (Object)"(A=@n1.ValuePin, B=@n3), B="Equal (Object)"(A=@n1.ValuePin, B=@n2)))
   SetRenderOpacity(execute=@n4.else, self=Get SizeBox_TileContents(self=@n1.ValuePin), InOpacity=0)
  ]}
 On Quickbar Contents Changed {nodes [
   n1 = On Quickbar Contents Changed!() -> (then, QuickbarIndex:EFortQuickBars Enum=Primary, ChangedSlots:Array of Integers) {is=title K2Node_FunctionEntry}
   > n2 = GetNumQuickbarSlots!(execute, self:Fort Player Controller Object Reference=Get FortPlayerController(), QuickBarType:EFortQuickBars Enum=Primary) -> (then, ReturnValue:int=0)
   > Set TotalSlots(TotalSlots:int=@n2) -> (Output_Get:int=0)
   > Switch on EFortQuickBars(Selection=@n1)
   > n3 = For Each Loop(Array:Array of Integers=Get ChangedSlots() -> (ChangedSlots:Array of Integers)) -> (Array Element:int)
   HideGizmoSlots(execute=@n3.Completed) {is=title K2Node_CallFunction}
   n4 = Set Current Slot(execute=@n3.LoopBody, Current Slot:int=@n3) -> (Output_Get:int=0)
   > Branch(Condition="integer >= integer"(A=@n4, B=6))
   n5 = Get FortInventoryContext()
   n6 = Get QuickbarIndex() -> (QuickbarIndex:EFortQuickBars Enum=Primary)
   n7 = Get Current Slot() -> (Current Slot:int=0)
   > Set Has Item In Current Slot(Has Item In Current Slot:bool=IsValid(Object=GetQuickBarSlottedItem!(self:Fort Inventory Context Object Reference=@n5, InQuickBar:EFortQuickBars Enum=@n6, Slot:int=@n7) -> (ReturnValue:Fort Item Object Reference))) -> (Output_Get:bool)
   n8 = IsValid(Object=GetSelectedGizmoSlot(CurrentSlot=Get Current Slot() -> (Current Slot=0)))
   > n9 = Branch(Condition=AND Boolean(A=Get Has Item In Current Slot() -> (Has Item In Current Slot:bool), B=NOT Boolean(A=@n8)))
   > n10 = Is Valid(InputObject=Find(TargetMap=Get GizmoSlotMap(), Key=Get Current Slot() -> (Current Slot=0)))
   n11 = Get GizmoSlot()
   n12 = Branch(execute=[@n19.then, @n10.Is Valid], Condition=IsValid(Object=@n11))
   n13 = Get GizmoSlotMap()
   n14 = Find(TargetMap=@n13, Key=Get Current Slot() -> (Current Slot=0))
   Set bKeybindsHidden!(execute=[@n17.then, @n12.then], bKeybindsHidden:bool=true, self:Fort Quick Bar Slot Base Object Reference=@n14) -> (then, Output_Get:bool)
   > AddChild(self=Get 6thSlot_Overlay(), Content=@n14)
   > InitializeDataModel(self=@n14, SlotNum=Get Current Slot() -> (Current Slot=0), Quick Bar Type=Get QuickbarIndex() -> (QuickbarIndex=Primary))
   > SetQuickbarIndexRuntime!(execute, self:Fort Quick Bar Slot Base Object Reference=@n14, Index:int=Get Current Slot() -> (Current Slot=0), bRefresh:bool=true) -> (then)
   > Set HasAdditionalSlot(HasAdditionalSlot=true)
   > SetVisibility(self=Get Overlay_AdditionalSlot(), InVisibility=HitTestInvisible)
   n15 = Get QuickbarSlot1()
   > SetVisibility(
    self=Get Swap()
    InVisibility=Select(
     Option 0=Collapsed
     Option 1=HitTestInvisible
     Index=Get bIsSelected!(self:Athena Quick Bar Slot Object Reference=@n15) -> (bIsSelected:bool)
    )
   )
   > Set bShowUnboundStatus(bShowUnboundStatus=true, self=Get KeybindWidget(self=Get QuickbarKeybindWidgetBottom(self=@n15)))
   n16 = Get bIsSelected(self=Get QuickbarSlot1())
   > HandleAdditionalSlotTag(IsAddTag=@n16) {is=title K2Node_CallFunction}
   > HideGizmoSlots()
   > Set IsGizmoSlotPreviewed(IsGizmoSlotPreviewed=false)
   > HandleGizmoSlotPreview(ShowGizmoPreview=@n16) {is=title K2Node_CallFunction}
   Branch(execute=@n12.else, Condition="integer > integer"(A=Length(TargetMap=@n13) -> (ReturnValue=0)))
   > n17 = Set GizmoSlot(GizmoSlot=@n14)
   n18 = Create Widget!(
    execute=@n10.Is Not Valid
    Class=Get AdditionalQuickbarSlotClass() -> (AdditionalQuickbarSlotClass:WBP BRGame UI Quickbar Slot Primary Class Reference)
    OwningPlayer=GetOwningPlayer()
    InstituteChirpQuickbarSlotViewModel:Institute Chirp - HUD Controller Quickbar Slot Viewmodel Object Reference
   ) -> (then, ReturnValue:WBP BRGame UI Quickbar Slot Primary Object Reference)
   > n19 = Add!(
    execute
    TargetMap:Map of Integers to WBP BRGame UI Quickbar Slot Primary Object References=Get GizmoSlotMap()
    Key:"Integer (by ref)"=Get Current Slot() -> (Current Slot=0)
    Value:"WBP BRGame UI Quickbar Slot Primary Object Reference (by ref)"=@n18
   ) -> (then) {is=title K2Node_CallFunction}
   n20 = Get Has Item In Current Slot()
   Branch(execute=@n9.else, Condition=AND Boolean(A=NOT Boolean(A=@n20), B=Get HasAdditionalSlot()))
   n21 = Get GizmoSlotMap()
   > n22 = Branch(Condition="integer > integer"(A=Length(TargetMap=@n21) -> (ReturnValue=0)))
   n23 = Get Current Slot() -> (Current Slot=0)
   > RemoveFromParent(self=[Find(TargetMap=@n21, Key=@n23), Get GizmoSlot()])
   > n24 = Remove!(execute, TargetMap:Map of Integers to WBP BRGame UI Quickbar Slot Primary Object References=@n21, Key:"Integer (by ref)"=@n23) -> (then, ReturnValue:bool)
   Set GizmoSlot(execute=[@n24.then, @n22.else])
   n25 = HasAnyValidAdditionalSlot() {is=title K2Node_CallFunction}
   > Set HasAdditionalSlot(HasAdditionalSlot=@n25)
   n26 = Select(Option 0=Collapsed, Option 1=HitTestInvisible, Index=@n25)
   > SetVisibility(self=Get Overlay_AdditionalSlot(), InVisibility=@n26)
   > SetVisibility(self=Get Swap(), InVisibility=@n26)
   n27 = Get QuickbarSlot1()
   > Set bShowUnboundStatus(bShowUnboundStatus=false, self=Get KeybindWidget(self=Get QuickbarKeybindWidgetBottom(self=@n27)))
   > GetFirstGizmoWithQuantity!(execute, self) -> (then, GizmoSlot:WBP BRGame UI Quickbar Slot Primary Object Reference) {is=title K2Node_CallFunction}
   > Set IsGizmoSlotPreviewed(IsGizmoSlotPreviewed=false)
   > HandleGizmoSlotPreview(ShowGizmoPreview=@n25)
   > Set IsGizmoSlotFocused(IsGizmoSlotFocused=false)
   > Swap Slots 1 to 6() {is=title K2Node_CallFunction}
   Set IsGizmoSlotPreviewed(IsGizmoSlotPreviewed=false)
   n28 = "Equal (Integer)"(A=Get CurrentWeaponSlot() -> (CurrentWeaponSlot=0))
   > HandleGizmoSlotPreview(ShowGizmoPreview=AND Boolean(A=@n28, B=HasAnyValidAdditionalSlot()))
   OR Boolean(A="integer >= integer"(A=Get CurrentWeaponSlot() -> (CurrentWeaponSlot=0), B=6), B=@n28)
   n29 = Get FortPlayerController()
   "Equal (Object)"(A=@n11, B=GetSelectedGizmoSlot(CurrentSlot=GetFocusedSlot(self=@n29, InQuickbar=Primary) -> (ReturnValue=0)))
   Get NonPickaxeSlotsNum() -> (NonPickaxeSlotsNum:int=0)
  ]}
 RefreshPickupPreview {nodes [
   RefreshPickupPreview() {is=title K2Node_FunctionEntry}
   n1 = Get CurrentPickupPreviewItem() -> (CurrentPickupPreviewItem:Fort Item Object Reference)
   > n2 = Is Valid(InputObject=@n1)
   > n3 = PreviewItemPickUp!(
    execute
    self:Fort HUDContext Object Reference=Get FortHUDContext()
    InItem:Fort Item Object Reference=@n1
    InQuickBarType:EFortQuickBars Enum=Primary
   ) -> (then, OutSlotIndex:int=0, OutItemsToPush:Map of Integers to Integers, OutItemsToDrop:Array of Integers, ReturnValue:bool)
   n4 = Get PckupPreviewTranslation() -> (PckupPreviewTranslation:"Array of Float (double-precision)s")
   n5 = Get bPlaylistHidesQuickbar() -> (bPlaylistHidesQuickbar:bool)
   > n6 = Branch(Condition=AND Boolean!(
     A=@n3.ReturnValue
     B=Is Valid Index!(TargetArray:"Array of Float (double-precision)s"=@n4, IndexToTest:int=@n3) -> (ReturnValue:bool)
     C:bool=NOT Boolean(A=@n5)
    ) -> (ReturnValue))
   n7 = GetChildrenCount!(self:Panel Widget Object Reference=Get HorizontalBox_SlotContainer() -> (HorizontalBox_SlotContainer:Horizontal Box Object Reference)) -> (ReturnValue:int=0)
   > n8 = Branch(Condition=AND Boolean!(
     A="integer > integer"(A=@n3, B=0) {is=title K2Node_CallFunction}
     B="integer <= integer"!(A:int=@n3, B:int=@n7) -> (ReturnValue:bool) {is=title K2Node_CallFunction}
    ) -> (ReturnValue))
   n9 = Break Widget Transform!(WidgetTransform:"Widget Transform Structure (by ref)"=Get RenderTransform!(self:Widget Object Reference=Get SizeBox_PickupPreview()) -> (RenderTransform:Widget Transform Structure)) -> (Translation:Vector 2D Structure, Scale:Vector 2D Structure, Shear:Vector 2D Structure, Angle:float=~) {is=title K2Node_BreakStruct}
   n10 = BreakVector2D(InVec=@n9)
   n11 = GetChildAt!(self:Panel Widget Object Reference=Get HorizontalBox_SlotContainer(), Index:int=int - int(A=@n3, B=1) {is=title K2Node_CallFunction}) -> (ReturnValue:Widget Object Reference)
   n12 = BreakVector2D(InVec=GetLocalTopLeft!(Geometry:"Geometry Structure (by ref)"=GetCachedGeometry!(self:Widget Object Reference=@n11) -> (ReturnValue:"Geometry Structure (by ref)")) -> (ReturnValue:Vector 2D Structure))
   > SetRenderTransform!(
    execute
    self:Widget Object Reference=Get SizeBox_PickupPreview()
    InTransform:Widget Transform Structure=Make Widget Transform!(
     Translation:Vector 2D Structure=MakeVector2D(X=@n12, Y=@n10.Y)
     Scale:Vector 2D Structure=@n9.Scale
     Shear:Vector 2D Structure=@n9.Shear
     Angle:float=@n9.Angle
    ) -> (WidgetTransform:Widget Transform Structure) {is=title K2Node_MakeStruct}
   ) -> (then)
   n13 = GetNumberOfSlotsTaken!(self:Fort Item Object Reference=Get CurrentPickupPreviewItem()) -> (ReturnValue:Byte=0)
   > SetWidthOverride(
    self=Get SizeBox_PickupPreview()
    InWidthOverride="Get (a ref)"(
     Array:"Array of Float (double-precision)s"=Get PickupPreviewWidth() -> (PickupPreviewWidth:"Array of Float (double-precision)s")
     Dimension 1="To Integer (Byte)"!(InByte:Byte=@n13) -> (ReturnValue:int=0)
    ) -> (Output:"Float (double-precision) (by ref)")
   )
   > SetVisibility(self=Get SizeBox_PickupPreview(), InVisibility=HitTestInvisible)
   SetVisibility(execute=[@n8.else, @n6.else, @n2.Is Not Valid], self=Get SizeBox_PickupPreview(), InVisibility=Hidden)
   n14 = Get CurrentPickupPreviewItem()
   n15 = Is Valid(InputObject=@n14)
   > n16 = PreviewItemPickUp(self=Get FortHUDContext(), InItem=@n14, InQuickBarType=Primary) -> (OutSlotIndex=0)
   > n17 = Branch(Condition=@n16.ReturnValue)
   n18 = Select!(
    Option 0:Widget Object Reference=Get QuickbarSlot1()
    Option 1:Widget Object Reference=Get QuickbarSlot2()
    Option 2:Widget Object Reference=Get QuickbarSlot3()
    Option 3:Widget Object Reference=Get QuickbarSlot4()
    Option 4:Widget Object Reference=Get QuickbarSlot5()
    Option 5:Widget Object Reference=Get QuickbarSlot6()
    Index:int=@n16
   ) -> (ReturnValue:Widget Object Reference)
   n19 = BreakVector2D(InVec=GetLocalTopLeft(Geometry=GetTickSpaceGeometry!(self:Widget Object Reference=@n18) -> (ReturnValue:"Geometry Structure (by ref)")))
   n20 = "To Integer (Byte)"(InByte=GetNumberOfSlotsTaken(self=Get CurrentPickupPreviewItem()) -> (ReturnValue=0)) -> (ReturnValue=0)
   n21 = Select(
    Option 0=Get QuickbarSlot1()
    Option 1=Get QuickbarSlot2()
    Option 2=Get QuickbarSlot3()
    Option 3=Get QuickbarSlot4()
    Option 4=Get QuickbarSlot5()
    Option 5=Get QuickbarSlot6()
    Index=int + int(A=int - int(A=@n20, B=1), B=@n16) -> (ReturnValue=0) {is=title K2Node_CommutativeAssociativeBinaryOperator}
   )
   n22 = BreakVector2D(InVec=GetLocalSize!(Geometry:"Geometry Structure (by ref)"=GetTickSpaceGeometry(self=@n21)) -> (ReturnValue:Vector 2D Structure))
   n23 = BreakVector2D(InVec=GetLocalTopLeft(Geometry=GetTickSpaceGeometry(self=@n21)))
   > SetWidthOverride(
    self=Get SizeBox_PickupPreview()
    InWidthOverride=float - float(A=float + float(A=@n23, B=@n22) {is=title K2Node_CommutativeAssociativeBinaryOperator}, B=@n19) {is=title K2Node_CallFunction}
   )
   n24 = float + float(A=@n23.Y, B=@n22.Y)
   n25 = GetTickSpaceGeometry(self=@n18)
   n26 = float + float(A=@n19.Y, B=BreakVector2D(InVec=GetLocalSize(Geometry=@n25)).Y)
   > SetHeightOverride(
    self=Get SizeBox_PickupPreview()
    InHeightOverride=float - float(
     A="Max (Float)"(A=@n24, B=@n26)
     B="Min (Float)"(A:double=@n23.Y, B:double=@n19.Y) -> (ReturnValue:double) {is=title K2Node_CommutativeAssociativeBinaryOperator}
    )
   )
   n27 = Break Widget Transform(WidgetTransform=Get RenderTransform(self=Get SizeBox_PickupPreview())) -> (Angle=~)
   n28 = BreakVector2D(InVec=@n27)
   > SetRenderTransform(
    self=Get SizeBox_PickupPreview()
    InTransform=Make Widget Transform(Translation=MakeVector2D(X=@n19, Y=@n28.Y), Scale=@n27.Scale, Shear=@n27.Shear, Angle=@n27.Angle)
   )
   > SetVisibility(self=Get SizeBox_PickupPreview(), InVisibility=HitTestInvisible)
   SetVisibility(execute=[@n17.else, @n15.Is Not Valid], self=Get SizeBox_PickupPreview(), InVisibility=Hidden)
   "Get (a ref)"(Array=Get PckupPreviewTranslation(), Dimension 1=@n3)
  ]}
 SetSwapableVisuals {nodes [
   n1 = SetSwapableVisuals!() -> (then, QuickbarSlot:WBP BRGame UI Quickbar Slot Primary Object Reference, IsFocused:bool=~) {is=title K2Node_FunctionEntry}
   > SetRenderTranslation(
    self=Get SizeBox_TileContents(self=@n1)
    Translation=Select!(Option 0:Vector 2D Structure=(X=0,Y=-35), Option 1:Vector 2D Structure=(X=0,Y=0), Index:bool=Get IsFocused() -> (IsFocused:bool)) -> (ReturnValue:Vector 2D Structure)
   )
   > n2 = GetDynamicMaterial(self=Get Image_TileImage(self=Get QuickbarSlot() -> (QuickbarSlot:WBP BRGame UI Quickbar Slot Primary Object Reference)))
   > SetScalarParameterValue(self=@n2, ParameterName=IsMasked, Value=SelectFloat(A=1, B=0, bPickA=Get IsFocused()))
   > SetScalarParameterValue(self=@n2, ParameterName=IconYOffset, Value=SelectFloat(A=0, B=0.15, bPickA=Get IsFocused()))
  ]}
 Show Primary Quickbar Rail {nodes [
   n1 = Show Primary Quickbar Rail!() -> (then, Show:bool=~) {is=title K2Node_FunctionEntry}
   > SetVisibility(
    self=Get SizeBox_WeaponsRail() -> (SizeBox_WeaponsRail:Size Box Object Reference)
    InVisibility=Select(Option 0:ESlateVisibility Enum=Hidden, Option 1:ESlateVisibility Enum=SelfHitTestInvisible, Index=@n1) -> (ReturnValue:ESlateVisibility Enum)
   )
  ]}
]
variables [
 AdditionalQuickbarSlotClass {type=Class, default=&/WBP_BRGameUI_QuickbarSlot_Slottable.WBP_BRGameUI_QuickbarSlot_Slottable_C}
 AllowTogglePickAxeTag {type=GameplayTagContainer, default {gameplayTags [{tagName=Gameplay.Status.TogglePickAxeEnabled}]}}
 Async Task Ref {type=AsyncAction_StartListeningToEvent, default=null}
 bPlaylistHidesQuickbar {type=boolean}
 Current Focused Quickbar {type=EFortQuickBars, default=Primary}
 CurrentPickupPreviewItem {type=FortItem, default=null}
 CurrentPreviewGizmoSlot {type=WBP_BRGameUI_QuickbarSlot_Primary_C, default=null}
 CurrentWeaponSlot {type=integer}
 FillRadialInputTimerHandle {type=TimerHandle}
 FortPlayerController {type=FortPlayerController, default=null}
 GizmoSlot {type=WBP_BRGameUI_QuickbarSlot_Primary_C, default=null}
 GizmoSlotMap {default {}}
 HasAdditionalSlot {type=boolean}
 InputHoldFillDuration {type=number}
 InputHoldFillElapsedTime {type=number}
 InputHoldFillRate {type=number, default=0.01}
 IsAdditionalSlotTagAdded {type=boolean}
 IsGizmoSlotFocused {type=boolean}
 IsGizmoSlotPreviewed {type=boolean}
 IsMaxSlotsSet {type=boolean}
 MaximizeAnimHasPlayed {type=boolean}
 MinimizeAnimHasPlayed {type=boolean}
 NonPickaxeSlotsNum {type=integer}
 PlayerController {type=PlayerController, default=null}
 Previous Focused Quickbar {type=EFortQuickBars, default=Primary}
 PreviousWeaponSlot {type=integer}
 QuickbarSlot1TileImageMID {type=MaterialInstanceDynamic}
 QuickbarTag {type=GameplayTag, default {tagName=HUD.Quickbar}}
 RadialHoldMID {type=MaterialInstanceDynamic}
 TotalSlots {type=integer}
]
widgets {
 info {
  inheritedWidgetCount=2
  namedSlotCount=1
  parentClass=/Game/UI/InGame/HUD/QuickbarBase.QuickbarBase_C
  rootWidgetClass=&Overlay
  widgetCount=37
 }
 widgets [
  Overlay_0 {widgetClassPath=&Overlay, widget=Overlay_0}
  SizeBox_0 {
   widgetClassPath=&SizeBox
   parent=Overlay_0
   widget=SizeBox_0
   properties {bOverride_HeightOverride=true, heightOverride=155}
   slot {class=OverlaySlot, properties {horizontalAlignment=HAlign_Fill, verticalAlignment=VAlign_Fill}}
  }
  HorizontalBox_1 {
   widgetClassPath=&HorizontalBox
   parent=SizeBox_0
   widget=HorizontalBox_1
   properties {renderTransformPivot=<1 0.8>}
   slot {class=SizeBoxSlot, properties {horizontalAlignment=HAlign_Right, verticalAlignment=VAlign_Top}}
  }
  InvalidationBox_Slot_1 {widgetClassPath=/Script/UMG.InvalidationBox, parent=HorizontalBox_1, widget=InvalidationBox_Slot_1, slot {class=HorizontalBoxSlot}}
  Slot1Overlay {widgetClassPath=&Overlay, bIsVariable=true, parent=InvalidationBox_Slot_1, widget=Slot1Overlay, slot {class=PanelSlot}}
  QuickbarSlot1_Overlay {widgetClassPath=&Overlay, bIsVariable=true, parent=Slot1Overlay, widget=QuickbarSlot1_Overlay, slot {class=OverlaySlot}}
  Overlay_AdditionalSlot {
   widgetClassPath=&Overlay
   bIsVariable=true
   parent=QuickbarSlot1_Overlay
   widget=Overlay_AdditionalSlot
   properties {bOverride_Cursor=true}
   slot {class=OverlaySlot, properties {horizontalAlignment=HAlign_Fill, verticalAlignment=VAlign_Fill}}
  }
  6thSlot {
   widgetClassPath=&SizeBox
   parent=Overlay_AdditionalSlot
   widget=6thSlot
   properties {bOverride_WidthOverride=true, widthOverride=90}
   slot {class=OverlaySlot, properties {horizontalAlignment=HAlign_Center, verticalAlignment=VAlign_Fill}}
  }
  6thSlot_Overlay {widgetClassPath=&Overlay, bIsVariable=true, parent=6thSlot, widget=6thSlot_Overlay, slot {class=SizeBoxSlot}}
  AdditionalSlot {
   widgetClassPath=&VerticalBox
   bIsVariable=true
   parent=QuickbarSlot1_Overlay
   widget=AdditionalSlot
   slot {class=OverlaySlot, properties {horizontalAlignment=HAlign_Center}}
  }
  QuickbarSlot1 {
   widgetClassPath=&/WBP_BRGameUI_QuickbarSlot_Primary
   bIsVariable=true
   parent=QuickbarSlot1_Overlay
   widget=QuickbarSlot1
   properties {cooldownTypesSupported [Death]}
   slot {class=OverlaySlot}
  }
  Overlay_2 {widgetClassPath=&Overlay, namedSlotHost=QuickbarSlot1, widget=Overlay_2}
  Swap {
   widgetClassPath=&Image
   bIsVariable=true
   parent=Overlay_2
   widget=Swap
   properties {
    visibility=HitTestInvisible
    brush {
     bIsDynamicallyLoaded=false
     drawAs=Image
     imageType=FullColor
     mirroring=NoMirror
     resourceName=None
     resourceObject=/Game/Athena/HUD/Art/BacchusHudIcons/Swap_Normal
     tiling=NoTile
     imageSize=<35 28>
     margin=<ltrb 0 0 0 0>
     outlineSettings {
      bUseBrushTransparency=false
      roundingType=HalfHeightRadius
      width=0
      color {colorUseRule=UseColor_Specified, specifiedColor=<rgba 0 0 0 0>}
      cornerRadii=<0 0 0 0>
     }
     tintColor {colorUseRule=UseColor_Specified, specifiedColor=<rgba 1 1 1 1>}
     uVRegion {bIsValid=false, max=<0 0>, min=<0 0>}
    }
    colorAndOpacity=<rgba 0.5879 0.5879 0.5879 0.85>
   }
   slot {class=OverlaySlot, properties {horizontalAlignment=HAlign_Center, verticalAlignment=VAlign_Center, padding=<ltrb 4 0 0 0>}}
  }
  Overlay_1 {widgetClassPath=&Overlay, parent=HorizontalBox_1, widget=Overlay_1, slot {class=HorizontalBoxSlot}}
  QuickbarPrimary_Underlay_171 {
   widgetClassPath=/Game/UI/ExtensionWidgets/QuickbarPrimary_Underlay
   parent=Overlay_1
   widget=QuickbarPrimary_Underlay_171
   slot {class=OverlaySlot, properties {horizontalAlignment=HAlign_Fill, verticalAlignment=VAlign_Fill}}
  }
  SizeBox_WeaponsRail {
   widgetClassPath=&SizeBox
   bIsVariable=true
   parent=Overlay_1
   widget=SizeBox_WeaponsRail
   properties {bOverride_HeightOverride=true, heightOverride=4, visibility=Hidden}
   slot {class=OverlaySlot, properties {horizontalAlignment=HAlign_Fill, verticalAlignment=VAlign_Bottom, padding=<ltrb 0 0 0 21>}}
  }
  CommonBorder_90 {
   widgetClassPath=&CommonBorder
   parent=SizeBox_WeaponsRail
   widget=CommonBorder_90
   properties {style=/FortUILibrary/Styles/Border/Border_HUD_Line_Horz.Border_HUD_Line_Horz_C}
   slot {class=SizeBoxSlot}
  }
  HorizontalBox_SlotContainer {widgetClassPath=&HorizontalBox, bIsVariable=true, parent=Overlay_1, widget=HorizontalBox_SlotContainer, slot {class=OverlaySlot}}
  InvalidationBox_Slot_2 {
   widgetClassPath=&InvalidationBox
   parent=HorizontalBox_SlotContainer
   widget=InvalidationBox_Slot_2
   slot {class=HorizontalBoxSlot}
  }
  Slot2Overlay {widgetClassPath=&Overlay, bIsVariable=true, parent=InvalidationBox_Slot_2, widget=Slot2Overlay, slot {class=PanelSlot}}
  QuickbarSlot2 {
   widgetClassPath=&WBP_BRGameUI_QuickbarSlot_Primary
   bIsVariable=true
   parent=Slot2Overlay
   widget=QuickbarSlot2
   properties {
    bShowBottomArrows=true
    slotIndex=1
    cooldownTypesSupported [Death, WeaponReloading, ItemActivation, AthenaWeaponFireCooldown, AmmoRegeneration]
   }
   slot {class=OverlaySlot}
  }
  InvalidationBox_Slot_3 {
   widgetClassPath=&InvalidationBox
   parent=HorizontalBox_SlotContainer
   widget=InvalidationBox_Slot_3
   slot {class=HorizontalBoxSlot}
  }
  Slot3Overlay {widgetClassPath=&Overlay, bIsVariable=true, parent=InvalidationBox_Slot_3, widget=Slot3Overlay, slot {class=PanelSlot}}
  QuickbarSlot3 {
   widgetClassPath=&WBP_BRGameUI_QuickbarSlot_Primary
   bIsVariable=true
   parent=Slot3Overlay
   widget=QuickbarSlot3
   properties {
    bCheckVisibilityBasedOffInventoryCapacity=true
    bShowBottomArrows=true
    slotIndex=2
    cooldownTypesSupported [Death, WeaponReloading, ItemActivation, AthenaWeaponFireCooldown, AmmoRegeneration]
   }
   slot {class=OverlaySlot}
  }
  InvalidationBox_Slot_4 {
   widgetClassPath=&InvalidationBox
   parent=HorizontalBox_SlotContainer
   widget=InvalidationBox_Slot_4
   slot {class=HorizontalBoxSlot}
  }
  Slot4Overlay {widgetClassPath=&Overlay, bIsVariable=true, parent=InvalidationBox_Slot_4, widget=Slot4Overlay, slot {class=PanelSlot}}
  QuickbarSlot4 {
   widgetClassPath=&WBP_BRGameUI_QuickbarSlot_Primary
   bIsVariable=true
   parent=Slot4Overlay
   widget=QuickbarSlot4
   properties {
    bCheckVisibilityBasedOffInventoryCapacity=true
    bShowBottomArrows=true
    slotIndex=3
    cooldownTypesSupported [Death, WeaponReloading, ItemActivation, AthenaWeaponFireCooldown, AmmoRegeneration]
   }
   slot {class=OverlaySlot}
  }
  InvalidationBox_Slot_5 {
   widgetClassPath=&InvalidationBox
   parent=HorizontalBox_SlotContainer
   widget=InvalidationBox_Slot_5
   slot {class=HorizontalBoxSlot}
  }
  Slot5Overlay {widgetClassPath=&Overlay, bIsVariable=true, parent=InvalidationBox_Slot_5, widget=Slot5Overlay, slot {class=PanelSlot}}
  QuickbarSlot5 {
   widgetClassPath=&WBP_BRGameUI_QuickbarSlot_Primary
   bIsVariable=true
   parent=Slot5Overlay
   widget=QuickbarSlot5
   properties {
    bCheckVisibilityBasedOffInventoryCapacity=true
    bShowBottomArrows=true
    slotIndex=4
    cooldownTypesSupported [Death, WeaponReloading, ItemActivation, AthenaWeaponFireCooldown, AmmoRegeneration]
   }
   slot {class=OverlaySlot}
  }
  InvalidationBox_Slot_6 {
   widgetClassPath=&InvalidationBox
   parent=HorizontalBox_SlotContainer
   widget=InvalidationBox_Slot_6
   slot {class=HorizontalBoxSlot}
  }
  Slot6Overlay {widgetClassPath=&Overlay, bIsVariable=true, parent=InvalidationBox_Slot_6, widget=Slot6Overlay, slot {class=PanelSlot}}
  QuickbarSlot6 {
   widgetClassPath=&WBP_BRGameUI_QuickbarSlot_Primary
   bIsVariable=true
   parent=Slot6Overlay
   widget=QuickbarSlot6
   properties {
    bCheckVisibilityBasedOffInventoryCapacity=true
    bShowBottomArrows=true
    slotIndex=5
    cooldownTypesSupported [Death, WeaponReloading, ItemActivation, AthenaWeaponFireCooldown, AmmoRegeneration]
   }
   slot {class=OverlaySlot}
  }
  SizeBox_PickupPreview {
   widgetClassPath=&SizeBox
   bIsVariable=true
   parent=Overlay_1
   widget=SizeBox_PickupPreview
   properties {bOverride_HeightOverride=true, bOverride_WidthOverride=true, heightOverride=110, visibility=Hidden, widthOverride=95}
   slot {class=OverlaySlot}
  }
  Image_PickupPreview {
   widgetClassPath=&Image
   bIsVariable=true
   parent=SizeBox_PickupPreview
   widget=Image_PickupPreview
   properties {
    brush {
     bIsDynamicallyLoaded=false
     drawAs=Image
     imageType=NoImage
     mirroring=NoMirror
     resourceName=None
     resourceObject=/BattleRoyaleGameUI/HUD/QuickBar/Assets/M_UI_QuickBarPickupPreview
     tiling=NoTile
     imageSize=<93 92>
     margin=<ltrb 0 0 0 0>
     outlineSettings {
      bUseBrushTransparency=false
      roundingType=HalfHeightRadius
      width=0
      color {colorUseRule=UseColor_Specified, specifiedColor=<rgba 0 0 0 0>}
      cornerRadii=<0 0 0 0>
     }
     tintColor {colorUseRule=UseColor_Specified, specifiedColor=<rgba 1 1 1 1>}
     uVRegion {bIsValid=false, max=<0 0>, min=<0 0>}
    }
    renderTransform {angle=0, scale=<1.8 1.8>, shear=<0 0>, translation=<0 0>}
   }
   slot {class=SizeBoxSlot, properties {horizontalAlignment=HAlign_Center, verticalAlignment=VAlign_Center}}
  }
  TutorialHighlight {
   widgetClassPath=&TutorialHighlight
   bIsVariable=true
   parent=Overlay_0
   widget=TutorialHighlight
   properties {customHighlightWidget=/Game/Athena/HUD/Tutorial/Tutorial_Outline_Square_Static.Tutorial_Outline_Square_Static_C}
   slot {class=OverlaySlot, properties {horizontalAlignment=HAlign_Fill, verticalAlignment=VAlign_Fill}}
  }
  KeybindWidget_SwitchModeKeybind {widgetClassPath=/Script/FortniteUI.FortKeybindWidget, bInherited=true, widget=null}
 ]
}
