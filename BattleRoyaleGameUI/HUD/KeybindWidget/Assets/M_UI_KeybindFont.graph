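# M_UI_KeybindFont is a translucent UI material, by its name for font text in the HUD keybind widget.
# It takes the Bright color from the shared MF_UI_AdaptiveColors function.
# A vertical UV gradient, flippable through FlipGradient, blends TopColorAmount and BottomColorAmount into the alpha of a final color lerp.
# Both lerp inputs are the same Bright output, so the gradient leaves the color unchanged.
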Material /BattleRoyaleGameUI/HUD/KeybindWidget/Assets/M_UI_KeybindFont
graph {
 properties {bAutomaticallySetUsageInEditor=false, blendMode=BLEND_Translucent, materialDomain=MD_UI, usedShadingModels=Default Lit}
 nodes [
  n1 = MaterialFunctionCall!(BufferSource) {materialFunction=/BattleRoyaleGameUI/Shared/MF_UI_AdaptiveColors}
  n2 = ComponentMask(None=TextureCoordinate()) {g=true}
  n3 = OneMinus(None=@n2)
  n4 = ScalarParameter() {defaultValue=1, parameterName=FlipGradient}
  LinearInterpolate(
   A=@n1.Bright
   B=@n1.Bright
   Alpha=LinearInterpolate(
    A=ScalarParameter() {parameterName=TopColorAmount}
    B=ScalarParameter() {defaultValue=1, parameterName=BottomColorAmount}
    Alpha=LinearInterpolate(A=@n2, B=@n3, Alpha=@n4)
   )
  )
 ]
}
