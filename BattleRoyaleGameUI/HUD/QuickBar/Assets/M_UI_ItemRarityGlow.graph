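# UI material that draws an item rarity glow for a quick bar slot, visible as IsSelected rises.
# A box SDF on scaled UVs, shaped by easing and a vertical gradient, forms the glow mask.
# The mask blends Color_A to Color_B, with IsAmmoEmpty shifting toward dimmed red from the shared parameter collection.
# Alpha-composite blending uses premultiplied color, with AdditiveRatio reducing opacity for a more additive look.
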Material /BattleRoyaleGameUI/HUD/QuickBar/Assets/M_UI_ItemRarityGlow
graph {
 properties {bAutomaticallySetUsageInEditor=false, blendMode=BLEND_AlphaComposite, materialDomain=MD_UI, usedShadingModels=Default Lit}
 nodes [
  n1 = MaterialFunctionCall() {materialFunction=&GetUserInterfaceUV}
  n2 = MaterialFunctionCall(UV=@n1.Normalized UV) {materialFunction=&MF_UI_UVScaling}
  n3 = LinearInterpolate(A=@n2.ScaleX, B=@n2.ScaleX, Alpha=ScalarParameter() {parameterName=UVScalingDirection})
  n4 = Saturate(None=MaterialFunctionCall(
    UVs=StaticSwitchParameter(True=@n3, False=@n2.Scaled to Fit) {parameterName=SelectUVScalingType}
    Size=ScalarParameter() {defaultValue=0.15, parameterName=GlowWidth}
   ) {materialFunction=&MF_UI_Box}.SDF)
  n5 = ComponentMask(None=@n1.Normalized UV) {g=true}
  n6 = Multiply(
   A=MaterialFunctionCall("Value (0-1)"=OneMinus(None=@n4)) {materialFunction=&MF_UI_EaseCurves}.EaseInQuint
   B=SmoothStep(
    Min=ScalarParameter() {defaultValue=-0.25, parameterName=VerticalGradient_EndPoint}
    Value=MaterialFunctionCall("Value (0-1)"=@n5) {materialFunction=&MF_UI_EaseCurves}.EaseInOutSine
   ) {constMin=-0.5}
  )
  n7 = CollectionParameter() {collection=/BattleRoyaleGameUI/Shared/MPC_BRGameUI, parameterName=Red}
  n8 = ScalarParameter() {defaultValue=0.75, parameterName=EmptyColorDim}
  n9 = LinearInterpolate(
   A=LinearInterpolate(
    A=VectorParameter() {parameterName=Color_A, defaultValue=<rgba 0.05 0.05 0.05 1>}.RGB
    B=VectorParameter() {parameterName=Color_B, defaultValue=<rgba 0.5 0.5 0.5 1>}.RGB
    Alpha=@n6
   )
   B=Multiply(A=ComponentMask(None=@n7) {b=true, g=true, r=true}, B=OneMinus(None=@n8)) {constB=0.5}
   Alpha=ScalarParameter() {parameterName=IsAmmoEmpty}
  ) {constB=0}
  n10 = ScalarParameter() {defaultValue=0.5, parameterName=DimColor_Stop}
  n11 = ScalarParameter() {parameterName=IsSelected}
  n12 = LinearInterpolate(A=@n9, B=Multiply(A=LinearInterpolate(B=@n10, Alpha=@n11), B=@n9), Alpha=ScalarParameter() {parameterName=IsSelected})
  n13 = ScalarParameter() {parameterName=IsSelected}
  Multiply(A=VertexColor().A, B=Multiply(A=Multiply(A=@n12, B=@n6), B=LinearInterpolate(Alpha=@n13)))
  n14 = ScalarParameter() {defaultValue=0.75, parameterName=AdditiveRatio}
  Multiply(A=Multiply(A=@n6, B=OneMinus(None=@n14)), B=LinearInterpolate(Alpha=ScalarParameter() {parameterName=IsSelected}))
 ]
}
