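# Translucent UI material for the quickbar's no-build indicator: a build icon inside a ringed circle with a prohibition slash.
# Circle functions form the fill and ring, while rotated distance fields joined by SDF union form an antialiased slash and a wider crop that cuts the icon.
# These masks blend adaptive colors and separate fill, ring and icon opacities.
# AnimateIconColor shifts ring and slash toward collection red while adjusting their thickness and icon scale.
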Material /BattleRoyaleGameUI/HUD/QuickBar/Assets/M_UI_NoBuildIndicator
graph {
 properties {blendMode=BLEND_Translucent, materialDomain=MD_UI, pixelDepthOffsetMode=PDOM_Legacy, usedShadingModels=Default Lit}
 nodes [
  n1 = MaterialFunctionCall!(BufferSource) {materialFunction=/BattleRoyaleGameUI/Shared/MF_UI_AdaptiveColors}
  n2 = ComponentMask(None=CollectionParameter() {collection=/BattleRoyaleGameUI/Shared/MPC_BRGameUI, parameterName=Red}) {b=true, g=true, r=true}
  n3 = ScalarParameter() {parameterName=AnimateIconColor}
  n4 = Add(
   A=MaterialFunctionCall() {materialFunction=&GetUserInterfaceUV}.Normalized UV
   B=AppendVector(A=ScalarParameter() {defaultValue=0.0125, parameterName=RingOffset_X}, B=ScalarParameter() {defaultValue=0.005, parameterName=RingOffset_Y})
  )
  n5 = ScalarParameter() {defaultValue=0.825, group=Ring, parameterName=InnerRadius}
  n6 = LinearInterpolate(A=@n5, B=Subtract(A=@n5) {constB=0.025}, Alpha=ScalarParameter() {parameterName=AnimateIconColor})
  n7 = MaterialFunctionCall(UVs=TextureCoordinate(), "Rotation Angle (0-1)"=ScalarParameter() {defaultValue=-0.125, parameterName=CrossRotation}) {materialFunction=&CustomRotator}
  n8 = Add(A=@n7.Rotated Values) {constB=-0.5}
  n9 = ScalarParameter() {defaultValue=0.05, parameterName=Cross_Thickness}
  n10 = LinearInterpolate(A=@n9, B=Subtract(A=@n9) {constB=-0.0125}, Alpha=ScalarParameter() {parameterName=AnimateIconColor})
  n11 = ScalarParameter() {defaultValue=-0.0125, parameterName=CrossRotation_2}
  n12 = Add(A=OneMinus(None=MaterialFunctionCall(UVs=@n7.Rotated Values, "Rotation Angle (0-1)"=@n11) {materialFunction=&CustomRotator}.Rotated Values)) {constB=-0.5}
  n13 = MaterialFunctionCall!(
   DistanceFieldA=ComponentMask(None=Add(A=@n8, B=@n10) {constB=-0.5}) {r=true}
   DistanceFieldB=ComponentMask(None=Add(A=@n12, B=@n10) {constB=-0.5}) {r=true}
   BlendAmount
  ) {materialFunction=/FortUILibrary/MaterialFunctions/Shapes/Modifiers/MF_UI_SDF_Union}
  n14 = LinearInterpolate(
   A=OneMinus(None=MaterialFunctionCall(UVs=@n4, Size=@n6) {materialFunction=&MF_UI_Circle}.Fill)
   Alpha=Saturate(None=MaterialFunctionCall(DistanceField=@n13.Exact) {materialFunction=&MF_UI_AntiAliasing}.Outer)
  )
  n15 = ScalarParameter() {defaultValue=0.85, group=Icon, parameterName=IconScale}
  n16 = MaterialFunctionCall(
   UVs=TextureCoordinate()
   Texture Scale=LinearInterpolate(A=@n15, B=Subtract(A=@n15) {constB=0.15}, Alpha=ScalarParameter() {parameterName=AnimateIconColor})
  ) {materialFunction=&ScaleUVsByCenter}
  n17 = ScalarParameter() {defaultValue=0.085, parameterName=Crop_Thickness}
  n18 = LinearInterpolate(A=@n17, B=Subtract(A=@n17) {constB=-0.0125}, Alpha=ScalarParameter() {parameterName=AnimateIconColor})
  n19 = Add(A=@n8, B=@n18) {constB=-0.5}
  n20 = Add(A=@n12, B=@n18) {constB=-0.5}
  n21 = MaterialFunctionCall(DistanceField=MaterialFunctionCall(DistanceFieldA=ComponentMask(None=@n19) {r=true}, DistanceFieldB=ComponentMask(None=@n20) {r=true}) {materialFunction=&MF_UI_SDF_Union}.Exact) {materialFunction=&MF_UI_AntiAliasing}
  n22 = Saturate(None=Multiply(
    A=TextureSample(UVs=@n16.UVs) {samplerSource=SSM_Clamp_WorldGroupSettings, texture=/Game/UI/Icons/Icon-QuickbarMode-Build-M}.A
    B=Saturate(None=@n21.Inner)
   ))
  LinearInterpolate(
   A=LinearInterpolate(A=@n1.Dark, B=LinearInterpolate(A=@n1.Bright, B=@n2, Alpha=@n3), Alpha=@n14)
   B=MaterialFunctionCall() {materialFunction=&MF_UI_AdaptiveColors}.White
   Alpha=@n22
  )
  n23 = ScalarParameter() {defaultValue=0.75, group=Fill, parameterName=FillAlpha}
  n24 = ScalarParameter() {defaultValue=1, group=Ring, parameterName=RingAlpha}
  Multiply(
   A=MaterialFunctionCall(Size=Constant() {r=1}) {materialFunction=&MF_UI_Circle}.Fill
   B=LinearInterpolate(A=LinearInterpolate(A=@n23, B=@n24, Alpha=@n14), B=ScalarParameter() {defaultValue=1, group=Icon, parameterName=IconAlpha}, Alpha=@n22)
  )
 ]
}
