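# UI material for the Battle Royale quick bar's type icon, drawing a tinted icon with alpha-composite blending.
# The icon texture is rotated and center-scaled, with its mask selectable between the alpha and green channels.
# Shared palette colors form Blue-to-Cyan and Cyan-to-White gradients over the icon mask, crossfaded by Anim_Color.
# Color is premultiplied by mask, opacity and vertex alpha, while AdditiveMixRatio lowers alpha for a partly additive glow.
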Material /BattleRoyaleGameUI/HUD/QuickBar/Assets/M_UI_QuickbarTypeIcon
graph {
 properties {bAutomaticallySetUsageInEditor=false, blendMode=BLEND_AlphaComposite, materialDomain=MD_UI, usedShadingModels=Default Lit}
 nodes [
  n1 = TextureCoordinate()
  n2 = ScalarParameter() {parameterName=Rotation}
  n3 = TextureSampleParameter2D(UVs=MaterialFunctionCall(
    UVs=MaterialFunctionCall(UVs=@n1, "Rotation Angle (0-1)"=@n2) {materialFunction=&CustomRotator}.Rotated Values
    Texture Scale=ScalarParameter() {defaultValue=0.9, parameterName=OverallIconScale}
   ) {materialFunction=&ScaleUVsByCenter}.UVs) {parameterName=Icon, samplerSource=SSM_Clamp_WorldGroupSettings, texture=/Game/UI/Icons/Icon-QuickbarMode-Fight-M}
  n4 = LinearInterpolate(
   A=NamedRerouteDeclaration(None=@n3.A) {name=IconMask, nodeColor=<rgba 0 0 0 1>}
   B=@n3.G
   Alpha=ScalarParameter() {parameterName=SelectChannelAsAlpha}
  )
  n5 = ScalarParameter() {defaultValue=0.25, parameterName=AdditiveMixRatio}
  Multiply(A=Multiply(A=@n4, B=OneMinus(None=@n5)), B=ScalarParameter() {defaultValue=0.9, parameterName=Opacity})
  n6 = ComponentMask(None=CollectionParameter() {collection=/BattleRoyaleGameUI/Shared/MPC_BRGameUI, parameterName=Cyan}) {b=true, g=true, r=true}
  n7 = LinearInterpolate(
   A=ComponentMask(None=CollectionParameter() {collection=&MPC_BRGameUI, parameterName=Blue}) {b=true, g=true, r=true}
   B=@n6
   Alpha=NamedRerouteUsage()
  )
  n8 = LinearInterpolate(
   A=@n6
   B=ComponentMask(None=CollectionParameter() {collection=&MPC_BRGameUI, parameterName=White}) {b=true, g=true, r=true}
   Alpha=NamedRerouteUsage()
  )
  n9 = ScalarParameter() {parameterName=Anim_Color}
  n10 = ScalarParameter() {defaultValue=0.9, parameterName=Opacity}
  Multiply(A=VertexColor().A, B=Multiply(A=LinearInterpolate(A=@n7, B=@n8, Alpha=@n9), B=Multiply(A=@n4, B=@n10)))
 ]
}
