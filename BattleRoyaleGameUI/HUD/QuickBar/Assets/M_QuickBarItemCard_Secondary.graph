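# UI material for secondary quickbar item cards, drawing a build piece or trap icon with selected, trap and out-of-materials looks.
# IsSelected shifts and rescales the icon and blends three-color gradients from shared UI collection colors, with NoMats switching to reds.
# IsTrap swaps in a separate color blend and circular SDF glow.
# Alpha-composite blending uses per-state ratios to reduce alpha for partly additive icons and glow.
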Material /BattleRoyaleGameUI/HUD/QuickBar/Assets/M_QuickBarItemCard_Secondary
graph {
 properties {bAutomaticallySetUsageInEditor=false, blendMode=BLEND_AlphaComposite, materialDomain=MD_UI, usedShadingModels=Default Lit}
 nodes [
  n1 = NamedRerouteUsage()
  n2 = ScalarParameter() {defaultValue=0.5, parameterName=TrapGlowAdditiveRatio_Selected}
  LinearInterpolate(A=Multiply(A=@n1), B=Multiply(A=@n1, B=OneMinus(None=@n2)), Alpha=ScalarParameter() {parameterName=IsSelected})
  n3 = NamedRerouteUsage()
  n4 = MaterialFunctionCall("Value (0-1)"=@n3) {materialFunction=&MF_UI_EaseCurves}
  LinearInterpolate(A=@n3, B=@n4.EaseInSine)
  n5 = LinearInterpolate(
   A=ScalarParameter() {defaultValue=0.75, parameterName=AdditiveMixRatioUnSelected}
   B=ScalarParameter() {defaultValue=0.125, parameterName=AdditiveMixRatioSelected}
   Alpha=ScalarParameter() {parameterName=IsSelected}
  )
  n6 = NamedRerouteUsage()
  n7 = MaterialFunctionCall() {materialFunction=&MF_UI_UVScaling}
  n8 = ComponentMask(None=@n7.ScaleY) {r=true}
  n9 = Subtract(
   A=ComponentMask(None=@n7.ScaleY) {g=true}
   B=LinearInterpolate(
    A=ScalarParameter() {defaultValue=0.025, parameterName=Unselected_Y_Position}
    B=ScalarParameter() {defaultValue=-0.0125, parameterName=Selected_Y_Position}
    Alpha=ScalarParameter() {parameterName=IsSelected}
   ) {constA=-0.05, constB=-0.2}
  )
  n10 = ScalarParameter() {defaultValue=0.875, parameterName=Unselected_Icon_Scale}
  n11 = ScalarParameter() {defaultValue=0.9, parameterName=Selected_Icon_Scale}
  n12 = ScalarParameter() {parameterName=IsSelected}
  n13 = MaterialFunctionCall(
   UVs=NamedRerouteDeclaration(None=AppendVector(A=@n8, B=@n9)) {name=IconUVs, nodeColor=<rgba 0 0 0 1>}
   Texture Scale=NamedRerouteDeclaration(None=LinearInterpolate(A=@n10, B=@n11, Alpha=@n12) {constA=0.6667, constB=0.75}) {name=IconScaleLogic, nodeColor=<rgba 0 0 0 1>}
  ) {materialFunction=&ScaleUVsByCenter}
  n14 = TextureSample(UVs=@n13.UVs, Tex=NamedRerouteUsage()) {samplerSource=SSM_Clamp_WorldGroupSettings}
  n15 = NamedRerouteDeclaration(None=@n14.A) {name=IconAlpha, nodeColor=<rgba 0 0 0 1>}
  n16 = LinearInterpolate(
   A=Multiply(A=Multiply(A=@n6, B=@n15), B=@n13.0-1 mask)
   B=Multiply(A=@n14.A, B=@n13.0-1 mask)
   Alpha=LinearInterpolate(
    A=ScalarParameter() {defaultValue=0.25, parameterName=DesatAlphaMaskMix_Unselected}
    B=ScalarParameter() {defaultValue=-0.25, parameterName=DesatAlphaMaskMix_Selected}
    Alpha=ScalarParameter() {parameterName=IsSelected}
   )
  )
  n17 = NamedRerouteUsage()
  n18 = NamedRerouteUsage()
  n19 = Multiply(
   A=LinearInterpolate(
    A=NamedRerouteDeclaration(None=@n16) {name=MaskAdjusted, nodeColor=<rgba 0 0 0 1>}
    B=LinearInterpolate(A=@n17, Alpha=@n18)
    Alpha=ScalarParameter() {parameterName=IsTrap}
   )
   B=ScalarParameter() {defaultValue=1, parameterName=IsIconLoaded}
  )
  n20 = NamedRerouteUsage()
  n21 = ScalarParameter() {parameterName=TrapGlowAdditiveRatio_Unselected}
  n22 = ScalarParameter() {defaultValue=0.5, parameterName=TrapGlowAdditiveRatio_Selected}
  n23 = LinearInterpolate(A=Multiply(A=@n20, B=OneMinus(None=@n21)), B=Multiply(A=@n20, B=OneMinus(None=@n22)), Alpha=ScalarParameter() {parameterName=IsSelected})
  n24 = LinearInterpolate(
   A=OneMinus(None=ScalarParameter() {defaultValue=0.25, parameterName=TrapIconAdditiveRatio_Unselected})
   B=OneMinus(None=ScalarParameter() {defaultValue=0.0625, parameterName=TrapIconAdditiveRatio_Selected})
   Alpha=ScalarParameter() {parameterName=IsSelected}
  )
  n25 = NamedRerouteUsage()
  LinearInterpolate(
   A=Multiply(A=@n19, B=OneMinus(None=@n5))
   B=Multiply(A=@n19, B=LinearInterpolate(A=@n23, B=@n24, Alpha=@n25))
   Alpha=ScalarParameter() {parameterName=IsTrap}
  )
  NamedRerouteDeclaration(None=Desaturation(None=@n14.RGB, Fraction=Constant() {r=1})) {name=DesatMask, nodeColor=<rgba 0 0 0 1>}
  n26 = ComponentMask(None=CollectionParameter() {collection=/BattleRoyaleGameUI/Shared/MPC_BRGameUI, parameterName=Blue}) {b=true, g=true, r=true}
  n27 = ComponentMask(None=CollectionParameter() {collection=&MPC_BRGameUI, parameterName=Cyan}) {b=true, g=true, r=true}
  n28 = NamedRerouteUsage()
  n29 = Multiply(
   A=@n4.EaseInSine
   B=LinearInterpolate(
    A=ScalarParameter() {defaultValue=1, parameterName=UnselectedTopFadeAmount}
    B=ScalarParameter() {defaultValue=0.5, parameterName=UnselectedBottomFadeAmount}
    Alpha=OneMinus(None=@n28)
   ) {constA=1}
  )
  n30 = ComponentMask(None=CollectionParameter() {collection=&MPC_BRGameUI, parameterName=Red}) {b=true, g=true, r=true}
  n31 = Multiply(A=@n30, B=ScalarParameter() {defaultValue=0.125})
  n32 = LinearInterpolate(
   A=MaterialFunctionCall(A=VectorParameter() {parameterName=UnselectedColor_A, defaultValue=<rgba 0 0 0.25 1>}.RGB, B=@n26, C=@n27, Alpha=@n29) {materialFunction=&Lerp_3Color}.Result
   B=MaterialFunctionCall(A=VectorParameter() {parameterName=NoMatColor3, defaultValue=<rgba 0.075 0 0.0312 1>}.RGB, B=@n31, C=@n30, Alpha=@n29) {materialFunction=&Lerp_3Color}.Result
   Alpha=ScalarParameter() {parameterName=NoMats}
  )
  n33 = ComponentMask(None=CollectionParameter() {collection=&MPC_BRGameUI, parameterName=White}) {b=true, g=true, r=true}
  n34 = Multiply(
   A=LinearInterpolate(B=ScalarParameter() {defaultValue=1, parameterName=SelectedBottomFadeAmount}, Alpha=NamedRerouteUsage()) {constA=1}
   B=NamedRerouteUsage()
  )
  n35 = LinearInterpolate(
   A=MaterialFunctionCall(A=@n26, B=@n27, C=@n33, Alpha=@n34) {materialFunction=&Lerp_3Color}.Result
   B=MaterialFunctionCall(A=@n31, B=@n30, C=@n33, Alpha=@n34) {materialFunction=&Lerp_3Color}.Result
   Alpha=ScalarParameter() {parameterName=NoMats}
  )
  n36 = ScalarParameter() {parameterName=IsSelected}
  n37 = LinearInterpolate(
   A=VectorParameter() {parameterName=Color_2, defaultValue=<rgba 0.0006 0.0782 0.0009 1>}.RGB
   B=VectorParameter() {parameterName=Color_1, defaultValue=<rgba 0.1195 0.521 0 1>}.RGB
   Alpha=NamedRerouteUsage()
  )
  n38 = NamedRerouteUsage()
  n39 = Multiply(A=NamedRerouteUsage(), B=OneMinus(None=ScalarParameter() {defaultValue=0.05, parameterName=TrapRarityColorMix}))
  NamedRerouteDeclaration(None=LinearInterpolate(
    A=LinearInterpolate(A=@n32, B=@n35, Alpha=@n36)
    B=LinearInterpolate(A=@n37, B=@n38, Alpha=@n39)
    Alpha=ScalarParameter() {parameterName=IsTrap}
   )) {name=Color, nodeColor=<rgba 0 0 0 1>}
  NamedRerouteDeclaration(None=TextureObjectParameter() {parameterName=Icon, texture=/Game/UI/Foundation/Textures/Icons/HUD/T_BuildMode_Prefab_Wall_128}) {name=IconTexture, nodeColor=<rgba 0 0 0 1>}
  NamedRerouteDeclaration(None=@n14.RGB) {name=IconRGB, nodeColor=<rgba 0 0 0 1>}
  n40 = MaterialFunctionCall() {materialFunction=&MF_UI_UVScaling}
  n41 = ScalarParameter() {defaultValue=1, parameterName=TrapGlowSize}
  n42 = OneMinus(None=SmoothStep(Value=MaterialFunctionCall(UVs=@n40.Scaled to Fit, Size=@n41) {materialFunction=&MF_UI_Circle}.SDF) {constMax=0, constMin=-1})
  NamedRerouteDeclaration(None=Saturate(None=MaterialFunctionCall("Value (0-1)"=@n42) {materialFunction=&MF_UI_EaseCurves}.EaseInSine)) {name=TrapGlow, nodeColor=<rgba 0 0 0 1>}
  NamedRerouteDeclaration(None=ComponentMask(None=@n13.UVs) {g=true}) {name=MaskG, nodeColor=<rgba 0 0 0 1>}
  OneMinus(None=ScalarParameter() {parameterName=TrapGlowAdditiveRatio_Unselected})
  n43 = NamedRerouteUsage()
  n44 = LinearInterpolate(
   A=OneMinus(None=ScalarParameter() {defaultValue=0.25, parameterName=TrapIconAdditiveRatio_Unselected})
   B=OneMinus(None=ScalarParameter() {defaultValue=0.0625, parameterName=TrapIconAdditiveRatio_Selected})
   Alpha=ScalarParameter() {parameterName=IsSelected}
  )
  n45 = NamedRerouteUsage()
  n46 = LinearInterpolate(A=@n19, B=Multiply(A=LinearInterpolate(A=@n1, B=@n44, Alpha=@n45) {constA=1}, B=@n19), Alpha=ScalarParameter() {parameterName=IsTrap})
  Reroute(None=Multiply(A=VertexColor().A, B=Multiply(A=@n43, B=@n46)))
  TextureObject() {texture=/Game/Athena/Items/Traps/Launchpad/Textures/T-Icon_Trap_Athena_Launchpad}
  VectorParameter() {parameterName=UnselectedColor_C, defaultValue=<rgba 0 0.6667 1 1>}
 ]
}
