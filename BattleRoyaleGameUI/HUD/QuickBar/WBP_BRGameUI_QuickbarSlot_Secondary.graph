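# Secondary quickbar slot widget for the Battle Royale HUD, showing an item's icon, quantity and rarity colors.
# Quickbar change events refresh the slot, which async-loads icons into a dynamic tile material and reads rarity and counts from an InstituteChirp viewmodel.
# Selection plays scale-and-glow animations, skippable on simplified non-touch HUDs, and offsets extension widgets.
# Touch overrides handle slot selection through a rollback CVar, and size lerps by an animated scale factor.
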WidgetBlueprint /BattleRoyaleGameUI/HUD/QuickBar/WBP_BRGameUI_QuickbarSlot_Secondary
animations [
 Selected {
  binding_count=3
  display_label=""
  length_seconds=0.4
  object=&self:Selected
  sequence {
   bindings [
    Overlay_TileContents {
     id=#1
     possessed_class=Overlay
     tracks [{
       type=MovieScene2DTransformTrack
       sections [{
         type=MovieScene2DTransformSection
         channels [
          Translation.X {type=MovieSceneScriptingFloatChannel, default=0}
          Translation.Y {type=MovieSceneScriptingFloatChannel, default=0, keys [[0,0],[12,-4]]}
          Angle {type=MovieSceneScriptingFloatChannel}
          Scale.X {type=MovieSceneScriptingFloatChannel, keys [[0,1],[12,1.1875]]}
          Scale.Y {type=MovieSceneScriptingFloatChannel, keys [[0,1],[12,1.1875]]}
          Shear.X {type=MovieSceneScriptingFloatChannel}
          Shear.Y {type=MovieSceneScriptingFloatChannel}
         ]
        }]
      }]
    }
    Image_TileImage {
     id=#2
     possessed_class=Image
     tracks [{
       type=MovieSceneWidgetMaterialTrack
       sections [{
         type=MovieSceneParameterSection
         end_frame=24
         start_frame=0
         channels [IsSelected {type=MovieSceneScriptingFloatChannel, keys [[0,0],[6,1.5],[24,1]]}]
        }]
      }]
    }
    Image_SelectedGlow {
     id=#3
     possessed_class=Image
     tracks [{
       type=MovieSceneWidgetMaterialTrack
       sections [{type=MovieSceneParameterSection, channels [IsSelected {type=MovieSceneScriptingFloatChannel, keys [[0,0],[3,2],[24,1]]}]}]
      }]
    }
   ]
   display_rate {numerator=60}
  }
 }
 Unselected {
  binding_count=3
  length_seconds=0.1167
  object=&self:Unselected
  sequence {
   bindings [
    Image_TileImage {
     id=#4
     possessed_class=Image
     tracks [{
       type=MovieSceneWidgetMaterialTrack
       sections [{
         type=MovieSceneParameterSection
         end_frame=7
         start_frame=0
         channels [IsSelected {type=MovieSceneScriptingFloatChannel, keys [[0,1],[7,0]]}]
        }]
      }]
    }
    Overlay_TileContents {
     id=#5
     display_name=Overlay_0
     possessed_class=Overlay
     tracks [{
       type=MovieScene2DTransformTrack
       sections [{
         type=MovieScene2DTransformSection
         channels [
          Translation.X {type=MovieSceneScriptingFloatChannel, default=0}
          Translation.Y {type=MovieSceneScriptingFloatChannel, default=0, keys [[0,-4],[7,0]]}
          Angle {type=MovieSceneScriptingFloatChannel}
          Scale.X {type=MovieSceneScriptingFloatChannel, keys [[0,1.1875],[7,1]]}
          Scale.Y {type=MovieSceneScriptingFloatChannel, keys [[0,1.1875],[7,1]]}
          Shear.X {type=MovieSceneScriptingFloatChannel}
          Shear.Y {type=MovieSceneScriptingFloatChannel}
         ]
        }]
      }]
    }
    Image_SelectedGlow {
     id=#3
     possessed_class=Image
     tracks [{
       type=MovieSceneWidgetMaterialTrack
       sections [{type=MovieSceneParameterSection, channels [IsSelected {type=MovieSceneScriptingFloatChannel, keys [[0,1],[7,0]]}]}]
      }]
    }
   ]
   display_rate {numerator=60}
  }
 }
]
graphs [
 EventGraph {nodes [
   n1 = Handle OnQuickbarContentsChanged!() -> (OutputDelegate, then, QuickbarIndex:EFortQuickBars Enum=Primary, ChangedSlots:Array of Integers) {is=title K2Node_CustomEvent}
   > RefreshItemIfSlotChanged!(
    execute
    self:Fort Quick Bar Slot Base Object Reference
    InQuickBar:EFortQuickBars Enum=@n1.QuickbarIndex
    InSlot:Array of Integers=@n1.ChangedSlots
   ) -> (then)
   n2 = HandleItemIconChanged!() -> (OutputDelegate, then, Texture:Texture 2D Soft Object Reference) {is=title K2Node_CustomEvent}
   > n3 = Async Load Asset(Asset=@n2.Texture)
   n4 = Get Image_TileImage() -> (Image_TileImage:Image Object Reference)
   > n5 = GetDynamicMaterial(self=@n4)
   > SetScalarParameterValue(self=@n5, ParameterName=IsIconLoaded, Value=0)
   n6 = GetDynamicMaterial(execute=@n3.Completed, self=@n4)
   > SetScalarParameterValue(self=@n6, ParameterName=IsIconLoaded, Value=1)
   > n7 = Cast To Texture2D(Object=@n3)
   > SetTextureParameterValue(self=@n6, ParameterName=Icon, Value=@n7)
   Event MaximizeSlot!() -> (OutputDelegate, then, bShouldSkipAnimation:bool) {is=title K2Node_Event}
   > Resize!(execute, self:Fort Quick Bar Slot Base Object Reference, CardSize:EFortItemCardSize Enum=S) -> (then)
   Event MinimizeSlot!() -> (OutputDelegate, then, bShouldSkipAnimation:bool) {is=title K2Node_Event}
   > Resize(CardSize=XS)
   Event PreConstruct()
   > UpdateItemBoxSize()
   > UpdateKeyBindingText(self:Fort Quick Bar Slot Base Object Reference)
   Event Construct()
   n8 = Get FortInventoryContext!() -> (ReturnValue:Fort Inventory Context Object Reference) {is=title K2Node_GetBlueprintContextEx}
   > Bind Event to On Quickbar Contents Changed(self:Fort Inventory Context Object Reference=@n8, Delegate=@n1) {is=title K2Node_AddDelegate}
   > InitializeInvalidation()
   > RefreshItem(self:Fort Quick Bar Slot Base Object Reference)
   n9 = Event SetSlotSelected!() -> (OutputDelegate, then, bSelected:bool) {is=title K2Node_Event}
   > n10 = Set bIsSelected(bIsSelected:bool=@n9.bSelected) -> (Output_Get:bool)
   > SetIsSelected!(
    execute
    self:Athena Quick Bar Native Cell Object Reference=Get NativeCell() -> (NativeCell:Athena Quick Bar Native Cell Object Reference)
    bInSelected:bool=@n10
   ) -> (then)
   > SetRenderTranslation(
    self=Get OverlayExtensionWidgets() -> (OverlayExtensionWidgets:Overlay Object Reference)
    Translation=Select(Option 0:Vector 2D Structure, Option 1:Vector 2D Structure=(X=0,Y=-13.5), Index=@n10) -> (ReturnValue:Vector 2D Structure)
   )
   > UpdateKeyBindingText()
   > UpdateKeyBindingVisibility(self:Fort Quick Bar Slot Base Object Reference)
   > UpdateGameplayTags()
   > n11 = Branch(Condition=@n10)
   > n12 = Branch(Condition=Get DoNotPlaySelectionAnimation() -> (DoNotPlaySelectionAnimation:bool))
   n13 = Get bCacheIsSelected() -> (bCacheIsSelected:bool)
   n14 = Branch(execute=@n12.else, Condition=@n13)
   Set bCacheIsSelected(execute=@n14.else, bCacheIsSelected:bool=true) -> (Output_Get:bool)
   n15 = Get Unselected() -> (Unselected:Widget Animation Object Reference)
   > n16 = Branch(Condition=IsAnimationPlaying(InAnimation=@n15))
   > n17 = StopAnimation(InAnimation=@n15)
   n18 = Get Selected() -> (Selected:Widget Animation Object Reference)
   PlayAnimation(execute=[@n17.then, @n16.else], InAnimation=@n18)
   n19 = Branch(execute=@n11.else, Condition=Get DoNotPlaySelectionAnimation())
   Branch(execute=@n19.else, Condition=@n13)
   > Set bCacheIsSelected(bCacheIsSelected=false)
   > n20 = Branch(Condition=IsAnimationPlaying(InAnimation=@n18))
   > n21 = StopAnimation(InAnimation=@n18)
   PlayAnimation(execute=[@n21.then, @n20.else], InAnimation=@n15)
   Event OnRefreshItem() {is=title K2Node_Event}
   > UpdateItemBoxSize()
   > UpdateGameplayTags()
   > SetupVisuals()
   n22 = Event OnExtensionWidgetAdded!() -> (OutputDelegate, then, ExtensionSlot:EUIExtensionSlot Enum=Invalid, ExtensionWidget:User Widget Object Reference) {is=title K2Node_Event}
   > n23 = AddChildToOverlay!(execute, self:Overlay Object Reference=Get OverlayExtensionWidgets(), Content:Widget Object Reference=@n22.ExtensionWidget) -> (then, ReturnValue:Overlay Slot Object Reference)
   > SetHorizontalAlignment(self=@n23, InHorizontalAlignment=HAlign_Fill)
   > SetVerticalAlignment(self=@n23, InVerticalAlignment=VAlign_Fill)
   Event Destruct()
   > Unbind Event from On Quickbar Contents Changed(self:Fort Inventory Context Object Reference=@n8, Delegate=@n1) {is=title K2Node_RemoveDelegate}
   Event OnInitialized()
   > n24 = Construct Object from Class(Class=/Script/BattleRoyaleGameUI.InstituteChirpQuickbarSlotViewModel) -> (ReturnValue:Institute Chirp - HUD Controller Quickbar Slot Viewmodel Object Reference)
   > n25 = Set InstituteChirpQuickbarSlotViewModel(InstituteChirpQuickbarSlotViewModel:Institute Chirp - HUD Controller Quickbar Slot Viewmodel Object Reference=@n24) -> (Output_Get:Institute Chirp - HUD Controller Quickbar Slot Viewmodel Object Reference)
   > Set CooldownTypesSupported!(
    execute
    CooldownTypesSupported:Array of EFortItemCooldownType Enums=Get CooldownTypesSupported() -> (CooldownTypesSupported:Array of EFortItemCooldownType Enums=None)
    self:Item Cooldown Viewmodel Object Reference=Get ItemCooldownViewmodel!(self:Device - HUD Controller Quickbar Slot Viewmodel Object Reference=@n25) -> (ItemCooldownViewmodel:Item Cooldown Viewmodel Object Reference)
   ) -> (then, Output_Get:Array of EFortItemCooldownType Enums=None)
   > InitializeDataModel!(execute, self, SlotNum:int=Get Slot Num() -> (Slot Num:int=0), Quick Bar Type:EFortQuickBars Enum=Primary) -> (then)
  ]}
 HandleItemCountChanged {nodes [
   n1 = HandleItemCountChanged!() -> (then, ItemCount:int) {is=title K2Node_FunctionEntry}
   n2 = Get InstituteChirpQuickbarSlotViewModel() -> (InstituteChirpQuickbarSlotViewModel:Institute Chirp - HUD Controller Quickbar Slot Viewmodel Object Reference)
   n3 = Total Ammo!(self:Device - HUD Controller Quickbar Slot Viewmodel Object Reference=@n2) -> (ReturnValue:int=0)
   n4 = "To Text (Integer)"(Value=@n3)
   n5 = "To Text (Integer)"(Value=Stack Count!(self:Device - HUD Controller Quickbar Slot Viewmodel Object Reference=@n2) -> (ReturnValue:int=0))
   n6 = "integer < integer"(A=@n3)
   > "SetText (Text)"(
    self=Get Text_ItemQuantity() -> (Text_ItemQuantity:Common Text Object Reference)
    InText=SelectText(
     A=SelectText(A=@n5, B=@n4, bPickA=@n6)
     B=GetInfinitySymbolText!() -> (ReturnValue:"Text (by ref)")
     bPickA="integer < integer"(A=@n1, B=999999)
    )
   )
  ]}
 InitializeDataModel {nodes [
   n1 = InitializeDataModel!() -> (then, SlotNum:int, Quick Bar Type:EFortQuickBars Enum=Primary) {is=title K2Node_FunctionEntry}
   > Branch(Condition="integer >= integer"(A=@n1))
   > n2 = Set Slot Num(Slot Num:int=@n1) -> (Output_Get:int=0)
   n3 = Get InstituteChirpQuickbarSlotViewModel()
   > Is Valid(InputObject=@n3)
   > InitializeWithType!(
    execute
    self:Institute Chirp - HUD Controller Quickbar Slot Viewmodel Object Reference=@n3
    InPlayerController:Player Controller Object Reference=GetOwningPlayer()
    InQuickBarType:EFortQuickBars Enum=@n1.Quick Bar Type
    InSlot:int=@n2
   ) -> (then)
  ]}
 InitializeInvalidation {nodes [
   InitializeInvalidation() {is=title K2Node_FunctionEntry}
   > n1 = Branch(Condition=IsUsingTouch(PlayerController=GetOwningPlayer()))
   Is Simplified HUD!(Exec:Exec=@n1.else) -> (Simplified:Exec, Default:Exec) {is=title K2Node_MacroInstance}
   > Set DoNotPlaySelectionAnimation(DoNotPlaySelectionAnimation:bool=true) -> (Output_Get:bool)
  ]}
 OnTouchEnded {nodes [
   OnTouchEnded!() -> (then, MyGeometry:Geometry Structure, InTouchEvent:"Pointer Event Structure (by ref)") {is=title K2Node_FunctionEntry}
   > n1 = Branch(Condition=AND Boolean(A=IsTouchInputPlatform!() -> (ReturnValue:bool), B=Get bShouldTakeItemSelectionInputFromHUD() -> (bShouldTakeItemSelectionInputFromHUD:bool)))
   > n2 = GetConsoleVariableBoolValue!(execute, VariableName:String=Fort.Rollback.IndependentTapAnywherePrimaryAction) -> (then, ReturnValue:bool)
   > n3 = Branch(Condition=@n2)
   > StopSelectSlot(self:Athena Quick Bar Slot Object Reference)
   > Return Node!(execute, ReturnValue:Event Reply Structure=Handled())
   Return Node(execute=@n3.else, ReturnValue=Unhandled())
   Return Node(execute=@n1.else, ReturnValue=Unhandled())
  ]}
 OnTouchStarted {nodes [
   OnTouchStarted!() -> (then, MyGeometry:Geometry Structure, InTouchEvent:"Pointer Event Structure (by ref)") {is=title K2Node_FunctionEntry}
   > n1 = Branch(Condition=AND Boolean(A=IsTouchInputPlatform(), B=Get bShouldTakeItemSelectionInputFromHUD()))
   > n2 = GetConsoleVariableBoolValue(VariableName=Fort.Rollback.IndependentTapAnywherePrimaryAction)
   > n3 = Branch(Condition=@n2)
   n4 = Cast To FortPlayerController!(Object=GetOwningPlayer()) -> (AsFort Player Controller, bSuccess:bool=~)
   n5 = NOT Boolean(A=IsSmartBuildEnabled!(self:Fort Controller Component Smart Build Object Reference=GetSmartBuildComponent!(self:Fort Player Controller Object Reference=@n4) -> (ReturnValue:Fort Controller Component Smart Build Object Reference)) -> (ReturnValue:bool))
   > StartSelectSlot!(execute, self:Athena Quick Bar Slot Object Reference, bReloadOrUseIfAlreadySelected:bool=@n5, bStopUseImmediately:bool=false) -> (then)
   > Return Node(ReturnValue=Handled())
   TakeItemSelectionInputFromHUD!(
    execute=@n3.else
    self:Athena HUDContext Object Reference=Get AthenaHUDContext!() -> (ReturnValue:Athena HUDContext Object Reference) {is=title K2Node_GetBlueprintContextEx}
    TouchIndex:int=Get Pointer Index!(Input:"Pointer Event Structure (by ref)"=Get InTouchEvent() -> (InTouchEvent:"Pointer Event Structure (by ref)")) -> (ReturnValue:int=0) {is=title K2Node_CallFunction}
    SlotIdx:int=Get SlotIndex() -> (SlotIndex:int=0)
    InQuickBar:EFortQuickBars Enum=Get QuickBarType() -> (QuickBarType:EFortQuickBars Enum=Primary)
    bReloadOrUseIfAlreadySelected:bool=@n5
    bStopUseImmediately:bool=true
   ) -> (then, bIsUsing:bool)
   > Return Node(ReturnValue=Handled())
   Return Node(execute=@n1.else, ReturnValue=Unhandled())
  ]}
 SetAnimatedBoxScaleFactor {nodes [
   n1 = SetAnimatedBoxScaleFactor!() -> (then, Value:double=~) {is=title K2Node_FunctionEntry}
   > Set AnimatedBoxScaleFactor(AnimatedBoxScaleFactor:double=@n1) -> (Output_Get:double)
   > UpdateItemBoxSize()
  ]}
 SetupVisuals {nodes [
   SetupVisuals() {is=title K2Node_FunctionEntry}
   n1 = Get Item() -> (Item:Fort Item Object Reference)
   > Is Valid(InputObject=@n1)
   n2 = "Not Equal (Enum)"!(
    A:EFortItemType Enum=GetType!(self:Fort Item Object Reference=@n1) -> (ReturnValue:EFortItemType Enum=WorldItem)
    B:EFortItemType Enum=BuildingPiece
   ) -> (ReturnValue:bool)
   > Set Visibility(
    Visibility=Select(Option 0:ESlateVisibility Enum=Collapsed, Option 1:ESlateVisibility Enum=SelfHitTestInvisible, Index=@n2) -> (ReturnValue:ESlateVisibility Enum)
    self=Get Text_ItemQuantity()
   )
   > n3 = GetDynamicMaterial(self=Get Image_TileImage())
   > SetScalarParameterValue(self=@n3, ParameterName=IsTrap, Value=Select(Option 0:float=0, Option 1:float=1, Index=@n2) -> (ReturnValue:float))
   > n4 = GetRarityData!(
    execute
    Rarity:EFortRarity Enum=Rarity!(self:Institute Chirp - HUD Controller Quickbar Slot Viewmodel Object Reference=Get InstituteChirpQuickbarSlotViewModel()) -> (ReturnValue:EFortRarity Enum=Common)
   ) -> (
    then
    ReturnValue_Name:Text
    ReturnValue_Color1:Linear Color Structure
    ReturnValue_Color2:Linear Color Structure
    ReturnValue_Color3:Linear Color Structure
    ReturnValue_Color4:Linear Color Structure
    ReturnValue_Color5:Linear Color Structure
    ReturnValue_Radius:float=~
    ReturnValue_Falloff:float=~
    ReturnValue_Brightness:float=~
    ReturnValue_Roughness:float=~
    ReturnValue_Glow:float=~
   )
   > n5 = GetDynamicMaterial(self=Get Image_TileImage())
   > SetVectorParameterValue(self=@n5, ParameterName=Color_1, Value=@n4.ReturnValue_Color2)
   > SetVectorParameterValue(self=@n5, ParameterName=Color_2, Value=@n4.ReturnValue_Color3)
   > n6 = GetDynamicMaterial(self=Get Image_SelectedGlow() -> (Image_SelectedGlow:Image Object Reference))
   > SetVectorParameterValue(
    self=@n6
    ParameterName=Color_B
    Value=Select(Option 0:Linear Color Structure=(R=0,G=0.1384,B=1,A=1), Option 1:Linear Color Structure=@n4.ReturnValue_Color2, Index=@n2) -> (ReturnValue:Linear Color Structure)
   )
   > SetVectorParameterValue(self=@n6, ParameterName=Color_A, Value=Select(Option 0=(R=0,G=0.05,B=1,A=1), Option 1=@n4.ReturnValue_Color3, Index=@n2))
  ]}
 UpdateGameplayTags {nodes [
   UpdateGameplayTags() {is=title K2Node_FunctionEntry}
   > HasMatchingGameplayTag!(execute:Exec, self:Object Reference=Get Item(), TagToCheck:Gameplay Tag Structure=(TagName="Athena.HUD.EquippedItem.HideQuickbarItemCount")) -> (then:Exec, ReturnValue:bool) {is=title K2Node_Message}
  ]}
 UpdateItemBoxSize {nodes [
   UpdateItemBoxSize() {is=title K2Node_FunctionEntry}
   > n1 = Branch(Condition=Get bUseSlotSizeBySlotUsage() -> (bUseSlotSizeBySlotUsage:bool))
   n2 = GetSlotSizeForItem!(self:Fort Quick Bar Slot Base Object Reference, InItem:Fort Item Object Reference=Get Item()) -> (ReturnValue:Vector 2D Structure)
   n3 = BreakVector2D(InVec=@n2)
   n4 = BreakVector2D(InVec=vector2d * float!(A:Vector 2D Structure=@n2, B:double=0.7) -> (ReturnValue:Vector 2D Structure))
   n5 = Get AnimatedBoxScaleFactor() -> (AnimatedBoxScaleFactor:double)
   n6 = Lerp(A=@n4, B=@n3, Alpha=@n5)
   n7 = Lerp(A=@n4.Y, B=@n3.Y, Alpha=@n5)
   > DesiredSizeOverride!(execute, self:Athena Quick Bar Native Cell Object Reference=Get NativeCell(), WidthOverride:float=@n6, HeightOverride:float=@n7) -> (then)
   n8 = Get SizeBox_ExtensionOverlay() -> (SizeBox_ExtensionOverlay:Size Box Object Reference)
   > SetHeightOverride(self=@n8, InHeightOverride=@n7)
   > SetWidthOverride(self=@n8, InWidthOverride=@n6)
   n9 = Lerp(A=64, B=92, Alpha=@n5)
   DesiredSizeOverride(execute=@n1.else, self=Get NativeCell(), WidthOverride=@n9, HeightOverride=@n9)
   n10 = Get SizeBox_ExtensionOverlay()
   > SetHeightOverride(self=@n10, InHeightOverride=@n9)
   > SetWidthOverride(self=@n10, InWidthOverride=@n9)
  ]}
]
variables [
 AnimatedBoxScaleFactor {type=number}
 bCacheIsSelected {type=boolean}
 bShouldTakeItemSelectionInputFromHUD {type=boolean, default=true}
 CooldownTypesSupported {type="array<EFortItemCooldownType>", default []}
 ItemDisplayStyle {type=ItemDisplayStyle, default=World}
 Slot Num {type=integer, default=-1}
]
widgets {
 info {inheritedWidgetCount=15, parentClass=/Script/FortniteUI.BacchusQuickbarSlotBase, rootWidgetClass=&VerticalBox, widgetCount=24}
 widgets [
  VerticalBox_0 {widgetClassPath=&VerticalBox, widget=VerticalBox_0}
  QuickbarKeybindWidgetTop {
   widgetClassPath=&/WBP_BRGameUI_QuickbarKeybindTop
   bInherited=true
   bIsVariable=true
   parent=VerticalBox_0
   widget=QuickbarKeybindWidgetTop
   slot {class=VerticalBoxSlot, properties {horizontalAlignment=HAlign_Center}}
  }
  Overlay_TileContents {
   widgetClassPath=&Overlay
   bIsVariable=true
   parent=VerticalBox_0
   widget=Overlay_TileContents
   slot {class=VerticalBoxSlot, properties {horizontalAlignment=HAlign_Center, verticalAlignment=VAlign_Center}}
  }
  Slot_BehindImage {
   widgetClassPath=&NamedSlot
   bIsVariable=true
   parent=Overlay_TileContents
   widget=Slot_BehindImage
   properties {bExposeOnInstanceOnly=true, visibility=Collapsed}
   slot {class=OverlaySlot, properties {horizontalAlignment=HAlign_Fill, verticalAlignment=VAlign_Fill}}
  }
  Image_SelectedGlow {
   widgetClassPath=&Image
   bIsVariable=true
   parent=Overlay_TileContents
   widget=Image_SelectedGlow
   properties {
    visibility=SelfHitTestInvisible
    brush {
     bIsDynamicallyLoaded=false
     drawAs=Image
     imageType=FullColor
     mirroring=NoMirror
     resourceName=None
     resourceObject=/BattleRoyaleGameUI/HUD/QuickBar/Assets/M_UI_ItemRarityGlow
     tiling=NoTile
     imageSize=<32 32>
     margin=<ltrb 0 0 0 0>
     outlineSettings {
      bUseBrushTransparency=false
      roundingType=HalfHeightRadius
      width=0
      color {colorUseRule=UseColor_Specified, specifiedColor=<rgba 0 0 0 0>}
      cornerRadii=<0 0 0 0>
     }
     tintColor {colorUseRule=UseColor_Specified, specifiedColor=<rgba 1 1 1 1>}
     uVRegion {bIsValid=false, max=<0 0>, min=<0 0>}
    }
    renderTransform {angle=0, scale=<1.5 2>, shear=<0 0>, translation=<0 0>}
    renderTransformPivot=<0.5 0.6>
   }
   slot {class=OverlaySlot, properties {horizontalAlignment=HAlign_Fill, verticalAlignment=VAlign_Fill}}
  }
  Image_TileImage {
   widgetClassPath=&Image
   bIsVariable=true
   parent=Overlay_TileContents
   widget=Image_TileImage
   properties {brush {
     bIsDynamicallyLoaded=false
     drawAs=Image
     imageType=FullColor
     mirroring=NoMirror
     resourceName=None
     resourceObject=/BattleRoyaleGameUI/HUD/QuickBar/Assets/M_QuickBarItemCard_Secondary
     tiling=NoTile
     imageSize=<80 80>
     margin=<ltrb 0 0 0 0>
     outlineSettings {
      bUseBrushTransparency=false
      roundingType=HalfHeightRadius
      width=0
      color {colorUseRule=UseColor_Specified, specifiedColor=<rgba 0 0 0 0>}
      cornerRadii=<0 0 0 0>
     }
     tintColor {colorUseRule=UseColor_Specified, specifiedColor=<rgba 1 1 1 1>}
     uVRegion {bIsValid=false, max=<0 0>, min=<0 0>}
    }}
   slot {class=OverlaySlot, properties {horizontalAlignment=HAlign_Fill, verticalAlignment=VAlign_Fill}}
  }
  SizeBox_ExtensionOverlay {
   widgetClassPath=&SizeBox
   bIsVariable=true
   parent=Overlay_TileContents
   widget=SizeBox_ExtensionOverlay
   properties {bOverride_HeightOverride=true, heightOverride=132}
   slot {class=OverlaySlot, properties {horizontalAlignment=HAlign_Fill, verticalAlignment=VAlign_Fill}}
  }
  OverlayExtensionWidgets {
   widgetClassPath=&Overlay
   bIsVariable=true
   parent=SizeBox_ExtensionOverlay
   widget=OverlayExtensionWidgets
   slot {class=SizeBoxSlot}
  }
  TutorialHighlight {
   widgetClassPath=&TutorialHighlight
   parent=Overlay_TileContents
   widget=TutorialHighlight
   properties {customHighlightWidget=/Game/Athena/HUD/Tutorial/Tutorial_Outline_Square.Tutorial_Outline_Square_C}
   slot {class=OverlaySlot, properties {horizontalAlignment=HAlign_Fill, verticalAlignment=VAlign_Fill}}
  }
  SizeBox_1 {
   widgetClassPath=&SizeBox
   parent=Overlay_TileContents
   widget=SizeBox_1
   properties {bOverride_MaxDesiredWidth=true, maxDesiredWidth=60, renderTransformPivot=<1 1>}
   slot {class=OverlaySlot, properties {horizontalAlignment=HAlign_Right, verticalAlignment=VAlign_Bottom, padding=<ltrb 0 0 8 8>}}
  }
  Text_ItemQuantity {
   widgetClassPath=&CommonTextBlock
   bIsVariable=true
   parent=SizeBox_1
   widget=Text_ItemQuantity
   properties {
    justification=Center
    text=99
    visibility=Collapsed
    font {
     bForceMonospaced=false
     bMaterialIsStencil=false
     fontMaterial=null
     fontObject=&BurbankSmall
     letterSpacing=0
     monospacedWidth=1
     size=13.5
     skewAmount=0
     typefaceFontName=Black
     outlineSettings {
      bApplyOutlineToDropShadows=false
      bMiteredCorners=false
      bSeparateFillAlpha=true
      outlineMaterial=/BattleRoyaleGameUI/Shared/MI_UI_Font_SoftShadow_SmallText
      outlineSize=4
      outlineColor=<rgba 1 1 1 1>
     }
    }
    margin=<ltrb -4 -8 0 -8>
    shadowOffset=<0 0>
   }
   slot {class=SizeBoxSlot, properties {verticalAlignment=VAlign_Center}}
  }
  SelectionHighlight {widgetClassPath=&Image, bInherited=true, widget=null}
  LockIcon {widgetClassPath=&Image, bInherited=true, widget=null}
  Image_Disabled {widgetClassPath=&Image, bInherited=true, widget=null}
  SizeBox_BottomKeybind {widgetClassPath=&SizeBox, bInherited=true, widget=null}
  SizeBox_TopKeybind {widgetClassPath=&SizeBox, bInherited=true, widget=null}
  QuickbarKeybindWidgetBottom {widgetClassPath=/Script/FortniteUI.FortQuickbarKeybindWidget, bInherited=true, widget=null}
  EmptyImage {widgetClassPath=&Image, bInherited=true, widget=null}
  ItemCardMaximized {widgetClassPath=/Script/FortniteUI.FortMultiSizeItemCard, bInherited=true, widget=null}
  ItemCardMinimized {widgetClassPath=&FortMultiSizeItemCard, bInherited=true, widget=null}
  QuickbarSlotCooldown {widgetClassPath=/Script/FortniteUI.FortItemCooldownWidget, bInherited=true, widget=null}
  QuickbarSlotDurability {widgetClassPath=/Script/FortniteUI.AthenaItemElementWidgetBase, bInherited=true, widget=null}
  SlotInteraction {widgetClassPath=&Image, bInherited=true, widget=null}
  NativeCell {widgetClassPath=/Script/FortniteUI.AthenaQuickBarNativeCell, bInherited=true, widget=null}
 ]
}
