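# Translucent UI material for a vertical frame divider in the Battle Royale game UI, drawn procedurally as a trapezoid.
# UI UVs are offset, warped by a perspective function, axis-swapped through a YX swizzle, then skewed before feeding an isosceles trapezoid shape function.
# The trapezoid fill is scaled by an Opacity parameter, with width and height also parameterized.
# The graph also holds the shared FrameColor collection parameter and an adaptive-colors function.
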Material /BattleRoyaleGameUI/Shared/FrameDecor/M_UI_SimpleDivider_Vertical
graph {
 properties {bAutomaticallySetUsageInEditor=false, blendMode=BLEND_Translucent, materialDomain=MD_UI, usedShadingModels=Default Lit}
 nodes [
  CollectionParameter() {collection=/BattleRoyaleGameUI/Shared/MPC_BRGameUI, parameterName=FrameColor}
  MaterialFunctionCall!(BufferSource) {materialFunction=/BattleRoyaleGameUI/Shared/MF_UI_AdaptiveColors}
  n1 = MaterialFunctionCall() {materialFunction=&GetUserInterfaceUV}
  n2 = Constant2Vector() {g=0.12}
  n3 = MaterialFunctionCall(XY=MaterialFunctionCall!(UV=Subtract(A=@n1.Normalized UV, B=@n2), Perspective Rotation=Constant() {r=0.3}, Perspective Amount=Constant() {r=0.4}) {materialFunction=/FortUILibrary/MaterialFunctions/Basics/MF_UI_Perspective}.Perspective UVs) {materialFunction=&Swizzle}
  n4 = Constant2Vector() {r=-8}
  Multiply(
   A=MaterialFunctionCall!(
    UVs=MaterialFunctionCall!(UV=@n3.YX, Skew Angle=@n4) {materialFunction=/FortUILibrary/MaterialFunctions/Basics/MF_UI_Skew}.Skewed UVs
    TopWidth=ScalarParameter() {defaultValue=0.8, parameterName=TopWidth}
    BottomWidth=ScalarParameter() {defaultValue=1.1, parameterName=BottomWidth}
    Height=ScalarParameter() {defaultValue=0.4, parameterName=Height}
    OutlineThickness
    OutlineAlignment
    GlowMin
    GlowMax
   ) {materialFunction=/FortUILibrary/MaterialFunctions/Shapes/MF_UI_IsoscelesTrapezoid}.Fill
   B=ScalarParameter() {defaultValue=0.5, parameterName=Opacity}
  )
 ]
}
