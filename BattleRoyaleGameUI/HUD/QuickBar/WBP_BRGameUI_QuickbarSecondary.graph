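# Battle Royale HUD widget for the secondary (build) quickbar, with five slots, a switch-mode keybind and icon, extending QuickbarBase.
# Maximize and Minimize swap the icon material and play scale animations.
# Async waits on the BlockBuildMode gameplay tag swap build tools for a blocked icon, which shakes with a sound on failed actions.
# Smart-build state can show a simple-build icon and, on gamepad, reverse slot order behind a console variable.
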WidgetBlueprint /BattleRoyaleGameUI/HUD/QuickBar/WBP_BRGameUI_QuickbarSecondary
animations [
 AnimCannotPerform {
  binding_count=1
  length_seconds=0.2
  object=&self:AnimCannotPerform
  sequence {
   bindings [BuildingBlocked_Icon {
     id=#1
     possessed_class=Image
     tracks [
      {
       type=MovieSceneWidgetMaterialTrack
       sections [{
         type=MovieSceneParameterSection
         channels [AnimateIconColor {type=MovieSceneScriptingFloatChannel, keys [[0,0],[3,1],[6,0],[9.001,1],[12,0]]}]
        }]
      }
      {
       type=MovieScene2DTransformTrack
       sections [{
         type=MovieScene2DTransformSection
         channels [Translation.X {type=MovieSceneScriptingFloatChannel, keys [[0,0],[3,-2],[6,2],[9.001,2],[12,0]]}]
        }]
      }
     ]
    }]
   display_rate {numerator=60}
  }
 }
 AnimSwitchIcon {
  binding_count=2
  length_seconds=0.2
  object=&self:AnimSwitchIcon
  sequence {
   bindings [
    Image_IconGlow {
     id=#2
     possessed_class=Image
     tracks [{
       type=MovieSceneWidgetMaterialTrack
       sections [{type=MovieSceneParameterSection, channels [IsSelected {type=MovieSceneScriptingFloatChannel, keys [[0,0],[12,0.5]]}]}]
      }]
    }
    Image_SwitchAction {
     id=#3
     possessed_class=Image
     tracks [
      {
       type=MovieSceneFloatTrack
       display_name=Render Opacity
       sections [{type=MovieSceneFloatSection, channels [None {type=MovieSceneScriptingFloatChannel, keys [[0,0],[12,1]]}]}]
      }
      {
       type=MovieScene2DTransformTrack
       sections [{
         type=MovieScene2DTransformSection
         channels [
          Angle {type=MovieSceneScriptingFloatChannel, default=0}
          Scale.X {type=MovieSceneScriptingFloatChannel, keys [[0,0.5],[12,1]]}
          Scale.Y {type=MovieSceneScriptingFloatChannel, keys [[0,0.5],[12,1]]}
         ]
        }]
      }
     ]
    }
   ]
   display_rate {numerator=60}
  }
 }
 AnimMaximize {
  binding_count=2
  length_seconds=0.0584
  object=&self:AnimMaximize
  sequence {
   bindings [
    HorizontalBox_1 {
     id=#4
     possessed_class=HorizontalBox
     tracks [{
       type=MovieScene2DTransformTrack
       sections [{
         type=MovieScene2DTransformSection
         end_frame=7
         start_frame=0
         channels [
          Translation.X {type=MovieSceneScriptingFloatChannel, default=0}
          Translation.Y {type=MovieSceneScriptingFloatChannel, default=0}
          Angle {type=MovieSceneScriptingFloatChannel, default=0}
          Scale.X {type=MovieSceneScriptingFloatChannel, keys [[0,1.25],[6,1]]}
          Scale.Y {type=MovieSceneScriptingFloatChannel, keys [[0,1.25],[6,1]]}
          Shear.X {type=MovieSceneScriptingFloatChannel, default=0}
          Shear.Y {type=MovieSceneScriptingFloatChannel, default=0}
         ]
        }]
      }]
    }
    SizeBox_0 {
     id=#5
     possessed_class=SizeBox
     tracks [{
       type=MovieScene2DTransformTrack
       sections [{
         type=MovieScene2DTransformSection
         end_frame=7
         start_frame=0
         channels [
          Translation.X {type=MovieSceneScriptingFloatChannel, default=0}
          Translation.Y {type=MovieSceneScriptingFloatChannel, default=0}
          Angle {type=MovieSceneScriptingFloatChannel, default=0}
          Scale.X {type=MovieSceneScriptingFloatChannel, keys [[0,0.8],[7,1]]}
          Scale.Y {type=MovieSceneScriptingFloatChannel, keys [[0,0.8],[7,1]]}
          Shear.X {type=MovieSceneScriptingFloatChannel, default=0}
          Shear.Y {type=MovieSceneScriptingFloatChannel, default=0}
         ]
        }]
      }]
    }
   ]
   display_rate {numerator=120}
  }
 }
]
graphs [
 BuildModeUpdated {nodes [
   n1 = BuildModeUpdated!() -> (then, bBuildingBlocked:bool=~) {is=title K2Node_FunctionEntry}
   > SetVisibility(
    self=Get BuildingBlocked_Icon() -> (BuildingBlocked_Icon:Image Object Reference)
    InVisibility=Select(Option 0=Hidden, Option 1=HitTestInvisible, Index=@n1)
   )
   > SetVisibility(
    self=Get Overlay_BuildTools() -> (Overlay_BuildTools:Overlay Object Reference)
    InVisibility=Select(Option 0=SelfHitTestInvisible, Option 1=Hidden, Index=@n1)
   )
   > SetVisibility(
    self=Get Overlay_NoBuild() -> (Overlay_NoBuild:Overlay Object Reference)
    InVisibility=Select(Option 0=SelfHitTestInvisible, Option 1=Hidden, Index=@n1)
   )
   n2 = Select(Option 0=HitTestInvisible, Option 1=Hidden, Index=Get bBuildingBlocked() -> (bBuildingBlocked:bool))
   > SetVisibility(self=Get KeybindWidget_SwitchModeKeybind() -> (KeybindWidget_SwitchModeKeybind:Fort Keybind Widget Object Reference), InVisibility=@n2)
   > n3 = For Each Loop(Array:Array of Fort Quick Bar Slot Base Object References=Get QuickbarSlots() -> (QuickbarSlots:Array of Fort Quick Bar Slot Base Object References)) -> (Array Element:Fort Quick Bar Slot Base Object Reference)
   > SetComboSwitcherVisibility!(
    execute
    self:Fort Quickbar Keybind Widget Object Reference=Get QuickbarKeybindWidgetTop!(self:Fort Quick Bar Slot Base Object Reference=@n3) -> (QuickbarKeybindWidgetTop:Fort Quickbar Keybind Widget Object Reference)
    InVisibility:ESlateVisibility Enum=@n2
   ) -> (then)
  ]}
 EventGraph {nodes [
   HandleInputMethodChanged_Bind!() -> (OutputDelegate, then, InputType:ECommonInputType Enum) {is=title K2Node_CustomEvent}
   > HandleShowHideRail()
   n1 = HandlePawnSet() {is=title K2Node_CustomEvent}
   K2Node_Composite_0 = InitAbilitySystem!(execute=[@n21.then, @n1.then]) -> (then:Exec) {is=title K2Node_Composite}
   > n2 = HasMatchingGameplayTag!(
    execute:Exec
    self:Object Reference=GetOwningPlayerPawn!(self:User Widget Object Reference) -> (ReturnValue:Pawn Object Reference)
    TagToCheck:Gameplay Tag Structure=(TagName="Gameplay.Action.Player.BlockBuildMode")
   ) -> (then:Exec, ReturnValue:bool) {is=title K2Node_Message}
   > BuildModeUpdated!(execute, self, bBuildingBlocked:bool=@n2) -> (then) {is=title K2Node_CallFunction}
   n3 = GetOwningPlayer()
   > Branch(Condition=IsValid(Object=GetSpectatorPawn!(self:Player Controller Object Reference=@n3) -> (ReturnValue:Spectator Pawn Object Reference)))
   > n4 = HasMatchingGameplayTag(
    self=GetViewTarget!(self:Controller Object Reference=@n3) -> (ReturnValue:Actor Object Reference)
    TagToCheck=(TagName="Gameplay.Action.Player.BlockBuildMode")
   )
   > BuildModeUpdated(bBuildingBlocked=@n4)
   n5 = OnUnableToPerformAction!() -> (OutputDelegate, then, FailedReason:Gameplay Tag Container Structure, FailureText:"Text (by ref)") {is=title K2Node_CustomEvent}
   > Branch(Condition=HasTag!(
     TagContainer:"Gameplay Tag Container Structure (by ref)"=@n5.FailedReason
     Tag:Gameplay Tag Structure=(TagName="Gameplay.Action.Player.BlockBuildMode")
     bExactMatch:bool=false
    ) -> (ReturnValue:bool))
   > PlayAnimation(InAnimation=Get AnimCannotPerform() -> (AnimCannotPerform:Widget Animation Object Reference))
   > PlaySound2D(Sound=/Game/Athena/Sounds/Items/Consumable_CantApply_Cue)
   n6 = Event OnQuickbarContentsChanged!() -> (OutputDelegate, then, QuickbarIndex:EFortQuickBars Enum=Primary, ChangedSlots:Array of Integers) {is=title K2Node_Event}
   > Parent: OnQuickbarContentsChanged!(execute, QuickbarIndex:EFortQuickBars Enum=@n6.QuickbarIndex, ChangedSlots:Array of Integers=@n6.ChangedSlots) -> (then) {is=title K2Node_CallParentFunction}
   Event Destruct()
   n7 = Get CommonInputSubsystem()
   > n8 = Is Valid(InputObject=@n7)
   > n9 = Unbind Event from On Input Method Changed(self=@n7, Delegate=Create Event())
   n10 = Get FortHUDContext!() -> (ReturnValue:Fort HUDContext Object Reference) {is=title K2Node_GetBlueprintContextEx}
   n11 = Is Valid(exec=[@n9.then, @n8.Is Not Valid], InputObject=@n10)
   > n12 = Unbind Event from On Unable to Perform Action(self:Fort HUDContext Object Reference=@n10, Delegate=Create Event()) {is=title K2Node_RemoveDelegate}
   Parent: Destruct(execute=[@n12.then, @n11.Is Not Valid])
   Event Maximize() {is=title K2Node_Event}
   > Parent: Maximize() {is=title K2Node_CallParentFunction}
   > SetBrushFromMaterial(
    self=Get Image_SwitchAction() -> (Image_SwitchAction:Image Object Reference)
    Material=/BattleRoyaleGameUI/HUD/QuickBar/Assets/M_UI_QuickbarTypeIcon
   )
   > HandleShowHideRail()
   > n13 = Branch(Condition=Get MaximizedAnimHasPlayed() -> (MaximizedAnimHasPlayed:bool))
   Set MinimizedAnimHasPlayed(execute=[@n13.then, @n14.then], MinimizedAnimHasPlayed:bool=false) -> (Output_Get:bool)
   > HandleSimpleBuild!(
    execute
    self
    IsEnabled:bool=IsSmartBuildEnabled!(self:Fort Controller Component Smart Build Object Reference=Get SmartBuildComponent() -> (SmartBuildComponent:Fort Controller Component Smart Build Object Reference)) -> (ReturnValue:bool)
   ) -> (then)
   PlayAnimationForward(execute=@n13.else, InAnimation=Get AnimMaximize() -> (AnimMaximize:Widget Animation Object Reference))
   > PlayAnimation(InAnimation=Get AnimSwitchIcon() -> (AnimSwitchIcon:Widget Animation Object Reference))
   > n14 = Set MaximizedAnimHasPlayed(MaximizedAnimHasPlayed:bool=true) -> (Output_Get:bool)
   Event Minimize() {is=title K2Node_Event}
   > Parent: Minimize() {is=title K2Node_CallParentFunction}
   > SetBrushFromMaterial(self=Get Image_SwitchAction(), Material=/BattleRoyaleGameUI/HUD/QuickBar/Assets/MI_UI_QuickbarTypeIcon_Build)
   > Show Secondary Quickbar Rail!(execute, self, Show:bool=false) -> (then)
   > n15 = Branch(Condition=Get MinimizedAnimHasPlayed() -> (MinimizedAnimHasPlayed:bool))
   Set MaximizedAnimHasPlayed(execute=[@n15.then, @n16.then], MaximizedAnimHasPlayed=false)
   PlayAnimationReverse(execute=@n15.else, InAnimation=Get AnimMaximize())
   > PlayAnimation(InAnimation=Get AnimSwitchIcon())
   > n16 = Set MinimizedAnimHasPlayed(MinimizedAnimHasPlayed=true)
   Event Construct()
   > Parent: Construct()
   > n17 = Sequence!(execute) -> (then_0, then_1, then_2:Exec, then_3:Exec, then_4:Exec, then_5:Exec, then_6:Exec)
   Parent: Construct(execute=@n17.then_1)
   Minimize(execute=@n17.then_4, self:Fort Quickbar Base Object Reference)
   InitSimpleBuild(execute=@n17.then_5)
   HandleKeyBindsChange(execute=@n17.then_6)
   Bind Event to On Input Method Changed(execute=@n17.then_2, self=Get CommonInputSubsystem(), Delegate=Create Event())
   n18 = Set QuickbarSlots(
    execute=@n17.then_0
    QuickbarSlots:Array of Fort Quick Bar Slot Base Object References=Make Array!(
     "[0]":Fort Quick Bar Slot Base Object Reference=Get QuickbarSlotF1() -> (QuickbarSlotF1:WBP BRGame UI Quickbar Slot Secondary Object Reference)
     "[1]":Fort Quick Bar Slot Base Object Reference=Get QuickbarSlotF2() -> (QuickbarSlotF2:WBP BRGame UI Quickbar Slot Secondary Object Reference)
     "[2]":Fort Quick Bar Slot Base Object Reference=Get QuickbarSlotF3() -> (QuickbarSlotF3:WBP BRGame UI Quickbar Slot Secondary Object Reference)
     "[3]":Fort Quick Bar Slot Base Object Reference=Get QuickbarSlotF4() -> (QuickbarSlotF4:WBP BRGame UI Quickbar Slot Secondary Object Reference)
     "[4]":Fort Quick Bar Slot Base Object Reference=Get QuickbarSlotF5() -> (QuickbarSlotF5:WBP BRGame UI Quickbar Slot Secondary Object Reference)
    ) -> (Array:Array of Fort Quick Bar Slot Base Object References)
   ) -> (Output_Get:Array of Fort Quick Bar Slot Base Object References)
   > n19 = For Each Loop(Array=@n18)
   > n20 = Cast To WBP_BRGameUI_QuickbarSlot_Secondary!(execute, Object=@n19) -> (then, CastFailed, AsWBP BRGame UI Quickbar Slot Secondary:WBP BRGame UI Quickbar Slot Secondary Object Reference) {is=title K2Node_DynamicCast}
   > InitializeDataModel!(
    execute
    self:WBP BRGame UI Quickbar Slot Secondary Object Reference=@n20
    SlotNum:int=@n19.Array Index
    Quick Bar Type:EFortQuickBars Enum=Secondary
   ) -> (then)
   Bind Event to On Unable to Perform Action(execute=@n17.then_3, self:Fort HUDContext Object Reference=Get FortHUDContext(), Delegate=Create Event()) {is=title K2Node_AddDelegate}
   > n21 = Bind Event to On Pawn Set(self:Fort HUDContext Object Reference=Get FortHUDContext(), Delegate=@n1) {is=title K2Node_AddDelegate}
   Event HandleKeybindsChanged() {is=title K2Node_Event}
   > HandleKeyBindsChange()
  ]}
 InitAbilitySystem {
  path=EventGraph.K2Node_Composite_0.InitAbilitySystem
  nodes [
   Bind Ability Async Events() {is=title K2Node_CustomEvent}
   > n1 = WaitGameplayTagAddToActor!(
    execute:Exec
    TargetActor:Actor Object Reference=GetOwningPlayerPawn()
    Tag:Gameplay Tag Structure=(TagName="Gameplay.Action.Player.BlockBuildMode")
    OnlyTriggerOnce:bool=false
   ) -> (then:Exec, AsyncTaskProxy:Ability Async Wait Gameplay Tag Added Object Reference, Added:Exec) {is=title K2Node_AsyncAction}
   n2 = GetOwningPlayerPawn()
   > n3 = WaitGameplayTagRemoveFromActor!(
    execute:Exec
    TargetActor:Actor Object Reference=@n2
    Tag:Gameplay Tag Structure=(TagName="Gameplay.Action.Player.BlockBuildMode")
    OnlyTriggerOnce:bool=false
   ) -> (then:Exec, AsyncTaskProxy:Ability Async Wait Gameplay Tag Removed Object Reference, Removed:Exec) {is=title K2Node_AsyncAction}
   n4 = Cast To FortPawn!(execute=@n3.Removed, Object=@n2) -> (then, CastFailed, AsFort Pawn:Fort Pawn Object Reference) {is=title K2Node_DynamicCast}
   > n5 = Branch(Condition=IsDead!(self:Fort Pawn Object Reference=@n4) -> (ReturnValue:bool))
   BuildModeUpdated(execute=@n5.else, bBuildingBlocked=false)
   BuildModeUpdated(execute=@n1.Added, bBuildingBlocked=true)
   n6 = OnAbilitySystemComponentInitialized_Event!() -> (OutputDelegate, then, AbilitySystemComponent:Fort Ability System Component Object Reference, AffectedPawn:Fort Player Pawn Object Reference) {is=title K2Node_CustomEvent}
   > Bind Ability Async Events() {is=title K2Node_CallFunction}
   Inputs()
   n7 = GetOwningPlayerPawn()
   > n8 = Is Valid(InputObject=GetAbilitySystemComponent!(Actor:Actor Object Reference=@n7) -> (ReturnValue:Ability System Component Object Reference))
   n9 = Cast To FortPlayerPawn(execute=@n8.Is Not Valid, Object=@n7)
   > n10 = Assign On Ability System Component Initialized!(execute:Exec, self:Fort Player Pawn Object Reference=@n9, Delegate:"Delegate (by ref)"=@n6) -> (then:Exec) {is=title K2Node_AssignDelegate}
   Outputs!(then:Exec=[@n10.then, @n11.then, @n9.CastFailed])
   n11 = Bind Ability Async Events(execute=@n8.Is Valid)
  ]
 }
 HandleKeyBindsChange {nodes [
   HandleKeyBindsChange() {is=title K2Node_FunctionEntry}
   > TryReverseSlots_ForSimpleBuildOnGamepad()
   n1 = Get KeybindWidget_SwitchModeKeybind()
   > SetBoundAction!(
    execute
    self:Fort Keybind Widget Object Reference=@n1
    NewBoundAction:Name=GetSwitchQuickBarActionName!(
     self:Fort Global UIContext Object Reference=Get FortGlobalUIContext!() -> (ReturnValue:Fort Global UIContext Object Reference) {is=title K2Node_GetBlueprintContextEx}
     OverrideInputType:ECommonInputType Enum=Get InputTypeOverride!(self:Fort Keybind Widget Object Reference=@n1) -> (InputTypeOverride:ECommonInputType Enum)
    ) -> (ReturnValue:Name=None)
   ) -> (then)
   > n2 = For Each Loop(Array=Get QuickbarSlots())
   > RefreshItem(self:Fort Quick Bar Slot Base Object Reference=@n2)
  ]}
 HandleQuickbarSlotFocusChanged {nodes [
   n1 = HandleQuickbarSlotFocusChanged!() -> (then, QuickbarIndex:EFortQuickBars Enum=Primary, Slot:int=0) {is=title K2Node_FunctionEntry}
   > Parent: HandleQuickbarSlotFocusChanged!(execute, QuickbarIndex:EFortQuickBars Enum=@n1, Slot:int=@n1.Slot) -> (then) {is=title K2Node_CallParentFunction}
   n2 = Get MyQuickbarIndex() -> (MyQuickbarIndex:EFortQuickBars Enum=Primary)
   > SetVisibility(
    self=Get Image_Separator() -> (Image_Separator:Image Object Reference)
    InVisibility=Select(Option 0=Collapsed, Option 1=Collapsed, Index="Equal (Enum)"(A:EFortQuickBars Enum=@n1, B:EFortQuickBars Enum=@n2))
   )
  ]}
 HandleShowHideRail {nodes [
   HandleShowHideRail() {is=title K2Node_FunctionEntry}
   n1 = Get CommonInputSubsystem!() -> (ReturnValue:Common Input Object Reference) {is=title K2Node_GetBlueprintContextEx}
   > n2 = Branch(Condition="Equal (Enum)"(A:ECommonInputType Enum=GetCurrentInputType(self=@n1), B:ECommonInputType Enum=Gamepad))
   > Show Secondary Quickbar Rail(Show=false)
   Show Secondary Quickbar Rail(execute=@n2.else, Show=false)
  ]}
 HandleSimpleBuild {nodes [
   n1 = IsAllBuildModeEnabled!(execute, self:Fort Controller Component Smart Build Object Reference=Get SmartBuildComponent()) -> (then, ReturnValue:bool)
   n2 = HandleSimpleBuild!() -> (then, IsEnabled:bool=~) {is=title K2Node_FunctionEntry}
   n3 = SetActiveWidgetIndex(
    execute=@n1.then
    self:Common Visibility Switcher Object Reference=Get Switcher_BuildTools() -> (Switcher_BuildTools:Common Visibility Switcher Object Reference)
    Index=Select(Option 0:int=0, Option 1:int=1, Index=AND Boolean(A=@n1, B=@n2)) -> (ReturnValue:int)
   )
   n4 = Branch(execute=[@n3.then, @n2.then], Condition=Get IsEnabled() -> (IsEnabled:bool))
   > TryReverseSlots_ForSimpleBuildOnGamepad()
   Branch(execute=@n4.else, Condition=Get IsReversed() -> (IsReversed:bool))
   > ReverseSlotIndices()
  ]}
 InitSimpleBuild {nodes [
   InitSimpleBuild() {is=title K2Node_FunctionEntry}
   > n1 = Cast To FortPlayerControllerAthena(Object=GetOwningPlayer())
   n2 = GetSmartBuildComponent!(self:Fort Player Controller Object Reference=@n1) -> (ReturnValue:Fort Controller Component Smart Build Object Reference)
   > Is Valid(InputObject=@n2)
   > n3 = Set SmartBuildComponent(SmartBuildComponent:Fort Controller Component Smart Build Object Reference=@n2) -> (Output_Get:Fort Controller Component Smart Build Object Reference)
   > Bind Event to On Simple Build Toggled(self:Fort Controller Component Smart Build Object Reference=@n3, Delegate=Create Event()) {is=title K2Node_AddDelegate}
   > HandleSimpleBuild(IsEnabled=IsSmartBuildEnabled(self=Get SmartBuildComponent())) {is=title K2Node_CallFunction}
  ]}
 ReverseSlotIndices {nodes [
   ReverseSlotIndices() {is=title K2Node_FunctionEntry}
   > n1 = GetConsoleVariableBoolValue!(execute, VariableName:String=SimpleBuild.Quickbar.ReverseOrder) -> (then, ReturnValue:bool)
   > Branch(Condition=@n1)
   n2 = int - int(
    A=Length(TargetArray:Array of Fort Quick Bar Slot Base Object References=Get QuickBarSlots() -> (QuickBarSlots:Array of Fort Quick Bar Slot Base Object References))
    B=2
   )
   > Set Count(Count:int=@n2) -> (Output_Get:int=0)
   > Branch(Condition="integer > integer"(A=@n2))
   > n3 = For Loop(LastIndex=@n2)
   n4 = Set IsReversed(execute=@n3.Completed, IsReversed:bool=InvertBool!(bValue:bool=Get IsReversed()) -> (ReturnValue:bool)) -> (Output_Get:bool)
   > PrintText!(
    execute
    InText=Format Text!(Format="LOCTEXT(\"Reverse :: Reversed Slots :: {0}\")", 0:bool=@n4) -> (Result)
    bPrintToScreen
    bPrintToLog
    TextColor
    Duration=10
    Key
   ) -> (then)
   n5 = "Get (a copy)"(Array:Array of Fort Quick Bar Slot Base Object References=Get QuickBarSlots(), Dimension 1=@n3) -> (Output:Fort Quick Bar Slot Base Object Reference)
   n6 = Get Count() -> (Count:int=0)
   SetQuickbarIndexRuntime!(
    execute=@n3.LoopBody
    self:Fort Quick Bar Slot Base Object Reference=@n5
    Index:int=Select(Option 0=@n6, Option 1=@n3, Index=Get IsReversed())
    bRefresh:bool=true
   ) -> (then)
   > UpdateKeyBindingText(self:Fort Quick Bar Slot Base Object Reference=@n5)
   > Decrement Int(Value=@n6)
   HandleKeyBindsChange() {is=title K2Node_CallFunction}
  ]}
 Show Secondary Quickbar Rail {nodes [
   n1 = Show Secondary Quickbar Rail!() -> (then, Show:bool=~) {is=title K2Node_FunctionEntry}
   > SetVisibility(
    self=Get SizeBox_Building() -> (SizeBox_Building:Size Box Object Reference)
    InVisibility=Select(Option 0:ESlateVisibility Enum=Hidden, Option 1:ESlateVisibility Enum=SelfHitTestInvisible, Index=@n1) -> (ReturnValue:ESlateVisibility Enum)
   )
  ]}
 TryReverseSlots_ForSimpleBuildOnGamepad {nodes [
   TryReverseSlots_ForSimpleBuildOnGamepad() {is=title K2Node_FunctionEntry}
   n1 = Get SmartBuildComponent()
   > Is Valid(InputObject=@n1)
   n2 = IsUsingGamepad!(self:Fort Global UIContext Object Reference=Get FortGlobalUIContext()) -> (ReturnValue:bool)
   n3 = Get IsReversed()
   > Branch(Condition=AND Boolean(A="Not Equal (Boolean)"(A=@n3, B=@n2) -> (ReturnValue=~), B=IsSmartBuildEnabled(self=@n1)))
   > ReverseSlotIndices() {is=title K2Node_CallFunction}
  ]}
]
variables [
 IsReversed {type=boolean}
 MaximizedAnimHasPlayed {type=boolean}
 MinimizedAnimHasPlayed {type=boolean}
 SmartBuildComponent {type=FortControllerComponent_SmartBuild, default=null}
]
widgets {
 info {inheritedWidgetCount=2, parentClass=/Game/UI/InGame/HUD/QuickbarBase.QuickbarBase_C, rootWidgetClass=&Overlay, widgetCount=30}
 widgets [
  Overlay_1 {widgetClassPath=&Overlay, widget=Overlay_1}
  SizeBox_0 {
   widgetClassPath=&SizeBox
   parent=Overlay_1
   widget=SizeBox_0
   properties {bOverride_HeightOverride=true, heightOverride=112, renderTransformPivot=<1 0.25>}
   slot {class=OverlaySlot, properties {horizontalAlignment=HAlign_Fill, verticalAlignment=VAlign_Fill}}
  }
  QuickbarSecondaryInvalidationBox {
   widgetClassPath=/Script/UMG.InvalidationBox
   parent=SizeBox_0
   widget=QuickbarSecondaryInvalidationBox
   slot {class=SizeBoxSlot, properties {horizontalAlignment=HAlign_Right, verticalAlignment=VAlign_Top}}
  }
  "[Horizontal Box] Camera Mode Content" {
   widgetClassPath=&HorizontalBox
   parent=QuickbarSecondaryInvalidationBox
   widget="[Horizontal Box] Camera Mode Content"
   slot {class=PanelSlot}
  }
  HorizontalBox_1 {
   widgetClassPath=&HorizontalBox
   parent="[Horizontal Box] Camera Mode Content"
   widget=HorizontalBox_1
   properties {renderTransformPivot=<1 0.5>}
   slot {class=HorizontalBoxSlot, properties {verticalAlignment=VAlign_Center, padding=<ltrb 0 16 0 0>}}
  }
  Switcher_SwitchModeUnbound {
   widgetClassPath=&CommonVisibilitySwitcher
   parent=HorizontalBox_1
   widget=Switcher_SwitchModeUnbound
   slot {class=HorizontalBoxSlot, properties {horizontalAlignment=HAlign_Center, verticalAlignment=VAlign_Center, padding=<ltrb 0 0 5 0>}}
  }
  KeybindWidget_SwitchModeKeybind {
   widgetClassPath=/BattleRoyaleGameUI/HUD/KeybindWidget/WBP_BRGameUI_KeybindWidget
   bInherited=true
   bIsVariable=true
   parent=Switcher_SwitchModeUnbound
   widget=KeybindWidget_SwitchModeKeybind
   slot {class=CommonVisibilitySwitcherSlot}
  }
  CanvasPanel {
   widgetClassPath=&CanvasPanel
   parent=Switcher_SwitchModeUnbound
   widget=CanvasPanel
   slot {class=CommonVisibilitySwitcherSlot, properties {horizontalAlignment=HAlign_Center, verticalAlignment=VAlign_Center, padding=<ltrb 0 0 4 0>}}
  }
  Image {
   widgetClassPath=&Image
   bIsVariable=true
   parent=CanvasPanel
   widget=Image
   properties {brush {
     bIsDynamicallyLoaded=false
     drawAs=Box
     imageType=FullColor
     mirroring=NoMirror
     resourceName=None
     resourceObject=/BattleRoyaleGameUI/HUD/KeybindWidget/Assets/M_UI_KeybindBackground
     tiling=NoTile
     imageSize=<16 32>
     margin=<ltrb 0.5 0.5 0.5 0.5>
     outlineSettings {
      bUseBrushTransparency=false
      roundingType=HalfHeightRadius
      width=0
      color {colorUseRule=UseColor_Specified, specifiedColor=<rgba 0 0 0 0>}
      cornerRadii=<0 0 0 0>
     }
     tintColor {colorUseRule=UseColor_Specified, specifiedColor=<rgba 1 1 1 1>}
     uVRegion {bIsValid=false, max=<0 0>, min=<0 0>}
    }}
   slot {class=CanvasPanelSlot, properties {layoutData {alignment=<0 0>, anchors {maximum=<1 1>, minimum=<0 0>}, offsets=<ltrb -12 -4 -12 -6>}}}
  }
  SizeBox {
   widgetClassPath=&SizeBox
   parent=CanvasPanel
   widget=SizeBox
   slot {
    class=CanvasPanelSlot
    properties {bAutoSize=true, layoutData {alignment=<0 0>, anchors {maximum=<1 1>, minimum=<0 0>}, offsets=<ltrb 0 0 0 0>}}
   }
  }
  Text_SwitchModeUnbound {
   widgetClassPath=&CommonTextBlock
   parent=SizeBox
   widget=Text_SwitchModeUnbound
   properties {
    justification=Center
    text=Switch Mode Unbound
    font {
     bForceMonospaced=false
     bMaterialIsStencil=false
     fontMaterial=/BattleRoyaleGameUI/HUD/KeybindWidget/Assets/M_UI_KeybindFont
     fontObject=&BurbankSmall
     letterSpacing=-10
     monospacedWidth=1
     size=13.5
     skewAmount=0
     typefaceFontName=Black
     outlineSettings {
      bApplyOutlineToDropShadows=false
      bMiteredCorners=false
      bSeparateFillAlpha=true
      outlineMaterial=null
      outlineSize=0
      outlineColor=<rgba 0 0 0 0>
     }
    }
    shadowOffset=<0 0>
   }
   slot {class=SizeBoxSlot}
  }
  Overlay_NoBuild {widgetClassPath=&Overlay, bIsVariable=true, parent=HorizontalBox_1, widget=Overlay_NoBuild, slot {class=HorizontalBoxSlot}}
  Image_IconGlow {
   widgetClassPath=&Image
   parent=Overlay_NoBuild
   widget=Image_IconGlow
   properties {
    visibility=Collapsed
    brush {
     bIsDynamicallyLoaded=false
     drawAs=Image
     imageType=FullColor
     mirroring=NoMirror
     resourceName=None
     resourceObject=/BattleRoyaleGameUI/HUD/QuickBar/Assets/M_UI_ItemRarityGlow
     tiling=NoTile
     imageSize=<32 32>
     margin=<ltrb 0 0 0 0>
     outlineSettings {
      bUseBrushTransparency=false
      roundingType=HalfHeightRadius
      width=0
      color {colorUseRule=UseColor_Specified, specifiedColor=<rgba 0 0 0 0>}
      cornerRadii=<0 0 0 0>
     }
     tintColor {colorUseRule=UseColor_Specified, specifiedColor=<rgba 0 0 1 1>}
     uVRegion {bIsValid=false, max=<0 0>, min=<0 0>}
    }
    renderTransform {angle=0, scale=<3 -3>, shear=<0 0>, translation=<0 4>}
   }
   slot {class=OverlaySlot, properties {horizontalAlignment=HAlign_Fill, verticalAlignment=VAlign_Fill}}
  }
  Image_SwitchAction {
   widgetClassPath=&Image
   bIsVariable=true
   parent=Overlay_NoBuild
   widget=Image_SwitchAction
   properties {brush {
     bIsDynamicallyLoaded=false
     drawAs=Image
     imageType=FullColor
     mirroring=NoMirror
     resourceName=None
     resourceObject=&M_UI_QuickbarTypeIcon
     tiling=NoTile
     imageSize=<40 40>
     margin=<ltrb 0 0 0 0>
     outlineSettings {
      bUseBrushTransparency=false
      roundingType=HalfHeightRadius
      width=0
      color {colorUseRule=UseColor_Specified, specifiedColor=<rgba 0 0 0 0>}
      cornerRadii=<0 0 0 0>
     }
     tintColor {colorUseRule=UseColor_Specified, specifiedColor=<rgba 1 1 1 1>}
     uVRegion {bIsValid=false, max=<0 0>, min=<0 0>}
    }}
   slot {class=OverlaySlot, properties {horizontalAlignment=HAlign_Center, verticalAlignment=VAlign_Center}}
  }
  Overlay_BuildTools {
   widgetClassPath=&Overlay
   bIsVariable=true
   parent="[Horizontal Box] Camera Mode Content"
   widget=Overlay_BuildTools
   slot {class=HorizontalBoxSlot}
  }
  SizeBox_Building {
   widgetClassPath=&SizeBox
   bIsVariable=true
   parent=Overlay_BuildTools
   widget=SizeBox_Building
   properties {bOverride_HeightOverride=true, heightOverride=4, visibility=Hidden}
   slot {class=OverlaySlot, properties {horizontalAlignment=HAlign_Fill, padding=<ltrb 0 21 0 0>}}
  }
  CommonBorder_0 {
   widgetClassPath=&CommonBorder
   parent=SizeBox_Building
   widget=CommonBorder_0
   properties {style=/FortUILibrary/Styles/Border/Border_HUD_Line_Horz.Border_HUD_Line_Horz_C}
   slot {class=SizeBoxSlot}
  }
  HorizontalBox_0 {widgetClassPath=&HorizontalBox, parent=Overlay_BuildTools, widget=HorizontalBox_0, slot {class=OverlaySlot}}
  Switcher_BuildTools {
   widgetClassPath=&CommonVisibilitySwitcher
   bIsVariable=true
   parent=HorizontalBox_0
   widget=Switcher_BuildTools
   slot {class=HorizontalBoxSlot}
  }
  F1toF4Slots {widgetClassPath=&HorizontalBox, parent=Switcher_BuildTools, widget=F1toF4Slots, slot {class=CommonVisibilitySwitcherSlot}}
  QuickbarSlotF1 {
   widgetClassPath=&/WBP_BRGameUI_QuickbarSlot_Secondary
   bIsVariable=true
   parent=F1toF4Slots
   widget=QuickbarSlotF1
   properties {bHideIfNoItemAssigned=true, bShowTopArrows=true, quickBarType=Secondary, cooldownTypesSupported [Death]}
   slot {class=HorizontalBoxSlot}
  }
  QuickbarSlotF2 {
   widgetClassPath=&WBP_BRGameUI_QuickbarSlot_Secondary
   bIsVariable=true
   parent=F1toF4Slots
   widget=QuickbarSlotF2
   properties {bHideIfNoItemAssigned=true, bShowTopArrows=true, quickBarType=Secondary, slotIndex=1, cooldownTypesSupported [Death]}
   slot {class=HorizontalBoxSlot}
  }
  QuickbarSlotF3 {
   widgetClassPath=&WBP_BRGameUI_QuickbarSlot_Secondary
   bIsVariable=true
   parent=F1toF4Slots
   widget=QuickbarSlotF3
   properties {bHideIfNoItemAssigned=true, bShowTopArrows=true, quickBarType=Secondary, slotIndex=2, cooldownTypesSupported [Death]}
   slot {class=HorizontalBoxSlot}
  }
  QuickbarSlotF4 {
   widgetClassPath=&WBP_BRGameUI_QuickbarSlot_Secondary
   bIsVariable=true
   parent=F1toF4Slots
   widget=QuickbarSlotF4
   properties {bHideIfNoItemAssigned=true, bShowTopArrows=true, quickBarType=Secondary, slotIndex=3, cooldownTypesSupported [Death]}
   slot {class=HorizontalBoxSlot}
  }
  SimpleBuild {
   widgetClassPath=&VerticalBox
   parent=Switcher_BuildTools
   widget=SimpleBuild
   slot {class=CommonVisibilitySwitcherSlot, properties {verticalAlignment=VAlign_Center}}
  }
  Image_SimpleBuild {
   widgetClassPath=&Image
   bIsVariable=true
   parent=SimpleBuild
   widget=Image_SimpleBuild
   properties {brush {
     bIsDynamicallyLoaded=false
     drawAs=Image
     imageType=FullColor
     mirroring=NoMirror
     resourceName=None
     resourceObject=/Game/UI/Foundation/Textures/Icons/HUD/T_SmartBuildMode_All_128
     tiling=NoTile
     imageSize=<128 128>
     margin=<ltrb 0 0 0 0>
     outlineSettings {
      bUseBrushTransparency=false
      roundingType=HalfHeightRadius
      width=0
      color {colorUseRule=UseColor_Specified, specifiedColor=<rgba 0 0 0 0>}
      cornerRadii=<0 0 0 0>
     }
     tintColor {colorUseRule=UseColor_Specified, specifiedColor=<rgba 1 1 1 1>}
     uVRegion {bIsValid=false, max=<0 0>, min=<0 0>}
    }}
   slot {class=VerticalBoxSlot, properties {verticalAlignment=VAlign_Center}}
  }
  Image_Separator {
   widgetClassPath=&Image
   bIsVariable=true
   parent=HorizontalBox_0
   widget=Image_Separator
   properties {
    visibility=Collapsed
    brush {
     bIsDynamicallyLoaded=false
     drawAs=Image
     imageType=FullColor
     mirroring=NoMirror
     resourceName=None
     resourceObject=/BattleRoyaleGameUI/Shared/FrameDecor/M_UI_SimpleDivider_Vertical
     tiling=Vertical
     imageSize=<12 48>
     margin=<ltrb 0 0 0 0>
     outlineSettings {
      bUseBrushTransparency=false
      roundingType=HalfHeightRadius
      width=0
      color {colorUseRule=UseColor_Specified, specifiedColor=<rgba 0 0 0 0>}
      cornerRadii=<0 0 0 0>
     }
     tintColor {colorUseRule=UseColor_Specified, specifiedColor=<rgba 1 1 1 1>}
     uVRegion {bIsValid=false, max=<0 0>, min=<0 0>}
    }
    colorAndOpacity=<rgba 1 1 1 0.8>
   }
   slot {class=HorizontalBoxSlot, properties {verticalAlignment=VAlign_Bottom, padding=<ltrb 0 0 0 16>}}
  }
  QuickbarSlotF5 {
   widgetClassPath=&WBP_BRGameUI_QuickbarSlot_Secondary
   bIsVariable=true
   parent=HorizontalBox_0
   widget=QuickbarSlotF5
   properties {bHideIfNoItemAssigned=true, quickBarType=Secondary, slotIndex=4, cooldownTypesSupported [Death]}
   slot {class=HorizontalBoxSlot}
  }
  TutorialHighlight {
   widgetClassPath=&TutorialHighlight
   bIsVariable=true
   parent=Overlay_1
   widget=TutorialHighlight
   properties {customHighlightWidget=/Game/Athena/HUD/Tutorial/Tutorial_Outline_Square_Static.Tutorial_Outline_Square_Static_C}
   slot {class=OverlaySlot, properties {horizontalAlignment=HAlign_Fill, verticalAlignment=VAlign_Fill}}
  }
  BuildingBlocked_Icon {
   widgetClassPath=&Image
   bIsVariable=true
   parent=Overlay_1
   widget=BuildingBlocked_Icon
   properties {
    visibility=Collapsed
    brush {
     bIsDynamicallyLoaded=false
     drawAs=Image
     imageType=FullColor
     mirroring=NoMirror
     resourceName=None
     resourceObject=/BattleRoyaleGameUI/HUD/QuickBar/Assets/M_UI_NoBuildIndicator
     tiling=NoTile
     imageSize=<56 56>
     margin=<ltrb 0 0 0 0>
     outlineSettings {
      bUseBrushTransparency=false
      roundingType=HalfHeightRadius
      width=0
      color {colorUseRule=UseColor_Specified, specifiedColor=<rgba 0 0 0 0>}
      cornerRadii=<0 0 0 0>
     }
     tintColor {colorUseRule=UseColor_Specified, specifiedColor=<rgba 1 1 1 1>}
     uVRegion {bIsValid=false, max=<0 0>, min=<0 0>}
    }
    renderTransform {angle=0, scale=<1 1>, shear=<0 0>, translation=<0 -30>}
   }
   slot {class=OverlaySlot, properties {horizontalAlignment=HAlign_Right, verticalAlignment=VAlign_Center, padding=<ltrb 0 30 20 0>}}
  }
 ]
}
