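# Overflow quickbar slot widget for the Battle Royale HUD, a child of the Primary quickbar slot with no widgets of its own.
# PreConstruct caches the tile image's dynamic material and turns on its rounded-mask parameter.
# Construct plays a short Anim_Show that first grows the tile's SizeBox from zero to full size, then fades the slot in, so it appears to make room before becoming visible.
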WidgetBlueprint /BattleRoyaleGameUI/HUD/QuickBar/WBP_BRGameUI_QuickbarSlot_Overflow
animations [Anim_Show {
  binding_count=2
  length_seconds=0.25
  object=&self:Anim_Show
  sequence {bindings [
    WBP_BRGameUI_QuickbarSlot_Overflow_C_30 {
     id=#1
     possessed_class=WBP_BRGameUI_QuickbarSlot_Overflow_C
     tracks [{
       type=MovieSceneFloatTrack
       display_name=Render Opacity
       sections [{type=MovieSceneFloatSection, end_frame=5, start_frame=0, channels [None {type=MovieSceneScriptingFloatChannel, keys [[0,0],[2,0],[5,1]]}]}]
      }]
    }
    SizeBox_TileContents {
     id=#2
     possessed_class=SizeBox
     tracks [
      {
       type=MovieSceneFloatTrack
       display_name=Height Override
       sections [{type=MovieSceneFloatSection, end_frame=2, start_frame=0, channels [None {type=MovieSceneScriptingFloatChannel, keys [[0,0],[2,110]]}]}]
      }
      {
       type=MovieSceneFloatTrack
       display_name=Width Override
       sections [{type=MovieSceneFloatSection, end_frame=2, start_frame=0, channels [None {type=MovieSceneScriptingFloatChannel, keys [[0,0],[2,95]]}]}]
      }
     ]
    }
   ]}
 }]
graphs [EventGraph {nodes [
   n1 = Event PreConstruct()
   > Parent: PreConstruct(IsDesignTime=@n1.IsDesignTime)
   > n2 = GetDynamicMaterial(self=Get Image_TileImage() -> (Image_TileImage:Image Object Reference))
   > n3 = Set MID_TileImage(MID_TileImage:Material Instance Dynamic Object Reference=@n2) -> (Output_Get:Material Instance Dynamic Object Reference)
   > SetScalarParameterValue(self=@n3, ParameterName=isRoundedMask, Value=1)
   Event Construct()
   > Parent: Construct()
   > PlayAnimationForward(InAnimation=Get Anim_Show() -> (Anim_Show:Widget Animation Object Reference))
   Event Tick()
  ]}]
variables [MID_TileImage {type=MaterialInstanceDynamic}]
widgets {
 info {
  inheritedWidgetCount=0
  parentClass=&/WBP_BRGameUI_QuickbarSlot_Primary.WBP_BRGameUI_QuickbarSlot_Primary_C
  rootWidgetClass=null
  widgetCount=0
 }
 widgets []
}
