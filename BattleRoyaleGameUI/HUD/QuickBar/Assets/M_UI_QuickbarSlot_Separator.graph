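# Translucent UI material that draws a slanted, tapered separator for a HUD quickbar slot.
# UVs are offset by position parameters, slightly rotated and skewed, then fed into an isosceles trapezoid function with zero bottom width and adjustable top width and height.
# The shape fill is tinted by a Color parameter and also masked by a top-to-bottom gradient, which the Fade parameter shifts.
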Material /BattleRoyaleGameUI/HUD/QuickBar/Assets/M_UI_QuickbarSlot_Separator
graph {
 properties {bAutomaticallySetUsageInEditor=false, blendMode=BLEND_Translucent, materialDomain=MD_UI, usedShadingModels=Default Lit}
 nodes [
  n1 = Add(A=TextureCoordinate(), B=AppendVector(A=ScalarParameter() {parameterName=PositionX}, B=ScalarParameter() {parameterName=PositionY}))
  n2 = Constant() {r=-0.025}
  n3 = MaterialFunctionCall!(
   UVs=MaterialFunctionCall!(
    UV=MaterialFunctionCall(UVs=@n1, "Rotation Angle (0-1)"=@n2) {materialFunction=&CustomRotator}.Rotated Values
    Skew Angle=Constant2Vector() {g=-6, r=-10}
   ) {materialFunction=/FortUILibrary/MaterialFunctions/Basics/MF_UI_Skew}.Skewed UVs
   TopWidth=ScalarParameter() {defaultValue=0.1, parameterName=TopWidth}
   BottomWidth=Constant()
   Height=ScalarParameter() {defaultValue=0.87, parameterName=Height}
   OutlineThickness
   OutlineAlignment
   GlowMin
   GlowMax
  ) {materialFunction=/FortUILibrary/MaterialFunctions/Shapes/MF_UI_IsoscelesTrapezoid}
  Multiply(A=VectorParameter() {parameterName=Color, defaultValue=<rgba 0.4102 0.4102 0.4102 0>}.RGB, B=@n3.Fill)
  n4 = TextureCoordinate()
  n5 = Subtract(A=OneMinus(None=ComponentMask(None=@n4) {g=true}), B=OneMinus(None=ScalarParameter() {defaultValue=1, parameterName=Fade}))
  Reroute(None=Multiply(A=@n3.Fill, B=Saturate(None=@n5)))
 ]
}
