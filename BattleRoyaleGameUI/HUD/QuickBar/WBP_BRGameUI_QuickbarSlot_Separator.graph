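# Quickbar slot separator widget for the Battle Royale HUD, drawing a thin vertical divider image that fades in when created.
# An Overlay centers a VerticalBox holding the material-driven separator image, widened by render scale, with a Spacer beneath it.
# On Construct it plays Anim_Show forward, a short animation that drives the material's Fade parameter from 0 to 1 rather than animating widget opacity.
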WidgetBlueprint /BattleRoyaleGameUI/HUD/QuickBar/WBP_BRGameUI_QuickbarSlot_Separator
animations [Anim_Show {
  binding_count=1
  length_seconds=0.2
  object=&self:Anim_Show
  sequence {bindings [Image_Separator {
     id=#1
     possessed_class=Image
     tracks [{
       type=MovieSceneWidgetMaterialTrack
       sections [{type=MovieSceneParameterSection, end_frame=4, start_frame=0, channels [Fade {type=MovieSceneScriptingFloatChannel, keys [[0,0],[4,1]]}]}]
      }]
    }]}
 }]
graphs [EventGraph {nodes [
   Event Construct()
   > PlayAnimationForward(InAnimation=Get Anim_Show() -> (Anim_Show:Widget Animation Object Reference))
  ]}]
variables []
widgets {
 info {inheritedWidgetCount=0, parentClass=&UserWidget, rootWidgetClass=&Overlay, widgetCount=4}
 widgets [
  Overlay_0 {widgetClassPath=&Overlay, widget=Overlay_0, properties {bOverride_Cursor=true}}
  VerticalBox_0 {
   widgetClassPath=&VerticalBox
   parent=Overlay_0
   widget=VerticalBox_0
   slot {class=OverlaySlot, properties {horizontalAlignment=HAlign_Center, verticalAlignment=VAlign_Center}}
  }
  Image_Separator {
   widgetClassPath=&Image
   bIsVariable=true
   parent=VerticalBox_0
   widget=Image_Separator
   properties {
    brush {
     bIsDynamicallyLoaded=false
     drawAs=Image
     imageType=FullColor
     mirroring=NoMirror
     resourceName=None
     resourceObject=/BattleRoyaleGameUI/HUD/QuickBar/Assets/M_UI_QuickbarSlot_Separator
     tiling=NoTile
     imageSize=<16 64>
     margin=<ltrb 0 0 0 0>
     outlineSettings {
      bUseBrushTransparency=false
      roundingType=HalfHeightRadius
      width=0
      color {colorUseRule=UseColor_Specified, specifiedColor=<rgba 0 0 0 0>}
      cornerRadii=<0 0 0 0>
     }
     tintColor {colorUseRule=UseColor_Specified, specifiedColor=<rgba 1 1 1 1>}
     uVRegion {bIsValid=false, max=<0 0>, min=<0 0>}
    }
    renderTransform {angle=0, scale=<2 1>, shear=<0 0>, translation=<0 0>}
   }
   slot {class=VerticalBoxSlot}
  }
  Spacer_35 {
   widgetClassPath=&Spacer
   parent=VerticalBox_0
   widget=Spacer_35
   properties {size=<1 20>}
   slot {class=VerticalBoxSlot, properties {horizontalAlignment=HAlign_Center, verticalAlignment=VAlign_Center}}
  }
 ]
}
