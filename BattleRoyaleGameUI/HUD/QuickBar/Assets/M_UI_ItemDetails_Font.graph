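# Translucent UI font material for item detail text in the battle royale HUD quick bar, with an empty-ammo color state.
# IsAmmoEmpty blends the adaptive Bright color toward a mix of shared-collection Red and adaptive Dim.
# Color_MixRatio has no effect because both its inputs are Bright.
# UseFade scales Opacity by a sine-eased vertical UV gradient, running from full at the top to FadeOpacity at the bottom.
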Material /BattleRoyaleGameUI/HUD/QuickBar/Assets/M_UI_ItemDetails_Font
graph {
 properties {bAutomaticallySetUsageInEditor=false, blendMode=BLEND_Translucent, materialDomain=MD_UI, usedShadingModels=Default Lit}
 nodes [
  ComponentMask(None=CollectionParameter() {collection=/BattleRoyaleGameUI/Shared/MPC_BRGameUI, parameterName=Charcoal}) {b=true, g=true, r=true}
  n1 = MaterialFunctionCall!(BufferSource) {materialFunction=/BattleRoyaleGameUI/Shared/MF_UI_AdaptiveColors}
  n2 = CollectionParameter() {collection=&MPC_BRGameUI, parameterName=Red}
  LinearInterpolate(
   A=LinearInterpolate(A=@n1.Bright, B=@n1.Bright, Alpha=ScalarParameter() {parameterName=Color_MixRatio})
   B=LinearInterpolate(A=ComponentMask(None=@n2) {b=true, g=true, r=true}, B=@n1.Dim, Alpha=ScalarParameter() {parameterName=EmptyColor_MixRatio})
   Alpha=ScalarParameter() {parameterName=IsAmmoEmpty}
  )
  n3 = ScalarParameter() {defaultValue=1, parameterName=Opacity}
  n4 = ScalarParameter() {parameterName=FadeOpacity}
  n5 = ComponentMask(None=MaterialFunctionCall() {materialFunction=&GetUserInterfaceUV}.Normalized UV) {g=true}
  n6 = Saturate(None=MaterialFunctionCall("Value (0-1)"=OneMinus(None=@n5)) {materialFunction=&MF_UI_EaseCurves}.EaseInSine)
  LinearInterpolate(A=@n3, B=Multiply(A=@n3, B=LinearInterpolate(A=@n4, Alpha=@n6)), Alpha=ScalarParameter() {parameterName=UseFade})
 ]
}
