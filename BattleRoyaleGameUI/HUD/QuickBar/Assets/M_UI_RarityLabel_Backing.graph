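# Translucent UI material for the rarity label backing in the Battle Royale quick bar HUD.
# Color starts from the shared adaptive Dark color and blends toward a Color parameter by ColorMix, with Anim_Color scaling that blend from none to full.
# A rounded box fill, with adjustable corner radius and kept aspect ratio, is multiplied by a shared low-end switcher value that lowers its strength on low-end devices.
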Material /BattleRoyaleGameUI/HUD/QuickBar/Assets/M_UI_RarityLabel_Backing
graph {
 properties {bAutomaticallySetUsageInEditor=false, blendMode=BLEND_Translucent, materialDomain=MD_UI, usedShadingModels=Default Lit}
 nodes [
  n1 = MaterialFunctionCall!(BufferSource) {materialFunction=/BattleRoyaleGameUI/Shared/MF_UI_AdaptiveColors}
  LinearInterpolate(
   A=@n1.Dark
   B=LinearInterpolate(
    A=@n1.Dark
    B=VectorParameter() {parameterName=Color, defaultValue=<rgba 0.0781 0.0781 0.0781 1>}.RGB
    Alpha=ScalarParameter() {parameterName=ColorMix}
   )
   Alpha=ScalarParameter() {parameterName=Anim_Color}
  )
  Multiply(
   A=MaterialFunctionCall(
    Size=Constant() {r=1}
    CornerRoundness=ScalarParameter() {defaultValue=0.25, parameterName=CornerRadius}
    KeepAspectRatio=StaticBool() {value=true}
   ) {materialFunction=&MF_UI_Box}.Fill
   B=MaterialFunctionCall!(Default=Constant() {r=0.75}, LowEnd=Constant() {r=0.464}) {materialFunction=/BattleRoyaleGameUI/Shared/MF_UI_LowEnd_ValueSwitcher}.Result
  ) {constB=0.75}
 ]
}
